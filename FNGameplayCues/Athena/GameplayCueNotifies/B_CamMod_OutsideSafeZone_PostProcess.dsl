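; Camera modifier that applies an outside-safe-zone post-process effect, by its name for Fortnite Athena gameplay cues.
; BlueprintModifyPostProcess returns the modifier's alpha as the blend weight together with a stored Post Process Settings variable.
; Fading comes from the camera modifier's alpha; the settings themselves are not modified per frame.
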(fn BlueprintModifyPostProcess
  (params (DeltaTime :type Float%20%28single-precision%29))
  (return (Variables|CameraModifier|GetAlpha) (Variables|Default|GetPost%20Process%20Settings)))
