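; Looping gameplay cue for the Athena Carrot that turns on a thermal-vision post-process effect while it is active.
; On loop start it enables its post-process component, sets ThermalFadeout in the shared MPC_ThermalGlobal parameter collection, and plays a timeline that drives ThermalScale.
; On removal it only disables the post-process component, leaving the collection values as they are.
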(event EventOnLoopingStart(Cascade)
  (params (MyTarget :type Actor%20Object%20Reference) (Parameters :type Gameplay%20Cue%20Parameters%20Structure%20%28by%20ref%29) (ParticleComponents :type Array%20of%20Cascade%20Particle%20System%20Component%20Object%20References) (AudioComponents :type Array%20of%20Audio%20Component%20Object%20References))
  (bind _postprocess (Variables|Default|GetPostProcess))
  (Variables|/Script/Engine.PostProcessComponent|SetbEnabled true _postprocess)
  (Rendering|Material|SetScalarParameterValue 0 "/Game/Athena/Items/Weapons/Abilities/Thermal/MPC_ThermalGlobal.MPC_ThermalGlobal" "ThermalFadeout" 1.0 :!class /Script/Engine.KismetMaterialLibrary)
  (bind (_direction _newtrack_0) (Timeline|Timeline_0 PlayFromStart)
    (:Update
      (Rendering|Material|SetScalarParameterValue _postprocess "/Game/Athena/Items/Weapons/Abilities/Thermal/MPC_ThermalGlobal.MPC_ThermalGlobal" "ThermalScale" _newtrack_0 :!class /Script/Engine.KismetMaterialLibrary))))

(event EventOnRemoval(Cascade)
  (params (MyTarget :type Actor%20Object%20Reference) (Parameters :type Gameplay%20Cue%20Parameters%20Structure%20%28by%20ref%29) (ParticleComponents :type Array%20of%20Cascade%20Particle%20System%20Component%20Object%20References) (AudioComponents :type Array%20of%20Audio%20Component%20Object%20References) (BurstCameraShakeInstance :type Camera%20Shake%20Base%20Object%20Reference) (BurstDecalInstance :type Decal%20Actor%20Object%20Reference))
  (Variables|/Script/Engine.PostProcessComponent|SetbEnabled false (Variables|Default|GetPostProcess)))

(fn ConstructionScript)
