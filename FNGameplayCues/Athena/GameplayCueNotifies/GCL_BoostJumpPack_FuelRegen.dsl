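; Gameplay cue notify for the boost jetpack that blinks the backpack mesh when fuel regenerates.
; An EventResetBlink event takes the player pawn and fetches its Backpack customization skeletal mesh.
; A timeline plays from start, writing its blink track to the FullBlink material scalar and holding Refueling at zero on each update.
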(event Jetpack|EventResetBlink
  (params (PlayerPawn :type Fort%20Player%20Pawn%20Object%20Reference))
  (bind _returnvalue (Customization|GetSkeletalMeshforPartType PlayerPawn "Backpack" :!class /Script/FortniteGame.FortPlayerPawn))
  ;[1 1 1] Fuel Recharge "Blink"
  (bind (_direction _blink) (Timeline|FullBlinkTimeline PlayFromStart)
    (:Update
      (Rendering|Material|SetScalarParameterValueonMaterials _returnvalue "FullBlink" _blink)
      (Rendering|Material|SetScalarParameterValueonMaterials _returnvalue "Refueling" 0.0))))

(fn ConstructionScript)
