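; Looping gameplay cue for the Boost Jump Pack hover, providing jetpack VFX and audio on the player pawn.
; OnBurst attaches hover Niagara FX to the backpack mesh, sets its material scalars and enables audio; OnCeaseRelevant disables audio and deactivates the FX.
; SetJetpackAudioEnabled toggles tick and handles start, stop and randomly offset idle/accel loop sounds.
; A first-jump flag chooses start or thrust sound, resetting when the pawn lands.
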(event Custom|On%20Pawn%20Landed
  (params (Hit :type Hit%20Result%20Structure%20%28by%20ref%29))
  ;[1 1 1] Reset "First-Jump" flag once we've landed
  (Variables|Sounds|SetbIsFirstJump true)
  (EventDispatchers|UnbindEventfromOnPawnLanded (Variables|GameplayCueNotify_Jetpack_Hovering|GetPlayerPawn) (delegate Custom|On%20Pawn%20Landed)))

(fn OnBurst
  (params (MyTarget :type Actor%20Object%20Reference) (Parameters :type Gameplay%20Cue%20Parameters%20Structure%20%28by%20ref%29))
  (bind _playerpawn (Variables|GameplayCueNotify_Jetpack_Hovering|GetPlayerPawn))
  (bind _returnvalue (Customization|GetSkeletalMeshforPartType _playerpawn "Backpack" :!class /Script/FortniteGame.FortPlayerPawn))
  (bind _returnvalue_2 (|CallParentFunction MyTarget Parameters))
  (if _returnvalue_2
    (Utilities|IsValid _returnvalue
      (:Is%20Valid
        (bind _returnvalue_3 (Effects|SpawnFXSystemAttached "/FNGameplayCues/Athena/Items/Gameplay/BackPacks/BoostJumpPack/FX/NS_Jetpack_Hover_Simplified.NS_Jetpack_Hover_Simplified" _returnvalue))
        (Variables|Default|SetVFX_Reference _returnvalue_3)
        (HelperMethods|SetJetpackAudioEnabled true)
        (Rendering|Material|SetScalarParameterValueonMaterials _returnvalue "FullBlink" 1.0)
        (Rendering|Material|SetScalarParameterValueonMaterials _returnvalue "Refueling" 0.0)
        (return true))
      (:Is%20Not%20Valid
        (return false)))
    (else
      (return false))))

(fn OnCeaseRelevant
  (params (MyTarget :type Actor%20Object%20Reference) (Parameters :type Gameplay%20Cue%20Parameters%20Structure%20%28by%20ref%29))
  (EventDispatchers|UnbindEventfromOnPawnLanded (Variables|GameplayCueNotify_Jetpack_Hovering|GetPlayerPawn) (EventDispatchers|CreateEvent :!event On%20Pawn%20Landed))
  (HelperMethods|SetJetpackAudioEnabled false)
  (|CallParentFunction (Variables|GetMyTarget) (Variables|GetParameters))
  (Components|Activation|Deactivate (Variables|Default|GetVFX_Reference))
  (return true))

(fn SetJetpackAudioEnabled
  (params (Enabled :type Boolean))
  (Variables|Sounds|SetbJetpackAudioEnabled Enabled)
  (Actor|Tick|SetActorTickEnabled (Variables|Sounds|GetbJetpackAudioEnabled))
  (Utilities|IsValid (Variables|GameplayCueNotify_Jetpack_Hovering|GetPlayerPawn)
    (:Is%20Valid
      (if (Variables|Sounds|GetbJetpackAudioEnabled)
        (EventDispatchers|BindEventtoOnPawnLanded (Variables|GameplayCueNotify_Jetpack_Hovering|GetPlayerPawn) (EventDispatchers|CreateEvent :!event On%20Pawn%20Landed))
        (Audio|SpawnSoundAttached (select (Variables|Sounds|GetbIsFirstJump) (Variables|Sounds|GetSoundOnStart) (Variables|SoundIndicator|GetSoundOnThrust)) (Variables|/Script/Engine.Actor|GetRootComponent (Variables|GameplayCueNotify_Jetpack_Hovering|GetPlayerPawn)))
        (Variables|Sounds|SetbIsFirstJump false)
        (Utilities|FlowControl|Sequence
          (:then_0
            (bind _returnvalue (Audio|SpawnSoundAttached :Sound (Variables|Sounds|GetSoundLoopOnIdle) :AttachToComponent (Variables|/Script/Engine.Actor|GetRootComponent (Variables|GameplayCueNotify_Jetpack_Hovering|GetPlayerPawn)) :AttachPointName "None" :LocationType "KeepRelativeOffset" :bStopWhenAttachedToDestroyed false :VolumeMultiplier 1.0 :PitchMultiplier 1.0 :StartTime (Math|Random|RandomFloatinRange 0.0 7.4) :bAutoDestroy true))
            (Variables|GameplayCueNotify_Jetpack_Hovering|SetIdleAudioComp _returnvalue))
          (:then_1
            (bind _returnvalue_2 (Audio|SpawnSoundAttached :Sound (Variables|Sounds|GetSoundLoopOnAccel) :AttachToComponent (Variables|/Script/Engine.Actor|GetRootComponent (Variables|GameplayCueNotify_Jetpack_Hovering|GetPlayerPawn)) :AttachPointName "None" :LocationType "KeepRelativeOffset" :bStopWhenAttachedToDestroyed false :VolumeMultiplier 1.0 :PitchMultiplier 1.0 :StartTime (Math|Random|RandomFloatinRange 0.0 5.0) :bAutoDestroy true))
            (Variables|GameplayCueNotify_Jetpack_Hovering|SetAccelAudioComp _returnvalue_2)))
        (else
          _)))
    (:Is%20Not%20Valid
      _))
  (Audio|Components|Audio|FadeOut (Variables|GameplayCueNotify_Jetpack_Hovering|GetIdleAudioComp) 0.25 0.0 "Linear" :!class /Script/Engine.AudioComponent)
  (Utilities|IsValid (Variables|GameplayCueNotify_Jetpack_Hovering|GetPlayerPawn)
    (:Is%20Valid
      (Audio|SpawnSoundAttached (Variables|Sounds|GetSoundOnStop) (Variables|/Script/Engine.Actor|GetRootComponent (Variables|GameplayCueNotify_Jetpack_Hovering|GetPlayerPawn))))))

(fn ConstructionScript)
