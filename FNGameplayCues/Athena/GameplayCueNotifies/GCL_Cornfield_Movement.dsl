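; Looping gameplay cue notify that handles foley and effects for an Athena player pawn moving in and out of cornfields.
; OnBurst adds the corn stalk sound library to the pawn through the SoundLibrary subsystem and plays the enter sound at its location.
; OnCeaseRelevant removes that library, plays the exit sound and spawns an exit particle emitter attached to the pawn's mesh.
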(fn OnBurst
  (params (MyTarget :type Actor%20Object%20Reference) (Parameters :type Gameplay%20Cue%20Parameters%20Structure%20%28by%20ref%29))
  (bind _asfort_player_pawn_athena (Utilities|Casting|CastToFortPlayerPawnAthena MyTarget))
  (SoundLibrary|AddLibrary (WorldSubsystems|GetSoundLibrarySubsystem _asfort_player_pawn_athena) _asfort_player_pawn_athena "/FNGameplayCues/Sounds/Fort_Foley_Sounds/Cornfield/CornStalks_Lib.CornStalks_Lib" :!class /Script/SoundLibrary.SoundLibrarySubsystem)
  (Audio|PlaySoundatLocation _asfort_player_pawn_athena "/FNGameplayCues/Sounds/Fort_Foley_Sounds/Cornfield/CornStalks_Enter_Cue.CornStalks_Enter_Cue" (Transformation|GetActorLocation _asfort_player_pawn_athena))
  (return false))

(fn OnCeaseRelevant
  (params (MyTarget :type Actor%20Object%20Reference) (Parameters :type Gameplay%20Cue%20Parameters%20Structure%20%28by%20ref%29))
  (bind _asfort_player_pawn_athena (Utilities|Casting|CastToFortPlayerPawnAthena MyTarget))
  (SoundLibrary|RemoveLibrary (WorldSubsystems|GetSoundLibrarySubsystem _asfort_player_pawn_athena) _asfort_player_pawn_athena "/FNGameplayCues/Sounds/Fort_Foley_Sounds/Cornfield/CornStalks_Lib.CornStalks_Lib" :!class /Script/SoundLibrary.SoundLibrarySubsystem)
  (Audio|PlaySoundatLocation _asfort_player_pawn_athena "/FNGameplayCues/Sounds/Fort_Foley_Sounds/Cornfield/CornStalks_Exit_Cue.CornStalks_Exit_Cue" (Transformation|GetActorLocation _asfort_player_pawn_athena))
  (Effects|SpawnEmitterAttached :EmitterTemplate "/FNGameplayCues/Effects/Fort_Effects/Gameplay/Corn/P_Player_Corn_Exit.P_Player_Corn_Exit" :AttachToComponent (Variables|/Script/Engine.Character|GetMesh _asfort_player_pawn_athena) :AttachPointName "None" :Scale "1.000000,1.000000,1.000000" :LocationType "SnapToTarget" :bAutoDestroy true :PoolingMethod "None" :bAutoActivate true)
  (return false))

(fn ConstructionScript)
