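; Intel Pack highlight gameplay cue actor that marks its owning item, such as a container, pawn or pickup, with a particle effect.
; BeginPlay creates an IntelMarker dynamic material for view radius, then feeds highlight color and bounds-based size into the particle system.
; A looping timer, rather than Tick, pushes the target's location into that material.
; It destroys itself when the target becomes invalid, searched, dead or picked up; debug prints remain.
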(event EventBeginPlay
  (bind (_origin _boxextent _sphereradius) (Collision|GetComponentBounds (Variables|/Script/Engine.Actor|GetRootComponent (Variables|Default|GetItem%20to%20Highlight))))
  (bind _returnvalue (Actor|GetOwner))
  (bind _item_to_highlight (Variables|Default|GetItem%20to%20Highlight))
  (bind _particlesystemcomponent (Variables|Particles|GetParticleSystemComponent))
  (Development|PrintString 0 (Utilities|GetDisplayName _returnvalue))
  (Variables|Default|SetItem%20to%20Highlight _returnvalue)
  (Utilities|IsValid _item_to_highlight
    (:Is%20Valid
      ;[0.4283 0.4172 1] Set VFX MID
      (bind _returnvalue_2 (Rendering|Material|CreateDynamicMaterialInstance _particlesystemcomponent 0 "/FNGameplayCues/Athena/Items/Gameplay/BackPacks/IntelPack/FX/M_IntelMarker.M_IntelMarker" "None" :!class /Script/Engine.PrimitiveComponent))
      (bind _output_get (Variables|Default|SetIntelFX_MID _returnvalue_2))
      (Effects|Components|ParticleSystem|SetMaterialParameter _particlesystemcomponent "IntelMID" _output_get)
      (Rendering|Material|SetScalarParameterValue _output_get "Radius" (Variables|Default|GetView%20Radius) :!class /Script/Engine.MaterialInstanceDynamic)
      (Effects|Components|ParticleSystem|SetVectorParameter _particlesystemcomponent "ItemTypeColor" (Math|Conversions|ToVector(LinearColor) (Variables|Default|GetHighlight%20Color)))
      (Effects|Components|ParticleSystem|SetVectorParameter _particlesystemcomponent "SizeScale" (Math|Conversions|ToVector(Float) (/ _sphereradius 1e+02)))
      (Utilities|Time|SetTimerbyEvent (delegate Custom|Set%20Pawn%20Location) 0.1 true))
    (:Is%20Not%20Valid
      (Actor|DestroyActor))))

(event Custom|Set%20Pawn%20Location
  (bind _item_to_highlight (Variables|Default|GetItem%20to%20Highlight))
  (Utilities|IsValid _item_to_highlight
    (:Is%20Valid
      (Utilities|FlowControl|Sequence
        (:then_0
          ;[0.2657 1 0.286] Different conditions besides being invalid should make this visually turn off
          (bind _asbuilding_container (Utilities|Casting|CastToBuildingContainer _item_to_highlight))
          (if (Variables|/Script/FortniteGame.BuildingContainer|GetbAlreadySearched _asbuilding_container)
            _))
        (:then_1
          (bind _asfort_pawn (Utilities|Casting|CastToFortPawn _item_to_highlight))
          (if (FortPawn|IsDead _asfort_pawn)
            _))
        (:then_2
          (bind _asfort_pickup (Utilities|Casting|CastToFortPickup _item_to_highlight))
          (if (Pickup|PickedUp _asfort_pickup)
            _))
        (:then_3
          (Rendering|Material|SetVectorParameterValue (Variables|Default|GetIntelFX_MID) "PawnLocation" (Math|Conversions|ToLinearColor(Vector) (Transformation|GetActorLocation _item_to_highlight)) :!class /Script/Engine.MaterialInstanceDynamic)))
      (Actor|DestroyActor))
    (:Is%20Not%20Valid
      (Actor|DestroyActor))))

(event EventEndPlay
  (params (EndPlayReason :type EEndPlayReason%20Enum :default "Destroyed"))
  (Development|PrintString 0 "stop"))

(event Collision|EventActorBeginOverlap
  (params (OtherActor :type Actor%20Object%20Reference)))

(event EventTick
  (params (DeltaSeconds :type Float%20%28single-precision%29)))

(fn ConstructionScript)
