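; Looping gameplay cue that plays enter and exit sounds for a player pawn moving through a jungle bush.
; On burst, it adds its foliage sound library to the target pawn through the SoundLibrary subsystem, then plays the tagged bush enter sound.
; On cease relevant, it plays the bush exit sound before removing that library, so the library is attached only while the cue is relevant.
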(fn OnBurst
  (params (MyTarget :type Actor%20Object%20Reference) (Parameters :type Gameplay%20Cue%20Parameters%20Structure%20%28by%20ref%29))
  (bind _returnvalue (WorldSubsystems|GetSoundLibrarySubsystem MyTarget))
  (Utilities|IsValid MyTarget
    (:Is%20Valid
      (bind _asfort_player_pawn_athena (Utilities|Casting|CastToFortPlayerPawnAthena MyTarget))
      (SoundLibrary|AddLibrary _returnvalue _asfort_player_pawn_athena (Variables|Default|GetFoliageSoundLibrary) :!class /Script/SoundLibrary.SoundLibrarySubsystem)
      (SoundLibrary|PlaySoundSimple _returnvalue _asfort_player_pawn_athena (Utilities|Struct|MakeSimpleSoundSettings "(TagName=\"SoundLibrary.Player.State.Bush.Enter\")" false false true false 0.4) :!class /Script/SoundLibrary.SoundLibrarySubsystem)
      (return false))))

(fn OnCeaseRelevant
  (params (MyTarget :type Actor%20Object%20Reference) (Parameters :type Gameplay%20Cue%20Parameters%20Structure%20%28by%20ref%29))
  (bind _returnvalue (WorldSubsystems|GetSoundLibrarySubsystem MyTarget))
  (Utilities|IsValid MyTarget
    (:Is%20Valid
      (bind _asfort_player_pawn_athena (Utilities|Casting|CastToFortPlayerPawnAthena MyTarget))
      (SoundLibrary|PlaySoundSimple _returnvalue _asfort_player_pawn_athena (Utilities|Struct|MakeSimpleSoundSettings "(TagName=\"SoundLibrary.Player.State.Bush.Exit\")" false false true false 0.4) :!class /Script/SoundLibrary.SoundLibrarySubsystem)
      (SoundLibrary|RemoveLibrary _returnvalue _asfort_player_pawn_athena (Variables|Default|GetFoliageSoundLibrary) :!class /Script/SoundLibrary.SoundLibrarySubsystem)
      (return false))))

(fn ConstructionScript)
