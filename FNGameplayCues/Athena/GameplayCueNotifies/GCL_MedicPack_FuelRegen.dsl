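; Looping gameplay cue for MedicPack fuel regeneration, driving backpack material and audio feedback on the player pawn while fuel recharges.
; On becoming relevant it caches max fuel and reserve attributes, enables tick, and spawns a ramp sound.
; Tick pushes fuel percentages to backpack materials and eases ramp volume and pitch along a curve.
; On ceasing, it fades the ramp, playing a recharged sound and timeline blink if fuel is nearly full.
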(event EventTick
  (params (DeltaSeconds :type Float%20%28single-precision%29))
  (bind _playerpawn (Variables|Default|GetPlayerPawn))
  (bind _returnvalue (Customization|GetSkeletalMeshforPartType _playerpawn "Backpack" :!class /Script/FortniteGame.FortPlayerPawn))
  ;[1 1 1] Tick
  (Utilities|IsValid _playerpawn
    (:Is%20Valid
      (Rendering|Material|SetScalarParameterValueonMaterials _returnvalue "FuelLevel" (CallFunction|GetFuelPercent))
      (Rendering|Material|SetScalarParameterValueonMaterials _returnvalue "ReserveFuelLevel" (CallFunction|GetReserveFuelPercent))
      (CallFunction|AudioTick))))

(event Custom|Start%20Fuel%20Blink%20FX
  (bind _playerpawn (Variables|Default|GetPlayerPawn))
  (bind _returnvalue (Customization|GetSkeletalMeshforPartType _playerpawn "Backpack" :!class /Script/FortniteGame.FortPlayerPawn))
  ;[1 1 1] Fuel Recharge "Blink"
  (if (>= (CallFunction|GetFuelPercent) 0.95)
    (Utilities|FlowControl|DoOnce false
      (:Completed
        (bind (_direction _blink) (Timeline|FullBlinkTimeline PlayFromStart)
          (:Update
            (Rendering|Material|SetScalarParameterValueonMaterials _returnvalue "FullBlink" _blink)
            (Rendering|Material|SetScalarParameterValueonMaterials _returnvalue "Refueling" 0.0)))))))

(event Custom|ResetBlink
  (bind _playerpawn (Variables|Default|GetPlayerPawn))
  (bind _returnvalue (Customization|GetSkeletalMeshforPartType _playerpawn "Backpack" :!class /Script/FortniteGame.FortPlayerPawn))
  ;[1 1 1] Fuel Recharge "Blink"
  (Utilities|FlowControl|DoOnce false
    (:Completed
      (Timeline|FullBlinkTimeline PlayFromStart))))

(event Custom|Audio%20Tick
  ;[0 0 0] Audio Land
  (Utilities|IsValid (Variables|Default|GetPlayerPawn)
    (:Is%20Valid
      (if (not (Variables|Sounds|GetbPlayedFuelRecharge))
        (Utilities|IsValid (Variables|Default|GetOnFuelRechargeRamp_Comp)
          (:Is%20Valid
            (if (Audio|Components|Audio|IsPlaying (Variables|Default|GetOnFuelRechargeRamp_Comp))
              (Variables|Sounds|SetFuelRechargeRampVolumeMultiplier (Math|Interpolation|FInterpTo (Variables|Sounds|GetFuelRechargeRampVolumeMultiplier) (Math|Curves|GetFloatValue (Variables|Sounds|GetFuelRechargeRampVolumeCurve) (CallFunction|GetFuelPercent)) (Utilities|Time|GetWorldDeltaSeconds (Variables|Default|GetPlayerPawn)) 5.0))
              (else
                (Audio|Components|Audio|Play (Variables|Default|GetOnFuelRechargeRamp_Comp) 0.0)))
            (Audio|Parameter|SetFloatParameter (Variables|Default|GetOnFuelRechargeRamp_Comp) "Fader" (Variables|Sounds|GetFuelRechargeRampVolumeMultiplier) :!class /Script/AudioExtensions.AudioParameterControllerInterface)
            (Audio|Components|Audio|SetPitchMultiplier (Variables|Default|GetOnFuelRechargeRamp_Comp) (Math|Float|MapRangeUnclamped (Variables|Sounds|GetFuelRechargeRampVolumeMultiplier) 0.0 1.0 1.0 1.25))))))))

(fn CacheAttributes
  (bind _reservemaxfuelattribute (LocalVariables|GetReserveMaxFuelAttribute))
  (bind _playerpawn (Variables|Default|GetPlayerPawn))
  (bind (_bsuccessfullyfoundattribute _returnvalue) (Ability|Attribute|GetFloatAttribute _playerpawn _reservemaxfuelattribute))
  (bind _maxfuelattribute (LocalVariables|GetMaxFuelAttribute))
  (bind _playerpawn_2 (Variables|Default|GetPlayerPawn))
  (bind (_bsuccessfullyfoundattribute_2 _returnvalue_2) (Ability|Attribute|GetFloatAttribute _playerpawn_2 _maxfuelattribute))
  (Utilities|IsValid (Variables|Default|GetPlayerPawn)
    (:Is%20Valid
      (if _bsuccessfullyfoundattribute_2
        (Variables|Attributes|SetFuelMax _returnvalue_2)
        (else
          (Variables|Attributes|SetFuelMax (GadgetAttributeSets|GetSimulatedAttributeValue (Variables|Default|GetPlayerPawn) (LocalVariables|GetMaxFuelAttribute) 1.0 :!class /Script/FortniteGame.FortPlayerPawn))))
      (if _bsuccessfullyfoundattribute
        (Variables|Attributes|SetReserveFuelMax _returnvalue)
        (else
          (Variables|Attributes|SetReserveFuelMax (GadgetAttributeSets|GetSimulatedAttributeValue (Variables|Default|GetPlayerPawn) (LocalVariables|GetReserveMaxFuelAttribute) 1.0 :!class /Script/FortniteGame.FortPlayerPawn)))))
    (:Is%20Not%20Valid
      (Variables|Attributes|SetFuelMax 1.0)
      (Variables|Attributes|SetReserveFuelMax 1.0))))

(fn GetFuel
  (bind _fuelattribute (LocalVariables|GetFuelAttribute))
  (bind _playerpawn (Variables|Default|GetPlayerPawn))
  (bind (_bsuccessfullyfoundattribute _returnvalue) (Ability|Attribute|GetFloatAttribute _playerpawn _fuelattribute))
  (Utilities|IsValid (Variables|Default|GetPlayerPawn)
    (:Is%20Valid
      (if _bsuccessfullyfoundattribute
        (return _returnvalue)
        (else
          (return (GadgetAttributeSets|GetSimulatedAttributeValue (Variables|Default|GetPlayerPawn) (LocalVariables|GetFuelAttribute) 0.0 :!class /Script/FortniteGame.FortPlayerPawn)))))
    (:Is%20Not%20Valid
      (return 0.0))))

(fn GetFuelPercent
  (return (/ (CallFunction|GetFuel) (Variables|Attributes|GetFuelMax))))

(fn GetReserveFuel
  (bind _reservefuelattribute (LocalVariables|GetReserveFuelAttribute))
  (bind _playerpawn (Variables|Default|GetPlayerPawn))
  (bind (_bsuccessfullyfoundattribute _returnvalue) (Ability|Attribute|GetFloatAttribute _playerpawn _reservefuelattribute))
  (Utilities|IsValid (Variables|Default|GetPlayerPawn)
    (:Is%20Valid
      (if _bsuccessfullyfoundattribute
        (return _returnvalue)
        (else
          (return (GadgetAttributeSets|GetSimulatedAttributeValue (Variables|Default|GetPlayerPawn) (LocalVariables|GetReserveFuelAttribute) 0.0 :!class /Script/FortniteGame.FortPlayerPawn)))))
    (:Is%20Not%20Valid
      (return 0.0))))

(fn GetReserveFuelPercent
  (return (/ (CallFunction|GetReserveFuel) (Variables|Attributes|GetReserveFuelMax))))

(fn OnCeaseRelevant
  (params (MyTarget :type Actor%20Object%20Reference) (Parameters :type Gameplay%20Cue%20Parameters%20Structure%20%28by%20ref%29))
  (Utilities|IsValid (Variables|Default|GetPlayerPawn)
    (:Is%20Valid
      (Rendering|Material|SetScalarParameterValueonMaterials (Customization|GetSkeletalMeshforPartType (Variables|Default|GetPlayerPawn) "Backpack" :!class /Script/FortniteGame.FortPlayerPawn) "Refueling" 0.0))
    (:Is%20Not%20Valid
      _))
  (Actor|Tick|SetActorTickEnabled false)
  (CallFunction|SetFuelRechargeAudioEnabled false)
  (CallFunction|StartFuelBlinkFX)
  (return true))

(fn SetFuelRechargeAudioEnabled
  (params (Enabled :type Boolean))
  (bind _output_get (Variables|Sounds|SetbFuelRechargeAudioEnabled Enabled))
  (if _output_get
    (Variables|Sounds|SetbPlayedFuelRecharge false)
    (Utilities|IsValid (Variables|Default|GetPlayerPawn)
      (:Is%20Valid
        (bind _returnvalue (Audio|SpawnSoundAttached (Variables|Sounds|GetSoundOnFuelChargeRamp) (Variables|/Script/Engine.Actor|GetRootComponent (Variables|Default|GetPlayerPawn))))
        (Variables|Default|SetOnFuelRechargeRamp_Comp _returnvalue)
        (Variables|Sounds|SetFuelRechargeRampVolumeMultiplier 0.0)))
    (else
      (Utilities|FlowControl|Sequence
        (:then_0
          (bind _onfuelrechargeramp_comp (Variables|Default|GetOnFuelRechargeRamp_Comp)
            (:then
              (Audio|Components|Audio|FadeOut _onfuelrechargeramp_comp 0.2 0.0 "Linear" :!class /Script/Engine.AudioComponent))
            (:else)))
        (:then_1
          (if (>= (CallFunction|GetFuelPercent) 0.95)
            (Utilities|IsValid (Variables|Default|GetPlayerPawn)
              (:Is%20Valid
                (Audio|SpawnSoundAttached (Variables|Sounds|GetSoundOnFuelRecharged) (Variables|/Script/Engine.Actor|GetRootComponent (Variables|Default|GetPlayerPawn)))))))))))

(fn ConstructionScript)

(fn OnBecomeRelevant
  (params (MyTarget :type Actor%20Object%20Reference) (Parameters :type Gameplay%20Cue%20Parameters%20Structure%20%28by%20ref%29))
  (bind _playerpawn (Variables|Default|GetPlayerPawn))
  (bind _returnvalue (Customization|GetSkeletalMeshforPartType _playerpawn "Backpack" :!class /Script/FortniteGame.FortPlayerPawn))
  (bind _asplayer_pawn_athena (Utilities|Casting|CastToPlayerPawn_Athena MyTarget)
    (:then
      (Variables|Default|SetPlayerPawn _asplayer_pawn_athena)
      (CallFunction|CacheAttributes)
      (Actor|Tick|SetActorTickEnabled true)
      (CallFunction|SetFuelRechargeAudioEnabled true)
      (Rendering|Material|SetScalarParameterValueonMaterials _returnvalue "Refueling" 1.0)
      (Rendering|Material|SetScalarParameterValueonMaterials _returnvalue "FullBlink" 1.0)
      (CallFunction|ResetBlink)
      (return true))
    (:CastFailed
      (return false))))
