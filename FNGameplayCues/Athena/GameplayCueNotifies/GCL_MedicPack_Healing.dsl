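; Looping gameplay cue named for MedicPack healing that drives fuel-based backpack effects and audio, using jetpack logic.
; On activation it attaches an emitter to the player's backpack mesh, sets material parameters and starts sound loops.
; Tick reads fuel attributes, with simulated fallback, to scale particles and material, crossfade loops into sputter and play a local low-fuel beep.
; Removal deactivates particles and audio; a landing event handler resets the first-jump sound flag.
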(event Custom|On%20Pawn%20Landed
  (params (Hit :type Hit%20Result%20Structure%20%28by%20ref%29))
  ;[1 1 1] Reset "First-Jump" flag once we've landed
  (Variables|Sounds|SetbIsFirstJump true)
  (EventDispatchers|UnbindEventfromOnPawnLanded (Variables|Default|GetPlayerPawn) (delegate Custom|On%20Pawn%20Landed)))

(event EventTick
  (params (DeltaSeconds :type Float%20%28single-precision%29))
  (bind _cachedfuelpercent (Variables|Default|GetCachedFuelPercent))
  (bind _returnvalue (* "1.000000,1.000000,1.000000" _cachedfuelpercent))
  (bind _spawnedparticleemitter (Variables|Default|GetSpawnedParticleEmitter))
  (Utilities|IsValid (Variables|Default|GetPlayerPawn)
    (:Is%20Valid
      (CallFunction|CalculateFuelPercentage)
      (CallFunction|CacheRenderingDetailMode)
      (CallFunction|AudioTick)
      ;[1 1 1] FX Control
      (Rendering|Material|SetScalarParameterValueonMaterials (Customization|GetSkeletalMeshforPartType (Variables|Default|GetPlayerPawn) "Backpack" :!class /Script/FortniteGame.FortPlayerPawn) "FuelLevel" _cachedfuelpercent)
      (Utilities|IsValid (Variables|Default|GetSpawnedParticleEmitter)
        (:Is%20Valid
          (Effects|Components|ParticleSystem|SetVectorParameter _spawnedparticleemitter "FlameScale" _returnvalue)
          (Effects|Components|ParticleSystem|SetVectorParameter _spawnedparticleemitter "SpriteFlame" _returnvalue)
          (Effects|Components|ParticleSystem|SetFloatParameter _spawnedparticleemitter "SpawnRate" _cachedfuelpercent))))))

(event Custom|Audio%20Tick
  ;[0 0 0] Audio land
  (Utilities|IsValid (Variables|Default|GetPlayerPawn)
    (:Is%20Valid
      (Utilities|FlowControl|Sequence
        (:then_0
          ;;[1 1 1] Modulate volume of idle/accel loops, and drop pitch of idle/accel loops based on fuel remaining
          (Audio|Components|Audio|SetVolumeMultiplier (Variables|Default|GetOnAccelAudio_Comp) (* (Variables|Default|GetScaleMultiplier) (- 1.0 (Variables|Sounds|GetCurrentFuelVolumeMultiplier))) :!class /Script/Engine.AudioComponent)
          (Audio|Components|Audio|SetVolumeMultiplier (Variables|Default|GetOnIdleAudio_Comp) (- 1.0 (Variables|Sounds|GetCurrentFuelVolumeMultiplier)) :!class /Script/Engine.AudioComponent)
          (Audio|Components|Audio|SetPitchMultiplier (Variables|Default|GetOnAccelAudio_Comp) (- 1.0 (Math|Float|MapRangeUnclamped (Variables|Sounds|GetCurrentFuelVolumeMultiplier) 0.0 1.0 0.0 0.5))))
        (:then_1
          ;;[1 1 1] Gradually ramp up the sputter audio loop volume
          (bind (_direction _fuel) (Timeline|Jetpack%20Audio%20Modulation SetNewTime :NewTime (- 1.0 (Variables|Default|GetCachedFuelPercent)))
            (:Update
              (bind _onoutoffuelaudio_comp (Variables|Default|GetOnOutOfFuelAudio_Comp)
                (:then
                  (bind _output_get (Variables|Sounds|SetTargetFuelVolumeMultiplier _fuel))
                  (if (not (and (> _output_get 0.0) (Audio|Components|Audio|IsPlaying _onoutoffuelaudio_comp)))
                    (Audio|Components|Audio|Play (Variables|Default|GetOnOutOfFuelAudio_Comp) 0.0)))
                (:else)))))
        (:then_2
          ;;[1 1 1] Fade in "sputter" audio loop based on fuel percentage
          (bind _output_get_2 (Variables|Sounds|SetCurrentFuelVolumeMultiplier (Math|Interpolation|FInterpTo (Variables|Sounds|GetCurrentFuelVolumeMultiplier) (Variables|Sounds|GetTargetFuelVolumeMultiplier) (Utilities|Time|GetWorldDeltaSeconds (Variables|Default|GetPlayerPawn)) 1e+01)))
          (Audio|Components|Audio|SetVolumeMultiplier (Variables|Default|GetOnOutOfFuelAudio_Comp) _output_get_2 :!class /Script/Engine.AudioComponent))
        (:then_3
          ;;[1 1 1] Play a warning beep when < 20% fuel
          (if (Math|Boolean|ANDBoolean (<= (Variables|Default|GetCachedFuelPercent) 0.1) (not (Variables|Sounds|GetbPlayedFuelWarning)) (Pawn|IsLocallyControlled (Variables|Default|GetPlayerPawn)) (>= (- (Utilities|Time|GetGameTimeinSeconds (Variables|Default|GetPlayerPawn)) (Variables|Default|GetJetpackActivatedTime)) 1.0))
            (Audio|PlaySound2D (Variables|Default|GetPlayerPawn) (Variables|Sounds|GetSoundOnFuelWarning))
            (Variables|Sounds|SetbPlayedFuelWarning true)))))))

(fn CacheAttributes
  (Variables|Sounds|SetCurrentFuelVolumeMultiplier 0.0)
  (Variables|Sounds|SetTargetFuelVolumeMultiplier 0.0))

(fn CacheRenderingDetailMode
  (Variables|Default|SetCachedRenderingDetailMode (Rendering|GetRenderingDetailMode)))

(fn CalculateFuelPercentage
  (Variables|Default|SetCachedFuelPercent (HelperMethods|GetFuelPercent)))

(fn GetFuel
  (bind _fuelattribute (LocalVariables|GetFuelAttribute))
  (bind _playerpawn (Variables|Default|GetPlayerPawn))
  (bind (_bsuccessfullyfoundattribute _returnvalue) (Ability|Attribute|GetFloatAttribute _playerpawn _fuelattribute))
  (Utilities|IsValid (Variables|Default|GetPlayerPawn)
    (:Is%20Valid
      (if _bsuccessfullyfoundattribute
        (return _returnvalue)
        (else
          (return (GadgetAttributeSets|GetSimulatedAttributeValue (Variables|Default|GetPlayerPawn) (LocalVariables|GetFuelAttribute) 0.0 :!class /Script/FortniteGame.FortPlayerPawn)))))
    (:Is%20Not%20Valid
      (return 0.0))))

(fn GetFuelPercent
  (return (/ (AttributeMethods|GetFuel) (AttributeMethods|GetMaxFuel))))

(fn GetMaxFuel
  (bind _maxfuelattribute (LocalVariables|GetMaxFuelAttribute))
  (bind _playerpawn (Variables|Default|GetPlayerPawn))
  (bind (_bsuccessfullyfoundattribute _returnvalue) (Ability|Attribute|GetFloatAttribute _playerpawn _maxfuelattribute))
  (Utilities|IsValid (Variables|Default|GetPlayerPawn)
    (:Is%20Valid
      (if _bsuccessfullyfoundattribute
        (return _returnvalue)
        (else
          (return (GadgetAttributeSets|GetSimulatedAttributeValue (Variables|Default|GetPlayerPawn) (LocalVariables|GetMaxFuelAttribute) 1.0 :!class /Script/FortniteGame.FortPlayerPawn)))))
    (:Is%20Not%20Valid
      (return 1.0))))

(fn OnCeaseRelevant
  (params (MyTarget :type Actor%20Object%20Reference) (Parameters :type Gameplay%20Cue%20Parameters%20Structure%20%28by%20ref%29))
  (EventDispatchers|UnbindEventfromOnPawnLanded (Variables|Default|GetPlayerPawn) (EventDispatchers|CreateEvent :!event On%20Pawn%20Landed))
  (Utilities|FlowControl|Sequence
    (:then_0
      (Utilities|IsValid (Variables|Default|GetSpawnedParticleEmitter)
        (:Is%20Valid
          (Components|Activation|Deactivate (Variables|Default|GetSpawnedParticleEmitter)))))
    (:then_1
      (HelperMethods|SetJetpackAudioEnabled false))
    (:then_2
      (return true))))

(fn SetJetpackAudioEnabled
  (params (Enabled :type Boolean))
  (Variables|Sounds|SetbJetpackAudioEnabled Enabled)
  (Actor|Tick|SetActorTickEnabled (Variables|Sounds|GetbJetpackAudioEnabled))
  (Variables|Sounds|SetbPlayedFuelWarning false)
  (Utilities|IsValid (Variables|Default|GetPlayerPawn)
    (:Is%20Valid
      (if (Variables|Sounds|GetbJetpackAudioEnabled)
        (Variables|Default|SetJetpackActivatedTime (Utilities|Time|GetGameTimeinSeconds (Variables|Default|GetPlayerPawn)))
        (EventDispatchers|BindEventtoOnPawnLanded (Variables|Default|GetPlayerPawn) (EventDispatchers|CreateEvent :!event On%20Pawn%20Landed))
        (Audio|SpawnSoundAttached (select (Variables|Sounds|GetbIsFirstJump) (Variables|Sounds|GetSoundOnStart) (Variables|Sounds|GetSoundOnThrust)) (Variables|/Script/Engine.Actor|GetRootComponent (Variables|Default|GetPlayerPawn)))
        (Variables|Sounds|SetbIsFirstJump false)
        (Utilities|FlowControl|Sequence
          (:then_0
            (bind _returnvalue (Audio|SpawnSoundAttached :Sound (Variables|Sounds|GetSoundLoopOnIdle) :AttachToComponent (Variables|/Script/Engine.Actor|GetRootComponent (Variables|Default|GetPlayerPawn)) :AttachPointName "None" :LocationType "KeepRelativeOffset" :bStopWhenAttachedToDestroyed false :VolumeMultiplier 1.0 :PitchMultiplier 1.0 :StartTime (Math|Random|RandomFloatinRange 0.0 7.4) :bAutoDestroy true))
            (Variables|Default|SetOnIdleAudio_Comp _returnvalue))
          (:then_1
            (bind _returnvalue_2 (Audio|SpawnSoundAttached :Sound (Variables|Sounds|GetSoundLoopOnAccel) :AttachToComponent (Variables|/Script/Engine.Actor|GetRootComponent (Variables|Default|GetPlayerPawn)) :AttachPointName "None" :LocationType "KeepRelativeOffset" :bStopWhenAttachedToDestroyed false :VolumeMultiplier 1.0 :PitchMultiplier 1.0 :StartTime (Math|Random|RandomFloatinRange 0.0 5.0) :bAutoDestroy true))
            (Variables|Default|SetOnAccelAudio_Comp _returnvalue_2))
          (:then_2
            (bind _returnvalue_3 (Audio|SpawnSoundAttached :Sound (Variables|Sounds|GetSoundLoopOnOutOfFuel) :AttachToComponent (Variables|/Script/Engine.Actor|GetRootComponent (Variables|Default|GetPlayerPawn)) :AttachPointName "None" :LocationType "KeepRelativeOffset" :bStopWhenAttachedToDestroyed false :VolumeMultiplier 0.0 :PitchMultiplier 1.0 :StartTime (Math|Random|RandomFloatinRange 0.0 5.5) :bAutoDestroy true))
            (Variables|Default|SetOnOutOfFuelAudio_Comp _returnvalue_3)))
        (else
          _)))
    (:Is%20Not%20Valid
      _))
  (Audio|Components|Audio|FadeOut (Variables|Default|GetOnIdleAudio_Comp) 0.25 0.0 "Linear" :!class /Script/Engine.AudioComponent)
  (Audio|Components|Audio|FadeOut (Variables|Default|GetOnOutOfFuelAudio_Comp) 0.25 0.0 "Linear" :!class /Script/Engine.AudioComponent)
  (Utilities|IsValid (Variables|Default|GetPlayerPawn)
    (:Is%20Valid
      (Audio|SpawnSoundAttached (select (and (>= (Variables|Sounds|GetTargetFuelVolumeMultiplier) 0.95) (> (- (Utilities|Time|GetGameTimeinSeconds (Variables|Default|GetPlayerPawn)) (Variables|Default|GetJetpackActivatedTime)) 4.0)) (Variables|Sounds|GetSoundOnOutOfFuel) (Variables|Sounds|GetSoundOnStop)) (Variables|/Script/Engine.Actor|GetRootComponent (Variables|Default|GetPlayerPawn))))))

(fn ConstructionScript)

(fn OnBecomeRelevant
  (params (MyTarget :type Actor%20Object%20Reference) (Parameters :type Gameplay%20Cue%20Parameters%20Structure%20%28by%20ref%29))
  (bind _asplayer_pawn_athena (Utilities|Casting|CastToPlayerPawn_Athena MyTarget)
    (:then
      (bind _output_get (Variables|Default|SetPlayerPawn _asplayer_pawn_athena))
      (Utilities|IsValid (Customization|GetSkeletalMeshforPartType _output_get "Backpack" :!class /Script/FortniteGame.FortPlayerPawn)
        (:Is%20Valid
          (bind _returnvalue (Effects|SpawnEmitterAttached 0 (Customization|GetSkeletalMeshforPartType _output_get "Backpack" :!class /Script/FortniteGame.FortPlayerPawn) "Root"))
          (Variables|Default|SetSpawnedParticleEmitter _returnvalue)
          (HelperMethods|SetJetpackAudioEnabled true)
          (HelperMethods|CacheAttributes)
          (Rendering|Material|SetScalarParameterValueonMaterials (Customization|GetSkeletalMeshforPartType _output_get "Backpack" :!class /Script/FortniteGame.FortPlayerPawn) "FullBlink" 1.0)
          (Rendering|Material|SetScalarParameterValueonMaterials (Customization|GetSkeletalMeshforPartType _output_get "Backpack" :!class /Script/FortniteGame.FortPlayerPawn) "Refueling" 0.0)
          (return true))
        (:Is%20Not%20Valid
          (return false))))
    (:CastFailed
      (return false))))
