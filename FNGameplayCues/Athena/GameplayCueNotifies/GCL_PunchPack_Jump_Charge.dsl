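; Looping gameplay cue for the Punch Pack jump charge, providing charge-finished feedback on the target Athena player pawn.
; OnBecomeRelevant stores the pawn and calls ChargeDelay, a custom event built on a timed delay and gate that leads to an attached sound and emitter, camera shake and force feedback.
; OnCeaseRelevant calls ChargeDelay with Stop set, returning true so the cue is destroyed for anim blueprint blends.
; EndPlay fades out the finish sound.
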(event Custom|ChargeDelay
  (params (Stop :type Boolean))
  (bind _targetplayer (Variables|Default|GetTargetPlayer))
  (bind _returnvalue (Transformation|GetActorLocation _targetplayer))
  (if (not Stop)
    (Utilities|FlowControl|Sequence
      (:then_0
        _)
      (:then_1
        (Utilities|FlowControl|Delay 0 3.0))))
  (Utilities|FlowControl|Gate true
    (:Exit
      (bind _returnvalue_2 (Audio|SpawnSoundAttached (Variables|Default|GetChargeFinishSound) (Variables|Particles|GetParticleSystemComponent)))
      (Camera|PlayWorldCameraShake 0 "/Game/Effects/Fort_Effects/BluePrints/CameraShakes/B_Small_Vertical_Jolt_CameraShake.B_Small_Vertical_Jolt_CameraShake_C" _returnvalue 4e+02 1.2e+03)
      (Effects|SpawnEmitterAttached :EmitterTemplate (Variables|Default|GetChargeFinish) :AttachToComponent (Variables|/Script/Engine.Character|GetMesh _targetplayer) :AttachPointName "root" :Scale "1.000000,1.000000,1.000000" :LocationType "SnapToTarget" :bAutoDestroy true :PoolingMethod "None" :bAutoActivate true)
      (Fortnite|PlayLocalForceFeedbackatLocation 0 "/Game/Weapons/ForceFeedback/FF_Generic_MeleeImpact.FF_Generic_MeleeImpact" _returnvalue 3e+02))))

(event EventBeginPlay
  (Development|PrintString self "start jump FX" false false))

(event EventEndPlay
  (params (EndPlayReason :type EEndPlayReason%20Enum :default "Destroyed"))
  (Development|PrintString self "end charge" false false)
  (Audio|Components|Audio|FadeOut (stale Custom|ChargeDelay _returnvalue_2) 0.3 0.0 "Linear" :!class /Script/Engine.AudioComponent))

(event Collision|EventActorBeginOverlap
  (params (OtherActor :type Actor%20Object%20Reference)))

(event EventTick
  (params (DeltaSeconds :type Float%20%28single-precision%29)))

(fn OnCeaseRelevant
  (params (MyTarget :type Actor%20Object%20Reference) (Parameters :type Gameplay%20Cue%20Parameters%20Structure%20%28by%20ref%29))
  ;[1 1 1] This is to ensure that the GCL is actually destroyed after every use so the ANIM BP can fire blends correctly.
  (CallFunction|ChargeDelay true)
  (return true))

(fn OnStartParticleSystemSpawned
  (params (SpawnedParticleSysComponent :type Cascade%20Particle%20System%20Component%20Object%20Reference) (Parameters :type Gameplay%20Cue%20Parameters%20Structure%20%28by%20ref%29))
  (return))

(fn OnWhileActiveParticleSystemDeactivate
  (params (WhileActiveParticleSysComponent :type Cascade%20Particle%20System%20Component%20Object%20Reference) (Parameters :type Gameplay%20Cue%20Parameters%20Structure%20%28by%20ref%29))
  (return))

(fn ConstructionScript)

(fn OnBecomeRelevant
  (params (MyTarget :type Actor%20Object%20Reference) (Parameters :type Gameplay%20Cue%20Parameters%20Structure%20%28by%20ref%29))
  (bind _asfort_player_pawn_athena (Utilities|Casting|CastToFortPlayerPawnAthena MyTarget))
  (Variables|Default|SetTargetPlayer _asfort_player_pawn_athena)
  (CallFunction|ChargeDelay false)
  (return false))
