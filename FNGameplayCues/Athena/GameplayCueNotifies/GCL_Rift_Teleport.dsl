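; Looping gameplay cue that plays player feedback for a Rift teleport when it becomes relevant on a Fort player pawn.
; It attaches a Niagara glow to the pawn's mesh, then triggers a crack-exit camera lens effect and a portal-warp camera shake.
; A helper function plays the teleport sound in 2D only for a locally viewed pawn.
; A driven-vehicle branch ends at a validity check with no action, and OnCeaseRelevant does no cleanup.
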(fn OnCeaseRelevant
  (params (MyTarget :type Actor%20Object%20Reference) (Parameters :type Gameplay%20Cue%20Parameters%20Structure%20%28by%20ref%29))
  (return true))

(fn OnStartParticleSystemSpawned
  (params (SpawnedParticleSysComponent :type Cascade%20Particle%20System%20Component%20Object%20Reference) (Parameters :type Gameplay%20Cue%20Parameters%20Structure%20%28by%20ref%29))
  (return))

(fn OnWhileActiveParticleSystemDeactivate
  (params (WhileActiveParticleSysComponent :type Cascade%20Particle%20System%20Component%20Object%20Reference) (Parameters :type Gameplay%20Cue%20Parameters%20Structure%20%28by%20ref%29))
  (return))

(fn PlayTeleportSoundLocal
  (params (PlayerPawn :type Fort%20Player%20Pawn%20Object%20Reference))
  (Utilities|IsValid PlayerPawn
    (:Is%20Valid
      (if (Pawn|IsLocallyViewed PlayerPawn)
        (Audio|PlaySound2D PlayerPawn (Variables|Default|GetTeleportSoundLocal))))
    (:Is%20Not%20Valid
      _))
  (return))

(fn ConstructionScript)

(fn OnBecomeRelevant
  (params (MyTarget :type Actor%20Object%20Reference) (Parameters :type Gameplay%20Cue%20Parameters%20Structure%20%28by%20ref%29))
  (Utilities|FlowControl|Sequence
    (:then_0
      (bind _asfort_player_pawn (Utilities|Casting|CastToFortPlayerPawn MyTarget))
      (Utilities|FlowControl|Sequence
        (:then_0
          (Niagara|SpawnSystemAttached "/FNGameplayCues/Athena/Items/ForagedItems/Rift/FX/NS_Athena_Rift_Player_Glow.NS_Athena_Rift_Player_Glow" (Variables|/Script/Engine.Character|GetMesh _asfort_player_pawn) "None" "0, 0, 0" "0, 0, 0" "KeepRelativeOffset" true)
          (CallFunction|PlayTeleportSoundLocal _asfort_player_pawn)
          (Feedback|SpawnGenericCameraLensEffect _asfort_player_pawn "/FNGameplayCues/Athena/Items/ForagedItems/Rift/FX/BP_CameraLens_CrackExit.BP_CameraLens_CrackExit_C")
          (Game|Feedback|ClientStartCameraShake :self (Pawn|GetLocalViewingPlayerController _asfort_player_pawn) :Shake "/FNGameplayCues/Effects/Fort_Effects/Camera_Effects/BP_CameraShake_PortalWarp.BP_CameraShake_PortalWarp_C" :Scale 2.0 :PlaySpace "CameraLocal"))
        (:then_1
          (if (Vehicle|IsDrivingVehicle _asfort_player_pawn)
            (Utilities|IsValid (Utilities|ToObject(Interface) (Vehicle|GetVehicleInterface _asfort_player_pawn)))))))
    (:then_1
      (return false))))
