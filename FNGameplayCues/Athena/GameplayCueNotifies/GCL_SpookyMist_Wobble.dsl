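; Looping gameplay cue for the Spooky Mist wobble, making damageable building pieces around the affected player pawn bounce.
; OnBurst caches the pawn, runs WiggleActors at once and repeats it on a short looping timer.
; Each pass sphere-overlaps building actors at a height offset from the pawn and starts a "Hit" bounce animation on damageable ones, oriented toward the pawn.
; OnCeaseRelevant clears the timer, but only while the target is valid.
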(event Custom|WiggleActors
  (bind _targetpawn (Variables|Default|GetTargetPawn))
  (Utilities|IsValid _targetpawn
    (:Is%20Valid
      (bind _outactors (Collision|SphereOverlapActors (Variables|Default|GetTargetPawn) (+ (Transformation|GetActorLocation _targetpawn) (Variables|Default|GetPlayerHeightOffset)) 1e+02 (Variables|Default|GetObjectTypes) "/Script/FortniteGame.BuildingActor"))
      (for _array_element _outactors
        (if (Variables|/Script/Engine.Actor|GetbCanBeDamaged _array_element)
          (bind _asbuilding_static_mesh (Utilities|Casting|CastToBuildingSMActor _array_element))
          (Class|BuildingStaticMesh|StartBounceAnimation _asbuilding_static_mesh (Utilities|Time|GetTimeSeconds (Variables|Default|GetTargetPawn)) 5e+02 (Math|Rotator|GetRotationXVector (Math|Rotator|FindLookatRotation (Transformation|GetActorLocation _asbuilding_static_mesh) (Transformation|GetActorLocation (Variables|Default|GetTargetPawn)))) (Transformation|GetActorLocation (Variables|Default|GetTargetPawn)) "Hit" true))))))

(event Collision|EventActorBeginOverlap
  (params (OtherActor :type Actor%20Object%20Reference)))

(event EventTick
  (params (DeltaSeconds :type Float%20%28single-precision%29)))

(event EventBeginPlay)

(fn OnBurst
  (params (MyTarget :type Actor%20Object%20Reference) (Parameters :type Gameplay%20Cue%20Parameters%20Structure%20%28by%20ref%29))
  (bind _asfort_player_pawn (Utilities|Casting|CastToFortPlayerPawn MyTarget))
  (Variables|Default|SetTargetPawn _asfort_player_pawn)
  (CallFunction|WiggleActors)
  (bind _returnvalue (Utilities|Time|SetTimerbyEvent (EventDispatchers|CreateEvent :!event WiggleActors) 0.2 true))
  (Variables|Default|SetWiggleTimer _returnvalue)
  (return true))

(fn OnCeaseRelevant
  (params (MyTarget :type Actor%20Object%20Reference) (Parameters :type Gameplay%20Cue%20Parameters%20Structure%20%28by%20ref%29))
  (Utilities|IsValid MyTarget
    (:Is%20Valid
      (Utilities|Array|Clear (Variables|Default|GetArrayOfActors))
      (Utilities|Time|ClearandInvalidateTimerbyHandle 0 (Variables|Default|GetWiggleTimer))
      (return false))))

(fn ConstructionScript)

(fn OnBecomeRelevant
  (params (MyTarget :type Actor%20Object%20Reference) (Parameters :type Gameplay%20Cue%20Parameters%20Structure%20%28by%20ref%29))
  (bind _asfort_player_pawn (Utilities|Casting|CastToFortPlayerPawn MyTarget))
  (Variables|Default|SetTargetPawn _asfort_player_pawn)
  (return false))
