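; Looping gameplay cue that animates the burning fuse on the thrown TNT projectile.
; Curve-table durations scale three timelines, offset by elapsed fuse time, with one rebuilding a socket-based spline and driving wick fade and emissive flash.
; Stage 1 moves the looping particle along that spline while intensifying it, then spawns death-warning FX and audio that stage 2 ramps.
; Removal deactivates those spawned components and stops the movement timeline.
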(event Custom|Control%20Stage%201%20Fuse%20Particles
  (bind _loopingparticle (Variables|Default|GetLoopingParticle))
  ;[1 1 1] Sets Values in Fuse Particle Systems
  (Utilities|IsValid _loopingparticle
    (:Is%20Valid
      (Effects|Components|ParticleSystem|SetFloatParameter _loopingparticle "AlphaIntensity" (Variables|Default|GetPSys%20Opacity))
      (Effects|Components|ParticleSystem|SetVectorParameter _loopingparticle "ColorIntensity" (Variables|Default|GetPSys%20Color))
      (Effects|Components|ParticleSystem|SetVectorParameter _loopingparticle "Size_Intensity" (Variables|Default|GetPsys%20Size))
      (Transformation|SetWorldLocation _loopingparticle (Variables|Default|GetPSysSplineLoca) false false))))

(event Custom|Control%20Stage%202%20Fuse%20Particles
  (bind _death_fx (Variables|Default|GetDeath%20FX))
  ;[1 1 1] Sets Values in Fuse Particle Systems
  (Utilities|IsValid _death_fx
    (:Is%20Valid
      (Effects|Components|ParticleSystem|SetFloatParameter _death_fx "AlphaIntensity" (Variables|Default|GetPSys%20Opacity))
      (Effects|Components|ParticleSystem|SetVectorParameter _death_fx "ColorIntensity" (Variables|Default|GetPSys%20Color))
      (Effects|Components|ParticleSystem|SetVectorParameter _death_fx "Size_Intensity" (Variables|Default|GetPsys%20Size)))))

(event EventOnLoopingStart(EitherVFX-preferthisversion)
  (params (MyTarget :type Actor%20Object%20Reference) (Parameters :type Gameplay%20Cue%20Parameters%20Structure%20%28by%20ref%29) (ParticleComponents :type Array%20of%20FXSystem%20Component%20Object%20References) (AudioComponents :type Array%20of%20Audio%20Component%20Object%20References))
  (bind _2nd_stage_fuse_duration (Variables|Default|Get2nd%20Stage%20Fuse%20Duration))
  (bind _fuseduration (Variables|Default|GetFuseDuration))
  (bind _returnvalue (- _fuseduration _2nd_stage_fuse_duration))
  (bind _projectile (Variables|Default|GetProjectile))
  (bind _tntbundle (Variables|/Game/Athena/Items/Consumables/TNT/B_Prj_Athena_TNT.B_Prj_Athena_TNT_C|GetTNTBundle _projectile))
  (bind _asb_prj_athena_tnt (Utilities|Casting|CastToB_Prj_Athena_TNT MyTarget))
  (Variables|Default|SetProjectile _asb_prj_athena_tnt)
  (Variables|Default|SetLoopingParticle (Utilities|Array|Get(acopy) ParticleComponents))
  (Variables|Default|SetLooping%20Audio (Utilities|Array|Get(acopy) AudioComponents))
  (bind _outxy (DataTable|EvaluateCurveTableRow "/Game/Athena/Balance/DataTables/AthenaGameData.AthenaGameData" "Default.TNT.FuseDuration")
    (:RowFound
      (bind _output_get (Variables|Default|SetFuseDuration _outxy))
      (Variables|Default|SetFuseElapsedTime (- _output_get (Variables|/Game/Athena/Items/Consumables/Grenade/B_Prj_Athena_Grenade_Base.B_Prj_Athena_Grenade_Base_C|GetFuseTime (Variables|Default|GetProjectile))))
      (bind _outxy_2 (DataTable|EvaluateCurveTableRow "/Game/Athena/Balance/DataTables/AthenaGameData.AthenaGameData" "Default.TNT.RomanCandleDuration")
        (:RowFound
          (Variables|Default|Set2nd%20Stage%20Fuse%20Duration _outxy_2)
          (Components|Timeline|SetPlayRate (Variables|GCL_TNT_Fuse_Projectile|GetFuse%20Movement) (/ 1.0 _fuseduration))
          (Components|Timeline|SetPlayRate (Variables|GCL_TNT_Fuse_Projectile|GetStage%201%20Fuse) (/ 1.0 _returnvalue))
          (Components|Timeline|SetPlayRate (Variables|GCL_TNT_Fuse_Projectile|GetStage%202%20Fuse) (/ 1.0 (Variables|Default|Get2nd%20Stage%20Fuse%20Duration)))
          (Utilities|FlowControl|Sequence
            (:then_0
              ;[1 1 1] Movement of Fuse along Wick and Fade Material
              (bind (_direction _lerp_control _material_flash) (Timeline|Fuse%20Movement SetNewTime :NewTime (/ (Variables|Default|GetFuseElapsedTime) (Variables|Default|GetFuseDuration)))
                (:Update
                  (bind _spline_location (CallFunction|SplineSetup _lerp_control))
                  (Variables|Default|SetPSysSplineLoca _spline_location)
                  (Rendering|Material|SetScalarParameterValueonMaterials _tntbundle "Fade along U" _lerp_control)
                  (Rendering|Material|SetVectorParameterValueonMaterials _tntbundle "EmissiveMultiplier" (Math|Vector|Lerp(Vector) "0, 0, 0" (Variables|Default|GetEmissive%20Color) _material_flash)))))
            (:then_1
              (Timeline|Fuse%20Movement Play))
            (:then_2
              (if (not (>= (Variables|Default|GetFuseElapsedTime) _returnvalue))
                (Utilities|FlowControl|Sequence
                  (:then_0
                    (bind (_direction_2 _death_warning) (Timeline|Stage%201%20Fuse SetNewTime :NewTime (/ (Variables|Default|GetFuseElapsedTime) (- (Variables|Default|GetFuseDuration) (Variables|Default|Get2nd%20Stage%20Fuse%20Duration))))
                      (:Update
                        (bind (_opacity _color _size) (CallFunction|WarningSystemSetup _death_warning))
                        (Variables|Default|SetPSys%20Opacity _opacity)
                        (Variables|Default|SetPSys%20Color _color)
                        (Variables|Default|SetPsys%20Size _size)
                        (CallFunction|ControlStage1FuseParticles))
                      (:Finished
                        (bind _returnvalue_2 (Effects|SpawnFXSystemAttached "/Game/Weapons/Grenades/TNT_Bundle/FX/NS_TNT_Fuse_Death_Converted.NS_TNT_Fuse_Death_Converted" (Variables|/Game/Athena/Items/Consumables/TNT/B_Prj_Athena_TNT.B_Prj_Athena_TNT_C|GetTNTBundle (Variables|Default|GetProjectile))))
                        (bind _output_get_2 (Variables|Default|SetDeath%20FX _returnvalue_2))
                        (Effects|Components|ParticleSystem|SetFloatParameter _output_get_2 "AlphaIntensity" (Variables|Default|GetPSys%20Opacity))
                        (Effects|Components|ParticleSystem|SetVectorParameter _output_get_2 "ColorIntensity" (Variables|Default|GetPSys%20Color))
                        (Effects|Components|ParticleSystem|SetVectorParameter _output_get_2 "Size_Intensity" (Variables|Default|GetPsys%20Size))
                        (bind _returnvalue_3 (Audio|SpawnSoundAttached "/Game/Sounds/Fort_GamePlay_Sounds/Explosives/TnT/TNT_Fuse_Warning_Cue.TNT_Fuse_Warning_Cue" (Variables|/Game/Athena/Items/Consumables/Grenade/B_Prj_Athena_Grenade_Base.B_Prj_Athena_Grenade_Base_C|GetMesh (Variables|Default|GetProjectile))))
                        (Variables|Default|SetDeathAudio _returnvalue_3)
                        (Utilities|FlowControl|Sequence
                          (:then_0
                            (bind (_direction_3 _death_warning_2) (Timeline|Stage%202%20Fuse SetNewTime :NewTime (/ (Variables|Default|GetFuseElapsedTime) (+ (Variables|Default|GetFuseDuration) (- (Variables|Default|GetFuseDuration) (Variables|Default|Get2nd%20Stage%20Fuse%20Duration)))))
                              (:Update
                                (bind (_opacity_2 _color_2 _size_2) (CallFunction|WarningSystemSetup _death_warning_2))
                                (Variables|Default|SetPSys%20Opacity _opacity_2)
                                (Variables|Default|SetPSys%20Color _color_2)
                                (Variables|Default|SetPsys%20Size _size_2)
                                (CallFunction|ControlStage2FuseParticles))))
                          (:then_1
                            (Timeline|Stage%202%20Fuse Play))))))
                  (:then_1
                    (Timeline|Stage%201%20Fuse Play)))
                (else
                  _)))))))))

(event EventOnRemoval(EitherVFX-preferthisversion)
  (params (MyTarget :type Actor%20Object%20Reference) (Parameters :type Gameplay%20Cue%20Parameters%20Structure%20%28by%20ref%29) (ParticleComponents :type Array%20of%20FXSystem%20Component%20Object%20References) (AudioComponents :type Array%20of%20Audio%20Component%20Object%20References) (BurstCameraShakeInstance :type Camera%20Shake%20Base%20Object%20Reference) (BurstDecalInstance :type Decal%20Actor%20Object%20Reference))
  (bind _deathaudio (Variables|Default|GetDeathAudio))
  (bind _death_fx (Variables|Default|GetDeath%20FX))
  ;[1 1 1] Ensure Cleanup on Spawned Particle System
  (Utilities|IsValid _death_fx
    (:Is%20Valid
      (Components|Activation|Deactivate _death_fx))
    (:Is%20Not%20Valid
      _))
  (Utilities|IsValid _deathaudio
    (:Is%20Valid
      (Components|Activation|Deactivate _deathaudio))
    (:Is%20Not%20Valid
      _))
  (Components|Timeline|Stop (Variables|GCL_TNT_Fuse_Projectile|GetFuse%20Movement)))

(fn SplineSetup
  (params (Alpha :type Float%20%28double-precision%29))
  (bind _projectile (Variables|Default|GetProjectile))
  (bind _spline (Variables|/Game/Athena/Items/Consumables/TNT/B_Prj_Athena_TNT.B_Prj_Athena_TNT_C|GetSpline _projectile))
  (Spline|SetSplinePoints _spline (Utilities|Array|MakeArray (Transformation|GetSocketLocation (Variables|/Game/Athena/Items/Consumables/TNT/B_Prj_Athena_TNT.B_Prj_Athena_TNT_C|GetTNTBundle (Variables|Default|GetProjectile)) "FX_FuseBegin") (Transformation|GetSocketLocation (Variables|/Game/Athena/Items/Consumables/TNT/B_Prj_Athena_TNT.B_Prj_Athena_TNT_C|GetTNTBundle (Variables|Default|GetProjectile)) "dyn_fuse2") (Transformation|GetSocketLocation (Variables|/Game/Athena/Items/Consumables/TNT/B_Prj_Athena_TNT.B_Prj_Athena_TNT_C|GetTNTBundle (Variables|Default|GetProjectile)) "dyn_fuse1") (Transformation|GetSocketLocation (Variables|/Game/Athena/Items/Consumables/TNT/B_Prj_Athena_TNT.B_Prj_Athena_TNT_C|GetTNTBundle (Variables|Default|GetProjectile)) "FX_FuseEnd")) "World" true)
  (return (Spline|GetLocationatDistanceAlongSpline _spline (Math|Float|Lerp 0.0 (Spline|GetSplineLength _spline) Alpha) "World")))

(fn ConstructionScript)

(fn WarningSystemSetup
  (params (Warning :type Float%20%28double-precision%29))
  (return (Math|Float|Lerp 1.0 (LocalVariables|GetMax%20Possible%20Opacity) Warning) (Math|Vector|Lerp(Vector) "1.000000,1.000000,1.000000" (LocalVariables|GetFinal%20Color) Warning) (Math|Vector|Lerp(Vector) "1.000000,1.000000,1.000000" (LocalVariables|GetMax%20Possible%20Size) Warning)))
