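; Looping gameplay cue for Athena that shows an invulnerability highlight on the target player pawn, stored as the downed player.
; On loop start it casts the target to PlayerPawn_Athena and applies the highlight with dynamic update off.
; The highlight copies the pawn's own Invulnerable Highlight settings but substitutes the cue's Glow Priority.
; It registers itself as the highlight source, so removal clears only that source's highlight.
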(event EventOnLoopingStart(Cascade)
  (params (MyTarget :type Actor%20Object%20Reference) (Parameters :type Gameplay%20Cue%20Parameters%20Structure%20%28by%20ref%29) (ParticleComponents :type Array%20of%20Cascade%20Particle%20System%20Component%20Object%20References) (AudioComponents :type Array%20of%20Audio%20Component%20Object%20References))
  (bind _asplayer_pawn_athena (Utilities|Casting|CastToPlayerPawn_Athena MyTarget))
  (Variables|Default|SetDownedPlayer _asplayer_pawn_athena)
  (CallFunction|ApplyInvulnerabilityEffect false))

(event Custom|Apply%20Invulnerability%20Effect
  (params (Dynamic%20Update :type Boolean))
  ;[1 1 1] Invuln FX
  (bind _downedplayer (Variables|Default|GetDownedPlayer)
    (:then
      (bind (_invulnerable_highlight_priority_28_e2e1b5344846e187b9c11b863a7f0698 _invulnerable_highlight_inner_21_4cc2801147ea190de16f59b34f36853e _invulnerable_highlight_outer_22_5a1d7d0543d303e8b54b66a7f7bd2e2e _invulnerable_highlight_fresnelbrightness_23_52b0f96447ff640f47df2895b0602e92 _invulnerable_highlight_fresnelexponent_24_b427cf0c441aa37ed49833bf7579de6d _invulnerable_highlight_usespulse_25_e29229f64e540f0617e4c4987ad77605) (Utilities|Struct|BreakPawnHighlight (Variables|/Game/Athena/PlayerPawn_Athena.PlayerPawn_Athena_C|GetInvulnerable%20Highlight_Priority_28_E2E1B5344846E187B9C11B863A7F0698 _downedplayer)))
      (PawnHighlight|HighlightPawnFromSource _downedplayer self (Utilities|Struct|MakePawnHighlight (Variables|Default|GetGlow%20Priority) _invulnerable_highlight_inner_21_4cc2801147ea190de16f59b34f36853e _invulnerable_highlight_outer_22_5a1d7d0543d303e8b54b66a7f7bd2e2e _invulnerable_highlight_fresnelbrightness_23_52b0f96447ff640f47df2895b0602e92 _invulnerable_highlight_fresnelexponent_24_b427cf0c441aa37ed49833bf7579de6d _invulnerable_highlight_usespulse_25_e29229f64e540f0617e4c4987ad77605) Dynamic%20Update))
    (:else)))

(event EventOnRemoval(Cascade)
  (params (MyTarget :type Actor%20Object%20Reference) (Parameters :type Gameplay%20Cue%20Parameters%20Structure%20%28by%20ref%29) (ParticleComponents :type Array%20of%20Cascade%20Particle%20System%20Component%20Object%20References) (AudioComponents :type Array%20of%20Audio%20Component%20Object%20References) (BurstCameraShakeInstance :type Camera%20Shake%20Base%20Object%20Reference) (BurstDecalInstance :type Decal%20Actor%20Object%20Reference))
  (bind _downedplayer (Variables|Default|GetDownedPlayer)
    (:then
      (PawnHighlight|RemovePawnHighlightFromSource _downedplayer self))
    (:else)))

(fn ConstructionScript)
