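; GameplayCue notify (apparently looping, per the GCNL prefix) that supplies hit glow colors based on the affected player's team affiliation to the local player.
; GetHitGlows resolves that affiliation and calls GetGlowColors, falling back to Neutral without an affected player.
; GetGlowColors looks up a base color in an affiliation-to-color map and scales it into inner and outer glows.
; A missing map entry is logged and returns zero glow.
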(fn GetGlowColors
  (params (Team%20Affiliation :type EFortTeamAffiliation%20Enum :default "Friendly"))
  (bind _teamaffiliationbasecolormap (Variables|Default|GetTeamAffiliationBaseColorMap))
  (bind (_value _returnvalue) (Utilities|Map|Find _teamaffiliationbasecolormap Team%20Affiliation))
  (if _returnvalue
    (return (Math|Color|LinearColor*Float _value (Variables|Default|GetGlow%20Inner%20Scalar)) (Math|Color|LinearColor*Float _value (Variables|Default|GetGlow%20Outer%20Scalar)))
    (else
      (Fortnite|FortShippingLog self (Utilities|String|Append "Failed to find color data in TeamAffiliationbaseColorMap  for team affiliation: [" (Utilities|String|EnumtoString Team%20Affiliation) "]") true)
      ;[1 0.4829 0.4583] Failed to find the team affiliation, returning black (i.e. no glow)
      (return 0 0))))

(fn GetHitGlows
  (bind (_glow_inner _glow_outer) (CallFunction|GetGlowColors "Neutral"))
  (bind _affectedplayer (Variables|Default|GetAffectedPlayer)
    (:then
      ;[1 0.4635 0.4702] Return Correct Team Affiliation Glows
      (return (CallFunction|GetGlowColors (Team|GetTeamAffiliationtoLocalPlayer _affectedplayer (Team|GetActorTeam _affectedplayer))) (.Glow%20Outer (CallFunction|GetGlowColors (Team|GetTeamAffiliationtoLocalPlayer _affectedplayer (Team|GetActorTeam _affectedplayer))))))
    (:else
      ;[1 0.4635 0.4702] Failsafe: Return Neutral Glows
      (return _glow_inner _glow_outer))))

(fn ConstructionScript)
