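; Looping gameplay cue notify that plays a Chrome surface ambient audio bank for the locally viewed player.
; On application, it casts the target to a Fort player pawn and, if locally viewed, stores it and registers the bank with the Ambient Audio subsystem as "ChromeSurface".
; On removal, it acts only if that stored target is valid, clearing it and removing the ambient entry.
; Both paths print debug strings.
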(event EventOnApplication(EitherVFX-preferthisone)
  (params (MyTarget :type Actor%20Object%20Reference) (Parameters :type Gameplay%20Cue%20Parameters%20Structure%20%28by%20ref%29) (ParticleComponents :type Array%20of%20FXSystem%20Component%20Object%20References) (AudioComponents :type Array%20of%20Audio%20Component%20Object%20References) (BurstCameraShakeInstance :type Camera%20Shake%20Base%20Object%20Reference) (BurstDecalInstance :type Decal%20Actor%20Object%20Reference))
  (bind _asfort_player_pawn (Utilities|Casting|CastToFortPlayerPawn MyTarget))
  (if (Pawn|IsLocallyViewed _asfort_player_pawn)
    (Variables|Default|SetLocalTarget _asfort_player_pawn)
    (Development|PrintString 0 "Applying Chrome surface GCNL")
    (AmbientAudio|AddAmbientEntry (WorldSubsystems|GetAmbientAudioSubsystem 0) "ChromeSurface" "/FNGameplayCues/Sounds/Artemis/AmbientBanks/Artemis_AmbientBank_Chrome_Surface_Lvl_01.Artemis_AmbientBank_Chrome_Surface_Lvl_01" 2 1e+01)))

(event EventOnRemoval(EitherVFX-preferthisversion)
  (params (MyTarget :type Actor%20Object%20Reference) (Parameters :type Gameplay%20Cue%20Parameters%20Structure%20%28by%20ref%29) (ParticleComponents :type Array%20of%20FXSystem%20Component%20Object%20References) (AudioComponents :type Array%20of%20Audio%20Component%20Object%20References) (BurstCameraShakeInstance :type Camera%20Shake%20Base%20Object%20Reference) (BurstDecalInstance :type Decal%20Actor%20Object%20Reference))
  (Variables|Default|GetLocalTarget
    (:then
      (Variables|Default|SetLocalTarget 0)
      (Development|PrintString 0 "Removing Chrome surface GCNL")
      (AmbientAudio|RemoveAmbientEntry (WorldSubsystems|GetAmbientAudioSubsystem 0) "ChromeSurface" -1.0))
    (:else)))

(fn ConstructionScript)
