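; Looping gameplay cue that shows Creative disguise "initiating" effects on a player pawn while DishFries is being applied.
; On loop start it attaches a phone-rays emitter to the right middle finger bone and a body emitter to the mesh, with the pawn as BoneSocketActor.
; Tick keeps the rays yawed toward the mesh origin with Z flattened.
; Removal prints a debug "EXIT", stops ticking and deactivates both emitters.
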(event EventOnLoopingStart(Cascade)
  (params (MyTarget :type Actor%20Object%20Reference) (Parameters :type Gameplay%20Cue%20Parameters%20Structure%20%28by%20ref%29) (ParticleComponents :type Array%20of%20Cascade%20Particle%20System%20Component%20Object%20References) (AudioComponents :type Array%20of%20Audio%20Component%20Object%20References))
  (bind _asplayer_pawn_athena (Utilities|Casting|CastToPlayerPawn_Athena MyTarget))
  (bind _output_get (Variables|Default|SetMyTarget _asplayer_pawn_athena))
  (bind _returnvalue (Effects|SpawnEmitterAttached :EmitterTemplate "/Game/Creative/Items/Gameplay/Disguise/P_Creative_Disguise_PhoneRays.P_Creative_Disguise_PhoneRays" :AttachToComponent (Variables|/Script/Engine.Character|GetMesh _output_get) :AttachPointName "middle_01_r" :Location (Variables|Default|GetRaysRelativeOffset) :Scale "1.000000,1.000000,1.000000" :LocationType "KeepRelativeOffset" :bAutoDestroy true :PoolingMethod "AutoRelease" :bAutoActivate true))
  (Variables|Default|SetRaysPS _returnvalue)
  (bind _returnvalue_2 (Effects|SpawnEmitterAttached :EmitterTemplate "/Game/Creative/Items/Gameplay/Disguise/P_Creative_Disguise_Initiating.P_Creative_Disguise_Initiating" :AttachToComponent (Variables|/Script/Engine.Character|GetMesh (Variables|Default|GetMyTarget)) :Scale "1.000000,1.000000,1.000000" :LocationType "KeepRelativeOffset" :bAutoDestroy true :PoolingMethod "AutoRelease" :bAutoActivate true))
  (bind _output_get_2 (Variables|Default|SetBodyPS _returnvalue_2))
  (Effects|Components|ParticleSystem|SetActorParameter _output_get_2 "BoneSocketActor" (Variables|Default|GetMyTarget))
  (Actor|Tick|SetActorTickEnabled true))

(event EventOnRemoval(Cascade)
  (params (MyTarget :type Actor%20Object%20Reference) (Parameters :type Gameplay%20Cue%20Parameters%20Structure%20%28by%20ref%29) (ParticleComponents :type Array%20of%20Cascade%20Particle%20System%20Component%20Object%20References) (AudioComponents :type Array%20of%20Audio%20Component%20Object%20References) (BurstCameraShakeInstance :type Camera%20Shake%20Base%20Object%20Reference) (BurstDecalInstance :type Decal%20Actor%20Object%20Reference))
  (Development|PrintString 0 "EXIT" true true "(R=1.000000,G=0.000000,B=0.424145,A=1.000000)" 5.0)
  (Actor|Tick|SetActorTickEnabled false)
  (bind _raysps (Variables|Default|GetRaysPS)
    (:then
      (Components|Activation|Deactivate _raysps))
    (:else
      _))
  (bind _bodyps (Variables|Default|GetBodyPS)
    (:then
      (Components|Activation|Deactivate _bodyps))
    (:else)))

(event EventTick
  (params (DeltaSeconds :type Float%20%28single-precision%29))
  (bind _raysps (Variables|Default|GetRaysPS)
    (:then
      (bind _mytarget (Variables|Default|GetMyTarget)
        (:then
          (Transformation|SetWorldRotation (Variables|Default|GetRaysPS) (Math|Conversions|RotationFromXVector (* (- (Transformation|GetWorldLocation (Variables|/Script/Engine.Character|GetMesh _mytarget)) (Transformation|GetWorldLocation _raysps)) "1.000000,1.000000, 0")) false false))
        (:else)))
    (:else)))

(fn ConstructionScript)
