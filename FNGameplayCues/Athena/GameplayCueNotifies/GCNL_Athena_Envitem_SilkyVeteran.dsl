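; Looping gameplay cue for the SilkyVeteran environment item, playing splash particle effects for a slurp vat on the target player.
; On loop start it stores the Cascade splash component and player pawn, then creates Ripple, Bubble and Mist dynamic materials fed the instigator's location and a vat radius.
; A character-movement-updated binding checks crouch state and sets the particle Standing and Crouching parameters through DoOnce gates.
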(event EventOnLoopingStart(Cascade)
  (params (MyTarget :type Actor%20Object%20Reference) (Parameters :type Gameplay%20Cue%20Parameters%20Structure%20%28by%20ref%29) (ParticleComponents :type Array%20of%20Cascade%20Particle%20System%20Component%20Object%20References) (AudioComponents :type Array%20of%20Audio%20Component%20Object%20References))
  (Variables|Default|SetSplashFX (Utilities|Array|Get(acopy) ParticleComponents))
  (CallFunction|SetFXMIDs Parameters)
  (bind _asfort_player_pawn_athena (Utilities|Casting|CastToFortPlayerPawnAthena MyTarget))
  (Variables|Default|SetPlayerPawn _asfort_player_pawn_athena)
  (Character|BindEventtoOnCharacterMovementUpdated _asfort_player_pawn_athena (EventDispatchers|CreateEvent :!event Custom|Crouch%20Check)))

(event Custom|Crouch%20Check
  (params (DeltaSeconds :type Float%20%28single-precision%29) (OldLocation :type Vector) (OldVelocity :type Vector))
  (bind _splashfx (Variables|Default|GetSplashFX))
  (bind _playerpawn (Variables|Default|GetPlayerPawn))
  (bind _splashfx_2 (Variables|Default|GetSplashFX))
  (Utilities|IsValid _playerpawn
    (:Is%20Valid
      (Utilities|IsValid (Variables|Default|GetSplashFX)
        (:Is%20Valid
          (if (AI|Components|NavMovement|IsCrouching (Variables|/Script/Engine.Character|GetCharacterMovement _playerpawn))
            (Utilities|FlowControl|DoOnce false
              (:Completed
                (Effects|Components|ParticleSystem|SetFloatParameter _splashfx "Standing" 0.0)
                (Effects|Components|ParticleSystem|SetFloatParameter _splashfx "Crouching" 1.0))))
          (Utilities|FlowControl|DoOnce false
            (:Completed
              (Effects|Components|ParticleSystem|SetFloatParameter _splashfx_2 "Standing" 1.0)
              (Effects|Components|ParticleSystem|SetFloatParameter _splashfx_2 "Crouching" 0.0))))))))

(fn SetFXMIDs
  (params (Gameplay%20Parameters :type Gameplay%20Cue%20Parameters%20Structure))
  (bind _returnvalue (Ability|GameplayCue|GetInstigatorActor Gameplay%20Parameters))
  (bind _returnvalue_2 (Transformation|GetActorLocation _returnvalue))
  (bind _returnvalue_3 (Math|Conversions|ToLinearColor(Vector) _returnvalue_2))
  (bind _splashfx (Variables|Default|GetSplashFX))
  (bind _returnvalue_4 (Rendering|Material|CreateNamedDynamicMaterialInstance _splashfx "Ripple" "/FNGameplayCues/Athena/Items/Gameplay/SilkyVeteran/FX/MI_SplashRipple_SlurpVat.MI_SplashRipple_SlurpVat"))
  (Rendering|Material|SetVectorParameterValue _returnvalue_4 "VatLocation" _returnvalue_3 :!class /Script/Engine.MaterialInstanceDynamic)
  (Rendering|Material|SetScalarParameterValue _returnvalue_4 "VatRadius" (Variables|Default|GetVatRadius) :!class /Script/Engine.MaterialInstanceDynamic)
  (bind _returnvalue_5 (Rendering|Material|CreateNamedDynamicMaterialInstance _splashfx "Bubble" "/Game/Athena/Items/Consumables/ChillBronco/FX/MI_BubbleGPU_ChillBronco.MI_BubbleGPU_ChillBronco"))
  (Rendering|Material|SetVectorParameterValue _returnvalue_5 "VatLocation" _returnvalue_3 :!class /Script/Engine.MaterialInstanceDynamic)
  (Rendering|Material|SetScalarParameterValue _returnvalue_5 "VatRadius" (Variables|Default|GetVatRadius) :!class /Script/Engine.MaterialInstanceDynamic)
  (bind _returnvalue_6 (Rendering|Material|CreateNamedDynamicMaterialInstance _splashfx "Mist" "/FNGameplayCues/Athena/Items/Gameplay/SilkyVeteran/FX/MI_IceTrail_FrostyWisps_SlurpSphereMask.MI_IceTrail_FrostyWisps_SlurpSphereMask"))
  (Rendering|Material|SetVectorParameterValue _returnvalue_6 "VatLocation" _returnvalue_3 :!class /Script/Engine.MaterialInstanceDynamic)
  (Rendering|Material|SetScalarParameterValue _returnvalue_6 "VatRadius" (Variables|Default|GetVatRadius) :!class /Script/Engine.MaterialInstanceDynamic))

(fn ConstructionScript)
