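; Looping gameplay cue that plays the Festivus fireworks finale effects and audio, only for the locally controlled player.
; After a short delay it traces down to place the main firework emitter below the pawn, spawns exit and background emitters at the pawn, and fades in 2D cue audio.
; Parallel timers add a close-up sound, then wind down emitters and audio in stages.
; Removal destroys all stored components as a failsafe.
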(event EventOnLoopingStart(Cascade)
  (params (MyTarget :type Actor%20Object%20Reference) (Parameters :type Gameplay%20Cue%20Parameters%20Structure%20%28by%20ref%29) (ParticleComponents :type Array%20of%20Cascade%20Particle%20System%20Component%20Object%20References) (AudioComponents :type Array%20of%20Audio%20Component%20Object%20References))
  (bind _emitter3 (Variables|Default|GetEmitter3))
  (bind _emitter1 (Variables|Default|GetEmitter1))
  (bind _sound1 (Variables|Default|GetSound1))
  (bind _emitter2 (Variables|Default|GetEmitter2))
  (if (Pawn|IsLocalPlayerController (Fortnite|GetFortPlayerControllerfromActor MyTarget))
    (bind _output_get (Variables|Default|SetPawnActor MyTarget))
    (Utilities|FlowControl|Delay 0 1.0)
    (bind _outhit (Collision|LineTraceByChannel _output_get (Transformation|GetActorLocation _output_get) "0, 0,-20000.000000" "TraceTypeQuery10" false (Utilities|Array|MakeArray _output_get) "Persistent"))
    (bind _returnvalue (Effects|SpawnEmitteratLocation 0 "/FNGameplayCues/Athena/Effects/Particles/Festivus/P_Festivus_Firework_Testing_v2.P_Festivus_Firework_Testing_v2" (Math|Vector|MakeVector (.x (Transformation|GetActorLocation _output_get)) (.y (Transformation|GetActorLocation _output_get)) (- (.z (Transformation|GetActorLocation _output_get)) (.Distance (Collision|BreakHitResult _outhit))))))
    (Variables|Default|SetEmitter1 _returnvalue)
    (bind _returnvalue_2 (Effects|SpawnEmitteratLocation 0 "/FNGameplayCues/Athena/Effects/Particles/Festivus/P_Festivus_Firework_Testing_OnExit.P_Festivus_Firework_Testing_OnExit" (Transformation|GetActorLocation _output_get)))
    (Variables|Default|SetEmitter2 _returnvalue_2)
    (bind _returnvalue_3 (Effects|SpawnEmitteratLocation 0 "/FNGameplayCues/Athena/Effects/Particles/Festivus/P_Festivus_Firework_Testing_v2_BACKGROUNDONLY.P_Festivus_Firework_Testing_v2_BACKGROUNDONLY" (Transformation|GetActorLocation _output_get)))
    (Variables|Default|SetEmitter3 _returnvalue_3)
    (bind _returnvalue_4 (Audio|SpawnSound2D (Variables|Default|GetPawnActor) "/FNGameplayCues/Athena/Sounds/LTMs/Stage/Festivus_Finale_Cue.Festivus_Finale_Cue"))
    (bind _output_get_2 (Variables|Default|SetSound1 _returnvalue_4))
    (Audio|Components|Audio|FadeIn _output_get_2 1.0 1.0 0.0 "Linear" :!class /Script/Engine.AudioComponent)
    (Utilities|FlowControl|Sequence
      (:then_0
        (Utilities|FlowControl|Sequence
          (:then_0
            (Utilities|FlowControl|Delay 0 6e+01)
            (Utilities|FlowControl|Sequence
              (:then_0
                (Utilities|IsValid _sound1
                  (:Is%20Valid
                    (Audio|Components|Audio|FadeOut _sound1 5.0 0.0 "Linear" :!class /Script/Engine.AudioComponent))))
              (:then_1
                (Utilities|IsValid _emitter2
                  (:Is%20Valid
                    (Components|Activation|Deactivate _emitter2))))
              (:then_2
                (Utilities|IsValid _emitter3
                  (:Is%20Valid
                    (Components|Activation|Deactivate _emitter3))))))
          (:then_1
            (Utilities|FlowControl|Delay 0 14.0)
            (Utilities|FlowControl|Sequence
              (:then_0
                (Utilities|IsValid _emitter1
                  (:Is%20Valid
                    (Components|Activation|Deactivate _emitter1))))
              (:then_1
                (bind _sound2 (Variables|Default|GetSound2)
                  (:then
                    (Audio|Components|Audio|FadeOut _sound2 5.0 0.0 "Linear" :!class /Script/Engine.AudioComponent))
                  (:else)))))))
      (:then_1
        (Utilities|FlowControl|Delay 0 2.0)
        (bind _returnvalue_5 (Audio|SpawnSound2D (Variables|Default|GetPawnActor) "/FNGameplayCues/Athena/Sounds/LTMs/Stage/Festivus_Finale_Close_01.Festivus_Finale_Close_01"))
        (Variables|Default|SetSound2 _returnvalue_5)))))

(event EventOnRemoval(Cascade)
  (params (MyTarget :type Actor%20Object%20Reference) (Parameters :type Gameplay%20Cue%20Parameters%20Structure%20%28by%20ref%29) (ParticleComponents :type Array%20of%20Cascade%20Particle%20System%20Component%20Object%20References) (AudioComponents :type Array%20of%20Audio%20Component%20Object%20References) (BurstCameraShakeInstance :type Camera%20Shake%20Base%20Object%20Reference) (BurstDecalInstance :type Decal%20Actor%20Object%20Reference))
  (bind _emitter1 (Variables|Default|GetEmitter1))
  (bind _emitter3 (Variables|Default|GetEmitter3))
  (bind _emitter2 (Variables|Default|GetEmitter2))
  (bind _sound1 (Variables|Default|GetSound1))
  (bind _sound2 (Variables|Default|GetSound2))
  ;[1 1 1] Failsafe destroy emitter hack. We will do this cleaner in the future
  (Utilities|FlowControl|Sequence
    (:then_0
      (Utilities|IsValid _emitter1
        (:Is%20Valid
          (Components|DestroyComponent _emitter1 _emitter1))))
    (:then_1
      (Utilities|IsValid _emitter2
        (:Is%20Valid
          (Components|DestroyComponent _emitter2 _emitter2))))
    (:then_2
      (Utilities|IsValid _emitter3
        (:Is%20Valid
          (Components|DestroyComponent _emitter3 _emitter3))))
    (:then_3
      (Utilities|IsValid _sound1
        (:Is%20Valid
          (Components|DestroyComponent _sound1 _sound1))))
    (:then_4
      (Utilities|IsValid _sound2
        (:Is%20Valid
          (Components|DestroyComponent _sound2 _sound2))))))

(fn ConstructionScript)
