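; Looping gameplay cue for the Hop Flopper low-gravity effect, attaching particles to the player and driving loop audio.
; Duration comes from curve tables or cue magnitude, with timers repeating a warning sound near expiry.
; Tick feeds interpolated fall speed, from vertical velocity and jump-apex time, into audio parameters and lowers pitch and volume toward expiry.
; It destroys itself when the pawn is invalid, preventing stuck audio, and clears timers and particles on removal.
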(event Custom|ExpireTell
  (bind _returnvalue (Utilities|Time|SetTimerbyEvent (delegate Custom|PlayExpirationSound) (Variables|Default|GetExpirationSoundPeriod) true))
  (Variables|Default|SetExpirationSoundTimer _returnvalue))

(event Custom|PlayExpirationSound
  (Audio|SpawnSoundAttached "/FNGameplayCues/Sounds/Fort_GamePlay_Sounds/Player/LowGravityStones/LowGravityStone_Warning_Cue.LowGravityStone_Warning_Cue" (Variables|/Script/Engine.Actor|GetRootComponent (Variables|Default|GetPlayerPawn))))

(event Custom|ClearExpireTell
  (Utilities|Time|ClearandInvalidateTimerbyHandle 0 (Variables|Default|GetExpireTellDelayTimer))
  (Utilities|Time|ClearandInvalidateTimerbyHandle 0 (Variables|Default|GetExpirationSoundTimer)))

(event EventBeginPlay)

(event GameplayCueNotify|EventHandleGameplayCue
  (params (MyTarget :type Actor%20Object%20Reference) (EventType :type EGameplayCueEvent%20Enum :default "OnActive") (Parameters :type Gameplay%20Cue%20Parameters%20Structure%20%28by%20ref%29))
  (bind (_normalizedmagnitude _rawmagnitude) (Ability|GameplayCue|BreakGameplayCueParameters Parameters))
  (bind _output_get (Variables|Default|SetMyTargetCached MyTarget))
  (switch Utilities|FlowControl|Switch|SwitchonEGameplayCueEvent EventType
    (:OnActive
      (Development|PrintString _output_get (Utilities|String|ToString(Float) _rawmagnitude)))
    (:WhileActive
      (Development|PrintString _output_get))))

(event Collision|EventActorBeginOverlap
  (params (OtherActor :type Actor%20Object%20Reference)))

(event EventOnLoopingStart(EitherVFX-preferthisversion)
  (params (MyTarget :type Actor%20Object%20Reference) (Parameters :type Gameplay%20Cue%20Parameters%20Structure%20%28by%20ref%29) (ParticleComponents :type Array%20of%20FXSystem%20Component%20Object%20References) (AudioComponents :type Array%20of%20Audio%20Component%20Object%20References))
  (bind _playerpawn (Variables|Default|GetPlayerPawn))
  (bind _returnvalue (Utilities|Time|GetTimeSeconds _playerpawn))
  (bind _expirationtellduration (Variables|Default|GetExpirationTellDuration))
  (bind _duration (Variables|Default|GetDuration))
  (bind _starttime (Variables|Default|GetStartTime))
  (bind _returnvalue_2 (+ _starttime _duration))
  (bind _returnvalue_3 (- _returnvalue_2 _expirationtellduration))
  (bind _returnvalue_4 (Math|Float|MapRangeClamped _returnvalue _returnvalue_3 _returnvalue_2 0.0 1.0))
  (bind _returnvalue_5 (- 1.0 _returnvalue_4))
  (bind _playerpawn_2 (Variables|Default|GetPlayerPawn))
  (bind _charactermovement (Variables|/Script/Engine.Character|GetCharacterMovement _playerpawn_2))
  (bind _returnvalue_6 (AI|Components|NavMovement|IsFalling _charactermovement))
  (bind _returnvalue_7 (* _returnvalue_5 0.25))
  (bind _returnvalue_8 (+ _returnvalue_7 0.75))
  (bind (_velocity_x _velocity_y _velocity_z) (Math|Vector|BreakVector (Variables|/Script/Engine.MovementComponent|GetVelocity_X _charactermovement)))
  (Variables|Default|SetAudioComponent (Utilities|Array|Get(acopy) AudioComponents))
  (Utilities|FlowControl|Gate true
    (:Exit
      (Variables|Default|GetPlayerPawn
        (:then
          (Variables|Default|SetTargetDelta (Math|Float|SelectFloat (Math|Float|MapRangeClamped (Math|Float|Absolute(Float) (* (Math|Float|Absolute(Float) _velocity_z) (PawnStats|TimefromJumpApex _playerpawn_2))) 0.0 1e+03 0.0 1.0) 0.0 _returnvalue_6))
          (bind _output_get (Variables|Default|SetInterpDelta (Math|Interpolation|FInterpTo (Variables|Default|GetInterpDelta) (Variables|Default|GetTargetDelta) (Utilities|Time|GetWorldDeltaSeconds (Variables|Default|GetPlayerPawn)) (Math|Float|SelectFloat 1e+01 2.0 (or (> (Variables|Default|GetTargetDelta) (Variables|Default|GetInterpDelta)) (not _returnvalue_6))))))
          (Audio|Parameter|SetFloatParameter (Variables|Default|GetAudioComponent) "FallSpeedInv" (- 1.0 _output_get) :!class /Script/AudioExtensions.AudioParameterControllerInterface)
          (Audio|Parameter|SetFloatParameter (Variables|Default|GetAudioComponent) "FallSpeed" _output_get :!class /Script/AudioExtensions.AudioParameterControllerInterface)
          (Audio|Components|Audio|SetPitchMultiplier (Variables|Default|GetAudioComponent) _returnvalue_8)
          (Audio|Components|Audio|SetVolumeMultiplier (Variables|Default|GetAudioComponent) _returnvalue_8 :!class /Script/Engine.AudioComponent))
        (:else
          ;[1 1 1] Fix for Looping Audio getting stuck on in world
          (Actor|DestroyActor))))))

(event EventTick
  (params (DeltaSeconds :type Float%20%28single-precision%29))
  (bind _playerpawn (Variables|Default|GetPlayerPawn))
  (bind _returnvalue (Utilities|Time|GetTimeSeconds _playerpawn))
  (bind _expirationtellduration (Variables|Default|GetExpirationTellDuration))
  (bind _duration (Variables|Default|GetDuration))
  (bind _starttime (Variables|Default|GetStartTime))
  (bind _returnvalue_2 (+ _starttime _duration))
  (bind _returnvalue_3 (- _returnvalue_2 _expirationtellduration))
  (bind _returnvalue_4 (Math|Float|MapRangeClamped _returnvalue _returnvalue_3 _returnvalue_2 0.0 1.0))
  (bind _returnvalue_5 (- 1.0 _returnvalue_4))
  (bind _playerpawn_2 (Variables|Default|GetPlayerPawn))
  (bind _charactermovement (Variables|/Script/Engine.Character|GetCharacterMovement _playerpawn_2))
  (bind _returnvalue_6 (AI|Components|NavMovement|IsFalling _charactermovement))
  (bind _returnvalue_7 (* _returnvalue_5 0.25))
  (bind _returnvalue_8 (+ _returnvalue_7 0.75))
  (bind (_velocity_x _velocity_y _velocity_z) (Math|Vector|BreakVector (Variables|/Script/Engine.MovementComponent|GetVelocity_X _charactermovement)))
  (Utilities|FlowControl|Gate true
    (:Exit
      (Variables|Default|GetPlayerPawn
        (:then
          (Variables|Default|SetTargetDelta (Math|Float|SelectFloat (Math|Float|MapRangeClamped (Math|Float|Absolute(Float) (* (Math|Float|Absolute(Float) _velocity_z) (PawnStats|TimefromJumpApex _playerpawn_2))) 0.0 1e+03 0.0 1.0) 0.0 _returnvalue_6))
          (bind _output_get (Variables|Default|SetInterpDelta (Math|Interpolation|FInterpTo (Variables|Default|GetInterpDelta) (Variables|Default|GetTargetDelta) (Utilities|Time|GetWorldDeltaSeconds (Variables|Default|GetPlayerPawn)) (Math|Float|SelectFloat 1e+01 2.0 (or (> (Variables|Default|GetTargetDelta) (Variables|Default|GetInterpDelta)) (not _returnvalue_6))))))
          (Audio|Parameter|SetFloatParameter (Variables|Default|GetAudioComponent) "FallSpeedInv" (- 1.0 _output_get) :!class /Script/AudioExtensions.AudioParameterControllerInterface)
          (Audio|Parameter|SetFloatParameter (Variables|Default|GetAudioComponent) "FallSpeed" _output_get :!class /Script/AudioExtensions.AudioParameterControllerInterface)
          (Audio|Components|Audio|SetPitchMultiplier (Variables|Default|GetAudioComponent) _returnvalue_8)
          (Audio|Components|Audio|SetVolumeMultiplier (Variables|Default|GetAudioComponent) _returnvalue_8 :!class /Script/Engine.AudioComponent))
        (:else
          ;[1 1 1] Fix for Looping Audio getting stuck on in world
          (Actor|DestroyActor))))))

(fn Activated
  (params (PlayerPawn :type Actor%20Object%20Reference))
  (bind _asfort_player_pawn_athena (Utilities|Casting|CastToFortPlayerPawnAthena PlayerPawn))
  (Variables|Default|SetPlayerPawn _asfort_player_pawn_athena)
  (bind _outxy (DataTable|EvaluateCurveTableRow "/Game/Athena/Balance/DataTables/AthenaGameData.AthenaGameData" "Default.HopFlopper.Duration")
    (:RowFound
      (bind _output_get (Variables|Default|SetDuration _outxy)))
    (:RowNotFound
      _))
  (Variables|Default|SetExpirationTellDuration (* _output_get 0.25))
  (CallFunction|ClearExpireTell)
  (bind _returnvalue (Utilities|Time|SetTimerbyEvent (EventDispatchers|CreateEvent :!event ExpireTell) (- (Variables|Default|GetDuration) (Variables|Default|GetExpirationTellDuration))))
  (Variables|Default|SetExpireTellDelayTimer _returnvalue)
  (Variables|Default|SetStartTime (Utilities|Time|GetTimeSeconds (Variables|Default|GetPlayerPawn))))

(fn OnExecute
  (params (MyTarget :type Actor%20Object%20Reference) (Parameters :type Gameplay%20Cue%20Parameters%20Structure%20%28by%20ref%29))
  (bind _normalizedmagnitude (Ability|GameplayCue|BreakGameplayCueParameters Parameters))
  (if (> _normalizedmagnitude 0.0)
    (Variables|Default|SetDuration _normalizedmagnitude)
    (Variables|Default|SetExpirationTellDuration (* (Variables|Default|GetDuration) 0.25))
    (else
      (bind _outxy (DataTable|EvaluateCurveTableRow "/Game/Athena/Balance/DataTables/AthenaGameData.AthenaGameData" "Default.LowGravRock.Duration")
        (:RowFound
          (Variables|Default|SetDuration _outxy))
        (:RowNotFound
          _))
      (bind _outxy_2 (DataTable|EvaluateCurveTableRow "/Game/Athena/Balance/DataTables/AthenaGameData.AthenaGameData" "Default.LowGravRock.ExpireTellDuration")
        (:RowFound
          (Variables|Default|SetExpirationTellDuration _outxy_2))
        (:RowNotFound
          _))))
  (CallFunction|ClearExpireTell)
  (bind _returnvalue (Utilities|Time|SetTimerbyEvent (EventDispatchers|CreateEvent :!event ExpireTell) (- (Variables|Default|GetDuration) (Variables|Default|GetExpirationTellDuration))))
  (Variables|Default|SetExpireTellDelayTimer _returnvalue)
  (Variables|Default|SetStartTime (Utilities|Time|GetTimeSeconds (Variables|Default|GetPlayerPawn)))
  (return true))

(fn OnCeaseRelevant
  (params (MyTarget :type Actor%20Object%20Reference) (Parameters :type Gameplay%20Cue%20Parameters%20Structure%20%28by%20ref%29))
  (Actor|Tick|SetActorTickEnabled false)
  (CallFunction|ClearExpireTell)
  (bind _lowgravitypscomponent (Variables|Default|GetLowGravityPSComponent)
    (:then
      (Components|Activation|Deactivate _lowgravitypscomponent)
      (return true))
    (:else)))

(fn ConstructionScript)

(fn OnBecomeRelevant
  (params (MyTarget :type Actor%20Object%20Reference) (Parameters :type Gameplay%20Cue%20Parameters%20Structure%20%28by%20ref%29))
  (CallFunction|Activated MyTarget)
  (Actor|Tick|SetActorTickEnabled true)
  (bind _returnvalue (Effects|SpawnEmitterAttached "/FNGameplayCues/Athena/Items/ForagedItems/LowGravity/FX/P_Athena_LowGravity_Loop.P_Athena_LowGravity_Loop" (Variables|/Script/Engine.Character|GetMesh (Variables|Default|GetPlayerPawn))))
  (bind _output_get (Variables|Default|SetLowGravityPSComponent _returnvalue))
  (LOD|SetMaxDrawDistance _output_get (Variables|Default|GetMaxDrawDistance))
  (return false))
