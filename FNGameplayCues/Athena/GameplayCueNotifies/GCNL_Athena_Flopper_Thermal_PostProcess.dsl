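; Looping gameplay cue for the Flopper thermal effect, driving a thermal post process on the instigator's local viewing player.
; On start it makes the Fort player controller ignore significance-based custom depth rendering and adds the thermal camera modifier.
; A timeline ramps ThermalScaleFlopper in the global thermal parameter collection at the fade-in rate.
; Removal restores the depth setting and reverses the timeline at the fade-out rate, which removes the modifier when finished.
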(event EventOnLoopingStart(Cascade)
  (params (MyTarget :type Actor%20Object%20Reference) (Parameters :type Gameplay%20Cue%20Parameters%20Structure%20%28by%20ref%29) (ParticleComponents :type Array%20of%20Cascade%20Particle%20System%20Component%20Object%20References) (AudioComponents :type Array%20of%20Audio%20Component%20Object%20References))
  (bind _instigator (Variables|/Script/Engine.Actor|GetInstigator MyTarget))
  (bind _returnvalue (Pawn|GetLocalViewingPlayerController _instigator))
  (bind _asfort_player_controller (Utilities|Casting|CastToFortPlayerController _returnvalue)
    (:then
      (bind _output_get (Variables|Default|SetFortPlayerController _asfort_player_controller))
      (Rendering|SetIgnoreSignifanceBasedCustomDepthRendering _output_get true 0))
    (:CastFailed
      _))
  (bind _output_get_2 (Variables|Default|SetPlayerCameraManager (Variables|/Script/Engine.PlayerController|GetPlayerCameraManager _returnvalue)))
  (bind _returnvalue_2 (Game|Player|AddNewCameraModifier _output_get_2 "/FNGameplayCues/Athena/Items/Consumables/Flopper/Thermal/B_Athena_Flopper_Thermal_CameraModifier.B_Athena_Flopper_Thermal_CameraModifier_C"))
  (Variables|Default|SetThermalPostProcessCameraModifier _returnvalue_2)
  (Components|Timeline|SetPlayRate (Variables|GCNL_Athena_Flopper_Thermal_PostProcess|GetTimeline_0) (Variables|Default|GetFadeInPlayrate))
  (bind (_direction _thermalramp) (Timeline|Timeline_0 PlayFromStart)
    (:Update
      (Rendering|Material|SetScalarParameterValue 0 "/Game/Athena/Items/Weapons/Abilities/Thermal/MPC_ThermalGlobal.MPC_ThermalGlobal" "ThermalScaleFlopper" _thermalramp :!class /Script/Engine.KismetMaterialLibrary))
    (:Finished
      (switch Utilities|FlowControl|Switch|SwitchonETimelineDirection _direction
        (:Backward
          (Game|Player|RemoveCameraModifier (Variables|Default|GetPlayerCameraManager) (Variables|Default|GetThermalPostProcessCameraModifier)))))))

(event EventOnRemoval(Cascade)
  (params (MyTarget :type Actor%20Object%20Reference) (Parameters :type Gameplay%20Cue%20Parameters%20Structure%20%28by%20ref%29) (ParticleComponents :type Array%20of%20Cascade%20Particle%20System%20Component%20Object%20References) (AudioComponents :type Array%20of%20Audio%20Component%20Object%20References) (BurstCameraShakeInstance :type Camera%20Shake%20Base%20Object%20Reference) (BurstDecalInstance :type Decal%20Actor%20Object%20Reference))
  (bind _fortplayercontroller (Variables|Default|GetFortPlayerController))
  (Utilities|IsValid _fortplayercontroller
    (:Is%20Valid
      (Rendering|SetIgnoreSignifanceBasedCustomDepthRendering _fortplayercontroller false 0))
    (:Is%20Not%20Valid
      _))
  (Components|Timeline|SetPlayRate (Variables|GCNL_Athena_Flopper_Thermal_PostProcess|GetTimeline_0) (Variables|Default|GetFadeOutPlayrate))
  (Timeline|Timeline_0 ReverseFromEnd))

(event Collision|EventActorBeginOverlap
  (params (OtherActor :type Actor%20Object%20Reference)))

(event EventTick
  (params (DeltaSeconds :type Float%20%28single-precision%29)))

(fn ConstructionScript)
