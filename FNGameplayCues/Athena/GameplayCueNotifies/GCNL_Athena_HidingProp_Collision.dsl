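; Looping gameplay cue notify for a player hiding in a prop, handling collision and animation on the target.
; When looping starts, it disables the target actor's collision and plays the hiding loop montage on it as a FortPlayerPawn.
; On removal, it re-enables collision and stops that montage.
; Begin play, overlap and tick events are empty.
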(event EventBeginPlay)

(event Collision|EventActorBeginOverlap
  (params (OtherActor :type Actor%20Object%20Reference)))

(event EventTick
  (params (DeltaSeconds :type Float%20%28single-precision%29)))

(event EventOnLoopingStart(Cascade)
  (params (MyTarget :type Actor%20Object%20Reference) (Parameters :type Gameplay%20Cue%20Parameters%20Structure%20%28by%20ref%29) (ParticleComponents :type Array%20of%20Cascade%20Particle%20System%20Component%20Object%20References) (AudioComponents :type Array%20of%20Audio%20Component%20Object%20References))
  (Collision|SetActorEnableCollision MyTarget false)
  (bind _asfort_player_pawn (Utilities|Casting|CastToFortPlayerPawn MyTarget))
  (Animation|PlayAnimMontage _asfort_player_pawn "/Game/Animation/Game/MainPlayer/Reacts/WorldInteraction/Hideable/Hiding_HidingLoopMontage.Hiding_HidingLoopMontage" 1.0 "None"))

(event EventOnRemoval(Cascade)
  (params (MyTarget :type Actor%20Object%20Reference) (Parameters :type Gameplay%20Cue%20Parameters%20Structure%20%28by%20ref%29) (ParticleComponents :type Array%20of%20Cascade%20Particle%20System%20Component%20Object%20References) (AudioComponents :type Array%20of%20Audio%20Component%20Object%20References) (BurstCameraShakeInstance :type Camera%20Shake%20Base%20Object%20Reference) (BurstDecalInstance :type Decal%20Actor%20Object%20Reference))
  (Collision|SetActorEnableCollision MyTarget true)
  (bind _asfort_player_pawn (Utilities|Casting|CastToFortPlayerPawn MyTarget))
  (Animation|StopAnimMontage _asfort_player_pawn "/Game/Animation/Game/MainPlayer/Reacts/WorldInteraction/Hideable/Hiding_HidingLoopMontage.Hiding_HidingLoopMontage"))

(fn ConstructionScript)
