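; Looping gameplay cue for secret passage travel through hiding props, giving the teleporting player first-person camera and audio feedback.
; On start, it clears any old lens effect and, for locally controlled pawns, spawns a looping camera lens effect, 2D travel sound and sound mix modifier.
; On removal, a timeline drives the lens particle's WhiteWash and triggers a flash before destroying it.
; The owning pawn's sound fades out and its mix pops.
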(event EventOnLoopingStart(Cascade)
  (params (MyTarget :type Actor%20Object%20Reference) (Parameters :type Gameplay%20Cue%20Parameters%20Structure%20%28by%20ref%29) (ParticleComponents :type Array%20of%20Cascade%20Particle%20System%20Component%20Object%20References) (AudioComponents :type Array%20of%20Audio%20Component%20Object%20References))
  (bind _lenseffectactor (Variables|Default|GetLensEffectActor))
  (Utilities|IsValid _lenseffectactor
    (:Is%20Valid
      (Actor|DestroyActor _lenseffectactor))
    (:Is%20Not%20Valid
      _))
  (bind _asfort_player_pawn (Utilities|Casting|CastToFortPlayerPawn MyTarget))
  (if (Pawn|IsLocallyControlled _asfort_player_pawn)
    (Variables|Default|SetOwningActor _asfort_player_pawn)
    (bind _returnvalue (Feedback|SpawnGenericCameraLensEffect _asfort_player_pawn "/FNGameplayCues/Effects/Fort_Effects/Camera_Effects/BP_CameraLens_HidingProp_Teleporting_Looping.BP_CameraLens_HidingProp_Teleporting_Looping_C"))
    (bind _output_get (Variables|Default|SetLensEffectInterface _returnvalue))
    (Variables|Default|SetLensEffectActor (Utilities|Casting|CastToActor (Utilities|ToObject(Interface) _output_get)))
    (Utilities|FlowControl|Sequence
      (:then_0
        (bind _1ptravelaudio (Variables|Default|Get1PTravelAudio)
          (:then
            (Audio|Components|Audio|FadeOut _1ptravelaudio 0.25 0.0 "Linear" :!class /Script/Engine.AudioComponent))
          (:else)))
      (:then_1
        (bind _returnvalue_2 (Audio|SpawnSound2D 0 "/FNGameplayCues/Sounds/Fort_Hideables/SecretPassage/SecretPassage_Travel_1P_Cue.SecretPassage_Travel_1P_Cue"))
        (Variables|Default|Set1PTravelAudio _returnvalue_2)
        (Audio|PushSoundMixModifier 0 (Variables|Default|Get1pMixMod))))))

(event EventOnRemoval(Cascade)
  (params (MyTarget :type Actor%20Object%20Reference) (Parameters :type Gameplay%20Cue%20Parameters%20Structure%20%28by%20ref%29) (ParticleComponents :type Array%20of%20Cascade%20Particle%20System%20Component%20Object%20References) (AudioComponents :type Array%20of%20Audio%20Component%20Object%20References) (BurstCameraShakeInstance :type Camera%20Shake%20Base%20Object%20Reference) (BurstDecalInstance :type Decal%20Actor%20Object%20Reference))
  (bind _lenseffectactor (Variables|Default|GetLensEffectActor))
  (bind _returnvalue (Actor|GetComponentbyClass _lenseffectactor "/Script/CascadeRuntime.ParticleSystemComponent"))
  (Utilities|FlowControl|Sequence
    (:then_0
      (Utilities|IsValid _lenseffectactor
        (:Is%20Valid
          (bind (_direction _lerpwhitewash) (Timeline|Timeline_0 PlayFromStart)
            (:Update
              (Effects|Components|ParticleSystem|SetFloatParameter _returnvalue "WhiteWash" _lerpwhitewash))
            (:DoFlash
              (Effects|Components|ParticleSystem|GenerateParticleEvent _returnvalue "SpawnFlash" 0.0 "0, 0, 0" "0, 0, 0" "0, 0, 0")
              (Utilities|FlowControl|Delay 0 0.4)
              (Components|Activation|Deactivate _returnvalue)
              (Actor|DestroyActor (Variables|Default|GetLensEffectActor)))))))
    (:then_1
      (bind _asfort_player_pawn (Utilities|Casting|CastToFortPlayerPawn MyTarget))
      (if (Utilities|Equal(Object) (Variables|Default|GetOwningActor) _asfort_player_pawn)
        (Audio|PopSoundMixModifier 0 (Variables|Default|Get1pMixMod))
        (bind _1ptravelaudio (Variables|Default|Get1PTravelAudio)
          (:then
            (Audio|Components|Audio|FadeOut _1ptravelaudio 0.5 0.0 "Linear" :!class /Script/Engine.AudioComponent))
          (:else))))))

(fn ConstructionScript)
