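; Looping gameplay cue that sets up a beam particle effect for the Lotus Heart GE delivery actor.
; On looping start it casts the cue's instigator to BGA_Lotus_Heart_GEDelivery and snaps the first particle component onto that actor's mesh.
; It then sets the particle's "Target" actor parameter to the effect causer and prints debug names for the component and instigator.
; Application only calls the parent.
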(event EventBeginPlay)

(event Collision|EventActorBeginOverlap
  (params (OtherActor :type Actor%20Object%20Reference)))

(event EventTick
  (params (DeltaSeconds :type Float%20%28single-precision%29)))

(event EventOnApplication(Cascade)
  (params (MyTarget :type Actor%20Object%20Reference) (Parameters :type Gameplay%20Cue%20Parameters%20Structure%20%28by%20ref%29) (ParticleComponents :type Array%20of%20Cascade%20Particle%20System%20Component%20Object%20References) (AudioComponents :type Array%20of%20Audio%20Component%20Object%20References) (BurstCameraShakeInstance :type Camera%20Shake%20Base%20Object%20Reference) (BurstDecalInstance :type Decal%20Actor%20Object%20Reference))
  (|CallParentFunction MyTarget Parameters ParticleComponents AudioComponents BurstCameraShakeInstance BurstDecalInstance))

(event EventOnLoopingStart(Cascade)
  (params (MyTarget :type Actor%20Object%20Reference) (Parameters :type Gameplay%20Cue%20Parameters%20Structure%20%28by%20ref%29) (ParticleComponents :type Array%20of%20Cascade%20Particle%20System%20Component%20Object%20References) (AudioComponents :type Array%20of%20Audio%20Component%20Object%20References))
  (bind (_normalizedmagnitude _rawmagnitude _effectcontext _matchedtagname _originaltag _aggregatedsourcetags _aggregatedtargettags _location _normal _instigator _effectcauser) (Ability|GameplayCue|BreakGameplayCueParameters Parameters))
  (bind _output (Utilities|Array|Get(acopy) ParticleComponents))
  (bind _asbga_lotus_heart_gedelivery (Utilities|Casting|CastToBGA_Lotus_Heart_GEDelivery _instigator))
  (Transformation|AttachComponentToComponent _output (Variables|/FNGameplayCues/Athena/Items/Gameplay/Lotus/Heart/BGA_Lotus_Heart_GEDelivery.BGA_Lotus_Heart_GEDelivery_C|GetMesh _asbga_lotus_heart_gedelivery) "None" "SnapToTarget" "KeepRelative" "KeepRelative" true)
  (Effects|Components|ParticleSystem|SetActorParameter _output "Target" _effectcauser)
  (Development|PrintString _output (Utilities|GetDisplayName _output))
  (Development|PrintString _output (Utilities|GetDisplayName _instigator) true true "(R=1.000000,G=0.875123,B=0.000000,A=1.000000)"))

(fn ConstructionScript)
