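; Looping gameplay cue for multi-player interactions that plays audio and a looping sound indicator on the effect causer pawn.
; On apply it attaches the audio to the head, fades it in from a random start time and binds to MultiInteractActorComponent's actor-count delegate.
; Count changes nudge pitch toward a count-scaled target with a single FInterpTo step; the repeating timer only checks for completion.
; Removal fades out, stops the indicator and unbinds.
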(event EventOnApplication(EitherVFX-preferthisone)
  (params (MyTarget :type Actor%20Object%20Reference) (Parameters_NormalizedMagnitude :type Float%20%28single-precision%29) (Parameters_RawMagnitude :type Float%20%28single-precision%29) (Parameters_EffectContext :type Gameplay%20Effect%20Context%20Handle%20Structure) (Parameters_MatchedTagName :type Gameplay%20Tag%20Structure) (Parameters_OriginalTag :type Gameplay%20Tag%20Structure) (Parameters_AggregatedSourceTags :type Gameplay%20Tag%20Container%20Structure) (Parameters_AggregatedTargetTags :type Gameplay%20Tag%20Container%20Structure) (Parameters_Location :type Vector :default "0, 0, 0") (Parameters_Normal :type Vector :default "0, 0, 0") (Parameters_Instigator :type Actor%20Object%20Reference) (Parameters_EffectCauser :type Actor%20Object%20Reference) (Parameters_SourceObject :type Object%20Reference) (Parameters_PhysicalMaterial :type Physical%20Material%20Object%20Reference) (Parameters_GameplayEffectLevel :type Integer) (Parameters_AbilityLevel :type Integer) (Parameters_TargetAttachComponent :type Scene%20Component%20Object%20Reference) (Parameters_bReplicateLocationWhenUsingMinimalRepProxy :type Boolean) (ParticleComponents :type Array%20of%20FXSystem%20Component%20Object%20References) (AudioComponents :type Array%20of%20Audio%20Component%20Object%20References) (BurstCameraShakeInstance :type Camera%20Shake%20Base%20Object%20Reference) (BurstDecalInstance :type Decal%20Actor%20Object%20Reference))
  (bind _asfort_player_pawn_athena (Utilities|Casting|CastToFortPlayerPawnAthena Parameters_EffectCauser))
  (bind _output_get (Variables|Default|SetEffect%20Causer%20Pawn _asfort_player_pawn_athena))
  (bind _audio (Variables|Default|GetAudio)
    (:then
      (Transformation|AttachComponentToComponent _audio (Variables|/Script/Engine.Character|GetMesh _output_get) "head" "KeepRelative" "KeepRelative" "KeepRelative" true)
      (Audio|Components|Audio|FadeIn _audio 0.7 1.0 (Math|Random|RandomFloatinRange 0.1 5.0) "Linear" :!class /Script/Engine.AudioComponent)
      (SoundIndicator|StartSoundIndicatorLooping (Variables|/Script/FortniteGame.FortPawn|GetSoundIndicatorComponent _output_get) (Transformation|GetActorLocation _output_get) (Variables|/Script/Engine.AudioComponent|GetSound _audio) 1.0 0)
      (bind _returnvalue (Actor|GetComponentbyClass _output_get "/Game/Athena/Items/Components/MultiInteractActorComponent.MultiInteractActorComponent_C"))
      (Utilities|IsValid _returnvalue
        (:Is%20Valid
          (EventDispatchers|BindEventtoActorCountChangedDelegate _returnvalue (delegate Custom|OnActorCountChanged)))))
    (:else)))

(event Custom|OnActorCountChanged
  (params (InteractingActorCount :type Integer) (MultiInteractMultiplier :type Float%20%28single-precision%29) (InteractingDuration :type Float%20%28single-precision%29))
  ;[0 0.9822 1] On Actor Count Changed
  (bind _returnvalue (Utilities|Time|SetTimerbyEvent (delegate Custom|CheckTargetProgress) 0.1 true))
  (Variables|Default|SetTargetProgressHandle _returnvalue)
  (bind _audio (Variables|Default|GetAudio)
    (:then
      (bind _output_get (Variables|Default|SetTargetPitch (* (Variables|Default|GetCurrentPitch) (Math|Float|MapRangeClamped (Math|Conversions|ToFloat(Integer) InteractingActorCount) 1.0 4.0 1.0 3.0))))
      (bind _output_get_2 (Variables|Default|SetCurrentPitch (Variables|/Script/Engine.AudioComponent|GetPitchMultiplier _audio)))
      (Audio|Components|Audio|SetPitchMultiplier _audio (Math|Interpolation|FInterpTo _output_get_2 _output_get (Utilities|Time|GetWorldDeltaSeconds _audio) 1.23)))
    (:else)))

(event Custom|CheckTargetProgress
  ;[0 0.9822 1] On Actor Count Changed
  (if (== (Variables|Default|GetCurrentPitch) (Variables|Default|GetTargetPitch))
    (Utilities|Time|ClearandInvalidateTimerbyHandle (stale Custom|OnActorCountChanged _audio) (Variables|Default|GetTargetProgressHandle))))

(event EventOnRemoval(EitherVFX-preferthisversion)
  (params (MyTarget :type Actor%20Object%20Reference) (Parameters :type Gameplay%20Cue%20Parameters%20Structure%20%28by%20ref%29) (ParticleComponents :type Array%20of%20FXSystem%20Component%20Object%20References) (AudioComponents :type Array%20of%20Audio%20Component%20Object%20References) (BurstCameraShakeInstance :type Camera%20Shake%20Base%20Object%20Reference) (BurstDecalInstance :type Decal%20Actor%20Object%20Reference))
  ;[1 0.6476 0] Event On Removal
  (Utilities|FlowControl|Sequence
    (:then_0
      (bind _audio (Variables|Default|GetAudio)
        (:then
          (Audio|Components|Audio|FadeOut _audio 0.6 0.0 "Linear" :!class /Script/Engine.AudioComponent)
          (bind _effect_causer_pawn (Variables|Default|GetEffect%20Causer%20Pawn)
            (:then
              (SoundIndicator|StopSoundIndicatorLooping (Variables|/Script/FortniteGame.FortPawn|GetSoundIndicatorComponent _effect_causer_pawn)))
            (:else)))
        (:else)))
    (:then_1
      (bind _effect_causer_pawn_2 (Variables|Default|GetEffect%20Causer%20Pawn)
        (:then
          (bind _returnvalue (Actor|GetComponentbyClass _effect_causer_pawn_2 "/Game/Athena/Items/Components/MultiInteractActorComponent.MultiInteractActorComponent_C"))
          (Utilities|IsValid _returnvalue
            (:Is%20Valid
              (EventDispatchers|UnbindEventfromActorCountChangedDelegate _returnvalue (EventDispatchers|CreateEvent :!event Custom|OnActorCountChanged)))
            (:Is%20Not%20Valid
              _))
          (Variables|Default|SetEffect%20Causer%20Pawn 0))
        (:else)))))

(fn ConstructionScript)
