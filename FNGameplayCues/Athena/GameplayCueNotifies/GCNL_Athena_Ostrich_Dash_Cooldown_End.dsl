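; Looping gameplay cue for the Ostrich mech's dash cooldown end, placing particle effects on the mech's thighs.
; On the Cascade removal event, it casts the target to FortCharacterVehicle, stores it as Mech and fetches its mech mesh.
; The first two particle components then snap in location and rotation to the left and right startup thigh sockets, keeping relative scale.
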(event EventOnRemoval(Cascade)
  (params (MyTarget :type Actor%20Object%20Reference) (Parameters :type Gameplay%20Cue%20Parameters%20Structure%20%28by%20ref%29) (ParticleComponents :type Array%20of%20Cascade%20Particle%20System%20Component%20Object%20References) (AudioComponents :type Array%20of%20Audio%20Component%20Object%20References) (BurstCameraShakeInstance :type Camera%20Shake%20Base%20Object%20Reference) (BurstDecalInstance :type Decal%20Actor%20Object%20Reference))
  (bind _asfort_character_vehicle (Utilities|Casting|CastToFortCharacterVehicle MyTarget))
  (bind _output_get (Variables|Default|SetMech _asfort_character_vehicle))
  (Transformation|AttachComponentToComponent (Utilities|Array|Get(acopy) ParticleComponents) (Mesh|GetMechMesh _output_get) "FX_Startup_Thigh_L" "SnapToTarget" "SnapToTarget" "KeepRelative" true) ; Left
  (Transformation|AttachComponentToComponent (Utilities|Array|Get(acopy) ParticleComponents 1) (Mesh|GetMechMesh _output_get) "FX_Startup_Thigh_R" "SnapToTarget" "SnapToTarget" "KeepRelative" true) ; Right
)

(fn ConstructionScript)
