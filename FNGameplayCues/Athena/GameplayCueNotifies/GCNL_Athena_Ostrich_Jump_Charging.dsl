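; Looping gameplay cue that plays the charge-up effects for the Ostrich mech vehicle's charged jump.
; Loop start attaches the first particle component to the mech mesh root and spawns an attached charge sound, cached to avoid duplicates.
; Removal fades out that sound and, for a positive raw magnitude, sets the IsFullyCharged parameter on the removal audio according to a charge threshold.
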(event EventOnLoopingStart(Cascade)
  (params (MyTarget :type Actor%20Object%20Reference) (Parameters :type Gameplay%20Cue%20Parameters%20Structure%20%28by%20ref%29) (ParticleComponents :type Array%20of%20Cascade%20Particle%20System%20Component%20Object%20References) (AudioComponents :type Array%20of%20Audio%20Component%20Object%20References))
  (bind _asfort_character_vehicle_ostrich (Utilities|Casting|CastToFortCharacterVehicle_Ostrich MyTarget))
  (Transformation|AttachComponentToComponent (Utilities|Array|Get(acopy) ParticleComponents) (Mesh|GetMechMesh _asfort_character_vehicle_ostrich) "root" "KeepRelative" "KeepRelative" "KeepRelative" true)
  (Variables|Default|GetChargeJumpSound
    (:then)
    (:else
      (bind _returnvalue (Audio|SpawnSoundAttached "/FNGameplayCues/Athena/Sounds/Vehicles/Ostrich/Veh_Ostrich_Charged_Jump_Charge_Cue.Veh_Ostrich_Charged_Jump_Charge_Cue" (Mesh|GetMechMesh _asfort_character_vehicle_ostrich)))
      (Variables|Default|SetChargeJumpSound _returnvalue))))

(event EventOnApplication(Cascade)
  (params (MyTarget :type Actor%20Object%20Reference) (Parameters :type Gameplay%20Cue%20Parameters%20Structure%20%28by%20ref%29) (ParticleComponents :type Array%20of%20Cascade%20Particle%20System%20Component%20Object%20References) (AudioComponents :type Array%20of%20Audio%20Component%20Object%20References) (BurstCameraShakeInstance :type Camera%20Shake%20Base%20Object%20Reference) (BurstDecalInstance :type Decal%20Actor%20Object%20Reference)))

(event EventOnRemoval(Cascade)
  (params (MyTarget :type Actor%20Object%20Reference) (Parameters :type Gameplay%20Cue%20Parameters%20Structure%20%28by%20ref%29) (ParticleComponents :type Array%20of%20Cascade%20Particle%20System%20Component%20Object%20References) (AudioComponents :type Array%20of%20Audio%20Component%20Object%20References) (BurstCameraShakeInstance :type Camera%20Shake%20Base%20Object%20Reference) (BurstDecalInstance :type Decal%20Actor%20Object%20Reference))
  (bind (_normalizedmagnitude _rawmagnitude) (Ability|GameplayCue|BreakGameplayCueParameters Parameters))
  (Utilities|FlowControl|Sequence
    (:then_0
      (bind _chargejumpsound (Variables|Default|GetChargeJumpSound)
        (:then
          (Audio|Components|Audio|FadeOut _chargejumpsound 0.25 0.0 "Linear" :!class /Script/Engine.AudioComponent))
        (:else)))
    (:then_1
      (if (> _rawmagnitude 0.0)
        (Audio|Parameter|SetIntegerParameter (Utilities|Array|Get(acopy) AudioComponents) "IsFullyCharged" (Math|Integer|SelectInt 1 0 (>= _rawmagnitude 0.75)) :!class /Script/AudioExtensions.AudioParameterControllerInterface)))))

(fn ConstructionScript)
