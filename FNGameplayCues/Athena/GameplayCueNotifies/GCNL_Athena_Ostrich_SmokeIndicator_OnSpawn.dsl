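; Looping gameplay cue notify for the Ostrich mech's spawn smoke indicator.
; On looping start, it casts the target to a FortCharacterVehicle and stores it as the mech.
; It then attaches the first spawned Cascade particle component to the root of the mech mesh, snapping location and rotation while keeping relative scale.
; The particle and audio assets are not defined in this graph, and the construction script is empty.
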(event EventOnLoopingStart(Cascade)
  (params (MyTarget :type Actor%20Object%20Reference) (Parameters :type Gameplay%20Cue%20Parameters%20Structure%20%28by%20ref%29) (ParticleComponents :type Array%20of%20Cascade%20Particle%20System%20Component%20Object%20References) (AudioComponents :type Array%20of%20Audio%20Component%20Object%20References))
  (bind _asfort_character_vehicle (Utilities|Casting|CastToFortCharacterVehicle MyTarget))
  (bind _output_get (Variables|Default|SetMech _asfort_character_vehicle))
  (Transformation|AttachComponentToComponent (Utilities|Array|Get(acopy) ParticleComponents) (Mesh|GetMechMesh _output_get) "root" "SnapToTarget" "SnapToTarget" "KeepRelative" true) ; Left
)

(fn ConstructionScript)
