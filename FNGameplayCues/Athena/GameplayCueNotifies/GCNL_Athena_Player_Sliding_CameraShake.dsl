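; Looping gameplay cue notify that adds speed-based camera shake and post-process handling while a player slides.
; On application it enables the post-process for locally viewed pawns; a repeating timer started on looping start plays the reused zipline downhill shake above a speed threshold.
; Cubed, clamped speed ratio scales the shake's Perlin rotational amplitude and frequency.
; A temporary altitude latch gates post-process blend, and removal disables it and clears the timer.
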(event EventOnApplication(EitherVFX-preferthisone)
  (params (MyTarget :type Actor%20Object%20Reference) (Parameters :type Gameplay%20Cue%20Parameters%20Structure%20%28by%20ref%29) (ParticleComponents :type Array%20of%20FXSystem%20Component%20Object%20References) (AudioComponents :type Array%20of%20Audio%20Component%20Object%20References) (BurstCameraShakeInstance :type Camera%20Shake%20Base%20Object%20Reference) (BurstDecalInstance :type Decal%20Actor%20Object%20Reference))
  ;[1 0 0.2766] Temp fixed
  (Variables|Default|SetbIsSuperHigh false)
  (Variables|Default|SetPlayerPawn (Utilities|Casting|CastToFortPlayerPawn MyTarget))
  (if (Pawn|IsLocallyViewed (Variables|Default|GetPlayerPawn))
    (Variables|/Script/Engine.PostProcessComponent|SetbEnabled true (Variables|Default|GetPostProcess))))

(event EventOnRemoval(EitherVFX-preferthisversion)
  (params (MyTarget :type Actor%20Object%20Reference) (Parameters :type Gameplay%20Cue%20Parameters%20Structure%20%28by%20ref%29) (ParticleComponents :type Array%20of%20FXSystem%20Component%20Object%20References) (AudioComponents :type Array%20of%20Audio%20Component%20Object%20References) (BurstCameraShakeInstance :type Camera%20Shake%20Base%20Object%20Reference) (BurstDecalInstance :type Decal%20Actor%20Object%20Reference))
  (Utilities|FlowControl|Sequence
    (:then_1
      (Variables|/Script/Engine.PostProcessComponent|SetBlendWeight 0.0 (Variables|Default|GetPostProcess))
      (Variables|/Script/Engine.PostProcessComponent|SetbEnabled false (Variables|Default|GetPostProcess))
      (Utilities|Time|ClearandInvalidateTimerbyHandle self (stale EventOnLoopingStart(EitherVFX-preferthisversion) _returnvalue)))) ; Stop timeline in case boost ends before timeline.
)

(event EventOnLoopingStart(EitherVFX-preferthisversion)
  (params (MyTarget :type Actor%20Object%20Reference) (Parameters :type Gameplay%20Cue%20Parameters%20Structure%20%28by%20ref%29) (ParticleComponents :type Array%20of%20FXSystem%20Component%20Object%20References) (AudioComponents :type Array%20of%20Audio%20Component%20Object%20References))
  ;[0 0.2812 0.2431] In some cases this runs before On Application, so make sure the pawn is up to date.
  (bind _asfort_player_pawn (Utilities|Casting|CastToFortPlayerPawn MyTarget))
  (Variables|Default|SetPlayerPawn _asfort_player_pawn)
  (bind _returnvalue (Utilities|Time|SetTimerbyEvent (delegate Custom|cameraShakeTimer) 0.3 true)))

(event Custom|cameraShakeTimer
  (bind _playerpawn (Variables|Default|GetPlayerPawn))
  (bind _returnvalue (Transformation|GetVelocity _playerpawn))
  (bind _returnvalue_2 (Math|Vector|VectorLength _returnvalue))
  (bind _returnvalue_3 (Pawn|GetLocalViewingPlayerController _playerpawn))
  (bind _postprocess (Variables|Default|GetPostProcess))
  (Utilities|FlowControl|Sequence
    (:then_0
      (Utilities|IsValid _returnvalue_3
        (:Is%20Valid
          (if (> _returnvalue_2 (Variables|Default|GetCameraShakeSpeed))
            (bind _returnvalue_4 (CameraShakes|StartCameraShake :self (Variables|/Script/Engine.PlayerController|GetPlayerCameraManager _returnvalue_3) :ShakeClass "/FNGameplayCues/Athena/Items/Consumables/ZipLine/CameraShake_ZipLine_DownhillSpeed.CameraShake_ZipLine_DownhillSpeed_C" :Scale 1.0 :PlaySpace "CameraLocal"))
            (bind _output_get (Variables|Default|SetmatineeCamShake (CameraShakes|ConvLegacyCameraShake _returnvalue_4)))
            (Variables|/Script/EngineCameras.LegacyCameraShake|SetRotOscillation (Utilities|Struct|MakeROscillator (Utilities|Struct|MakeFOscillator (Math|Float|Lerp (Variables|Default|GetMinCamShakeAmp) (Variables|Default|GetMaxCamShakeAmp) (Math|Float|Clamp(Float) (Math|Float|float*float (/ _returnvalue_2 (ScalableFloat|GetValueatLevel (Variables|Default|GetMaxSlideSpeed))) (/ _returnvalue_2 (ScalableFloat|GetValueatLevel (Variables|Default|GetMaxSlideSpeed))) (/ _returnvalue_2 (ScalableFloat|GetValueatLevel (Variables|Default|GetMaxSlideSpeed)))))) (Math|Float|Lerp (Variables|Default|GetMinCamShakeFreq) (Variables|Default|GetMaxCamShakeFreq) (Math|Float|Clamp(Float) (Math|Float|float*float (/ _returnvalue_2 (ScalableFloat|GetValueatLevel (Variables|Default|GetMaxSlideSpeed))) (/ _returnvalue_2 (ScalableFloat|GetValueatLevel (Variables|Default|GetMaxSlideSpeed))) (/ _returnvalue_2 (ScalableFloat|GetValueatLevel (Variables|Default|GetMaxSlideSpeed)))))) "PerlinNoise") (Utilities|Struct|MakeFOscillator (Math|Float|Lerp (Variables|Default|GetMinCamShakeAmp) (Variables|Default|GetMaxCamShakeAmp) (Math|Float|Clamp(Float) (Math|Float|float*float (/ _returnvalue_2 (ScalableFloat|GetValueatLevel (Variables|Default|GetMaxSlideSpeed))) (/ _returnvalue_2 (ScalableFloat|GetValueatLevel (Variables|Default|GetMaxSlideSpeed))) (/ _returnvalue_2 (ScalableFloat|GetValueatLevel (Variables|Default|GetMaxSlideSpeed)))))) (Math|Float|Lerp (Variables|Default|GetMinCamShakeFreq) (Variables|Default|GetMaxCamShakeFreq) (Math|Float|Clamp(Float) (Math|Float|float*float (/ _returnvalue_2 (ScalableFloat|GetValueatLevel (Variables|Default|GetMaxSlideSpeed))) (/ _returnvalue_2 (ScalableFloat|GetValueatLevel (Variables|Default|GetMaxSlideSpeed))) (/ _returnvalue_2 (ScalableFloat|GetValueatLevel (Variables|Default|GetMaxSlideSpeed)))))) "PerlinNoise") (Utilities|Struct|MakeFOscillator (Math|Float|Lerp (Variables|Default|GetMinCamShakeAmp) (Variables|Default|GetMaxCamShakeAmp) (Math|Float|Clamp(Float) (Math|Float|float*float (/ _returnvalue_2 (ScalableFloat|GetValueatLevel (Variables|Default|GetMaxSlideSpeed))) (/ _returnvalue_2 (ScalableFloat|GetValueatLevel (Variables|Default|GetMaxSlideSpeed))) (/ _returnvalue_2 (ScalableFloat|GetValueatLevel (Variables|Default|GetMaxSlideSpeed)))))) (Math|Float|Lerp (Variables|Default|GetMinCamShakeFreq) (Variables|Default|GetMaxCamShakeFreq) (Math|Float|Clamp(Float) (Math|Float|float*float (/ _returnvalue_2 (ScalableFloat|GetValueatLevel (Variables|Default|GetMaxSlideSpeed))) (/ _returnvalue_2 (ScalableFloat|GetValueatLevel (Variables|Default|GetMaxSlideSpeed))) (/ _returnvalue_2 (ScalableFloat|GetValueatLevel (Variables|Default|GetMaxSlideSpeed)))))) "PerlinNoise")) _output_get)))))
    (:then_1
      ;[1 0 0.2897] Temp fixed
      (if (or (> (.z (Transformation|GetActorLocation (Variables|Default|GetPlayerPawn))) 6e+04) (not (Utilities|IsValid (Variables|Default|GetPlayerPawn) :!class /Script/Engine.KismetSystemLibrary)))
        (Variables|Default|SetbIsSuperHigh true)
        (elif (Variables|Default|GetbIsSuperHigh)
          _
          (else
            (Variables|/Script/Engine.PostProcessComponent|SetBlendWeight 1.0 _postprocess))))
      (Variables|/Script/Engine.PostProcessComponent|SetBlendWeight 0.0 _postprocess))))

(fn ConstructionScript)
