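; Looping gameplay cue for player sliding, driving slide audio, dust particles and the sound indicator.
; Start and stop one-shots and an attached layered loop vary by local view, affiliation and surface data-table rows.
; A fast timer feeds velocity and surface changes into the audio.
; A slower timer updates dust surface type and mutes dust and audio during the falling grace period; removal stops audio and dust.
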(event Custom|ResumeSlidingAudio
  (Sound|StartSlideAudio)
  (CallFunction|ToggleLoopingSoundIndicator true))

(event EventOnLoopingStart(EitherVFX-preferthisversion)
  (params (MyTarget :type Actor%20Object%20Reference) (Parameters :type Gameplay%20Cue%20Parameters%20Structure%20%28by%20ref%29) (ParticleComponents :type Array%20of%20FXSystem%20Component%20Object%20References) (AudioComponents :type Array%20of%20Audio%20Component%20Object%20References))
  (bind _returnvalue (Actor|GetComponentbyClass MyTarget "/Script/FortniteGame.FortMovementComp_Character"))
  (bind _asfort_player_pawn (Utilities|Casting|CastToFortPlayerPawn MyTarget))
  (Variables|FortGameplayCueNotifyLoop_Sliding|SetPlayerPawn _asfort_player_pawn)
  (Variables|Default|SetDust_FX (Utilities|Array|Get(acopy) ParticleComponents))
  (CallFunction|UpdateSurfaceType)
  (Utilities|FlowControl|Sequence
    (:then_0
      (Sound|StartSlideAudio)
      (CallFunction|ToggleLoopingSoundIndicator true))
    (:then_1
      ;[1 0.877 0 0.089 14] To hide dust/stop spawning slide dust in-air/falling
      (Utilities|IsValid _returnvalue
        (:Is%20Valid
          (Variables|Default|SetCached_Movement_ComponentRef _returnvalue))))
    (:then_2
      (bind _returnvalue_2 (Utilities|Time|SetTimerbyEvent (delegate Custom|Update%20Speed%20Line) 1.0 true))
      (Variables|Default|SetSpeedLineTimerHandle _returnvalue_2))))

(event Custom|Sliding%20Failsafe%20Check
  ;[1 1 1] Failsafe to address sliding being stuck on
  (if (not (Variables|/Script/FortniteGame.FortPawn|GetbIsSliding (Variables|FortGameplayCueNotifyLoop_Sliding|GetPlayerPawn)))
    (CallFunction|StopSlidingAudio)
    (Components|Activation|Deactivate (Variables|Default|GetDust_FX))))

(event Custom|Update%20Speed%20Line
  (Utilities|FlowControl|Sequence
    (:then_0
      ;[0.422 0.422 0.422 0.062 14] Update : Test if state has changed
      (bind _cached_movement_componentref (Variables|Default|GetCached_Movement_ComponentRef)
        (:then
          (if (!= (Variables|Default|GetbIsGraceFalling) (Variables|/Script/FortniteGame.FortMovementComp_Character|GetbInFallingGracePeriod _cached_movement_componentref))
            ;[1 1 1] Stop audio / dust fx while in the sliding grace falling state
            (bind _output_get (Variables|Default|SetbIsGraceFalling (Variables|/Script/FortniteGame.FortMovementComp_Character|GetbInFallingGracePeriod _cached_movement_componentref)))
            (bind _dust_fx (Variables|Default|GetDust_FX)
              (:then
                (Effects|Components|ParticleSystem|SetBoolParameter _dust_fx "IsSpawnDust" (not _output_get)))
              (:else
                _))
            (if _output_get
              (CallFunction|StopSlidingAudio)
              (else
                (CallFunction|ResumeSlidingAudio)))))
        (:else)))
    (:then_1
      (CallFunction|UpdateSurfaceType))))

(event Custom|UpdateSurfaceType
  (bind _dust_fx (Variables|Default|GetDust_FX))
  (bind _surfacetypeint (CallFunction|CheckSurfaceType (Class|FortSurfaceTrackingComponent|GetCurrentSurfaceType (Footsteps|GetSurfaceTrackingComponent (Variables|FortGameplayCueNotifyLoop_Sliding|GetPlayerPawn)))))
  (Effects|Components|ParticleSystem|SetIntParameter (Variables|Default|GetDust_FX) "SurfaceTypeIndex" _surfacetypeint)
  ;[1 1 1] TurnOffDustWhileOnTrain
  (Utilities|FlowControl|DelayUntilNextTick 0)
  (if (Variables|/Script/FortniteGame.FortPlayerPawn|GetbFXPlayDustOnMovement (Variables|FortGameplayCueNotifyLoop_Sliding|GetPlayerPawn))
    (Components|Activation|Activate _dust_fx false)
    (else
      (Components|Activation|Deactivate _dust_fx))))

(event EventOnRemoval(EitherVFX-preferthisversion)
  (params (MyTarget :type Actor%20Object%20Reference) (Parameters :type Gameplay%20Cue%20Parameters%20Structure%20%28by%20ref%29) (ParticleComponents :type Array%20of%20FXSystem%20Component%20Object%20References) (AudioComponents :type Array%20of%20Audio%20Component%20Object%20References) (BurstCameraShakeInstance :type Camera%20Shake%20Base%20Object%20Reference) (BurstDecalInstance :type Decal%20Actor%20Object%20Reference))
  (CallFunction|StopSlidingAudio)
  (Utilities|Time|ClearandInvalidateTimerbyHandle MyTarget (Variables|Default|GetSpeedLineTimerHandle))
  (Components|Activation|Deactivate (Variables|Default|GetDust_FX)))

(event Custom|Stop%20Sliding%20Audio
  (Sound|ResetAudioData)
  (Sound|StopSlideAudio)
  (CallFunction|ToggleLoopingSoundIndicator false)
  (Utilities|Time|ClearandInvalidateTimerbyHandle (Variables|FortGameplayCueNotifyLoop_Sliding|GetPlayerPawn) (Variables|Default|GetSlidingFailsafeTimerHandle)))

(fn CheckSurfaceType
  (params (CurrentSurfaceType :type EPhysicalSurface%20Enum :default "SurfaceType_Default"))
  (if (Utilities|Enum|Equal(Enum) CurrentSurfaceType "SurfaceType23")
    (Variables|Default|SetSurfaceTypeInt 1)
    (elif (Utilities|Enum|Equal(Enum) CurrentSurfaceType "SurfaceType26")
      (Variables|Default|SetSurfaceTypeInt 2)
      (elif (Utilities|Enum|Equal(Enum) CurrentSurfaceType "SurfaceType24")
        (Variables|Default|SetSurfaceTypeInt 3)
        (elif (or (Utilities|Enum|Equal(Enum) CurrentSurfaceType "SurfaceType8") (Utilities|Enum|Equal(Enum) CurrentSurfaceType "SurfaceType30"))
          (Variables|Default|SetSurfaceTypeInt 4)
          (elif (or (Utilities|Enum|Equal(Enum) CurrentSurfaceType "SurfaceType10") (Utilities|Enum|Equal(Enum) CurrentSurfaceType "SurfaceType9"))
            (Variables|Default|SetSurfaceTypeInt 5)
            (else
              (Variables|Default|SetSurfaceTypeInt 0)))))))
  (return (Variables|Default|GetSurfaceTypeInt)))

(fn OnTargetPawnActorHit
  (params (SelfActor :type Actor%20Object%20Reference) (OtherActor :type Actor%20Object%20Reference) (NormalImpulse :type Vector) (Hit :type Hit%20Result%20Structure%20%28by%20ref%29))
  (Audio|OnCollision (Variables|Sound|GetFortCollisionAudio) (Transformation|GetActorLocation SelfActor) (Transformation|GetVelocity (Variables|FortGameplayCueNotifyLoop_Sliding|GetPlayerPawn)))
  (return))

(fn ResetAudioData
  (Class|FortLayeredAudioComponent|ResetAllFloatParameters (Variables|FortGameplayCueNotifyLoop_Sliding|GetLayeredAudioComp))
  (Collision|UnbindEventfromOnActorHit (Variables|FortGameplayCueNotifyLoop_Sliding|GetPlayerPawn) (EventDispatchers|CreateEvent :!event On%20Target%20Pawn%20Actor%20Hit)))

(fn SetSurfaceSounds
  (params (Surface :type EPhysicalSurface%20Enum :default "SurfaceType_Default"))
  (bind _audiodatatable (Variables|Sound|GetAudioDataTable))
  (bind _output_get (Variables|Sound|SetAudioSurfaceType Surface))
  (Utilities|IsValid _audiodatatable
    (:Is%20Valid
      (bind _returnvalue (Utilities|GetDataTableRow _audiodatatable (Fortnite|Utility|GetPhysicalSurfaceName _output_get))
        (:then
          (LocalVariables|SetLayeredVolume (.LayeredVolume_38_7E4D4BBC41D4C0FFA9888B8D5CCAD6E2 (Utilities|Struct|BreakStrutAudioCrouchSlide _returnvalue)))
          (LocalVariables|SetBasic%20Layer%20Waves (.BaseLayerWaves_9_7835FCA9472E1E4F7B69F782E5B9C888 (Utilities|Struct|BreakStrutAudioCrouchSlide _returnvalue)))
          (Variables|Sound|SetStartSurfaceWaves (.StartWaves_19_4C22E28241588AE34A9939B262F25AE4 (Utilities|Struct|BreakStrutAudioCrouchSlide _returnvalue)))
          (Variables|Sound|SetStopSurfaceWaves (.StopWaves_20_7B2B0E6642EF50E6D812D082272CF015 (Utilities|Struct|BreakStrutAudioCrouchSlide _returnvalue)))
          (Variables|Sound|SetStart%20Volume (.StartVolume_35_78ED97024257094AE87F2EAE56B1D9AC (Utilities|Struct|BreakStrutAudioCrouchSlide _returnvalue)))
          (Variables|Sound|SetStop%20Volume (.StopVolume_36_DC617BF74E54F699D38B4491E14B03DF (Utilities|Struct|BreakStrutAudioCrouchSlide _returnvalue))))
        (:RowNotFound
          (bind _returnvalue_2 (Utilities|GetDataTableRow (Variables|Sound|GetAudioDataTable) "Default")
            (:then
              (LocalVariables|SetLayeredVolume (.LayeredVolume_38_7E4D4BBC41D4C0FFA9888B8D5CCAD6E2 (Utilities|Struct|BreakStrutAudioCrouchSlide _returnvalue_2)))
              (LocalVariables|SetBasic%20Layer%20Waves (.BaseLayerWaves_9_7835FCA9472E1E4F7B69F782E5B9C888 (Utilities|Struct|BreakStrutAudioCrouchSlide _returnvalue_2)))
              (Variables|Sound|SetStartSurfaceWaves (.StartWaves_19_4C22E28241588AE34A9939B262F25AE4 (Utilities|Struct|BreakStrutAudioCrouchSlide _returnvalue_2)))
              (Variables|Sound|SetStopSurfaceWaves (.StopWaves_20_7B2B0E6642EF50E6D812D082272CF015 (Utilities|Struct|BreakStrutAudioCrouchSlide _returnvalue_2)))
              (Variables|Sound|SetStart%20Volume (.StartVolume_35_78ED97024257094AE87F2EAE56B1D9AC (Utilities|Struct|BreakStrutAudioCrouchSlide _returnvalue_2)))
              (Variables|Sound|SetStop%20Volume (.StopVolume_36_DC617BF74E54F699D38B4491E14B03DF (Utilities|Struct|BreakStrutAudioCrouchSlide _returnvalue_2))))
            (:RowNotFound
              (Development|PrintString 0 "CrouchSliding: No Default Sound Selected!")))))
      (Audio|Parameter|SetObjectArrayParameter (Variables|FortGameplayCueNotifyLoop_Sliding|GetLayeredAudioComp) "BaseWaves" (LocalVariables|GetBasic%20Layer%20Waves))
      (Audio|Parameter|SetFloatParameter (Variables|FortGameplayCueNotifyLoop_Sliding|GetLayeredAudioComp) "VolumeMultiplier" (LocalVariables|GetLayeredVolume) :!class /Script/AudioExtensions.AudioParameterControllerInterface)
      (Audio|Parameter|ExecuteTriggerParameter (Variables|FortGameplayCueNotifyLoop_Sliding|GetLayeredAudioComp) "AssetSwitch"))
    (:Is%20Not%20Valid
      (Development|PrintString 0 "CrouchSliding: No Table Selected!"))))

(fn SetAudioAffiliation
  (params (Owner :type Fort%20Player%20Pawn%20Object%20Reference) (AudioComponent :type Audio%20Component%20Object%20Reference))
  (if (not (Pawn|IsLocallyViewed Owner))
    (Audio|Parameter|SetIntegerParameter (Variables|GetAudioComponent) "Affiliation" (Utilities|Select :Index (Team|GetTeamAffiliationforActortoLocalPlayer Owner true true) :Friendly 1 :Neutral 1 :Hostile 2) :!class /Script/AudioExtensions.AudioParameterControllerInterface)))

(fn StartSlideAudio
  (bind _layeredaudiocomp (Variables|FortGameplayCueNotifyLoop_Sliding|GetLayeredAudioComp))
  (Utilities|FlowControl|Sequence
    (:then_0
      (bind _playerpawn (Variables|FortGameplayCueNotifyLoop_Sliding|GetPlayerPawn)
        (:then
          (bind _returnvalue (Audio|SpawnSoundAttached (select (Pawn|IsLocallyViewed _playerpawn) (Variables|Sound|GetStartSlideSound1p) (Variables|Sound|GetStartSlideSound3p)) (Variables|/Script/Engine.Actor|GetRootComponent _playerpawn)))
          (Sound|UpdateSoundComponentParameters _playerpawn _returnvalue (Variables|Sound|GetStartSurfaceWaves) (Variables|Sound|GetStart%20Volume))
          (Collision|BindEventtoOnActorHit (Variables|FortGameplayCueNotifyLoop_Sliding|GetPlayerPawn) (EventDispatchers|CreateEvent :!event On%20Target%20Pawn%20Actor%20Hit))
          (bind _stop_slide_audio (Variables|Sound|GetStop%20Slide%20Audio)
            (:then
              (Audio|Components|Audio|FadeOut _stop_slide_audio 0.15 0.0 "Linear" :!class /Script/Engine.AudioComponent))
            (:else)))
        (:else)))
    (:then_1
      ;[1 1 1] On start slide attach and setup parameters based on local player and team affiliation.
      (bind _playerpawn_2 (Variables|FortGameplayCueNotifyLoop_Sliding|GetPlayerPawn)
        (:then
          (Transformation|AttachComponentToComponent (Variables|FortGameplayCueNotifyLoop_Sliding|GetLayeredAudioComp) (Variables|/Script/Engine.Actor|GetRootComponent _playerpawn_2) "None" "SnapToTarget" "SnapToTarget" "SnapToTarget" true)
          (Audio|Components|Audio|SetSound _layeredaudiocomp (select (Pawn|IsLocallyViewed (Variables|FortGameplayCueNotifyLoop_Sliding|GetPlayerPawn)) "/FNGameplayCues/Athena/Sounds/Character/Sliding/MSS_CrouchSliding_Loop_1p.MSS_CrouchSliding_Loop_1p" "/FNGameplayCues/Athena/Sounds/Character/Sliding/MSS_CrouchSliding_Loop_3p.MSS_CrouchSliding_Loop_3p"))
          (Audio|Components|Audio|FadeIn _layeredaudiocomp 1.0 1.0 0.0 "Linear" :!class /Script/Engine.AudioComponent)
          (CallFunction|SetAudioAffiliation (Variables|FortGameplayCueNotifyLoop_Sliding|GetPlayerPawn) _layeredaudiocomp)
          (Default|SetSurfaceSounds (Class|FortSurfaceTrackingComponent|GetCurrentSurfaceType (Footsteps|GetSurfaceTrackingComponent (Variables|FortGameplayCueNotifyLoop_Sliding|GetPlayerPawn))))
          (bind _returnvalue_2 (Utilities|Time|SetTimerbyEvent (EventDispatchers|CreateEvent :!event UpdateSlideAudio) 0.1 true))
          (Variables|Sound|SetSlideAudioTimerHandle _returnvalue_2)
          (Variables|Sound|SetbUpdateAudioParams true))
        (:else)))))

(fn StopSlideAudio
  (Audio|Components|Audio|FadeOut (Variables|FortGameplayCueNotifyLoop_Sliding|GetLayeredAudioComp) 0.25 0.0 "Linear" :!class /Script/Engine.AudioComponent)
  (Utilities|Time|ClearandInvalidateTimerbyHandle (Variables|FortGameplayCueNotifyLoop_Sliding|GetPlayerPawn) (Variables|Sound|GetSlideAudioTimerHandle))
  (Variables|Sound|SetbUpdateAudioParams false)
  (Utilities|FlowControl|Sequence
    (:then_0
      (bind _playerpawn (Variables|FortGameplayCueNotifyLoop_Sliding|GetPlayerPawn)
        (:then
          (bind _returnvalue (Audio|SpawnSoundAttached (select (Pawn|IsLocallyViewed _playerpawn) (Variables|Sound|GetStopSlideSound1p) (Variables|Sound|GetStopSlideSound3p)) (Variables|/Script/Engine.Actor|GetRootComponent _playerpawn)))
          (bind _output_get (Variables|Sound|SetStop%20Slide%20Audio _returnvalue))
          (Sound|UpdateSoundComponentParameters _playerpawn _output_get (Variables|Sound|GetStopSurfaceWaves) (Variables|Sound|GetStop%20Volume)))
        (:else)))
    (:then_1
      (bind _layeredaudiocomp (Variables|FortGameplayCueNotifyLoop_Sliding|GetLayeredAudioComp)
        (:then
          (Transformation|DetachFromComponent _layeredaudiocomp "KeepRelative" "KeepRelative" "KeepRelative" true))
        (:else)))))

(fn ToggleLoopingSoundIndicator
  (params (bEnableSoundIndicator :type Boolean))
  (if bEnableSoundIndicator
    (SoundIndicator|StartTargetedSoundIndicatorLooping (Variables|/Script/FortniteGame.FortPawn|GetSoundIndicatorComponent (Variables|FortGameplayCueNotifyLoop_Sliding|GetPlayerPawn)) (Variables|FortGameplayCueNotifyLoop_Sliding|GetPlayerPawn) (Variables|/Script/Engine.AudioComponent|GetSound (Variables|FortGameplayCueNotifyLoop_Sliding|GetLayeredAudioComp)) 1.0 0)
    (else
      (SoundIndicator|StopSoundIndicatorLooping (Variables|/Script/FortniteGame.FortPawn|GetSoundIndicatorComponent (Variables|FortGameplayCueNotifyLoop_Sliding|GetPlayerPawn))))))

(fn UpdateSlideAudio
  (bind _playerpawn (Variables|FortGameplayCueNotifyLoop_Sliding|GetPlayerPawn))
  (bind _returnvalue (Footsteps|GetSurfaceTrackingComponent _playerpawn))
  (bind _returnvalue_2 (Class|FortSurfaceTrackingComponent|GetCurrentSurfaceType _returnvalue))
  ;[1 1 1] Update audio parameters based on pawns velocity and zheight.
  (Networking|CanExecuteCosmeticEvents
    (:True
      (Utilities|FlowControl|Sequence
        (:then_0
          (bind _playerpawn_2 (Variables|FortGameplayCueNotifyLoop_Sliding|GetPlayerPawn)
            (:then
              (if (Variables|Sound|GetbUpdateAudioParams)
                (bind _output_get (Variables|Sound|SetVelocityLengthAudioParam (Math|Float|MapRangeClamped (Math|Vector|VectorLength (Pawn|Components|CharacterMovement|GetLastUpdateVelocity (Variables|/Script/Engine.Character|GetCharacterMovement _playerpawn_2))) 0.0 1e+03 0.0 1.0)))
                (Class|FortLayeredAudioComponent|SetFloatParameterExt (Variables|FortGameplayCueNotifyLoop_Sliding|GetLayeredAudioComp) "VelocityLength" _output_get)
                ;;[1 1 1] Set parameter to zero when going uphill
                (bind _output_get_2 (Variables|Sound|SetZVelocityAudioParam (Math|Float|MapRangeClamped (Math|Vector|VectorLength (Pawn|Components|CharacterMovement|GetLastUpdateVelocity (Variables|/Script/Engine.Character|GetCharacterMovement _playerpawn_2))) 5e+02 1.2e+03 0.0 1.0)))
                (Class|FortLayeredAudioComponent|SetFloatParameterExt (Variables|FortGameplayCueNotifyLoop_Sliding|GetLayeredAudioComp) "ZVelocity" _output_get_2)))
            (:else)))
        (:then_1
          (if (Utilities|Enum|NotEqual(Enum) _returnvalue_2 (Variables|Sound|GetAudioSurfaceType))
            (Default|SetSurfaceSounds _returnvalue_2)))))))

(fn UpdateSoundComponentParameters
  (params (PlayerPawn :type Fort%20Player%20Pawn%20Object%20Reference) (AudioComponent :type Audio%20Component%20Object%20Reference) (SurfaceWaves :type Array%20of%20Object%20References) (VolumeMultiplier :type Float%20%28single-precision%29))
  (Audio|Parameter|SetObjectArrayParameter AudioComponent "Waves" SurfaceWaves)
  (Audio|Parameter|SetFloatParameter AudioComponent "VolumeMultiplier" VolumeMultiplier :!class /Script/AudioExtensions.AudioParameterControllerInterface)
  (CallFunction|SetAudioAffiliation PlayerPawn AudioComponent))

(fn ConstructionScript)
