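; Looping gameplay cue for player sliding speedline effects that manages the slide's dust particles.
; On start it caches the pawn, movement component and viewing controller, and deactivates dust for first-person cameras.
; A short timer, refreshed on view target changes, mirrors the falling grace period into the IsSpawnDust parameter for a locally viewed pawn.
; The timer is re-armed rather than looped to avoid multiple runs per frame, and cleared on removal.
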(event EventOnLoopingStart(EitherVFX-preferthisversion)
  (params (MyTarget :type Actor%20Object%20Reference) (Parameters :type Gameplay%20Cue%20Parameters%20Structure%20%28by%20ref%29) (ParticleComponents :type Array%20of%20FXSystem%20Component%20Object%20References) (AudioComponents :type Array%20of%20Audio%20Component%20Object%20References))
  (Variables|Default|SetDustFX (Utilities|Array|Get(acopy) ParticleComponents))
  (Variables|Default|SetCachedMovementCmponent (Actor|GetComponentbyClass MyTarget "/Script/FortniteGame.FortMovementComp_Character"))
  (bind _aspawn (Utilities|Casting|CastToPawn MyTarget))
  (bind _output_get (Variables|Default|SetCachedPlayerPawn _aspawn))
  (bind _asfort_player_controller (Utilities|Casting|CastToFortPlayerController (Pawn|GetLocalViewingPlayerController _output_get)))
  (bind _output_get_2 (Variables|Default|SetCachedLocalViewingPlayerController _asfort_player_controller))
  (if (Camera|IsUsingFirstPersonCamera _asfort_player_controller)
    (Components|Activation|Deactivate (Variables|Default|GetDustFX))
    (Variables|Default|SetDustFX 0)
    (EventDispatchers|BindEventtoOnViewTargetChanged _output_get_2 (delegate Custom|On%20View%20Target%20Changed))
    (CallFunction|RefreshDustStateTimer)))

(event Custom|On%20View%20Target%20Changed
  (params (PlayerController :type Fort%20Player%20Controller%20Object%20Reference) (OldViewTarget :type Actor%20Object%20Reference) (NewViewTarget :type Actor%20Object%20Reference))
  (CallFunction|RefreshDustStateTimer))

(event EventOnRemoval(EitherVFX-preferthisversion)
  (params (MyTarget :type Actor%20Object%20Reference) (Parameters :type Gameplay%20Cue%20Parameters%20Structure%20%28by%20ref%29) (ParticleComponents :type Array%20of%20FXSystem%20Component%20Object%20References) (AudioComponents :type Array%20of%20Audio%20Component%20Object%20References) (BurstCameraShakeInstance :type Camera%20Shake%20Base%20Object%20Reference) (BurstDecalInstance :type Decal%20Actor%20Object%20Reference))
  (EventDispatchers|UnbindEventfromOnViewTargetChanged (Variables|Default|GetCachedLocalViewingPlayerController) (delegate Custom|On%20View%20Target%20Changed))
  (Utilities|Time|ClearandInvalidateTimerbyHandle (Variables|Default|GetCachedPlayerPawn) (Variables|Default|GetDustStateTimerHandle)))

(event Custom|UpdateDustState
  (bind _dustfx (Variables|Default|GetDustFX))
  (bind _cachedmovementcmponent (Variables|Default|GetCachedMovementCmponent))
  (bind _binfallinggraceperiod (Variables|/Script/FortniteGame.FortMovementComp_Character|GetbInFallingGracePeriod _cachedmovementcmponent))
  (bind _cachedplayerpawn (Variables|Default|GetCachedPlayerPawn))
  (Utilities|IsValid _cachedplayerpawn
    (:Is%20Valid
      (if (Pawn|IsLocallyViewed _cachedplayerpawn)
        (Utilities|IsValid _cachedmovementcmponent
          (:Is%20Valid
            (if (!= (Variables|Default|GetbIsGraceFalling) _binfallinggraceperiod)
              (Variables|Default|SetbIsGraceFalling _binfallinggraceperiod)
              (Utilities|IsValid _dustfx
                (:Is%20Valid
                  (Effects|Components|ParticleSystem|SetBoolParameter _dustfx "IsSpawnDust" (not _binfallinggraceperiod)))
                (:Is%20Not%20Valid
                  _))
              ;[1 1 1] Reset timer instead of enabling looping to more accurately simulate the behavior of a tick. Looping timers have a chance to execute multiple times in a frame, whereas resetting the timer will result in one execution.
              (Utilities|Time|SetTimerbyFunctionName 0 "UpdateDustState" 0.05)
              (Variables|Default|SetDustStateTimerHandle)
              (else
                _))))))))

(fn RefreshDustStateTimer
  (bind _cachedplayerpawn (Variables|Default|GetCachedPlayerPawn))
  (Utilities|IsValid _cachedplayerpawn
    (:Is%20Valid
      (if (Pawn|IsLocallyViewed _cachedplayerpawn)
        (bind _returnvalue (Utilities|Time|SetTimerbyFunctionName 0 "UpdateDustState" 0.05))
        (Variables|Default|SetDustStateTimerHandle _returnvalue)
        (else
          _)))
    (:Is%20Not%20Valid
      _))
  (Utilities|Time|ClearandInvalidateTimerbyHandle _cachedplayerpawn (Variables|Default|GetDustStateTimerHandle)))

(fn ConstructionScript)
