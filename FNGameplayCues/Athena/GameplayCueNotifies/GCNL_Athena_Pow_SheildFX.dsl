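; Looping gameplay cue for the Pow mutator that highlights an Athena player pawn according to its current shield.
; On start it caches the pawn, player state and game-state mutator, then applies the highlight and refreshes it on a repeating timer.
; Each update scales two glow colors by clamped shield value and toggles a highlight parameter when shield exceeds 100.
; On removal it deactivates the first particle component, clears the timer and removes the highlight.
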(event Custom|Update%20Visuals%20Dynamic
  (CallFunction|UpdateVisuals true))

(event Custom|UpdateVisuals
  (params (Update%20Dynamic :type Boolean))
  (bind _powmutator (Variables|Default|GetPowMutator)
    (:then
      (bind _pawn (Variables|Default|GetPawn)
        (:then
          (bind _output_get (Variables|Default|SetShieldValue (Class|FortAthenaMutatorPow|GetPlayerCurrentShield _powmutator (Variables|Default|GetPlayerState))))
          (Variables|Default|SetRangeOutput (Math|Float|MapRangeClamped _output_get 0.0 1e+03 0.0 0.75))
          (PawnHighlight|HighlightPawnFromSource _pawn self (Utilities|Struct|MakePawnHighlight (Variables|Default|GetGlow%20Priority) (Math|Color|LinearColor*Float (Variables|Default|GetColorVar_Dark) (Variables|Default|GetRangeOutput)) (Math|Color|LinearColor*Float (Variables|Default|GetColorVar) (Variables|Default|GetRangeOutput)) 1.4 1.0 (Math|Conversions|ToFloat(Boolean) (> (Variables|Default|GetShieldValue) 1e+02))) Update%20Dynamic))
        (:else)))
    (:else)))

(event EventOnLoopingStart(Cascade)
  (params (MyTarget :type Actor%20Object%20Reference) (Parameters :type Gameplay%20Cue%20Parameters%20Structure%20%28by%20ref%29) (ParticleComponents :type Array%20of%20Cascade%20Particle%20System%20Component%20Object%20References) (AudioComponents :type Array%20of%20Audio%20Component%20Object%20References))
  (bind _returnvalue (Game|GetGameState MyTarget))
  (Utilities|IsValid MyTarget
    (:Is%20Valid
      (Utilities|FlowControl|Sequence
        (:then_0
          (bind _asplayer_pawn_athena (Utilities|Casting|CastToPlayerPawn_Athena MyTarget))
          (Variables|Default|SetPawn _asplayer_pawn_athena)
          (Variables|Default|SetPlayerState (Utilities|Casting|CastToFortPlayerStateAthena (Variables|/Script/Engine.Pawn|GetPlayerState _asplayer_pawn_athena)))
          (Utilities|IsValid _returnvalue
            (:Is%20Valid
              (bind _asfort_athena_mutator_pow (Utilities|Casting|CastToFortAthenaMutator_Pow (Modifier|GetMutatorbyClass (Actor|GetComponentbyClass _returnvalue "/Script/FortniteGame.FortMutatorListComponent") "/Script/FortniteGame.FortAthenaMutator_Pow")))
              (Variables|Default|SetPowMutator _asfort_athena_mutator_pow))))
        (:then_1
          (CallFunction|UpdateVisuals false)
          (bind _returnvalue_2 (Utilities|Time|SetTimerbyEvent (EventDispatchers|CreateEvent :!event Custom|Update%20Visuals%20Dynamic) (Variables|Default|GetPeriodicUpdateFrequency) true))
          (Variables|Default|SetVisual%20Update%20TImer _returnvalue_2))))))

(event EventOnRemoval(Cascade)
  (params (MyTarget :type Actor%20Object%20Reference) (Parameters :type Gameplay%20Cue%20Parameters%20Structure%20%28by%20ref%29) (ParticleComponents :type Array%20of%20Cascade%20Particle%20System%20Component%20Object%20References) (AudioComponents :type Array%20of%20Audio%20Component%20Object%20References) (BurstCameraShakeInstance :type Camera%20Shake%20Base%20Object%20Reference) (BurstDecalInstance :type Decal%20Actor%20Object%20Reference))
  (Components|Activation|Deactivate (Utilities|Array|Get(acopy) ParticleComponents))
  (Utilities|Time|ClearandInvalidateTimerbyHandle 0 (Variables|Default|GetVisual%20Update%20TImer))
  (bind _pawn (Variables|Default|GetPawn)
    (:then
      (PawnHighlight|RemovePawnHighlightFromSource _pawn self))
    (:else)))

(fn ConstructionScript)
