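; Looping gameplay cue for the on-fire state of the gas pump reactive prop, intensifying its Cascade fire effect as the prop loses health.
; Loop start sets the first particle component's SpawnRamp parameter and starts a repeating 0.2-second timer that refreshes it.
; The value is how far health percent has fallen below the exploding prop's OnFireThreshold, normalized and mapped through SpawnRampCurve.
; The timer loop reuses the start event's target and component outputs.
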(event Custom|StartLoop
  (Utilities|Time|SetTimerbyEvent (delegate Custom|Loop) 0.2 true))

(event Custom|Loop
  (bind _asathena_prop_burst_reaction_exploding (Utilities|Casting|CastToAthena_Prop_BurstReaction_Exploding (stale EventOnLoopingStart(Cascade) MyTarget)))
  (Utilities|IsValid (Utilities|Array|Get(acopy) (stale EventOnLoopingStart(Cascade) ParticleComponents))
    (:Is%20Valid
      (Effects|Components|ParticleSystem|SetInstanceParams (Utilities|Array|Get(acopy) (stale EventOnLoopingStart(Cascade) ParticleComponents)) (Utilities|Array|MakeArray (Utilities|Struct|MakeParticleSysParam "SpawnRamp" "PSPT_Scalar" (Math|Curves|GetFloatValue (Variables|Default|GetSpawnRampCurve) (Math|Float|NormalizetoRange (- (Math|Float|MakeLiteralFloat (Variables|/Game/Athena/Items/EnvironmentalItems/ReactiveProps/Athena_Prop_BurstReaction_Exploding.Athena_Prop_BurstReaction_Exploding_C|GetOnFireThreshold _asathena_prop_burst_reaction_exploding)) (Health|GetHealthPercent _asathena_prop_burst_reaction_exploding :!class /Script/FortniteGame.BuildingActor)) 0.0 (Variables|/Game/Athena/Items/EnvironmentalItems/ReactiveProps/Athena_Prop_BurstReaction_Exploding.Athena_Prop_BurstReaction_Exploding_C|GetOnFireThreshold _asathena_prop_burst_reaction_exploding))) (Math|Curves|GetFloatValue (Variables|Default|GetSpawnRampCurve) (Math|Float|NormalizetoRange (- (Math|Float|MakeLiteralFloat (Variables|/Game/Athena/Items/EnvironmentalItems/ReactiveProps/Athena_Prop_BurstReaction_Exploding.Athena_Prop_BurstReaction_Exploding_C|GetOnFireThreshold _asathena_prop_burst_reaction_exploding)) (Health|GetHealthPercent _asathena_prop_burst_reaction_exploding :!class /Script/FortniteGame.BuildingActor)) 0.0 (Variables|/Game/Athena/Items/EnvironmentalItems/ReactiveProps/Athena_Prop_BurstReaction_Exploding.Athena_Prop_BurstReaction_Exploding_C|GetOnFireThreshold _asathena_prop_burst_reaction_exploding)))))))))

(event EventBeginPlay)

(event EventOnLoopingStart(Cascade)
  (params (MyTarget :type Actor%20Object%20Reference) (Parameters :type Gameplay%20Cue%20Parameters%20Structure%20%28by%20ref%29) (ParticleComponents :type Array%20of%20Cascade%20Particle%20System%20Component%20Object%20References) (AudioComponents :type Array%20of%20Audio%20Component%20Object%20References))
  (bind _output (Utilities|Array|Get(acopy) ParticleComponents))
  (CallFunction|StartLoop)
  (bind _asathena_prop_burst_reaction_exploding (Utilities|Casting|CastToAthena_Prop_BurstReaction_Exploding MyTarget))
  (Utilities|IsValid _output
    (:Is%20Valid
      (Effects|Components|ParticleSystem|SetInstanceParams _output (Utilities|Array|MakeArray (Utilities|Struct|MakeParticleSysParam "SpawnRamp" "PSPT_Scalar" (Math|Curves|GetFloatValue (Variables|Default|GetSpawnRampCurve) (Math|Float|NormalizetoRange (- (Math|Float|MakeLiteralFloat (Variables|/Game/Athena/Items/EnvironmentalItems/ReactiveProps/Athena_Prop_BurstReaction_Exploding.Athena_Prop_BurstReaction_Exploding_C|GetOnFireThreshold _asathena_prop_burst_reaction_exploding)) (Health|GetHealthPercent _asathena_prop_burst_reaction_exploding :!class /Script/FortniteGame.BuildingActor)) 0.0 (Variables|/Game/Athena/Items/EnvironmentalItems/ReactiveProps/Athena_Prop_BurstReaction_Exploding.Athena_Prop_BurstReaction_Exploding_C|GetOnFireThreshold _asathena_prop_burst_reaction_exploding))) (Math|Curves|GetFloatValue (Variables|Default|GetSpawnRampCurve) (Math|Float|NormalizetoRange (- (Math|Float|MakeLiteralFloat (Variables|/Game/Athena/Items/EnvironmentalItems/ReactiveProps/Athena_Prop_BurstReaction_Exploding.Athena_Prop_BurstReaction_Exploding_C|GetOnFireThreshold _asathena_prop_burst_reaction_exploding)) (Health|GetHealthPercent _asathena_prop_burst_reaction_exploding :!class /Script/FortniteGame.BuildingActor)) 0.0 (Variables|/Game/Athena/Items/EnvironmentalItems/ReactiveProps/Athena_Prop_BurstReaction_Exploding.Athena_Prop_BurstReaction_Exploding_C|GetOnFireThreshold _asathena_prop_burst_reaction_exploding)))))))))

(event Collision|EventActorBeginOverlap
  (params (OtherActor :type Actor%20Object%20Reference)))

(event EventTick
  (params (DeltaSeconds :type Float%20%28single-precision%29)))

(fn ConstructionScript)
