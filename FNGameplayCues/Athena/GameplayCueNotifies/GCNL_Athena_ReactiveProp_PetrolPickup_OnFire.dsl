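; Looping GameplayCue notify that drives the on-fire Cascade effect for the petrol pickup reactive prop.
; On loop start it orients the particle to the cue's normal and sets its SpawnRamp parameter from inverted building health through a curve.
; A repeating short timer reapplies both, reusing the start event's target, parameters and components, so the fire follows damage.
; The timer handle is stored, but no logic here clears it.
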(event Custom|StartLoop
  (bind _returnvalue (Utilities|Time|SetTimerbyEvent (delegate Custom|Loop) 0.2 true))
  (Variables|Default|SetTimer%20Handle _returnvalue))

(event Custom|Loop
  (bind _asbuilding_actor (Utilities|Casting|CastToBuildingActor (stale EventOnLoopingStart(Cascade) MyTarget)))
  (Transformation|SetWorldRotation (Utilities|Array|Get(acopy) (stale EventOnLoopingStart(Cascade) ParticleComponents)) (Math|Conversions|RotationFromXVector (.Normal (Ability|GameplayCue|BreakGameplayCueParameters (stale EventOnLoopingStart(Cascade) Parameters)))) false false)
  (Utilities|IsValid (Utilities|Array|Get(acopy) (stale EventOnLoopingStart(Cascade) ParticleComponents))
    (:Is%20Valid
      (Effects|Components|ParticleSystem|SetInstanceParams (Utilities|Array|Get(acopy) (stale EventOnLoopingStart(Cascade) ParticleComponents)) (Utilities|Array|MakeArray (Utilities|Struct|MakeParticleSysParam "SpawnRamp" "PSPT_Scalar" (Math|Curves|GetFloatValue (Variables|Default|GetSpawnRampCurve) (Math|Float|NormalizetoRange (- 1.0 (Health|GetHealthPercent _asbuilding_actor :!class /Script/FortniteGame.BuildingActor)) 0.0 1.0)) (Math|Curves|GetFloatValue (Variables|Default|GetSpawnRampCurve) (Math|Float|NormalizetoRange (- 1.0 (Health|GetHealthPercent _asbuilding_actor :!class /Script/FortniteGame.BuildingActor)) 0.0 1.0))))))))

(event EventOnLoopingStart(Cascade)
  (params (MyTarget :type Actor%20Object%20Reference) (Parameters :type Gameplay%20Cue%20Parameters%20Structure%20%28by%20ref%29) (ParticleComponents :type Array%20of%20Cascade%20Particle%20System%20Component%20Object%20References) (AudioComponents :type Array%20of%20Audio%20Component%20Object%20References))
  (bind (_normalizedmagnitude _rawmagnitude _effectcontext _matchedtagname _originaltag _aggregatedsourcetags _aggregatedtargettags _location _normal) (Ability|GameplayCue|BreakGameplayCueParameters Parameters))
  (bind _output (Utilities|Array|Get(acopy) ParticleComponents))
  (CallFunction|StartLoop)
  (bind _asbuilding_actor (Utilities|Casting|CastToBuildingActor MyTarget))
  (Transformation|SetWorldRotation _output (Math|Conversions|RotationFromXVector _normal) false false)
  (Utilities|IsValid _output
    (:Is%20Valid
      (Effects|Components|ParticleSystem|SetInstanceParams _output (Utilities|Array|MakeArray (Utilities|Struct|MakeParticleSysParam "SpawnRamp" "PSPT_Scalar" (Math|Curves|GetFloatValue (Variables|Default|GetSpawnRampCurve) (Math|Float|NormalizetoRange (- 1.0 (Health|GetHealthPercent _asbuilding_actor :!class /Script/FortniteGame.BuildingActor)) 0.0 1.0)) (Math|Curves|GetFloatValue (Variables|Default|GetSpawnRampCurve) (Math|Float|NormalizetoRange (- 1.0 (Health|GetHealthPercent _asbuilding_actor :!class /Script/FortniteGame.BuildingActor)) 0.0 1.0))))))))

(fn ConstructionScript)
