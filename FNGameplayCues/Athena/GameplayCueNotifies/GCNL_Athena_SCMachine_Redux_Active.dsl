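; Looping gameplay cue for the active Reboot Van (SCMachine), driving its spawn warmup and removal effects.
; On loop start it places the first particle effect above the van, caches it and binds to the van's OnPawnResurrected event.
; An enhanced resurrection is stored and switches the warmup Niagara "Variant" to 1 after a short delay.
; On removal it unbinds, sets the removal effect's variant from that flag and positions it above the van.
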(event EventOnLoopingStart(EitherVFX-preferthisversion)
  (params (MyTarget :type Actor%20Object%20Reference) (Parameters :type Gameplay%20Cue%20Parameters%20Structure%20%28by%20ref%29) (ParticleComponents :type Array%20of%20FXSystem%20Component%20Object%20References) (AudioComponents :type Array%20of%20Audio%20Component%20Object%20References))
  (bind _output (Utilities|Array|Get(acopy) ParticleComponents))
  ;[0 0.2083 0.03411] OnLoopingStart: Update the position of the spawned nigara effect and bind to know when enhanced resurrection is happening on the van.
  (|CallParentFunction MyTarget Parameters ParticleComponents AudioComponents)
  (bind _asbga_athena_scmachine (Utilities|Casting|CastToBGA_Athena_SCMachine MyTarget))
  (if (Utilities|Array|IsValidIndex ParticleComponents)
    (Utilities|IsValid _output
      (:Is%20Valid
        (Variables|Default|SetSpawnWarmupFXSystemComponent _output)
        (Default|BindEventtoOnPawnResurrected _asbga_athena_scmachine (EventDispatchers|CreateEvent :!event Custom|HandlePawnResurrected))
        (Transformation|SetWorldLocation _output (+ (Transformation|GetActorLocation _asbga_athena_scmachine) "0.000000,0.000000,275.000000") false false)))))

(event EventOnRemoval(EitherVFX-preferthisversion)
  (params (MyTarget :type Actor%20Object%20Reference) (Parameters :type Gameplay%20Cue%20Parameters%20Structure%20%28by%20ref%29) (ParticleComponents :type Array%20of%20FXSystem%20Component%20Object%20References) (AudioComponents :type Array%20of%20Audio%20Component%20Object%20References) (BurstCameraShakeInstance :type Camera%20Shake%20Base%20Object%20Reference) (BurstDecalInstance :type Decal%20Actor%20Object%20Reference))
  (bind _output (Utilities|Array|Get(acopy) ParticleComponents))
  ;[0.15 0.07615 0 0.5] OnRemoval: Unbind delegates.
  (|CallParentFunction MyTarget Parameters ParticleComponents AudioComponents BurstCameraShakeInstance BurstDecalInstance)
  (bind _asbga_athena_scmachine (Utilities|Casting|CastToBGA_Athena_SCMachine MyTarget))
  (Default|UnbindEventfromOnPawnResurrected _asbga_athena_scmachine (EventDispatchers|CreateEvent :!event Custom|HandlePawnResurrected))
  (bind _asniagara_particle_system_component (Utilities|Casting|CastToNiagaraComponent _output))
  (if (Variables|Default|GetEnhanced%20Resurrection)
    (Niagara|SetNiagaraVariable(Int32) _asniagara_particle_system_component "Variant" 1)
    (else
      (Niagara|SetNiagaraVariable(Int32) _asniagara_particle_system_component "Variant" 0)))
  (Transformation|SetWorldLocation _output (+ (Transformation|GetActorLocation _asbga_athena_scmachine) "0.000000,0.000000,275.000000") false false))

(event Custom|HandlePawnResurrected
  (params (FortPawn :type Fort%20Pawn%20Object%20Reference) (bEnhancedResurrection :type Boolean))
  ;[0.15 0 0.1222 0.5] HandlePawnResurrected: When enhanced resurrection is happening, update the warmup fx to variant 1 for a short duration.
  (Variables|Default|SetEnhanced%20Resurrection bEnhancedResurrection)
  (if bEnhancedResurrection
    (bind _spawnwarmupfxsystemcomponent (Variables|Default|GetSpawnWarmupFXSystemComponent)
      (:then
        (bind _asniagara_particle_system_component (Utilities|Casting|CastToNiagaraComponent _spawnwarmupfxsystemcomponent))
        (Utilities|FlowControl|Delay self 0.5)
        (Niagara|SetNiagaraVariable(Int32) _asniagara_particle_system_component "Variant" 1))
      (:else))))

(fn ConstructionScript)
