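; Looping gameplay cue notify for the active SmallHealthSpeed effect, a no-audio variant by name, that adapts its particles to the target's state.
; When the loop starts, it registers tag waits on the target that hide the first particle component when the cloak tag is added and show it on removal.
; It also sets that component's "Is Aiming" parameter once from the player controller's aim-down-sights state, if the controller is valid.
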(event EventOnLoopingStart(EitherVFX-preferthisversion)
  (params (MyTarget :type Actor%20Object%20Reference) (Parameters :type Gameplay%20Cue%20Parameters%20Structure%20%28by%20ref%29) (ParticleComponents :type Array%20of%20FXSystem%20Component%20Object%20References) (AudioComponents :type Array%20of%20Audio%20Component%20Object%20References))
  (bind _returnvalue (Fortnite|GetFortPlayerControllerfromActor MyTarget))
  (bind _output (Utilities|Array|Get(acopy) ParticleComponents))
  ; Turn invisible if the player goes invisible
  (Ability|Tasks|WaitGameplayTagAddToActor MyTarget (Variables|Default|GetCloakTag)
    (:then
      (Ability|Tasks|WaitGameplayTagRemoveFromActor MyTarget (Variables|Default|GetCloakTag)
        (:then
          (Utilities|IsValid _returnvalue
            (:Is%20Valid
              (Effects|Components|ParticleSystem|SetBoolParameter _output "Is Aiming" (Input|IsAimingDownSightsorScope _returnvalue)))))
        (:Removed
          (Rendering|SetVisibility _output true true)
          (Rendering|SetVisibility _output true true))))
    (:Added
      (Rendering|SetVisibility _output false true)
      (Rendering|SetVisibility _output false true))))

(fn ConstructionScript)
