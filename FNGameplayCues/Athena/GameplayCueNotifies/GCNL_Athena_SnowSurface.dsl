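; Looping gameplay cue notify that drives Niagara footstep effects for pawns on snow or ice surfaces.
; On non-mobile loop start, it removes the effect for targets off the local player's team, with separate handling for replays that caches the spectator controller.
; Each footstep outside water feeds the Niagara component the foot socket position, direction, side, ground normal and velocity.
; An "Ice" substring in the cue's original tag switches the effect to ice.
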(event EventOnLoopingStart(Niagara)
  (params (MyTarget :type Actor%20Object%20Reference) (Parameters :type Gameplay%20Cue%20Parameters%20Structure%20%28by%20ref%29) (ParticleComponents :type Array%20of%20Niagara%20Particle%20System%20Component%20Object%20References) (AudioComponents :type Array%20of%20Audio%20Component%20Object%20References))
  (bind _returnvalue (Fortnite|GetFirstLocalFortPlayerController 0 false true))
  (bind _asfort_replay_spectator_athena (Utilities|Casting|CastToFortReplaySpectatorAthena _returnvalue))
  (bind _output (Utilities|Array|Get(acopy) ParticleComponents))
  (Variables|Default|SetbNotValid false)
  (Variables|Default|SetbReplayMode false)
  (if (not (Fortnite|IsMobilePlatform))
    (if (Replay|GetIsPlayingReplay 0)
      (Variables|Default|SetbReplayMode true)
      (Utilities|IsValid _asfort_replay_spectator_athena
        (:Is%20Valid
          (Variables|Default|SetReplayController _asfort_replay_spectator_athena))
        (:Is%20Not%20Valid
          (Development|PrintString 0 "SSI unexcept replay issue.")))
      (elif (Team|OnLocalPlayersTeam MyTarget)
        _
        (else
          (Variables|Default|SetbNotValid true)
          (Utilities|FlowControl|DelayUntilNextTick 0)
          (Components|Activation|Deactivate _output)
          (Components|DestroyComponent _output _output))))
    (Variables|Default|SetbNotValid false)
    (Variables|Default|SetFootStepFX _output)
    (Rendering|SetTranslucentSortPriority _output 150)
    (Effects|Components|ParticleSystem|SetBoolParameter _output "bIsSnow" true)))

(event GameplayCueNotify|EventHandleGameplayCue
  (params (MyTarget :type Actor%20Object%20Reference) (EventType :type EGameplayCueEvent%20Enum :default "OnActive") (Parameters :type Gameplay%20Cue%20Parameters%20Structure%20%28by%20ref%29))
  (|CallParentFunction MyTarget EventType Parameters)
  (Variables|Default|SetGCParameters Parameters))

(event EventOnLoopingStart(Cascade)
  (params (MyTarget :type Actor%20Object%20Reference) (Parameters :type Gameplay%20Cue%20Parameters%20Structure%20%28by%20ref%29) (ParticleComponents :type Array%20of%20Cascade%20Particle%20System%20Component%20Object%20References) (AudioComponents :type Array%20of%20Audio%20Component%20Object%20References))
  ;[1 1 1] FORT-432673
  (|CallParentFunction MyTarget Parameters ParticleComponents AudioComponents))

(event EventOnRemoval(Cascade)
  (params (MyTarget :type Actor%20Object%20Reference) (Parameters :type Gameplay%20Cue%20Parameters%20Structure%20%28by%20ref%29) (ParticleComponents :type Array%20of%20Cascade%20Particle%20System%20Component%20Object%20References) (AudioComponents :type Array%20of%20Audio%20Component%20Object%20References) (BurstCameraShakeInstance :type Camera%20Shake%20Base%20Object%20Reference) (BurstDecalInstance :type Decal%20Actor%20Object%20Reference))
  ;[1 1 1] FORT-432673
  (|CallParentFunction MyTarget Parameters ParticleComponents AudioComponents BurstCameraShakeInstance BurstDecalInstance))

(fn OnPlayerFootstep
  (params (FootIndex :type Integer))
  (bind _targetpawn (Variables|Default|GetTargetPawn))
  (bind _mesh (Variables|/Script/Engine.Character|GetMesh _targetpawn))
  (bind _returnvalue (Transformation|GetSocketTransform _mesh "ball_r" "RTS_World"))
  (bind _returnvalue_2 (AthenaPawn|GetCachedFootstepIndex _targetpawn))
  (bind _returnvalue_3 (== _returnvalue_2 1))
  (bind _returnvalue_4 (Transformation|GetSocketTransform _mesh "ball_l" "RTS_World"))
  (bind _returnvalue_5 (Math|Transform|SelectTransform _returnvalue _returnvalue_4 _returnvalue_3))
  (bind (_normalizedmagnitude _rawmagnitude _effectcontext _matchedtagname _originaltag) (Ability|GameplayCue|BreakGameplayCueParameters (Variables|Default|GetGCParameters)))
  (bind _targetpawn_2 (Variables|Default|GetTargetPawn))
  (bind _footstepfx (Variables|Default|GetFootStepFX))
  (bind _replaycontroller (Variables|Default|GetReplayController))
  (|CallParentFunction FootIndex)
  (if (Variables|Default|GetbNotValid)
    _
    (elif (Variables|Default|GetbReplayMode)
      (if (Utilities|IsValid _replaycontroller :!class /Script/Engine.KismetSystemLibrary)
        (if (Team|OnSameTeam (Variables|Default|GetTargetPawn) (Pawn|GetViewTarget _replaycontroller))
          _)
        (else
          _))
      (Utilities|IsValid (Variables|Default|GetFootStepFX)
        (:Is%20Valid
          (if (not (Variables|Default|GetIs%20in%20Water))
            (Niagara|NiagaraSetVector4Array _footstepfx "SpawnList" (Utilities|Array|MakeArray (Math|Vector4|MakeVector4 (.x (.location _returnvalue_5)) (.y (.location _returnvalue_5)) (.z (.location _returnvalue_5)) 1.0)))
            (Niagara|SetNiagaraVariable(Vector3) _footstepfx "FootDirection" (Math|Vector|SelectVector (* (Math|Vector|GetForwardVector (.rotation _returnvalue_5)) "-1.000000,-1.000000,-1.000000") (Math|Vector|GetForwardVector (.rotation _returnvalue_5)) _returnvalue_3))
            (Effects|Components|ParticleSystem|SetBoolParameter _footstepfx "bIsRight" _returnvalue_3)
            (Niagara|SetNiagaraVariable(Vector3) _footstepfx "GroundNormal" (AthenaPawn|GetCachedStepSurfaceNormal _targetpawn_2))
            (Niagara|SetNiagaraVariable(Vector3) _footstepfx "BoneSpeed" (Transformation|GetVelocity _targetpawn_2))
            (Effects|Components|ParticleSystem|SetBoolParameter _footstepfx "bIsIce" (Utilities|String|Contains (Utilities|String|ToString(Name) (GameplayTags|GetTagName _originaltag)) "Ice")))))
      (else
        _))))

(fn ConstructionScript)
