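; GCNL_Athena_Surface_Parent is a looping gameplay cue notify, named as the parent for Athena surface cues, that tracks player footsteps and landings.
; On loop start it stores the target FortPlayerPawn and binds its footstep and landed events, unbinding both on removal.
; Each footstep refreshes Is in Water through InterfacePlayerPawn.
; The landing handler is empty, and no effects are triggered here.
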(event EventOnLoopingStart(Cascade)
  (params (MyTarget :type Actor%20Object%20Reference) (Parameters :type Gameplay%20Cue%20Parameters%20Structure%20%28by%20ref%29) (ParticleComponents :type Array%20of%20Cascade%20Particle%20System%20Component%20Object%20References) (AudioComponents :type Array%20of%20Audio%20Component%20Object%20References))
  (bind _asfort_player_pawn (Utilities|Casting|CastToFortPlayerPawn MyTarget))
  (bind _output_get (Variables|Default|SetTargetPawn _asfort_player_pawn))
  (EventDispatchers|BindEventtoOnFootstepEvent _output_get (EventDispatchers|CreateEvent :!event OnPlayerFootstep))
  (EventDispatchers|BindEventtoOnPawnLanded _output_get (delegate Custom|OnPlayerLanded)))

(event Custom|OnPlayerLanded
  (params (Hit :type Hit%20Result%20Structure%20%28by%20ref%29)))

(event EventOnRemoval(Cascade)
  (params (MyTarget :type Actor%20Object%20Reference) (Parameters :type Gameplay%20Cue%20Parameters%20Structure%20%28by%20ref%29) (ParticleComponents :type Array%20of%20Cascade%20Particle%20System%20Component%20Object%20References) (AudioComponents :type Array%20of%20Audio%20Component%20Object%20References) (BurstCameraShakeInstance :type Camera%20Shake%20Base%20Object%20Reference) (BurstDecalInstance :type Decal%20Actor%20Object%20Reference))
  (bind _asfort_player_pawn (Utilities|Casting|CastToFortPlayerPawn MyTarget))
  (EventDispatchers|UnbindEventfromOnFootstepEvent _asfort_player_pawn (EventDispatchers|CreateEvent :!event OnPlayerFootstep))
  (EventDispatchers|UnbindEventfromOnPawnLanded _asfort_player_pawn (delegate Custom|OnPlayerLanded)))

(fn OnPlayerFootstep
  (params (FootIndex :type Integer))
  (bind _asinterface_player_pawn (Utilities|Casting|CastToInterfacePlayerPawn (Variables|Default|GetTargetPawn)))
  (bind _isinwater (InterfacePlayerPawn|GetinWater _asinterface_player_pawn :!class /Game/Blueprints/InterfacePlayerPawn.InterfacePlayerPawn_C))
  (Variables|Default|SetIs%20in%20Water _isinwater))

(fn ConstructionScript)
