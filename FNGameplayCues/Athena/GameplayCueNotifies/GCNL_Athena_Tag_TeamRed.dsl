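; Looping gameplay cue notify that applies a highlight glow to a player pawn, named for the red team tag in Athena.
; On looping start it stores the target as the affected player and highlights it using its glow colors, priority and fresnel variables.
; It binds to the pawn's character-parts MID initialized event to reapply the highlight when materials initialize.
; On removal it clears its highlight and unbinds that event.
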(event EventOnLoopingStart(Cascade)
  (params (MyTarget :type Actor%20Object%20Reference) (Parameters :type Gameplay%20Cue%20Parameters%20Structure%20%28by%20ref%29) (ParticleComponents :type Array%20of%20Cascade%20Particle%20System%20Component%20Object%20References) (AudioComponents :type Array%20of%20Audio%20Component%20Object%20References))
  ;[0 1 0.01898] Looping Start
  (bind _asplayer_pawn_athena (Utilities|Casting|CastToPlayerPawn_Athena MyTarget))
  (Variables|Default|SetAffectedPlayer _asplayer_pawn_athena)
  (CallFunction|ApplyEffect false))

(event Custom|Apply%20Effect
  (params (Dynamic%20Update :type Boolean))
  (bind (_hitglowinner _hitglowouter) (CallFunction|GetHitGlows))
  ;[0.6073 0.01531 1] Apply Effect
  (Variables|Default|SetOnlyUpdateIfCurrent Dynamic%20Update)
  (bind _affectedplayer (Variables|Default|GetAffectedPlayer)
    (:then
      (PawnHighlight|HighlightPawnFromSource _affectedplayer self (Utilities|Struct|MakePawnHighlight (Variables|Default|GetGlow%20Priority) _hitglowinner _hitglowouter (Variables|Default|GetHighlight%20Fresnel%20Brightness) (Variables|Default|GetHighlight%20Fresnel%20Exponent) 0.0) (Variables|Default|GetOnlyUpdateIfCurrent))
      (Default|BindEventtoOnCharacterPartsMIDInitialized _affectedplayer (delegate Custom|Pawn%20MID%20Initialized)))
    (:else)))

(event Custom|Pawn%20MID%20Initialized
  (bind (_hitglowinner _hitglowouter) (CallFunction|GetHitGlows))
  ;[0.6073 0.01531 1] Apply Effect
  (bind _affectedplayer (Variables|Default|GetAffectedPlayer)
    (:then
      (PawnHighlight|HighlightPawnFromSource _affectedplayer self (Utilities|Struct|MakePawnHighlight (Variables|Default|GetGlow%20Priority) _hitglowinner _hitglowouter (Variables|Default|GetHighlight%20Fresnel%20Brightness) (Variables|Default|GetHighlight%20Fresnel%20Exponent) 0.0) (Variables|Default|GetOnlyUpdateIfCurrent))
      (Default|BindEventtoOnCharacterPartsMIDInitialized _affectedplayer (delegate Custom|Pawn%20MID%20Initialized)))
    (:else)))

(event EventOnRemoval(Cascade)
  (params (MyTarget :type Actor%20Object%20Reference) (Parameters :type Gameplay%20Cue%20Parameters%20Structure%20%28by%20ref%29) (ParticleComponents :type Array%20of%20Cascade%20Particle%20System%20Component%20Object%20References) (AudioComponents :type Array%20of%20Audio%20Component%20Object%20References) (BurstCameraShakeInstance :type Camera%20Shake%20Base%20Object%20Reference) (BurstDecalInstance :type Decal%20Actor%20Object%20Reference))
  ;[1 0.57 0] On Removal
  (bind _affectedplayer (Variables|Default|GetAffectedPlayer)
    (:then
      (PawnHighlight|RemovePawnHighlightFromSource _affectedplayer self)
      (Default|UnbindEventfromOnCharacterPartsMIDInitialized _affectedplayer (delegate Custom|Pawn%20MID%20Initialized)))
    (:else)))

(fn GetHitGlows
  (return (Variables|Default|GetHitGlowInner) (Variables|Default|GetHitGlowOuter)))

(fn ConstructionScript)
