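; Looping gameplay cue that gives escalating drowning-damage feedback to the locally controlled Fort player pawn.
; Start resets intensity, pushes a drown sound mix and stores the local pawn.
; Each recurrence raises intensity, which drives a camera lens effect's particle parameters, a blur-pulse camera shake's scale and a 2D damage tick sound's parameter.
; Removal pops the mix without a local-control check and resets intensity for the stored pawn.
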(event EventOnLoopingStart(Cascade)
  (params (MyTarget :type Actor%20Object%20Reference) (Parameters :type Gameplay%20Cue%20Parameters%20Structure%20%28by%20ref%29) (ParticleComponents :type Array%20of%20Cascade%20Particle%20System%20Component%20Object%20References) (AudioComponents :type Array%20of%20Audio%20Component%20Object%20References))
  (bind _asfort_player_pawn (Utilities|Casting|CastToFortPlayerPawn MyTarget))
  (if (Pawn|IsLocallyControlled _asfort_player_pawn)
    (Variables|Default|SetIntensity 0.1)
    (Audio|PushSoundMixModifier 0 (Variables|Default|GetDrownSoundMix))
    (Variables|Default|SetLocalTargetPawn MyTarget)))

(event EventOnRecurring(Cascade)
  (params (MyTarget :type Actor%20Object%20Reference) (Parameters :type Gameplay%20Cue%20Parameters%20Structure%20%28by%20ref%29) (ParticleComponents :type Array%20of%20Cascade%20Particle%20System%20Component%20Object%20References) (AudioComponents :type Array%20of%20Audio%20Component%20Object%20References) (BurstCameraShakeInstance :type Camera%20Shake%20Base%20Object%20Reference) (BurstDecalInstance :type Decal%20Actor%20Object%20Reference))
  (bind _intensity (Variables|Default|GetIntensity))
  (bind _asfort_player_pawn (Utilities|Casting|CastToFortPlayerPawn MyTarget))
  (if (Pawn|IsLocallyControlled _asfort_player_pawn)
    (Variables|Default|SetIntensity (+ (Variables|Default|GetIntensity) 0.08))
    (bind _returnvalue (Feedback|SpawnGenericCameraLensEffect _asfort_player_pawn "/FNGameplayCues/Effects/Fort_Effects/Camera_Effects/B_CameraLens_Drown_Damage.B_CameraLens_Drown_Damage_C"))
    (Effects|Components|ParticleSystem|SetFloatParameter (CameraLensEffect|GetPrimaryParticleComponent _returnvalue) "Intensity" _intensity)
    (Effects|Components|ParticleSystem|SetFloatParameter (CameraLensEffect|GetPrimaryParticleComponent _returnvalue) "Intensity2" (Math|Float|MapRangeClamped _intensity 0.0 1.0 0.3 0.8))
    (Game|Feedback|ClientStartCameraShake :self (Pawn|GetLocalViewingPlayerController _asfort_player_pawn) :Shake "/FNGameplayCues/Effects/Fort_Effects/BluePrints/CameraShakes/BP_CamShake_BlurPulse.BP_CamShake_BlurPulse_C" :Scale (Math|Float|Clamp(Float) _intensity 0.2 1.5) :PlaySpace "CameraLocal")
    (bind _returnvalue_2 (Audio|SpawnSound2D 0 (Variables|Default|GetDamageTickSound)))
    (Audio|Parameter|SetFloatParameter _returnvalue_2 "Intensity" _intensity :!class /Script/AudioExtensions.AudioParameterControllerInterface)))

(event EventOnRemoval(Cascade)
  (params (MyTarget :type Actor%20Object%20Reference) (Parameters :type Gameplay%20Cue%20Parameters%20Structure%20%28by%20ref%29) (ParticleComponents :type Array%20of%20Cascade%20Particle%20System%20Component%20Object%20References) (AudioComponents :type Array%20of%20Audio%20Component%20Object%20References) (BurstCameraShakeInstance :type Camera%20Shake%20Base%20Object%20Reference) (BurstDecalInstance :type Decal%20Actor%20Object%20Reference))
  (Audio|PopSoundMixModifier 0 (Variables|Default|GetDrownSoundMix))
  (bind _asfort_player_pawn (Utilities|Casting|CastToFortPlayerPawn MyTarget))
  (if (Utilities|Equal(Object) _asfort_player_pawn (Variables|Default|GetLocalTargetPawn))
    (Variables|Default|SetIntensity 0.1)))

(fn ConstructionScript)
