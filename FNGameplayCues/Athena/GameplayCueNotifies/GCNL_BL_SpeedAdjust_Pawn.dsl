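; Looping gameplay cue for the Battle Lab speed adjust effect, showing speed-line particles on the affected player pawn.
; On application, only for the locally controlled pawn, it attaches a Cascade speed-lines emitter to the root and enables tick.
; Each tick maps pawn velocity magnitude through a configurable range to the emitter's Alpha parameter, so the lines follow actual speed.
; Removal stops tick, deactivates the emitter and clears its reference.
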(event EventOnApplication(Cascade)
  (params (MyTarget :type Actor%20Object%20Reference) (Parameters :type Gameplay%20Cue%20Parameters%20Structure%20%28by%20ref%29) (ParticleComponents :type Array%20of%20Cascade%20Particle%20System%20Component%20Object%20References) (AudioComponents :type Array%20of%20Audio%20Component%20Object%20References) (BurstCameraShakeInstance :type Camera%20Shake%20Base%20Object%20Reference) (BurstDecalInstance :type Decal%20Actor%20Object%20Reference))
  (bind _asfort_player_pawn_athena (Utilities|Casting|CastToFortPlayerPawnAthena MyTarget))
  (if (Pawn|IsLocallyControlled _asfort_player_pawn_athena)
    (Variables|Default|SetLocalPawn _asfort_player_pawn_athena)
    (Actor|Tick|SetActorTickEnabled (Pawn|IsLocallyControlled _asfort_player_pawn_athena))
    (bind _returnvalue (Effects|SpawnEmitterAttached :EmitterTemplate "/FNGameplayCues/Creative/Items/BattleLab/Blueprint/SpeedIncrease/P_BL_SpeedIncrease_Lines.P_BL_SpeedIncrease_Lines" :AttachToComponent (Variables|/Script/Engine.Actor|GetRootComponent (Variables|Default|GetLocalPawn)) :AttachPointName "None" :Scale "1.000000,1.000000,1.000000" :LocationType "KeepRelativeOffset" :bAutoDestroy true :PoolingMethod "AutoRelease" :bAutoActivate true))
    (Variables|Default|SetLoopingPS _returnvalue)))

(event EventTick
  (params (DeltaSeconds :type Float%20%28single-precision%29))
  (bind _localpawn (Variables|Default|GetLocalPawn)
    (:then
      (bind _loopingps (Variables|Default|GetLoopingPS)
        (:then
          (Effects|Components|ParticleSystem|SetFloatParameter _loopingps "Alpha" (Math|Float|MapRangeClamped (Math|Vector|VectorLength (Transformation|GetVelocity _localpawn)) (Variables|Default|GetInRangeA) (Variables|Default|GetInRangeB) 0.0 1.0)))
        (:else)))
    (:else)))

(event EventOnRemoval(Cascade)
  (params (MyTarget :type Actor%20Object%20Reference) (Parameters :type Gameplay%20Cue%20Parameters%20Structure%20%28by%20ref%29) (ParticleComponents :type Array%20of%20Cascade%20Particle%20System%20Component%20Object%20References) (AudioComponents :type Array%20of%20Audio%20Component%20Object%20References) (BurstCameraShakeInstance :type Camera%20Shake%20Base%20Object%20Reference) (BurstDecalInstance :type Decal%20Actor%20Object%20Reference))
  (Actor|Tick|SetActorTickEnabled false)
  (Components|Activation|Deactivate (Variables|Default|GetLoopingPS))
  (Variables|Default|SetLoopingPS 0))

(fn ConstructionScript)
