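; Looping gameplay cue that shows an alert or threatened status icon above an Athena pawn's head, cosmetic only.
; Effect level selects the state: negative hides the icon, otherwise a reused static mesh attaches to the head bone, with colors and animation timing passed through custom primitive data.
; State 1 seeds a question-mark fill and enables tick; removal hides the mesh and stops tick.
; A helper reads the MANG pawn's stealth meter.
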(event EventOnApplication(Cascade)
  (params (MyTarget :type Actor%20Object%20Reference) (Parameters :type Gameplay%20Cue%20Parameters%20Structure%20%28by%20ref%29) (ParticleComponents :type Array%20of%20Cascade%20Particle%20System%20Component%20Object%20References) (AudioComponents :type Array%20of%20Audio%20Component%20Object%20References) (BurstCameraShakeInstance :type Camera%20Shake%20Base%20Object%20Reference) (BurstDecalInstance :type Decal%20Actor%20Object%20Reference))
  (bind (_normalizedmagnitude _rawmagnitude _effectcontext _matchedtagname _originaltag _aggregatedsourcetags _aggregatedtargettags _location _normal _instigator _effectcauser _sourceobject _physicalmaterial _gameplayeffectlevel) (Ability|GameplayCue|BreakGameplayCueParameters Parameters))
  (bind (_returnvalue_r _returnvalue_g _returnvalue_b) (Math|Color|BreakColor (Utilities|Select :Index (Variables|Default|GetIndex) :Option%200 (Variables|Default|GetIconColorAlert) :Option%201 (Variables|Default|GetIconColorThreatend))))
  (bind (_iconcoloralertinner_r _iconcoloralertinner_g _iconcoloralertinner_b _iconcoloralertinner_a) (Math|Color|BreakColor (Variables|GetIconColorAlertInner_R)))
  (Networking|CanExecuteCosmeticEvents
    (:True
      (Actor|Tick|SetTickGroup "TG_DuringPhysics")
      (Actor|Tick|SetActorTickEnabled false :!pinned true) ; SET TICK = FALSE
      (bind _asplayer_pawn_athena (Utilities|Casting|CastToPlayerPawn_Athena MyTarget)
        (:then
          (Variables|Default|SetPawn _asplayer_pawn_athena)
          (bind _output_get (Variables|Default|SetIndex _gameplayeffectlevel))
          (if (< _output_get 0)
            (CallFunction|HideSpriteMesh)
            (else
              (bind _spritemesh (Variables|Default|GetSpriteMesh)
                (:then
                  (Rendering|SetVisibility _spritemesh true false))
                (:else
                  (bind _returnvalue (AddComponent|NODE_AddStaticMeshComponent-0))
                  (Variables|Default|SetSpriteMesh _returnvalue)))
              (Transformation|AttachComponentToComponent (Variables|Default|GetSpriteMesh) (Variables|/Script/Engine.Character|GetMesh (Variables|Default|GetPawn)) "head" "KeepRelative" "KeepRelative" "KeepRelative" true)
              (Transformation|SetRelativeTransform (Variables|Default|GetSpriteMesh) (Variables|Default|GetMeshRelativeTransform) false false)
              (Rendering|Material|SetCustomPrimitiveDataVector4 (Variables|Default|GetSpriteMesh) 0 (Math|Vector4|MakeVector4 _returnvalue_r _returnvalue_g _returnvalue_b (Math|Conversions|ToFloat(Integer) (Variables|Default|GetIndex))))
              (Rendering|Material|SetCustomPrimitiveDataVector4 (Variables|Default|GetSpriteMesh) 6 (Math|Vector4|MakeVector4 _iconcoloralertinner_r _iconcoloralertinner_g _iconcoloralertinner_b _iconcoloralertinner_a))
              (Rendering|Material|SetCustomPrimitiveDataFloat (Variables|Default|GetSpriteMesh) 4 (Utilities|Time|GetGameTimeinSeconds 0))
              (Rendering|Material|SetCustomPrimitiveDataFloat (Variables|Default|GetSpriteMesh) 5 (Variables|Default|GetAnimationLength))
              ;[1 1 1] Use this to update "?" progress (normalized unsigned value)
              (if (== (Variables|Default|GetIndex) 1)
                (bind _output_get_2 (Variables|Default|SetStealthMeterPctFilled (Variables|Default|GetStealthMeterStartValue)))
                (CallFunction|UpdateQuestionMarkFill _output_get_2)
                (Actor|Tick|SetTickGroup "TG_DuringPhysics")
                (Actor|Tick|SetActorTickEnabled true :!pinned true) ; SET TICK = TRUE
                (Variables|GameplayCueNotify_BaconSnack_State|SetMesh (Variables|Default|GetSpriteMesh))
                (else
                  (Actor|Tick|SetTickGroup "TG_DuringPhysics")
                  (Actor|Tick|SetActorTickEnabled false :!pinned true) ; SET TICK = FALSE
                )))))
        (:CastFailed
          (CallFunction|HideSpriteMesh))))) ; Department of Redundancy Department?
)

(event Custom|HideSpriteMesh
  (Networking|CanExecuteCosmeticEvents
    (:True
      (bind _spritemesh (Variables|Default|GetSpriteMesh)
        (:then
          (Rendering|SetVisibility _spritemesh false false))
        (:else)))) ; Department of Redundancy Department?
)

(event EventOnRemoval(Cascade)
  (params (MyTarget :type Actor%20Object%20Reference) (Parameters :type Gameplay%20Cue%20Parameters%20Structure%20%28by%20ref%29) (ParticleComponents :type Array%20of%20Cascade%20Particle%20System%20Component%20Object%20References) (AudioComponents :type Array%20of%20Audio%20Component%20Object%20References) (BurstCameraShakeInstance :type Camera%20Shake%20Base%20Object%20Reference) (BurstDecalInstance :type Decal%20Actor%20Object%20Reference))
  (Networking|CanExecuteCosmeticEvents
    (:True
      (Variables|Default|SetStealthMeterStartValue 0.0)
      (Actor|Tick|SetTickGroup "TG_DuringPhysics")
      (Actor|Tick|SetActorTickEnabled false :!pinned true) ; SET TICK = FALSE
      (bind _spritemesh (Variables|Default|GetSpriteMesh)
        (:then
          (Rendering|SetVisibility _spritemesh false false))
        (:else)))) ; Department of Redundancy Department?
)

(fn GetStealthMeterInfo
  (bind _pawn (Variables|Default|GetPawn)
    (:then
      (bind _asbp_mang_player_pawn_parent (Utilities|Casting|CastToBP_MangPlayerPawn_Parent _pawn)
        (:then
          (return (> (Utilities|Struct|BreakStructureMANGStealthMeter (Variables|/Game/Athena/AI/MANG/BP_MangPlayerPawn_Parent.BP_MangPlayerPawn_Parent_C|GetStealthMeterPawn _asbp_mang_player_pawn_parent)) 0.0) (.StealthMeterTargetTime_5_627E99734167FD2903748490D5FC2A57 (Utilities|Struct|BreakStructureMANGStealthMeter (Variables|/Game/Athena/AI/MANG/BP_MangPlayerPawn_Parent.BP_MangPlayerPawn_Parent_C|GetStealthMeterPawn _asbp_mang_player_pawn_parent)))))
        (:CastFailed
          _)))
    (:else
      _))
  (return false 0.0))

(fn UpdateQuestionMarkFill
  (params (Value :type Float%20%28double-precision%29))
  (Rendering|Material|SetCustomPrimitiveDataFloat (Variables|Default|GetSpriteMesh) 10 Value))

(fn ConstructionScript)
