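; Looping gameplay cue that provides camera, particle and sound feedback while a player pawn is under the Cube Area effect.
; On start it spawns a camera lens effect and an applied sound on the pawn.
; For the locally controlled pawn, it caches the first Cascade component and binds its particle-spawn event, so "Shield" events play attached tick sounds.
; Removal destroys the lens effect, plays a removal sound and fades out the latest tick sound.
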(event EventBeginPlay)

(event Collision|EventActorBeginOverlap
  (params (OtherActor :type Actor%20Object%20Reference)))

(event EventTick
  (params (DeltaSeconds :type Float%20%28single-precision%29)))

(event EventOnLoopingStart(Cascade)
  (params (MyTarget :type Actor%20Object%20Reference) (Parameters :type Gameplay%20Cue%20Parameters%20Structure%20%28by%20ref%29) (ParticleComponents :type Array%20of%20Cascade%20Particle%20System%20Component%20Object%20References) (AudioComponents :type Array%20of%20Audio%20Component%20Object%20References))
  (bind _output (Utilities|Array|Get(acopy) ParticleComponents))
  (bind _asfort_player_pawn_athena (Utilities|Casting|CastToFortPlayerPawnAthena MyTarget))
  (bind _returnvalue (Feedback|SpawnGenericCameraLensEffect _asfort_player_pawn_athena "/FNGameplayCues/Effects/Fort_Effects/Camera_Effects/B_CameraLens_Cube_Area.B_CameraLens_Cube_Area_C"))
  (Variables|Default|SetLensEffectInterface _returnvalue)
  (Audio|SpawnSoundAttached (Variables|Default|GetOnAppliedSound) (Variables|/Script/Engine.Actor|GetRootComponent _asfort_player_pawn_athena))
  (if (Pawn|IsLocallyControlled _asfort_player_pawn_athena)
    (Utilities|IsValid _output
      (:Is%20Valid
        (bind _output_get (Variables|Default|SetShieldFX _output))
        (EventDispatchers|BindEventtoOnParticleSpawn _output_get (EventDispatchers|CreateEvent :!event Custom|On%20Shield%20Squares%20Spawn))))))

(event Custom|On%20Shield%20Squares%20Spawn
  (params (EventName :type Name) (EmitterTime :type Float%20%28single-precision%29) (Location :type Vector) (Velocity :type Vector))
  (if (Utilities|Name|Equal(Name) EventName "Shield")
    (bind _returnvalue (Audio|SpawnSoundAttached (Variables|Default|GetOnTickSound) (Variables|Default|GetShieldFX)))
    (Variables|Default|SetCurrentTickSound _returnvalue)))

(event EventOnRemoval(Cascade)
  (params (MyTarget :type Actor%20Object%20Reference) (Parameters :type Gameplay%20Cue%20Parameters%20Structure%20%28by%20ref%29) (ParticleComponents :type Array%20of%20Cascade%20Particle%20System%20Component%20Object%20References) (AudioComponents :type Array%20of%20Audio%20Component%20Object%20References) (BurstCameraShakeInstance :type Camera%20Shake%20Base%20Object%20Reference) (BurstDecalInstance :type Decal%20Actor%20Object%20Reference))
  (Actor|DestroyActor (Utilities|Casting|CastToActor (Variables|Default|GetLensEffectInterface)))
  (Audio|SpawnSoundAttached (Variables|Default|GetOnRemovedSound) (Variables|Default|GetShieldFX))
  (bind _currentticksound (Variables|Default|GetCurrentTickSound)
    (:then
      (Audio|Components|Audio|FadeOut _currentticksound 1.0 0.0 "Linear" :!class /Script/Engine.AudioComponent))
    (:else)))

(fn ConstructionScript)
