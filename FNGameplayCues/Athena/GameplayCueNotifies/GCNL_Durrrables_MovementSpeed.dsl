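; Looping gameplay cue notify for the Durrrables movement speed effect, adding foot and head particles to the affected actor.
; On becoming relevant it attaches the speed emitter at FX_Feet, passing the actor as PlayerMeshActor, and ear steam at FX_Head.
; Tick drives the foot emitter's Opacity from the actor's clamped, scaled velocity magnitude.
; On cease it destroys only the foot emitter, as the ear steam component is not stored.
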(event EventTick
  (params (DeltaSeconds :type Float%20%28single-precision%29))
  (bind _targetactor (Variables|Default|GetTargetActor))
  (Utilities|IsValid (Variables|Default|GetPickupFX)
    (:Is%20Valid
      (Utilities|IsValid _targetactor
        (:Is%20Valid
          (Effects|Components|ParticleSystem|SetFloatParameter (Variables|Default|GetPickupFX) "Opacity" (Math|Float|Clamp(Float) (/ (Math|Vector|VectorLength (Transformation|GetVelocity _targetactor)) 1e+03))))))))

(fn OnCeaseRelevant
  (params (MyTarget :type Actor%20Object%20Reference) (Parameters :type Gameplay%20Cue%20Parameters%20Structure%20%28by%20ref%29))
  (Utilities|IsValid (Variables|Default|GetPickupFX)
    (:Is%20Valid
      (Components|DestroyComponent (Variables|Default|GetPickupFX) (Variables|GetMyTarget))
      (Variables|Default|SetPickupFX 0))
    (:Is%20Not%20Valid
      _))
  (return false))

(fn ConstructionScript)

(fn OnBecomeRelevant
  (params (MyTarget :type Actor%20Object%20Reference) (Parameters :type Gameplay%20Cue%20Parameters%20Structure%20%28by%20ref%29))
  (Utilities|IsValid MyTarget
    (:Is%20Valid
      (bind _output_get (LocalVariables|SetTargetComp (GameplayCueNotify|GetTargetComponent MyTarget Parameters)))
      (Variables|Default|SetTargetActor MyTarget)
      (Utilities|IsValid _output_get
        (:Is%20Valid
          (bind _returnvalue (Effects|SpawnEmitterAttached (Variables|Default|GetMovementSpeedParticleSystem) (LocalVariables|GetTargetComp) "FX_Feet"))
          (bind _output_get_2 (Variables|Default|SetPickupFX _returnvalue))
          (Effects|Components|ParticleSystem|SetActorParameter _output_get_2 "PlayerMeshActor" MyTarget)
          (Effects|SpawnEmitterAttached (Variables|Default|GetEarSteamParticleSystem) (LocalVariables|GetTargetComp) "FX_Head")
          (return false))
        (:Is%20Not%20Valid
          _)))
    (:Is%20Not%20Valid
      _))
  (return false))
