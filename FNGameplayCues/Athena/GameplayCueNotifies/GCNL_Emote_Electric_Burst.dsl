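; Looping gameplay cue notify for the Electric Burst emote that feeds a player's rank progression into its effect.
; When the loop starts, it stores the first spawned particle component and casts the cue target to a Fort player pawn.
; It reads that pawn's StatManager.HabaneroProgression cosmetic stat and passes it to the effect as the RankProgression integer parameter.
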(event EventOnLoopingStart(EitherVFX-preferthisversion)
  (params (MyTarget :type Actor%20Object%20Reference) (Parameters :type Gameplay%20Cue%20Parameters%20Structure%20%28by%20ref%29) (ParticleComponents :type Array%20of%20FXSystem%20Component%20Object%20References) (AudioComponents :type Array%20of%20Audio%20Component%20Object%20References))
  (Variables|Default|SetFXSystem (Utilities|Array|Get(acopy) ParticleComponents))
  ; SetRankProgression
  (bind _asfort_player_pawn (Utilities|Casting|CastToFortPlayerPawn MyTarget))
  (bind _output_get (Variables|Default|SetPlayerPawn _asfort_player_pawn))
  (bind _returnvalue (Class|FortCosmeticFunctionLibrary|GetCosmeticStatValue _output_get "(TagName=\"StatManager.HabaneroProgression\")"))
  (Effects|Components|ParticleSystem|SetIntParameter (Variables|Default|GetFXSystem) "RankProgression" _returnvalue))

(fn ConstructionScript)
