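; Looping gameplay cue for a doused environmental campfire that controls the fade of its smoke particle effect.
; On loop start, a timeline, with play rate set to the inverse of a scalable-float duration, drives the first Cascade component's "Alpha" parameter.
; It also binds to the target's OnDestroyed event to zero that Alpha immediately.
; On removal, the timeline stops.
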(event Custom|DestroyOnDestroy
  (params (DestroyedActor :type Actor%20Object%20Reference))
  (Effects|Components|ParticleSystem|SetFloatParameter (Utilities|Array|Get(acopy) (stale EventOnLoopingStart(Cascade) ParticleComponents)) "Alpha" 0.0))

(event EventOnLoopingStart(Cascade)
  (params (MyTarget :type Actor%20Object%20Reference) (Parameters :type Gameplay%20Cue%20Parameters%20Structure%20%28by%20ref%29) (ParticleComponents :type Array%20of%20Cascade%20Particle%20System%20Component%20Object%20References) (AudioComponents :type Array%20of%20Audio%20Component%20Object%20References))
  (Game|BindEventtoOnDestroyed MyTarget (delegate Custom|DestroyOnDestroy))
  (Components|Timeline|SetPlayRate (Variables|GCNL_EnvCampFire_Doused|GetFade%20Doused%20Smoke) (/ 1.0 (ScalableFloat|GetValueatLevel (Variables|Default|GetRow_Duration))))
  (bind (_direction _fade) (Timeline|Fade%20Doused%20Smoke PlayFromStart)
    (:Update
      (Effects|Components|ParticleSystem|SetFloatParameter (Utilities|Array|Get(acopy) ParticleComponents) "Alpha" _fade))))

(event EventOnRemoval(Cascade)
  (params (MyTarget :type Actor%20Object%20Reference) (Parameters :type Gameplay%20Cue%20Parameters%20Structure%20%28by%20ref%29) (ParticleComponents :type Array%20of%20Cascade%20Particle%20System%20Component%20Object%20References) (AudioComponents :type Array%20of%20Audio%20Component%20Object%20References) (BurstCameraShakeInstance :type Camera%20Shake%20Base%20Object%20Reference) (BurstDecalInstance :type Decal%20Actor%20Object%20Reference))
  (Timeline|Fade%20Doused%20Smoke Stop))

(fn ConstructionScript)
