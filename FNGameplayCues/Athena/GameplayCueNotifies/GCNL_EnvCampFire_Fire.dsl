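; Looping gameplay cue for the campfire's fire effects.
; On loop start it moves particles to the campfire's fire location and runs a log-darkening timeline over tick interval times max heals, driving wood color and particle alpha.
; A repeating timer checks the stoked state, with timelines animating the Niagara light scale behind played-state gates.
; On removal it stops darkening and aligns particle rotation to the wood.
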(event EventBeginPlay)

(event Collision|EventActorBeginOverlap
  (params (OtherActor :type Actor%20Object%20Reference)))

(event EventTick
  (params (DeltaSeconds :type Float%20%28single-precision%29)))

(event EventOnLoopingStart(EitherVFX-preferthisversion)
  (params (MyTarget :type Actor%20Object%20Reference) (Parameters :type Gameplay%20Cue%20Parameters%20Structure%20%28by%20ref%29) (ParticleComponents :type Array%20of%20FXSystem%20Component%20Object%20References) (AudioComponents :type Array%20of%20Audio%20Component%20Object%20References))
  (bind _output (Utilities|Array|Get(acopy) ParticleComponents))
  ;[1 1 1] Set the location of the fire to match that of the FireLoc component in the BGA
  (bind _ascampfire_interface (Utilities|Casting|CastToCampfire_Interface MyTarget))
  (bind _worldlocation (Campfire|GetLocationtoSpawnFireVFX _ascampfire_interface))
  (bind _woodcomponent (Campfire|GetWoodComponent _ascampfire_interface))
  (Variables|Default|SetParentCampfire _ascampfire_interface)
  (for _array_element ParticleComponents
    (Transformation|SetWorldLocation _array_element _worldlocation false false)
    (bind _asniagara_particle_system_component (Utilities|Casting|CastToNiagaraComponent _output))
    (Variables|Default|SetFx%20Variable _asniagara_particle_system_component)
    (bind _returnvalue (Utilities|Time|SetTimerbyEvent (delegate Custom|StokeCheckLoop) 0.2 true)))
  (Components|Timeline|SetPlayRate (Variables|GCNL_EnvCampFire_Fire|GetDarken%20Logs) (/ 1.0 (* (ScalableFloat|GetValueatLevel (Variables|Default|GetRow_TickInterval)) (ScalableFloat|GetValueatLevel (Variables|Default|GetRow_MaxHeals)))))
  (bind (_direction _lerp_control) (Timeline|Darken%20Logs PlayFromStart)
    (:Update
      (Rendering|Material|SetVectorParameterValueonMaterials _woodcomponent "Base Color Multiplier" (Math|Vector|Lerp(Vector) "1.000000,1.000000,1.000000" "0.100000,0.100000,0.100000" _lerp_control))
      (Effects|Components|ParticleSystem|SetFloatParameter _output "Alpha" (Math|Float|Lerp 0.0 1.0 _lerp_control)))))

(event EventOnRemoval(EitherVFX-preferthisversion)
  (params (MyTarget :type Actor%20Object%20Reference) (Parameters :type Gameplay%20Cue%20Parameters%20Structure%20%28by%20ref%29) (ParticleComponents :type Array%20of%20FXSystem%20Component%20Object%20References) (AudioComponents :type Array%20of%20Audio%20Component%20Object%20References) (BurstCameraShakeInstance :type Camera%20Shake%20Base%20Object%20Reference) (BurstDecalInstance :type Decal%20Actor%20Object%20Reference))
  ;[1 1 1] Make sure wood is rotated to match the wood in the pit
  (Utilities|FlowControl|Sequence
    (:then_0
      (Timeline|Darken%20Logs Stop))
    (:then_1
      (bind _ascampfire_interface (Utilities|Casting|CastToCampfire_Interface MyTarget))
      (bind _woodcomponent (Campfire|GetWoodComponent _ascampfire_interface))
      (for _array_element ParticleComponents
        (Transformation|SetWorldRotation _array_element (Transformation|GetWorldRotation _woodcomponent) false false)))))

(event Custom|StokeCheckLoop
  (bind _fx_variable (Variables|Default|GetFx%20Variable))
  (Campfire|IsStoked (Variables|Default|GetParentCampfire))
  (if true
    (if (Variables|Default|GetHasPlayed)
      _
      (else
        _))
    (elif (Variables|Default|GetHasPlayed)
      (Utilities|FlowControl|DoOnce false
        (:Completed
          (bind (_direction _newtrack_0) (Timeline|Timeline_0 PlayFromStart)
            (:Update
              _)
            (:Finished
              (Variables|Default|SetHasPlayed false)
              (Utilities|FlowControl|Sequence
                (:then_0
                  _)
                (:then_1
                  _))
              (Utilities|FlowControl|DoOnce false
                (:Completed
                  (bind (_direction_2 _newtrack_0_2) (Timeline|StokeTimeline PlayFromStart)
                    (:Update
                      (Effects|Components|ParticleSystem|SetFloatParameter _fx_variable "LightScale" _newtrack_0_2))
                    (:Finished
                      (Variables|Default|SetHasPlayed true)))))))
          (Effects|Components|ParticleSystem|SetFloatParameter _fx_variable "LightScale" _newtrack_0))))))

(fn ConstructionScript)
