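; Looping gameplay cue for the environmental campfire's burning state, using the Janus campfire Niagara effect.
; On loop start it spawns the effect just below the campfire actor's FireLoc and stores the component.
; A timeline darkens the Wood material's base color multiplier, with its play rate scaled to tick interval times max heals.
; Removal stops the timeline, aligns removal Cascade particles to the wood's rotation, then deactivates and clears the flames.
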(event EventOnLoopingStart(EitherVFX-preferthisversion)
  (params (MyTarget :type Actor%20Object%20Reference) (Parameters :type Gameplay%20Cue%20Parameters%20Structure%20%28by%20ref%29) (ParticleComponents :type Array%20of%20FXSystem%20Component%20Object%20References) (AudioComponents :type Array%20of%20Audio%20Component%20Object%20References))
  ;[1 1 1] Set the location of the fire to match that of the FireLoc component in the BGA
  (bind _asb_bga_athena_env_camp_fire (Utilities|Casting|CastToB_BGA_Athena_EnvCampFire MyTarget))
  (bind _returnvalue (Niagara|SpawnSystematLocation 0 "/Game/Effects/Fort_Effects/World/Enviroment/Smoke/NS_Janus_Campfire.NS_Janus_Campfire" (+ (Transformation|GetWorldLocation (Variables|/Game/Athena/Items/ForagedItems/EnvCampFire/B_BGA_Athena_EnvCampFire.B_BGA_Athena_EnvCampFire_C|GetFireLoc _asb_bga_athena_env_camp_fire)) "0, 0,-30.000000")))
  (Variables|Default|SetNiagara%20Flames _returnvalue)
  (Components|Timeline|SetPlayRate (Variables|GCNL_EnvCampFire_Fire_Janus|GetDarken%20Logs) (/ 1.0 (* (ScalableFloat|GetValueatLevel (Variables|Default|GetRow_TickInterval)) (ScalableFloat|GetValueatLevel (Variables|Default|GetRow_MaxHeals)))))
  (bind (_direction _lerp_control) (Timeline|Darken%20Logs PlayFromStart)
    (:Update
      (Rendering|Material|SetVectorParameterValueonMaterials (Variables|/Game/Athena/Items/ForagedItems/EnvCampFire/B_BGA_Athena_EnvCampFire.B_BGA_Athena_EnvCampFire_C|GetWood _asb_bga_athena_env_camp_fire) "Base Color Multiplier" (Math|Vector|Lerp(Vector) "1.000000,1.000000,1.000000" "0.100000,0.100000,0.100000" _lerp_control)))))

(event EventOnRemoval(Cascade)
  (params (MyTarget :type Actor%20Object%20Reference) (Parameters :type Gameplay%20Cue%20Parameters%20Structure%20%28by%20ref%29) (ParticleComponents :type Array%20of%20Cascade%20Particle%20System%20Component%20Object%20References) (AudioComponents :type Array%20of%20Audio%20Component%20Object%20References) (BurstCameraShakeInstance :type Camera%20Shake%20Base%20Object%20Reference) (BurstDecalInstance :type Decal%20Actor%20Object%20Reference))
  ;[1 1 1] Make sure wood is rotated to match the wood in the pit
  (Utilities|FlowControl|Sequence
    (:then_0
      (Timeline|Darken%20Logs Stop))
    (:then_1
      (bind _asb_bga_athena_env_camp_fire (Utilities|Casting|CastToB_BGA_Athena_EnvCampFire MyTarget))
      (for _array_element ParticleComponents
        (Transformation|SetWorldRotation _array_element (Transformation|GetWorldRotation (Variables|/Game/Athena/Items/ForagedItems/EnvCampFire/B_BGA_Athena_EnvCampFire.B_BGA_Athena_EnvCampFire_C|GetWood _asb_bga_athena_env_camp_fire)) false false))
      (Components|Activation|Deactivate (Variables|Default|GetNiagara%20Flames))
      (Variables|Default|SetNiagara%20Flames 0))))

(fn ConstructionScript)
