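; Looping gameplay cue notify for stoking an environment campfire, placing its Cascade fire effects on the campfire.
; When the loop starts, it casts the target to Campfire_Interface and asks it for the fire VFX spawn location.
; It then moves every particle component in the cue to that world location, intended to match the campfire's FireLoc component.
; BeginPlay, overlap and tick events are empty.
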(event EventBeginPlay)

(event Collision|EventActorBeginOverlap
  (params (OtherActor :type Actor%20Object%20Reference)))

(event EventTick
  (params (DeltaSeconds :type Float%20%28single-precision%29)))

(event EventOnLoopingStart(Cascade)
  (params (MyTarget :type Actor%20Object%20Reference) (Parameters :type Gameplay%20Cue%20Parameters%20Structure%20%28by%20ref%29) (ParticleComponents :type Array%20of%20Cascade%20Particle%20System%20Component%20Object%20References) (AudioComponents :type Array%20of%20Audio%20Component%20Object%20References))
  ;[1 1 1] Set the location of the fire to match that of the FireLoc component in the BGA
  (bind _ascampfire_interface (Utilities|Casting|CastToCampfire_Interface MyTarget))
  (bind _worldlocation (Campfire|GetLocationtoSpawnFireVFX _ascampfire_interface :!class /Game/Athena/Items/ForagedItems/EnvCampFire/Campfire_Interface.Campfire_Interface_C))
  (for _array_element ParticleComponents
    (Transformation|SetWorldLocation _array_element _worldlocation false false)))

(fn ConstructionScript)
