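; Looping gameplay cue that drives the on-fire VFX for the industrial giant space heater, an exploding burst-reaction prop.
; On start and on a repeating timer, it sets the first particle component's SpawnRamp from a curve of how far health has fallen below the prop's OnFireThreshold.
; On start it also sets Niagara source location, length and width on the particle components to fit the heater's shape.
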(event Custom|StartLoop
  (Utilities|Time|SetTimerbyEvent (delegate Custom|Loop) 0.2 true))

(event Custom|Loop
  (bind _asathena_prop_burst_reaction_exploding (Utilities|Casting|CastToAthena_Prop_BurstReaction_Exploding (stale EventOnLoopingStart(EitherVFX-preferthisversion) MyTarget)))
  (Utilities|IsValid (Utilities|Array|Get(acopy) (stale EventOnLoopingStart(EitherVFX-preferthisversion) ParticleComponents))
    (:Is%20Valid
      (Effects|Components|ParticleSystem|SetInstanceParams (Utilities|Array|Get(acopy) (stale EventOnLoopingStart(EitherVFX-preferthisversion) ParticleComponents)) (Utilities|Array|MakeArray (Utilities|Struct|MakeParticleSysParam "SpawnRamp" "PSPT_Scalar" (Math|Curves|GetFloatValue (Variables|Default|GetSpawnRampCurve) (Math|Float|NormalizetoRange (- (Math|Float|MakeLiteralFloat (Variables|/Game/Athena/Items/EnvironmentalItems/ReactiveProps/Athena_Prop_BurstReaction_Exploding.Athena_Prop_BurstReaction_Exploding_C|GetOnFireThreshold _asathena_prop_burst_reaction_exploding)) (Health|GetHealthPercent _asathena_prop_burst_reaction_exploding :!class /Script/FortniteGame.BuildingActor)) 0.0 (Variables|/Game/Athena/Items/EnvironmentalItems/ReactiveProps/Athena_Prop_BurstReaction_Exploding.Athena_Prop_BurstReaction_Exploding_C|GetOnFireThreshold _asathena_prop_burst_reaction_exploding))) (Math|Curves|GetFloatValue (Variables|Default|GetSpawnRampCurve) (Math|Float|NormalizetoRange (- (Math|Float|MakeLiteralFloat (Variables|/Game/Athena/Items/EnvironmentalItems/ReactiveProps/Athena_Prop_BurstReaction_Exploding.Athena_Prop_BurstReaction_Exploding_C|GetOnFireThreshold _asathena_prop_burst_reaction_exploding)) (Health|GetHealthPercent _asathena_prop_burst_reaction_exploding :!class /Script/FortniteGame.BuildingActor)) 0.0 (Variables|/Game/Athena/Items/EnvironmentalItems/ReactiveProps/Athena_Prop_BurstReaction_Exploding.Athena_Prop_BurstReaction_Exploding_C|GetOnFireThreshold _asathena_prop_burst_reaction_exploding)))))))))

(event EventOnLoopingStart(EitherVFX-preferthisversion)
  (params (MyTarget :type Actor%20Object%20Reference) (Parameters :type Gameplay%20Cue%20Parameters%20Structure%20%28by%20ref%29) (ParticleComponents :type Array%20of%20FXSystem%20Component%20Object%20References) (AudioComponents :type Array%20of%20Audio%20Component%20Object%20References))
  (bind _output (Utilities|Array|Get(acopy) ParticleComponents))
  (Utilities|FlowControl|Sequence
    (:then_0
      (CallFunction|StartLoop)
      (bind _asathena_prop_burst_reaction_exploding (Utilities|Casting|CastToAthena_Prop_BurstReaction_Exploding MyTarget))
      (Utilities|IsValid _output
        (:Is%20Valid
          (Effects|Components|ParticleSystem|SetInstanceParams _output (Utilities|Array|MakeArray (Utilities|Struct|MakeParticleSysParam "SpawnRamp" "PSPT_Scalar" (Math|Curves|GetFloatValue (Variables|Default|GetSpawnRampCurve) (Math|Float|NormalizetoRange (- (Math|Float|MakeLiteralFloat (Variables|/Game/Athena/Items/EnvironmentalItems/ReactiveProps/Athena_Prop_BurstReaction_Exploding.Athena_Prop_BurstReaction_Exploding_C|GetOnFireThreshold _asathena_prop_burst_reaction_exploding)) (Health|GetHealthPercent _asathena_prop_burst_reaction_exploding :!class /Script/FortniteGame.BuildingActor)) 0.0 (Variables|/Game/Athena/Items/EnvironmentalItems/ReactiveProps/Athena_Prop_BurstReaction_Exploding.Athena_Prop_BurstReaction_Exploding_C|GetOnFireThreshold _asathena_prop_burst_reaction_exploding))) (Math|Curves|GetFloatValue (Variables|Default|GetSpawnRampCurve) (Math|Float|NormalizetoRange (- (Math|Float|MakeLiteralFloat (Variables|/Game/Athena/Items/EnvironmentalItems/ReactiveProps/Athena_Prop_BurstReaction_Exploding.Athena_Prop_BurstReaction_Exploding_C|GetOnFireThreshold _asathena_prop_burst_reaction_exploding)) (Health|GetHealthPercent _asathena_prop_burst_reaction_exploding :!class /Script/FortniteGame.BuildingActor)) 0.0 (Variables|/Game/Athena/Items/EnvironmentalItems/ReactiveProps/Athena_Prop_BurstReaction_Exploding.Athena_Prop_BurstReaction_Exploding_C|GetOnFireThreshold _asathena_prop_burst_reaction_exploding)))))))))
    (:then_1
      (for _array_element ParticleComponents
        ;[1 1 1] Alters the OnFire VFX scale and height to match the Heater's structure
        (bind _asniagara_particle_system_component (Utilities|Casting|CastToNiagaraComponent _array_element))
        (Niagara|SetNiagaraVariable(Vector3) _asniagara_particle_system_component "SourceLocation" (Variables|Default|GetOnFireVFXHeightOffset))
        (Niagara|SetNiagaraVariable(Float) _asniagara_particle_system_component "SourceLength" 3.5e+02)
        (Niagara|SetNiagaraVariable(Float) _asniagara_particle_system_component "SourceWidth" 35.0)))))

(fn ConstructionScript)
