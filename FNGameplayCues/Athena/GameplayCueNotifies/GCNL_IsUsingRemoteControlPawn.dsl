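; Gameplay cue notify for the remote-control pawn state, which shows the RC Rocket remote loop effect on the controlling player's weapon.
; OnBurst casts the target to PlayerPawn_Athena and, if valid, attaches the emitter to the current weapon mesh at hologramSocket.
; The graph stores no reference to the emitter and has no removal logic; EndPlay and Destroyed only print debug strings.
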(event EventBeginPlay)

(event Collision|EventActorBeginOverlap
  (params (OtherActor :type Actor%20Object%20Reference)))

(event EventTick
  (params (DeltaSeconds :type Float%20%28single-precision%29)))

(event EventEndPlay
  (params (EndPlayReason :type EEndPlayReason%20Enum :default "Destroyed"))
  (Development|PrintString 0 "EndPlay"))

(event EventDestroyed
  (Development|PrintString 0 "Destroyed"))

(fn OnBurst
  (params (MyTarget :type Actor%20Object%20Reference) (Parameters :type Gameplay%20Cue%20Parameters%20Structure%20%28by%20ref%29))
  (bind _asplayer_pawn_athena (Utilities|Casting|CastToPlayerPawn_Athena MyTarget))
  (Utilities|IsValid _asplayer_pawn_athena
    (:Is%20Valid
      (Effects|SpawnEmitterAttached :EmitterTemplate "/Game/Athena/Items/Weapons/Abilities/RCRocket/P_RC_Rocket_RemoteLoop.P_RC_Rocket_RemoteLoop" :AttachToComponent (Variables|/Script/FortniteGame.FortWeapon|GetWeaponMesh (Variables|/Script/FortniteGame.FortPawn|GetCurrentWeapon _asplayer_pawn_athena)) :AttachPointName "hologramSocket" :Scale "1.000000,1.000000,1.000000" :LocationType "SnapToTarget" :bAutoDestroy true :PoolingMethod "None" :bAutoActivate true)
      (Development|PrintString 0 "Remoting!")
      (return false))))

(fn ConstructionScript)
