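; Looping gameplay cue for island portal skydives that attaches the player's contrail effect to the pawn's mesh.
; On start it async-loads the pawn's contrail asset, then spawns it as Cascade or, if that cast fails, Niagara, unless one already exists.
; A movement-mode delegate shows the contrail only while skydiving.
; Removal deactivates the emitter, which auto-destroys.
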(event EventOnLoopingStart(EitherVFX-preferthisversion)
  (params (MyTarget :type Actor%20Object%20Reference) (Parameters :type Gameplay%20Cue%20Parameters%20Structure%20%28by%20ref%29) (ParticleComponents :type Array%20of%20FXSystem%20Component%20Object%20References) (AudioComponents :type Array%20of%20Audio%20Component%20Object%20References))
  ;[0.196 1 0.1095] When the gameplay cue is activated, start the async load of the player's contrail effect.
  (bind _asfort_player_pawn (Utilities|Casting|CastToFortPlayerPawn MyTarget))
  (bind _output_get (Variables|Default|SetPlayerPawn _asfort_player_pawn))
  (bind _returnvalue (Parachute|GetContrailParticleSystemSoftRef _output_get))
  (bind _object (Utilities|AsyncLoadAsset _returnvalue)
    (:Completed
      (bind _asfxsystem_asset (Utilities|Casting|CastToFXSystemAsset _object))
      (Variables|Default|SetContrailParticleSystemAsset _asfxsystem_asset)
      (CallFunction|SpawnContrailsEffect))))

(event Custom|SpawnContrailsEffect
  (bind _playerpawn (Variables|Default|GetPlayerPawn))
  (bind _mesh (Variables|/Script/Engine.Character|GetMesh _playerpawn))
  (bind _playerpawn_2 (Variables|Default|GetPlayerPawn))
  (bind _contrailparticlesystemasset (Variables|Default|GetContrailParticleSystemAsset))
  ;[0 1 0.5572] Spawn the contrails effect. It may be a cascade effect or a niagara effect, so account for both possibilities. ONLY spawn the contrails effect if one has not already been spawned.
  ;;[1 0.8849 0.1165] Only  continue spawning the contrails if neither of the cached particle emitters are vaild (i.e. the contrails has not already been spawned).
  (if (not (or (Utilities|IsValid (Variables|Default|GetCascadeContrailEmitter) :!class /Script/Engine.KismetSystemLibrary) (Utilities|IsValid (Variables|Default|GetNiagaraContrailEmitter) :!class /Script/Engine.KismetSystemLibrary)))
    (Utilities|IsValid _playerpawn_2
      (:Is%20Valid
        (Character|AssignMovementModeChangedDelegate _playerpawn_2 (EventDispatchers|CreateEvent :!event Custom|HandleMovementModeChanged))
        (bind _asparticle_system (Utilities|Casting|CastToParticleSystem _contrailparticlesystemasset)
          (:then
            (bind _returnvalue (Effects|SpawnEmitterAttached :EmitterTemplate _asparticle_system :AttachToComponent _mesh :AttachPointName "None" :Scale "1.000000,1.000000,1.000000" :LocationType "SnapToTarget" :bAutoDestroy true :PoolingMethod "None" :bAutoActivate true))
            (Variables|Default|SetCascadeContrailEmitter _returnvalue))
          (:CastFailed
            (bind _asniagara_system (Utilities|Casting|CastToNiagaraSystem _contrailparticlesystemasset))
            (bind _returnvalue_2 (Effects|SpawnFXSystemAttached :FXSystemAsset _asniagara_system :AttachToComponent _mesh :AttachPointName "None" :Scale "1.000000,1.000000,1.000000" :LocationType "SnapToTarget" :bAutoDestroy true :PoolingMethod "None" :bAutoActivate true))
            (Variables|Default|SetNiagaraContrailEmitter _returnvalue_2)))))))

(event Custom|HandleMovementModeChanged
  (params (Character :type Character%20Object%20Reference) (PrevMovementMode :type EMovementMode%20Enum :default "MOVE_None") (PreviousCustomMode :type Byte))
  ;[1 0.6105 0.09279] Update contrail visibility when the player's movement mode is changed.
  ;;[0.5525 0 1] Listen for the player's movement mode changing (bound to FortPlayerPawn::MovementModeChangedDelegate in SpawnContrailsEffect). If the player is skydiving, show the contrails, otherwise hide the contrails.
  (Utilities|IsValid Character
    (:Is%20Valid
      (bind _ascharacter_movement_component (Utilities|Casting|CastToCharacterMovementComponent (Pawn|GetMovementComponent Character)))
      (if (Utilities|Enum|Equal(Enum) (Variables|/Script/Engine.CharacterMovementComponent|GetMovementMode _ascharacter_movement_component) "MOVE_Custom")
        (if (Utilities|Enum|Equal(Enum) (Utilities|Enum|BytetoEnumEFortCustomMovement (Variables|/Script/Engine.CharacterMovementComponent|GetCustomMovementMode _ascharacter_movement_component)) "Skydiving")
          (CallFunction|ShowContrails)
          (else
            _))
        (else
          _))
      (CallFunction|HideContrails))))

(event Custom|HideContrails
  (bind _cascadecontrailemitter (Variables|Default|GetCascadeContrailEmitter))
  (bind _niagaracontrailemitter (Variables|Default|GetNiagaraContrailEmitter))
  ;[1 0.6105 0.09279] Update contrail visibility when the player's movement mode is changed.
  ;;[1 0 0.6088] Hide the contrails particle system component. Check both the Cascade and Niagara versions (though only one should be valid).
  (Utilities|FlowControl|Sequence
    (:then_0
      (Utilities|IsValid _cascadecontrailemitter
        (:Is%20Valid
          (Rendering|SetVisibility _cascadecontrailemitter false false))))
    (:then_1
      (Utilities|IsValid _niagaracontrailemitter
        (:Is%20Valid
          (Rendering|SetVisibility _niagaracontrailemitter false false))))))

(event Custom|ShowContrails
  (bind _niagaracontrailemitter (Variables|Default|GetNiagaraContrailEmitter))
  (bind _cascadecontrailemitter (Variables|Default|GetCascadeContrailEmitter))
  ;[1 0.6105 0.09279] Update contrail visibility when the player's movement mode is changed.
  ;;[0.1244 1 0.1589] Show the contrails particle system component. Check both the Cascade and Niagara versions (though only one should be valid).
  (Utilities|FlowControl|Sequence
    (:then_0
      (Utilities|IsValid _cascadecontrailemitter
        (:Is%20Valid
          (Rendering|SetVisibility _cascadecontrailemitter true false))))
    (:then_1
      (Utilities|IsValid _niagaracontrailemitter
        (:Is%20Valid
          (Rendering|SetVisibility _niagaracontrailemitter true false))))))

(event EventOnRemoval(EitherVFX-preferthisversion)
  (params (MyTarget :type Actor%20Object%20Reference) (Parameters :type Gameplay%20Cue%20Parameters%20Structure%20%28by%20ref%29) (ParticleComponents :type Array%20of%20FXSystem%20Component%20Object%20References) (AudioComponents :type Array%20of%20Audio%20Component%20Object%20References) (BurstCameraShakeInstance :type Camera%20Shake%20Base%20Object%20Reference) (BurstDecalInstance :type Decal%20Actor%20Object%20Reference))
  (bind _niagaracontrailemitter (Variables|Default|GetNiagaraContrailEmitter))
  (bind _cascadecontrailemitter (Variables|Default|GetCascadeContrailEmitter))
  ;[1 0.3476 0] Attempt to deactivate both emitters (though only one should be valid). Because they were spawned with 'bAutoDestroy', deactivating the components will also destroy them.
  (Utilities|FlowControl|Sequence
    (:then_0
      (Utilities|IsValid _cascadecontrailemitter
        (:Is%20Valid
          (Components|Activation|Deactivate _cascadecontrailemitter))))
    (:then_1
      (Utilities|IsValid _niagaracontrailemitter
        (:Is%20Valid
          (Components|Activation|Deactivate _niagaracontrailemitter))))))

(fn ConstructionScript)
