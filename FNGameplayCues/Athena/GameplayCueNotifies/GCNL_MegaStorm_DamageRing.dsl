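; Looping gameplay cue that displays the mega storm damage ring effect and drives storm damage on the safe zone indicator material.
; On looping start, a timeline animates the indicator MID's StormDamage parameter, obtained through an interface.
; Its first update spawns the Cascade ring through a DoOnce at the cue location, scaled by raw magnitude.
; On removal, a second timeline drives StormDamage, deactivates the ring, then calls Reset on that spawn logic.
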(event EventBeginPlay)

(event Collision|EventActorBeginOverlap
  (params (OtherActor :type Actor%20Object%20Reference)))

(event EventTick
  (params (DeltaSeconds :type Float%20%28single-precision%29)))

(event EventOnLoopingStart(Cascade)
  (params (MyTarget :type Actor%20Object%20Reference) (Parameters :type Gameplay%20Cue%20Parameters%20Structure%20%28by%20ref%29) (ParticleComponents :type Array%20of%20Cascade%20Particle%20System%20Component%20Object%20References) (AudioComponents :type Array%20of%20Audio%20Component%20Object%20References))
  (bind (_normalizedmagnitude _rawmagnitude _effectcontext _matchedtagname _originaltag _aggregatedsourcetags _aggregatedtargettags _location) (Ability|GameplayCue|BreakGameplayCueParameters Parameters))
  (bind _rawmagnitude_2 (Variables|Default|GetRawMagnitude))
  (bind _returnvalue (/ _rawmagnitude_2 1e+03))
  (bind (_returnvalue_x _returnvalue_y) (Math|Vector|BreakVector (Math|Conversions|ToVector(Float) _returnvalue)))
  (bind _assafe_zone_indicator_material_interface (Utilities|Casting|CastToSafeZoneIndicatorMaterial_Interface MyTarget))
  (bind _mid (SafeZoneIndicatorMaterial|GetSafeZoneIndicatorMID _assafe_zone_indicator_material_interface))
  (bind (_direction _stormdamage) (Timeline|Timeline_0 PlayFromStart)
    (:Update
      (Rendering|Material|SetScalarParameterValue _mid "StormDamage" _stormdamage :!class /Script/Engine.MaterialInstanceDynamic)
      (Utilities|FlowControl|DoOnce false
        (:Completed
          (Variables|Default|SetRawMagnitude _rawmagnitude)
          (bind _returnvalue_2 (Effects|SpawnEmitteratLocation MyTarget "/FNGameplayCues/Effects/Fort_Effects/Effects/Storm/P_SuperStorm_DamageRing.P_SuperStorm_DamageRing" _location))
          (bind _output_get (Variables|Default|SetStormDamageFX _returnvalue_2))
          (Effects|Components|ParticleSystem|SetFloatParameter _output_get "RingScaleCylinder" _rawmagnitude_2)
          (Effects|Components|ParticleSystem|SetVectorParameter _output_get "RingScale" (Math|Vector|MakeVector _returnvalue_x _returnvalue_y 2.0))
          (Effects|Components|ParticleSystem|SetFloatParameter _output_get "UVScale" (* _returnvalue 0.15)))))))

(event Custom|Reset
  (bind _rawmagnitude (Variables|Default|GetRawMagnitude))
  (bind _returnvalue (/ _rawmagnitude 1e+03))
  (bind (_returnvalue_x _returnvalue_y) (Math|Vector|BreakVector (Math|Conversions|ToVector(Float) _returnvalue)))
  (Utilities|FlowControl|DoOnce false
    (:Completed
      (Variables|Default|SetRawMagnitude (.RawMagnitude (Ability|GameplayCue|BreakGameplayCueParameters (stale EventOnLoopingStart(Cascade) Parameters))))
      (bind _returnvalue_2 (Effects|SpawnEmitteratLocation (stale EventOnLoopingStart(Cascade) MyTarget) "/FNGameplayCues/Effects/Fort_Effects/Effects/Storm/P_SuperStorm_DamageRing.P_SuperStorm_DamageRing" (.Location (Ability|GameplayCue|BreakGameplayCueParameters (stale EventOnLoopingStart(Cascade) Parameters)))))
      (bind _output_get (Variables|Default|SetStormDamageFX _returnvalue_2))
      (Effects|Components|ParticleSystem|SetFloatParameter _output_get "RingScaleCylinder" _rawmagnitude)
      (Effects|Components|ParticleSystem|SetVectorParameter _output_get "RingScale" (Math|Vector|MakeVector _returnvalue_x _returnvalue_y 2.0))
      (Effects|Components|ParticleSystem|SetFloatParameter _output_get "UVScale" (* _returnvalue 0.15)))))

(event EventOnRemoval(Cascade)
  (params (MyTarget :type Actor%20Object%20Reference) (Parameters :type Gameplay%20Cue%20Parameters%20Structure%20%28by%20ref%29) (ParticleComponents :type Array%20of%20Cascade%20Particle%20System%20Component%20Object%20References) (AudioComponents :type Array%20of%20Audio%20Component%20Object%20References) (BurstCameraShakeInstance :type Camera%20Shake%20Base%20Object%20Reference) (BurstDecalInstance :type Decal%20Actor%20Object%20Reference))
  (bind _assafe_zone_indicator_material_interface (Utilities|Casting|CastToSafeZoneIndicatorMaterial_Interface MyTarget))
  (bind _mid (SafeZoneIndicatorMaterial|GetSafeZoneIndicatorMID _assafe_zone_indicator_material_interface))
  (bind (_direction _stormdamage) (Timeline|Timeline_1 PlayFromStart)
    (:Update
      (Rendering|Material|SetScalarParameterValue _mid "StormDamage" _stormdamage :!class /Script/Engine.MaterialInstanceDynamic)
      (Components|Activation|Deactivate (Variables|Default|GetStormDamageFX)))
    (:Finished
      (CallFunction|Reset))))

(fn ConstructionScript)
