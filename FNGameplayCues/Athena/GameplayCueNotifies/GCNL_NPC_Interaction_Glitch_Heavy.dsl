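; Looping gameplay cue that puts a heavy glitch effect on a player pawn, named for NPC interactions.
; On start it plays spawn audio and applies a local glitch material override, tracked by GUID for later removal.
; Once character customization finishes, a cosmetic-part callback gathers part and current weapon meshes into a set and enables custom depth on them.
; Removal clears the set and override but leaves custom depth on.
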(event EventOnLoopingStart(Niagara)
  (params (MyTarget :type Actor%20Object%20Reference) (Parameters :type Gameplay%20Cue%20Parameters%20Structure%20%28by%20ref%29) (ParticleComponents :type Array%20of%20Niagara%20Particle%20System%20Component%20Object%20References) (AudioComponents :type Array%20of%20Audio%20Component%20Object%20References))
  (Utilities|Set|Clear (Variables|Default|GetHeavyGlitchContainer))
  (bind _asplayer_pawn_athena (Utilities|Casting|CastToPlayerPawn_Athena MyTarget)
    (:then
      (bind _output_get (Variables|Default|SetPawn _asplayer_pawn_athena))
      (Customization|BindEventtoOnFinishedCharacterCustomization _output_get (EventDispatchers|CreateEvent :!event Custom|OnFinishedCustomization))
      (Audio|Components|Audio|Play (Variables|Default|GetSpawnAudio) 0.0)
      (Utilities|FlowControl|Sequence
        (:then_0
          (if (Guid|IsValidGuid (Variables|Default|GetMaterial%20Override%20ID))
            (Class|FortPlayerPawn|RemoveMaterialOverrideLocal (Variables|Default|GetPawn) (Variables|Default|GetMaterial%20Override%20ID))
            (Variables|Default|SetMaterial%20Override%20ID (Utilities|Struct|MakeGuid))))
        (:then_1
          (bind _returnvalue (Class|FortPlayerPawn|ApplyMaterialOverrideLocalwithFlags :self (Variables|Default|GetPawn) :Material (Utilities|ResolveSoftReference (Variables|Default|GetGlitchMaterial)) :Priority 0.0 :Flags 2))
          (Variables|Default|SetMaterial%20Override%20ID _returnvalue))))
    (:CastFailed
      (Development|PrintString 0 "CastToFortPawn Failed!!" true true "(R=1.000000,G=0.016463,B=0.000000,A=1.000000)" 6.0))))

(event EventOnRemoval(EitherVFX-preferthisversion)
  (params (MyTarget :type Actor%20Object%20Reference) (Parameters :type Gameplay%20Cue%20Parameters%20Structure%20%28by%20ref%29) (ParticleComponents :type Array%20of%20FXSystem%20Component%20Object%20References) (AudioComponents :type Array%20of%20Audio%20Component%20Object%20References) (BurstCameraShakeInstance :type Camera%20Shake%20Base%20Object%20Reference) (BurstDecalInstance :type Decal%20Actor%20Object%20Reference))
  (Utilities|Set|Clear (Variables|Default|GetHeavyGlitchContainer))
  (if (Guid|IsValidGuid (Variables|Default|GetMaterial%20Override%20ID))
    (Class|FortPlayerPawn|RemoveMaterialOverrideLocal (Variables|Default|GetPawn) (Variables|Default|GetMaterial%20Override%20ID))))

(event Custom|OnFinishedCustomization
  (params (Pawn :type Fort%20Player%20Pawn%20Object%20Reference))
  (if (Customization|IsCharacterCustomizationLoadingCompleted Pawn)
    (Development|PrintString 0 "Is Character Customization Loading Completed - It's Done!!" false false "(R=0.000000,G=1.000000,B=0.017305,A=1.000000)" 15.0)
    ;[1 1 1] Apply Glitch FX on Pawn
    (bind _delegatehandlecontroller_out (Cosmetics|WhenCosmeticPartApplied Pawn (delegate Custom|OnCosmeticPart) true))
    (Variables|Default|SetDelegate%20Handle%20Controller _delegatehandlecontroller_out)
    (else
      (Development|PrintString 0 "Is Character Customization Loading Completed - Not Done Yet!!" true false "(R=1.000000,G=0.000000,B=0.000000,A=1.000000)" 15.0))))

(event Custom|OnCosmeticPart
  (params (Pawn :type Fort%20Player%20Pawn%20Object%20Reference) (Part :type Custom%20Character%20Part%20Object%20Reference) (MeshPart :type Skeletal%20Mesh%20Component%20Object%20Reference) (PartType :type EFortCustomPartType%20Enum :default "Head"))
  (bind _heavyglitchcontainer (Variables|Default|GetHeavyGlitchContainer))
  ;[1 1 1] Apply Glitch FX on Pawn
  (Utilities|Set|Add _heavyglitchcontainer MeshPart)
  (Utilities|Set|AddItems _heavyglitchcontainer)
  (Utilities|IsValid (Variables|/Script/FortniteGame.FortPawn|GetCurrentWeapon (stale Custom|OnFinishedCustomization Pawn))
    (:Is%20Valid
      (Utilities|Set|AddItems _heavyglitchcontainer (Visual|GetWeaponMeshes (Variables|/Script/FortniteGame.FortPawn|GetCurrentWeapon (stale Custom|OnFinishedCustomization Pawn)))))
    (:Is%20Not%20Valid
      _))
  (bind _result (Utilities|Set|ToArray _heavyglitchcontainer))
  (Variables|Default|SetHeavyGlitchSKmeshes _result)
  (for _array_element (Variables|Default|GetHeavyGlitchSKmeshes)
    (Rendering|SetRenderCustomDepth _array_element true)))

(fn ConstructionScript)
