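; Looping gameplay cue for the NightNight event that plays a party bus pass-by for the locally controlled target player.
; It attaches itself to the player, plays a delayed horn, then reveals the bus and drives it along a spline with timeline-controlled tumbling and scale.
; A gated tick aims an arrow at a zero-point location while a spotlight fades in.
; Instances on other players disable tick and destroy their spotlight.
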(event EventOnLoopingStart(Cascade)
  (params (MyTarget :type Actor%20Object%20Reference) (Parameters :type Gameplay%20Cue%20Parameters%20Structure%20%28by%20ref%29) (ParticleComponents :type Array%20of%20Cascade%20Particle%20System%20Component%20Object%20References) (AudioComponents :type Array%20of%20Audio%20Component%20Object%20References))
  (bind _asfort_player_pawn (Utilities|Casting|CastToFortPlayerPawn MyTarget))
  (Variables|Default|SetTargetPlayer _asfort_player_pawn)
  (if (Pawn|IsLocallyControlled _asfort_player_pawn)
    (Utilities|FlowControl|Sequence
      (:then_0
        (Transformation|AttachActorToActor MyTarget "None" "SnapToTarget" "SnapToTarget" "SnapToTarget")
        (Utilities|FlowControl|Sequence
          (:then_0
            (Utilities|FlowControl|Delay 0 3.483)
            (Audio|PlaySound2D 0 "/FNGameplayCues/Athena/Sounds/Events/NightNight/NightNight_Bus_PassBy_Cue.NightNight_Bus_PassBy_Cue" 2.0) ; Bus horn
          )
          (:then_1
            (Utilities|FlowControl|Delay 0 5.0)
            (Rendering|SetVisibility (Variables|Default|GetPartyBus_RT_Anim) true false)
            (bind (_direction _path _scale) (Timeline|Timeline_0 PlayFromStart)
              (:Update
                (Transformation|SetRelativeTransform (Variables|Default|GetPartyBus_RT_Anim) (Math|Transform|MakeTransform (Spline|GetLocationatTime (Variables|Default|GetSpline) _path "Local" false) (Math|Rotator|CombineRotators (Math|Conversions|RotationFromXVector (Spline|GetDirectionatTime (Variables|Default|GetSpline) _path "Local" false)) (Math|Rotator|Lerp(Rotator) "0.000000,80.000000,150.000000" "0.000000,120.000000,-600.000000" _path)) (Math|Conversions|ToVector(Float) _scale)) false false))))))
      (:then_1
        (Utilities|FlowControl|Gate true
          (:Exit
            (Transformation|SetWorldRotation (Variables|Default|GetArrow) (Math|Rotator|FindLookatRotation (Transformation|GetWorldLocation (Variables|Default|GetArrow)) (Variables|Default|GetZeroPointLocation)) false false))))
      (:then_2
        (bind (_direction_2 _newtrack_0) (Timeline|FadeIn PlayFromStart)
          (:Update
            (Rendering|Components|Light|SetIntensity (Variables|Default|GetSpotLight) (* _newtrack_0 1.5e+02))))))
    (else
      ;[1 1 1] Get rid of light for instances of GC on other players
      (Actor|Tick|SetActorTickEnabled false)
      (Components|DestroyComponent (Variables|Default|GetSpotLight) 0))))

(event EventTick
  (params (DeltaSeconds :type Float%20%28single-precision%29))
  (Utilities|FlowControl|Gate true
    (:Exit
      (Transformation|SetWorldRotation (Variables|Default|GetArrow) (Math|Rotator|FindLookatRotation (Transformation|GetWorldLocation (Variables|Default|GetArrow)) (Variables|Default|GetZeroPointLocation)) false false))))

(fn ConstructionScript)
