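; Looping gameplay cue notify that plays audio while an Ostrich is initiating self-destruct.
; On application it spawns the initiating sound attached to the target's root component, set to stop if that is destroyed, stores the audio component and plays it.
; On removal it checks the stored component, quickly fades it out linearly and clears the reference.
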(event EventOnApplication(Cascade)
  (params (MyTarget :type Actor%20Object%20Reference) (Parameters :type Gameplay%20Cue%20Parameters%20Structure%20%28by%20ref%29) (ParticleComponents :type Array%20of%20Cascade%20Particle%20System%20Component%20Object%20References) (AudioComponents :type Array%20of%20Audio%20Component%20Object%20References) (BurstCameraShakeInstance :type Camera%20Shake%20Base%20Object%20Reference) (BurstDecalInstance :type Decal%20Actor%20Object%20Reference))
  (bind _returnvalue (Audio|SpawnSoundAttached :Sound (Variables|Default|GetWhileInitiatingSoundCue) :AttachToComponent (Variables|/Script/Engine.Actor|GetRootComponent MyTarget) :AttachPointName "None" :LocationType "KeepRelativeOffset" :bStopWhenAttachedToDestroyed true :VolumeMultiplier 1.0 :PitchMultiplier 1.0 :StartTime 0.0 :bAutoDestroy true))
  (bind _output_get (Variables|Default|SetLoopingAudioComponent _returnvalue))
  (Audio|Components|Audio|Play _output_get 0.0))

(event EventOnRemoval(Cascade)
  (params (MyTarget :type Actor%20Object%20Reference) (Parameters :type Gameplay%20Cue%20Parameters%20Structure%20%28by%20ref%29) (ParticleComponents :type Array%20of%20Cascade%20Particle%20System%20Component%20Object%20References) (AudioComponents :type Array%20of%20Audio%20Component%20Object%20References) (BurstCameraShakeInstance :type Camera%20Shake%20Base%20Object%20Reference) (BurstDecalInstance :type Decal%20Actor%20Object%20Reference))
  (bind _loopingaudiocomponent (Variables|Default|GetLoopingAudioComponent)
    (:then
      (Audio|Components|Audio|FadeOut _loopingaudiocomponent 0.1 0.0 "Linear" :!class /Script/Engine.AudioComponent)
      (Variables|Default|SetLoopingAudioComponent 0))
    (:else)))

(fn ConstructionScript)
