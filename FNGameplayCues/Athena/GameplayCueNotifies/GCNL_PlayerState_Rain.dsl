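; Looping gameplay cue that drives rain audio on a player pawn.
; Start and removal fade the global rain control bus mix and start or fade out an optional attached MetaSound for the locally viewed pawn.
; A repeating timer fires async downward traces, mapping pawn-to-hit-actor origin distance (not impact point) to an interpolated "Amb.Height" parameter, lower when farther, on the ambient audio component and MetaSound.
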(event EventOnLoopingStart(EitherVFX-preferthisversion)
  (params (MyTarget :type Actor%20Object%20Reference) (Parameters :type Gameplay%20Cue%20Parameters%20Structure%20%28by%20ref%29) (ParticleComponents :type Array%20of%20FXSystem%20Component%20Object%20References) (AudioComponents :type Array%20of%20Audio%20Component%20Object%20References))
  (Development|PrintString MyTarget "Start Rain on Player" true true "(R=0.000000,G=0.256970,B=1.000000,A=1.000000)")
  (Audio|Modulation|SetGlobalControlBusMixValue MyTarget "/FNGameplayCues/Sounds/Modulation/Parameters/CB_RainAmount.CB_RainAmount" 1.0 2.0)
  (CallFunction|AddAudioParameter MyTarget)
  (bind _asfort_player_pawn (Utilities|Casting|CastToFortPlayerPawn MyTarget))
  (Variables|Default|SetPlayer%20Pawn _asfort_player_pawn)
  (CallFunction|EnableRainMetaSound true)
  ;[1 1 1] Start trace to know how far player is from terrain or building below them.
  (bind _returnvalue (Utilities|Time|SetTimerbyEvent (EventDispatchers|CreateEvent :!event StartTrace) 0.5 true))
  (Variables|Default|SetHandle%20Trace _returnvalue))

(event EventOnRemoval(EitherVFX-preferthisversion)
  (params (MyTarget :type Actor%20Object%20Reference) (Parameters :type Gameplay%20Cue%20Parameters%20Structure%20%28by%20ref%29) (ParticleComponents :type Array%20of%20FXSystem%20Component%20Object%20References) (AudioComponents :type Array%20of%20Audio%20Component%20Object%20References) (BurstCameraShakeInstance :type Camera%20Shake%20Base%20Object%20Reference) (BurstDecalInstance :type Decal%20Actor%20Object%20Reference))
  (Development|PrintString MyTarget "Stop Rain On Player" true true "(R=1.000000,G=0.272892,B=0.000000,A=1.000000)")
  (Audio|Modulation|SetGlobalControlBusMixValue MyTarget "/FNGameplayCues/Sounds/Modulation/Parameters/CB_RainAmount.CB_RainAmount" 0.0 2.0)
  (bind _player_pawn (Variables|Default|GetPlayer%20Pawn)
    (:then
      (CallFunction|EnableRainMetaSound false)
      (Utilities|Time|ClearandInvalidateTimerbyHandle _player_pawn (Variables|Default|GetHandle%20Trace)))
    (:else)))

(fn AddAudioParameter
  (params (WorldContext :type Object%20Reference))
  (bind _returnvalue (WorldSubsystems|GetAmbientAudioSubsystem WorldContext))
  (Variables|Default|SetRainDistanceParam 0.0)
  (Utilities|IsValid _returnvalue
    (:Is%20Valid
      (bind _returnvalue_2 (AmbientAudio|GetAudioParameterComponent _returnvalue))
      (bind _output_get (Variables|Default|SetAudioParamComponent _returnvalue_2))
      (Audio|Parameter|SetFloatParameter _output_get "Amb.Height" (Variables|Default|GetRainDistanceParam) :!class /Script/AudioExtensions.AudioParameterControllerInterface))))

(fn EnableRainMetaSound
  (params (bPlay :type Boolean))
  (bind _player_pawn (Variables|Default|GetPlayer%20Pawn)
    (:then
      (if (and (Variables|Default|GetbUseMetaSound) (Pawn|IsLocallyViewed _player_pawn))
        (if bPlay
          (bind _mssrainaudiocomponent (Variables|Default|GetMSSRainAudioComponent)
            (:then
              (Audio|Components|Audio|FadeOut _mssrainaudiocomponent 0.3 0.0 "Linear" :!class /Script/Engine.AudioComponent))
            (:else
              _))
          (bind _returnvalue (Audio|SpawnSoundAttached :Sound (Variables|Default|GetRainMetaSound) :AttachToComponent (Variables|/Script/Engine.Character|GetMesh _player_pawn) :AttachPointName "None" :LocationType "KeepRelativeOffset" :bStopWhenAttachedToDestroyed true :VolumeMultiplier 1.0 :PitchMultiplier 1.0 :StartTime 0.0 :bAutoDestroy true))
          (Variables|Default|SetMSSRainAudioComponent _returnvalue)
          (else
            (bind _mssrainaudiocomponent_2 (Variables|Default|GetMSSRainAudioComponent)
              (:then
                (Audio|Components|Audio|FadeOut _mssrainaudiocomponent_2 (Variables|Default|GetRainMSSFadeOut) 0.0 "Linear" :!class /Script/Engine.AudioComponent))
              (:else))))))
    (:else)))

(fn StartTrace
  ;[1 1 1] We start trace to help give realism to reduce rain parameters as player enters air space vs near ground.
  (bind _player_pawn (Variables|Default|GetPlayer%20Pawn)
    (:then
      (Class|AsyncTraceQueue|AsyncTraceImmediate (Variables|Default|GetPlayer%20Pawn) (Transformation|GetActorLocation _player_pawn) (- (Transformation|GetActorLocation _player_pawn) "0.000000,0.000000,10000.000000") (Utilities|Array|MakeArray _player_pawn) "ECC_WorldStatic" (EventDispatchers|CreateEvent :!event UpdateTraceParam)))
    (:else)))

(fn UpdateTraceParam
  (params (Succeeded :type Boolean) (ActorHit :type Actor%20Object%20Reference) (HitResult :type Hit%20Result%20Structure))
  (Utilities|IsValid ActorHit
    (:Is%20Valid
      (bind _player_pawn (Variables|Default|GetPlayer%20Pawn)
        (:then
          (bind _output_get (Variables|Default|SetRainDistanceParam (Math|Interpolation|FInterpTo (Variables|Default|GetRainDistanceParam) (Math|Float|MapRangeClamped (Math|Vector|Distance(Vector) (Transformation|GetActorLocation ActorHit) (Transformation|GetActorLocation _player_pawn)) 1e+03 1e+04 1.0 0.4) (Utilities|Time|GetWorldDeltaSeconds (Variables|Default|GetPlayer%20Pawn)) 5.0)))
          (bind _audioparamcomponent (Variables|Default|GetAudioParamComponent)
            (:then
              (Audio|Parameter|SetFloatParameter _audioparamcomponent "Amb.Height" _output_get :!class /Script/AudioExtensions.AudioParameterControllerInterface)
              (if (Variables|Default|GetbUseMetaSound)
                (bind _mssrainaudiocomponent (Variables|Default|GetMSSRainAudioComponent)
                  (:then
                    (Audio|Parameter|SetFloatParameter _mssrainaudiocomponent "Amb.Height" (Variables|Default|GetRainDistanceParam) :!class /Script/Engine.AudioComponent))
                  (:else))))
            (:else)))
        (:else)))))

(fn ConstructionScript)
