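; Base looping gameplay cue that draws a colored outline around a player pawn using planes attached to its skeletal mesh.
; On start it spawns planes at mapped bones with absolute rotation, passing the scaled GetColor result through custom primitive data.
; Local control picks glow or no-glow material and can hide the effect.
; A timer re-runs UpdateColor as a retry; each update and cue removal destroys existing planes.
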(event EventOnLoopingStart(EitherVFX-preferthisversion)
  (params (MyTarget :type Actor%20Object%20Reference) (Parameters :type Gameplay%20Cue%20Parameters%20Structure%20%28by%20ref%29) (ParticleComponents :type Array%20of%20FXSystem%20Component%20Object%20References) (AudioComponents :type Array%20of%20Audio%20Component%20Object%20References))
  ;[0.3564 1 0.2417] Attempt to spawn a plane attached to the target player pawn's skeletal mesh that draws an outline around the pawn.
  (|CallParentFunction MyTarget Parameters ParticleComponents AudioComponents)
  ;[1 0 0] This may just work now
  (bind _asplayer_pawn_athena (Utilities|Casting|CastToPlayerPawn_Athena MyTarget)
    (:then
      (Variables|Default|SetTarget%20Pawn _asplayer_pawn_athena)
      (Default|UpdateColor))
    (:CastFailed
      ;;[1 0.4635 0.4635] Log Warning
      (Fortnite|FortShippingLog self (Utilities|String|Append "CGNL_Player_ColoredOutline_Base::EventOnLoopingStart() target: [" (Fortnite|GetPlayerorObjectNameSafe MyTarget) "] could NOT be cast to PlayerPawn_Athena.") true))))

(event Custom|RetryUpdateColor
  ;[0.1205 1 0.7266] Retry updating the color because the last attempt failed.
  (Default|UpdateColor))

(event EventOnRemoval(EitherVFX-preferthisversion)
  (params (MyTarget :type Actor%20Object%20Reference) (Parameters :type Gameplay%20Cue%20Parameters%20Structure%20%28by%20ref%29) (ParticleComponents :type Array%20of%20FXSystem%20Component%20Object%20References) (AudioComponents :type Array%20of%20Audio%20Component%20Object%20References) (BurstCameraShakeInstance :type Camera%20Shake%20Base%20Object%20Reference) (BurstDecalInstance :type Decal%20Actor%20Object%20Reference))
  ;[1 0.5193 0.1054] Remove the planes when this GC is removed.
  (|CallParentFunction MyTarget Parameters ParticleComponents AudioComponents BurstCameraShakeInstance BurstDecalInstance)
  (CallFunction|DestroyPlanesonPawn))

(fn AttachPlanesToPawn
  (bind _currentplane (LocalVariables|GetCurrentPlane))
  (bind _target_pawn (Variables|Default|GetTarget%20Pawn)
    (:then
      (Utilities|IsValid (Variables|/Script/Engine.Character|GetMesh _target_pawn)
        (:Is%20Valid
          (Variables|Default|SetPawnSkelMesh (Variables|/Script/Engine.Character|GetMesh _target_pawn))
          (bind _keys (Utilities|Map|Keys (Variables|ColoredOutline|GetBonesAndRelativeTransforms)))
          (for _array_element _keys
            (Utilities|FlowControl|Sequence
              (:then_0
                ;[0.1312 1 0.4542] Create the plane mesh and set up its attachments.
                (if (Transformation|DoesSocketExist (Variables|Default|GetPawnSkelMesh) _array_element)
                  (bind _returnvalue (AddComponent|NODE_AddStaticMeshComponent-0))
                  (bind _output_get (LocalVariables|SetCurrentPlane _returnvalue))
                  (Transformation|AttachComponentToComponent _output_get (Variables|Default|GetPawnSkelMesh) _array_element "KeepRelative" "KeepRelative" "KeepRelative" true)
                  (Transformation|SetRelativeTransform _output_get (Utilities|Map|Find (Variables|ColoredOutline|GetBonesAndRelativeTransforms) _array_element) false true)
                  (Variables|/Script/Engine.SceneComponent|SetbAbsoluteRotation true _output_get)
                  (Transformation|SetWorldRotation _output_get "0, 0,90.000000" false false)
                  (else
                    ;;[1 0.9163 0.04337] Log a warning that the socket doesn't exist.
                    (Fortnite|FortShippingLog self (Utilities|String|Append "GCNL_Player_ColoredOutline_Base::AttachPlanesToPawn() partially failed for player: [" (Fortnite|GetPlayerNameSafe (Variables|Default|GetTarget%20Pawn)) "] , does not have a socket named: [" (Utilities|String|ToString(Name) _array_element) "].") true))))
              (:then_1
                ;[0.2314 0.2677 1] Set the material and color on the newly created plane mesh.
                (Rendering|Material|SetMaterial _currentplane 0 (CallFunction|GetOutlineMaterial))
                (Utilities|Array|AddUnique (Variables|Default|GetGlowPlanes) _currentplane)
                (bind _outlinecolor (Default|GetColor))
                (bind (_returnvalue_r _returnvalue_g _returnvalue_b _returnvalue_a) (Math|Color|BreakColor (Math|Color|LinearColor*Float _outlinecolor (Variables|ColoredOutline|GetOutlineColorScale))))
                (Rendering|Material|SetCustomPrimitiveDataVector4 _currentplane 0 (Math|Vector4|MakeVector4 _returnvalue_r _returnvalue_g _returnvalue_b _returnvalue_a))))))
        (:Is%20Not%20Valid
          ;[1 0.4688 0.4688] Log warnings that the player pawn or its mesh are not valid and set a timer to retry applying the colored outline.
          (Fortnite|FortShippingLog self (Utilities|String|Append "GCNL_TwoTeam_AllyVsEnemyColor_Outline::AttachPlanesToPawn() failed for player: [" (Fortnite|GetPlayerNameSafe (Variables|Default|GetTarget%20Pawn)) "]  Target Pawn does NOT have a Mesh Component.") true))))
    (:else
      (Fortnite|FortShippingLog self "GCNL_TwoTeam_AllyVsEnemyColor_Outline::AttachPlanesToPawn() failed, Target Pawn is nullptr." true)))
  (Utilities|Time|SetTimerbyEvent (EventDispatchers|CreateEvent :!event RetryUpdateColor) (Variables|ColoredOutline|GetRetryUpdateColorTime))
  (return))

(fn DestroyPlanesOnPawn
  (for _array_element (Variables|Default|GetGlowPlanes)
    (Utilities|IsValid _array_element
      (:Is%20Valid
        (Components|DestroyComponent _array_element 0))))
  (Utilities|Array|Clear (Variables|Default|GetGlowPlanes)))

(fn GetColor
  (return (Variables|ColoredOutline|GetDefaultColor)))

(fn GetOutlineMaterial
  (bind _target_pawn (Variables|Default|GetTarget%20Pawn)
    (:then
      (return (select (Pawn|IsLocallyControlled _target_pawn) (Variables|ColoredOutline|GetOutlineMaterialNoGlow) (Variables|ColoredOutline|GetOutlineMaterialGlow))))
    (:else
      (return (Variables|ColoredOutline|GetOutlineMaterialNoGlow)))))

(fn ShouldShowEffectOnPawn
  (bind _target_pawn (Variables|Default|GetTarget%20Pawn)
    (:then
      (return (or (Variables|ColoredOutline|GetbShowOutlineOnLocalPawn) (not (Pawn|IsLocallyControlled _target_pawn)))))
    (:else)))

(fn UpdateColor
  (CallFunction|DestroyPlanesonPawn)
  (bind _show_effect (CallFunction|ShouldShowEffectonPawn))
  (if _show_effect
    (CallFunction|AttachPlanestoPawn)))

(fn ConstructionScript
  (|CallParentFunction))
