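; Looping gameplay cue for the creative ranged flashlight that switches its light, bulb emissive and sounds on or off.
; Each recurring event reads the toggle from normalized magnitude and the weapon from the instigator.
; It attaches to the weapon, choosing a decal on mobile or a spotlight elsewhere, plus a beam mesh.
; A quick retrigger that ends off instead pulses the light on until a timer switches it off.
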(event EventOnRecurring(Cascade)
  (params (MyTarget :type Actor%20Object%20Reference) (Parameters :type Gameplay%20Cue%20Parameters%20Structure%20%28by%20ref%29) (ParticleComponents :type Array%20of%20Cascade%20Particle%20System%20Component%20Object%20References) (AudioComponents :type Array%20of%20Audio%20Component%20Object%20References) (BurstCameraShakeInstance :type Camera%20Shake%20Base%20Object%20Reference) (BurstDecalInstance :type Decal%20Actor%20Object%20Reference))
  (Utilities|Time|ClearandInvalidateTimerbyHandle 0 (Variables|State|GetPulseFlashTimerHandle))
  (CallFunction|GetToggledandWeaponfromParams Parameters)
  (CallFunction|ToggleFlashLight)
  (CallFunction|ChecktoPulseFlash (delegate Custom|PulseFlash))
  (Variables|State|SetLastActivationTime (Variables|State|GetActivationTime)))

(event Custom|PulseFlash
  (Variables|State|SetbToggled false)
  (CallFunction|ToggleFlashLight)
  (Utilities|Time|ClearandInvalidateTimerbyHandle 0 (Variables|State|GetPulseFlashTimerHandle)))

(event Custom|ToggleFlashLight
  (bind _flashlight (CallFunction|SetLightVisuals))
  (CallFunction|SetWeaponMeshEmissive _flashlight)
  (CallFunction|PlaySounds))

(fn CheckToPulseFlash
  (params (Delegate :type Delegate))
  (if (and (CallFunction|ToggleFlash) (not (Variables|State|GetbToggled)))
    (Variables|State|SetbToggled true)
    (bind _returnvalue (Utilities|Time|SetTimerbyEvent Delegate (LocalVariables|GetPulseTime)))
    (Variables|State|SetPulseFlashTimerHandle _returnvalue)
    (CallFunction|ToggleFlashLight)))

(fn GetToggledAndWeaponFromParams
  (params (Params :type Gameplay%20Cue%20Parameters%20Structure))
  (bind (_normalizedmagnitude _rawmagnitude _effectcontext _matchedtagname _originaltag _aggregatedsourcetags _aggregatedtargettags _location _normal _instigator) (Ability|GameplayCue|BreakGameplayCueParameters Params))
  (Variables|State|SetActivationTime (Utilities|Time|GetTimeSeconds 0))
  (Variables|State|SetbToggled (Math|Float|NearlyEqual(Float) _normalizedmagnitude 1.0 0.00015))
  (Variables|Weapon|SetFortWeapon (Utilities|Casting|CastToFortWeapon _instigator))
  (return))

(fn PlaySounds
  (bind _switchaudio (Variables|Components|GetSwitchAudio))
  (bind _flashlight_loop_cue (Variables|Components|GetFlashlight_Loop_Cue))
  (Audio|Components|Audio|SetSound _switchaudio (select (Variables|State|GetbToggled) "/Game/Sounds/Creative/Gadgets/Flashlight/Flashlight_Start_Cue.Flashlight_Start_Cue" "/Game/Sounds/Creative/Gadgets/Flashlight/Flashlight_Stop_Cue.Flashlight_Stop_Cue"))
  (Audio|Components|Audio|Play _switchaudio 0.0)
  (if (Variables|State|GetbToggled)
    (Audio|Components|Audio|FadeIn _flashlight_loop_cue 0.5 1.0 0.0 "Linear" :!class /Script/Engine.AudioComponent)
    (else
      (Audio|Components|Audio|FadeOut _flashlight_loop_cue 0.5 0.0 "Linear" :!class /Script/Engine.AudioComponent))))

(fn SetLightVisibilitiesBasedOnPlatform
  (params (FlashlightActor :type B%20Ranged%20Flashlight%20Object%20Reference) (Visibility :type Boolean) (IsMobilePlatform :type Boolean))
  (Rendering|SetVisibility (Variables|/Game/Athena/Items/Weapons/Creative/Ranged/B_Ranged_Flashlight.B_Ranged_Flashlight_C|GetMobileLightDecal (Variables|GetFlashlightActor)) (and (Variables|GetVisibility) (Variables|GetIsMobilePlatform)) false)
  (Rendering|SetVisibility (Variables|/Game/Athena/Items/Weapons/Creative/Ranged/B_Ranged_Flashlight.B_Ranged_Flashlight_C|GetSpotLight (Variables|GetFlashlightActor)) (and (Variables|GetVisibility) (not (Variables|GetIsMobilePlatform))) false)
  (Rendering|SetVisibility (Variables|/Game/Athena/Items/Weapons/Creative/Ranged/B_Ranged_Flashlight.B_Ranged_Flashlight_C|GetSm_CreativeMoveTool_Lightbeam (Variables|GetFlashlightActor)) (Variables|GetVisibility) false))

(fn SetLightVisuals
  (Transformation|AttachActorToActor (Variables|Weapon|GetFortWeapon) "None" "SnapToTarget" "SnapToTarget" "KeepWorld" false)
  (bind _asb_ranged_flashlight (Utilities|Casting|CastToB_Ranged_Flashlight (Variables|Weapon|GetFortWeapon))
    (:then
      (LocalVariables|SetLocalFlashlight _asb_ranged_flashlight)
      (CallFunction|SetLightVisibilitiesBasedonPlatform (LocalVariables|GetLocalFlashlight) (Variables|State|GetbToggled) (Fortnite|IsMobilePlatform))
      (Class|BRangedFlashlight|InitLightShadowQuality (LocalVariables|GetLocalFlashlight))
      (return (LocalVariables|GetLocalFlashlight)))
    (:CastFailed
      (return 0))))

(fn SetWeaponMeshEmissive
  (params (Flashlight :type Fort%20Weapon%20Object%20Reference))
  (Utilities|IsValid Flashlight
    (:Is%20Valid
      (Rendering|Material|SetVectorParameterValueonMaterials (Variables|/Script/FortniteGame.FortWeapon|GetWeaponMesh Flashlight) "EmissiveMultiplier" (Math|Conversions|ToVector(LinearColor) (select (Variables|State|GetbToggled) (Variables|Material|GetBulbMaterialEmissiveColorActive) 0))))))

(fn ToggleFlash
  (return (< (- (Variables|State|GetActivationTime) (Variables|State|GetLastActivationTime)) (Variables|State|GetLowRetriggerTime))))

(fn ConstructionScript)
