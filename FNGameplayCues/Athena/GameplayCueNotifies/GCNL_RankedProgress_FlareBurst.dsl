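; Looping gameplay cue notify that tunes a ranked-progress flare burst effect to the target player's rank progression.
; On looping start it casts the target to a Fort player pawn and reads its HabaneroProgression cosmetic stat.
; That stat indexes RankedCurveSettings to set Ranked_Curve on the first particle component, and a zero stat sets that component's Unranked flag.
; A DEBUG flag prints the stat value.
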(event EventOnLoopingStart(EitherVFX-preferthisversion)
  (params (MyTarget :type Actor%20Object%20Reference) (Parameters :type Gameplay%20Cue%20Parameters%20Structure%20%28by%20ref%29) (ParticleComponents :type Array%20of%20FXSystem%20Component%20Object%20References) (AudioComponents :type Array%20of%20Audio%20Component%20Object%20References))
  (bind _output (Utilities|Array|Get(acopy) ParticleComponents))
  (bind _asfort_player_pawn (Utilities|Casting|CastToFortPlayerPawn MyTarget))
  (bind _returnvalue (Class|FortCosmeticFunctionLibrary|GetCosmeticStatValue _asfort_player_pawn "(TagName=\"StatManager.HabaneroProgression\")"))
  (Effects|Components|ParticleSystem|SetFloatParameter _output "Ranked_Curve" (Utilities|Array|Get(acopy) (Variables|Default|GetRankedCurveSettings) _returnvalue))
  (Effects|Components|ParticleSystem|SetBoolParameter _output "Unranked" (== _returnvalue 0))
  (if (Variables|Default|GetDEBUG)
    (Development|PrintString 0 (Utilities|String|ToString(Integer) _returnvalue))))

(fn ConstructionScript)
