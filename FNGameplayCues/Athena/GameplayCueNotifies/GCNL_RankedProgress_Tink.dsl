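; Looping gameplay cue notify that drives a ranked-progression particle effect on a Fortnite player pawn.
; On loop start it reads the pawn's HabaneroProgression cosmetic stat and uses it as an index into the RankedCurveSettings array.
; The looked-up value sets the Ranked_Curve float parameter on the first particle component.
; It always prints the target's name, and prints the stat value when DEBUG is set.
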(event EventOnLoopingStart(EitherVFX-preferthisversion)
  (params (MyTarget :type Actor%20Object%20Reference) (Parameters :type Gameplay%20Cue%20Parameters%20Structure%20%28by%20ref%29) (ParticleComponents :type Array%20of%20FXSystem%20Component%20Object%20References) (AudioComponents :type Array%20of%20Audio%20Component%20Object%20References))
  (Development|PrintString 0 (Utilities|GetDisplayName MyTarget))
  (bind _asfort_player_pawn (Utilities|Casting|CastToFortPlayerPawn MyTarget))
  (bind _returnvalue (Class|FortCosmeticFunctionLibrary|GetCosmeticStatValue _asfort_player_pawn "(TagName=\"StatManager.HabaneroProgression\")"))
  (Effects|Components|ParticleSystem|SetFloatParameter (Utilities|Array|Get(acopy) ParticleComponents) "Ranked_Curve" (Utilities|Array|Get(acopy) (Variables|Default|GetRankedCurveSettings) _returnvalue))
  (if (Variables|Default|GetDEBUG)
    (Development|PrintString 0 (Utilities|String|ToString(Integer) _returnvalue))))

(fn ConstructionScript)
