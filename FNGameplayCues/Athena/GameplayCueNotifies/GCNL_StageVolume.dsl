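; Looping gameplay cue for a stage volume that adjusts the audio mix for the locally controlled Fort pawn while the cue is active.
; On application it pushes two in-volume sound mixes and overrides the live in-world Papaya music class within both emote music mixes (3P and standard).
; On removal it pops the two mixes and re-applies both overrides at zero volume rather than clearing them.
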(event EventOnApplication(Cascade)
  (params (MyTarget :type Actor%20Object%20Reference) (Parameters :type Gameplay%20Cue%20Parameters%20Structure%20%28by%20ref%29) (ParticleComponents :type Array%20of%20Cascade%20Particle%20System%20Component%20Object%20References) (AudioComponents :type Array%20of%20Audio%20Component%20Object%20References) (BurstCameraShakeInstance :type Camera%20Shake%20Base%20Object%20Reference) (BurstDecalInstance :type Decal%20Actor%20Object%20Reference))
  (bind _asfort_pawn (Utilities|Casting|CastToFortPawn MyTarget))
  (if (Pawn|IsLocallyControlled _asfort_pawn)
    (Audio|PushSoundMixModifier 0 (Variables|Default|GetInVolumeSoundMix1))
    (Audio|PushSoundMixModifier 0 (Variables|Default|GetInVolumeSoundMix2))
    (Audio|SetSoundMixClassOverride 0 "/Game/Sounds/FortSoundClassesAndModes/Fort_Music_Emote_3P_MixMod.Fort_Music_Emote_3P_MixMod" "/Game/Sounds/FortSoundClassesAndModes/Fort_Live_InWorld_Papaya_Music.Fort_Live_InWorld_Papaya_Music")
    (Audio|SetSoundMixClassOverride 0 "/Game/Sounds/FortSoundClassesAndModes/Fort_Music_Emote_MixMod.Fort_Music_Emote_MixMod" "/Game/Sounds/FortSoundClassesAndModes/Fort_Live_InWorld_Papaya_Music.Fort_Live_InWorld_Papaya_Music")))

(event EventOnRemoval(Cascade)
  (params (MyTarget :type Actor%20Object%20Reference) (Parameters :type Gameplay%20Cue%20Parameters%20Structure%20%28by%20ref%29) (ParticleComponents :type Array%20of%20Cascade%20Particle%20System%20Component%20Object%20References) (AudioComponents :type Array%20of%20Audio%20Component%20Object%20References) (BurstCameraShakeInstance :type Camera%20Shake%20Base%20Object%20Reference) (BurstDecalInstance :type Decal%20Actor%20Object%20Reference))
  (bind _asfort_pawn (Utilities|Casting|CastToFortPawn MyTarget))
  (if (Pawn|IsLocallyControlled _asfort_pawn)
    (Audio|PopSoundMixModifier 0 (Variables|Default|GetInVolumeSoundMix1))
    (Audio|PopSoundMixModifier 0 (Variables|Default|GetInVolumeSoundMix2))
    (Audio|SetSoundMixClassOverride 0 "/Game/Sounds/FortSoundClassesAndModes/Fort_Music_Emote_3P_MixMod.Fort_Music_Emote_3P_MixMod" "/Game/Sounds/FortSoundClassesAndModes/Fort_Live_InWorld_Papaya_Music.Fort_Live_InWorld_Papaya_Music" 0.0)
    (Audio|SetSoundMixClassOverride 0 "/Game/Sounds/FortSoundClassesAndModes/Fort_Music_Emote_MixMod.Fort_Music_Emote_MixMod" "/Game/Sounds/FortSoundClassesAndModes/Fort_Live_InWorld_Papaya_Music.Fort_Live_InWorld_Papaya_Music" 0.0)))

(fn ConstructionScript)
