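; Looping gameplay cue for the green glop variant of the dance-stunned status, showing a particle effect on the affected player.
; On looping start it casts the target to PlayerPawn_Athena and, only if that pawn is locally controlled, attaches the GreenGlop player emitter to the mesh root.
; The component is stored in the "Active Confetti" variable and deactivated on removal if valid, with auto-destroy set at spawn.
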(event EventOnLoopingStart(Cascade)
  (params (MyTarget :type Actor%20Object%20Reference) (Parameters :type Gameplay%20Cue%20Parameters%20Structure%20%28by%20ref%29) (ParticleComponents :type Array%20of%20Cascade%20Particle%20System%20Component%20Object%20References) (AudioComponents :type Array%20of%20Audio%20Component%20Object%20References))
  (bind _asplayer_pawn_athena (Utilities|Casting|CastToPlayerPawn_Athena MyTarget))
  (if (Pawn|IsLocallyControlled _asplayer_pawn_athena)
    (bind _returnvalue (Effects|SpawnEmitterAttached :EmitterTemplate "/FNGameplayCues/Effects/Fort_Effects/Events/GreenGlop/P_GreenGlop_Player.P_GreenGlop_Player" :AttachToComponent (Variables|/Script/Engine.Character|GetMesh _asplayer_pawn_athena) :AttachPointName "root" :Scale "1.000000,1.000000,1.000000" :LocationType "SnapToTarget" :bAutoDestroy true :PoolingMethod "None" :bAutoActivate true))
    (Variables|Default|SetActive%20Confetti _returnvalue)))

(event EventOnRemoval(Cascade)
  (params (MyTarget :type Actor%20Object%20Reference) (Parameters :type Gameplay%20Cue%20Parameters%20Structure%20%28by%20ref%29) (ParticleComponents :type Array%20of%20Cascade%20Particle%20System%20Component%20Object%20References) (AudioComponents :type Array%20of%20Audio%20Component%20Object%20References) (BurstCameraShakeInstance :type Camera%20Shake%20Base%20Object%20Reference) (BurstDecalInstance :type Decal%20Actor%20Object%20Reference))
  (bind _active_confetti (Variables|Default|GetActive%20Confetti))
  (Utilities|IsValid _active_confetti
    (:Is%20Valid
      (Components|Activation|Deactivate _active_confetti))))

(fn ConstructionScript)
