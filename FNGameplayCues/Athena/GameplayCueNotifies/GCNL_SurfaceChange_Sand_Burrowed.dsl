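; Looping gameplay cue notify that plays the effects and audio for a player burrowed in sand.
; On start it attaches speed-line FX to the capsule, offset ahead of the pawn near the ground, and attaches a looping audio component to the pawn.
; The loop sound is chosen for local instigator, ally or enemy.
; On removal it plays a burst, camera shake and force feedback near the capsule base, destroys the speed lines and fades out the audio.
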(event Custom|SetLoopingAudioSound
  ;[1 1 1] Set Looping Audio Sound based on team
  (if (Ability|GameplayCue|IsInstigatorLocallyControlledPlayer (Variables|FortGameplayCueNotifyLoop_SandBurrow|GetCachedGCParameters))
    (Audio|Components|Audio|SetSound (Variables|FortGameplayCueNotifyLoop_SandBurrow|GetLayeredAudioComp) (Variables|Default|GetSound_FirstPerson))
    (elif (Utilities|Enum|Equal(Enum) (Team|GetTeamAffiliationtoLocalPlayer (Variables|FortGameplayCueNotifyLoop_SandBurrow|GetPlayerPawn) (Team|GetActorTeam (Variables|FortGameplayCueNotifyLoop_SandBurrow|GetPlayerPawn))) "Friendly")
      (Audio|Components|Audio|SetSound (Variables|FortGameplayCueNotifyLoop_SandBurrow|GetLayeredAudioComp) (Variables|Default|GetSound_ThirdPerson_Ally))
      (else
        (Audio|Components|Audio|SetSound (Variables|FortGameplayCueNotifyLoop_SandBurrow|GetLayeredAudioComp) (Variables|Default|GetSound_ThirdPerson_Enemy))))))

(event EventOnLoopingStart(EitherVFX-preferthisversion)
  (params (MyTarget :type Actor%20Object%20Reference) (Parameters :type Gameplay%20Cue%20Parameters%20Structure%20%28by%20ref%29) (ParticleComponents :type Array%20of%20FXSystem%20Component%20Object%20References) (AudioComponents :type Array%20of%20Audio%20Component%20Object%20References))
  (bind (_returnvalue_x _returnvalue_y) (Math|Vector|BreakVector (* (Math|Vector|Normalize (Math|Rotator|GetRotationXVector (Transformation|GetActorRotation (Variables|FortGameplayCueNotifyLoop_SandBurrow|GetPlayerPawn)))) (Variables|Default|GetAttachedFX_DistanceInFrontOfPawn))))
  (bind (_asfort_player_pawn _bsuccess) (Utilities|Casting|CastToFortPlayerPawn MyTarget))
  ;[1 1 1] On Looping Start
  (|CallParentFunction MyTarget Parameters ParticleComponents AudioComponents)
  (Variables|FortGameplayCueNotifyLoop_SandBurrow|SetCachedGCParameters Parameters)
  (if _bsuccess
    (Variables|FortGameplayCueNotifyLoop_SandBurrow|SetPlayerPawn _asfort_player_pawn)
    (bind _returnvalue (Effects|SpawnFXSystemAttached (Variables|Default|GetFX_Speed_Lines) (Variables|/Script/Engine.Character|GetCapsuleComponent (Variables|FortGameplayCueNotifyLoop_SandBurrow|GetPlayerPawn)) "None" (Math|Vector|vector+vector (Transformation|GetWorldLocation (Variables|/Script/Engine.Character|GetCapsuleComponent (Variables|FortGameplayCueNotifyLoop_SandBurrow|GetPlayerPawn))) (Math|Vector|MakeVector _returnvalue_x _returnvalue_y (+ (* (Components|Capsule|GetScaledCapsuleHalfHeight (Variables|/Script/Engine.Character|GetCapsuleComponent (Variables|FortGameplayCueNotifyLoop_SandBurrow|GetPlayerPawn))) -1.0) (Variables|Default|GetAttachedFX_HeightOffsetFromGround)))) (Transformation|GetWorldRotation (Variables|/Script/Engine.Character|GetCapsuleComponent (Variables|FortGameplayCueNotifyLoop_SandBurrow|GetPlayerPawn))) "1.000000,1.000000,1.000000" "KeepWorldPosition" true "ManualRelease"))
    (Variables|FortGameplayCueNotifyLoop_SandBurrow|SetFXSpeedLinesComp _returnvalue)
    (CallFunction|SetLoopingAudioSound)
    (Transformation|AttachComponentToComponent (Variables|FortGameplayCueNotifyLoop_SandBurrow|GetLayeredAudioComp) (Variables|/Script/Engine.Actor|GetRootComponent (Variables|FortGameplayCueNotifyLoop_SandBurrow|GetPlayerPawn)) "None" "SnapToTarget" "SnapToTarget" "SnapToTarget" true)))

(event EventOnRemoval(EitherVFX-preferthisversion)
  (params (MyTarget :type Actor%20Object%20Reference) (Parameters :type Gameplay%20Cue%20Parameters%20Structure%20%28by%20ref%29) (ParticleComponents :type Array%20of%20FXSystem%20Component%20Object%20References) (AudioComponents :type Array%20of%20Audio%20Component%20Object%20References) (BurstCameraShakeInstance :type Camera%20Shake%20Base%20Object%20Reference) (BurstDecalInstance :type Decal%20Actor%20Object%20Reference))
  (bind (_asfort_pawn _bsuccess) (Utilities|Casting|CastToFortPawn MyTarget))
  (bind _capsulecomponent (Variables|/Script/Engine.Character|GetCapsuleComponent _asfort_pawn))
  (bind _returnvalue (Transformation|GetWorldLocation _capsulecomponent))
  (bind _returnvalue_2 (Components|Capsule|GetScaledCapsuleHalfHeight _capsulecomponent))
  (bind _removalfx_heightoffsetfromground (Variables|Default|GetRemovalFX_HeightOffsetFromGround))
  (bind _returnvalue_3 (* _returnvalue_2 -1.0))
  (bind _returnvalue_4 (+ _returnvalue_3 _removalfx_heightoffsetfromground))
  (bind _returnvalue_5 (Math|Vector|vector+vector _returnvalue (Math|Vector|MakeVector 0.0 0.0 _returnvalue_4)))
  (bind _layeredaudiocomp (Variables|FortGameplayCueNotifyLoop_SandBurrow|GetLayeredAudioComp))
  (bind _fxspeedlinescomp (Variables|FortGameplayCueNotifyLoop_SandBurrow|GetFXSpeedLinesComp))
  ;[1 1 1] On Removal
  (|CallParentFunction MyTarget Parameters ParticleComponents AudioComponents BurstCameraShakeInstance BurstDecalInstance)
  (if _bsuccess
    (Effects|Components|ParticleSystem|SpawnFXAtLocation _asfort_pawn (Variables|Default|GetFX_BurstRemoval) _returnvalue_5 (Transformation|GetWorldRotation _capsulecomponent) (Variables|Default|GetFX_BurstRemoval_Scale) true "AutoRelease")
    (Camera|PlayWorldCameraShake MyTarget "/Game/Effects/Fort_Effects/BluePrints/CameraShakes/B_Medium_Explosion_CameraShake.B_Medium_Explosion_CameraShake_C" _returnvalue_5 0.0 5e+03)
    (Fortnite|PlayLocalForceFeedbackatLocationMulti MyTarget "/Game/Weapons/ForceFeedback/FF_Generic_Explosion_Small_Near.FF_Generic_Explosion_Small_Near" "/Game/Weapons/ForceFeedback/FF_Generic_Explosion_Small_Far.FF_Generic_Explosion_Small_Far" _returnvalue_5 0.0 3e+03)
    (Components|Activation|Deactivate _fxspeedlinescomp)
    (Utilities|IsValid _fxspeedlinescomp
      (:Is%20Valid
        (Components|DestroyComponent _fxspeedlinescomp 0))
      (:Is%20Not%20Valid
        _))
    (Utilities|IsValid _layeredaudiocomp
      (:Is%20Valid
        (if (Audio|Components|Audio|IsPlaying _layeredaudiocomp)
          (Audio|Components|Audio|FadeOut _layeredaudiocomp 1.0 0.0 "Linear" :!class /Script/Engine.AudioComponent)
          (else
            _))
        (Transformation|DetachFromComponent _layeredaudiocomp "KeepWorld" "KeepWorld" "KeepWorld" true)))))

(fn ConstructionScript
  (|CallParentFunction))
