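; Looping gameplay cue notify that plays effects on a Fort pawn, named for sinking into a sand surface.
; On looping start, it attaches an FX system to the capsule just above its base and stores the handle.
; It also plays an explosion-style world camera shake and local force feedback at that point.
; On removal, it waits a short delay before deactivating the stored FX.
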(event EventOnLoopingStart(EitherVFX-preferthisversion)
  (params (MyTarget :type Actor%20Object%20Reference) (Parameters :type Gameplay%20Cue%20Parameters%20Structure%20%28by%20ref%29) (ParticleComponents :type Array%20of%20FXSystem%20Component%20Object%20References) (AudioComponents :type Array%20of%20Audio%20Component%20Object%20References))
  (bind (_asfort_pawn _bsuccess) (Utilities|Casting|CastToFortPawn MyTarget))
  (bind _capsulecomponent (Variables|/Script/Engine.Character|GetCapsuleComponent _asfort_pawn))
  (bind _returnvalue (Transformation|GetWorldLocation _capsulecomponent))
  (bind _returnvalue_2 (Components|Capsule|GetScaledCapsuleHalfHeight _capsulecomponent))
  (bind _returnvalue_3 (* _returnvalue_2 -1.0))
  (bind _returnvalue_4 (+ _returnvalue_3 8.0))
  (bind _returnvalue_5 (Math|Vector|vector+vector _returnvalue (Math|Vector|MakeVector 0.0 0.0 _returnvalue_4)))
  (|CallParentFunction MyTarget Parameters ParticleComponents AudioComponents)
  (if _bsuccess
    (bind _returnvalue_6 (Effects|SpawnFXSystemAttached (Variables|Default|GetFX) (Variables|/Script/Engine.Character|GetCapsuleComponent _asfort_pawn) "None" _returnvalue_5 (Transformation|GetWorldRotation _capsulecomponent) (Variables|Default|GetFX_Scale) "KeepWorldPosition" true "AutoRelease"))
    (Variables|Default|SetHandle_FX _returnvalue_6)
    (Camera|PlayWorldCameraShake MyTarget "/Game/Effects/Fort_Effects/BluePrints/CameraShakes/B_Medium_Explosion_CameraShake.B_Medium_Explosion_CameraShake_C" _returnvalue_5 0.0 5e+03)
    (Fortnite|PlayLocalForceFeedbackatLocationMulti MyTarget "/Game/Weapons/ForceFeedback/FF_Generic_Explosion_Small_Near.FF_Generic_Explosion_Small_Near" "/Game/Weapons/ForceFeedback/FF_Generic_Explosion_Small_Far.FF_Generic_Explosion_Small_Far" _returnvalue_5 0.0 3e+03)))

(event EventOnRemoval(EitherVFX-preferthisversion)
  (params (MyTarget :type Actor%20Object%20Reference) (Parameters :type Gameplay%20Cue%20Parameters%20Structure%20%28by%20ref%29) (ParticleComponents :type Array%20of%20FXSystem%20Component%20Object%20References) (AudioComponents :type Array%20of%20Audio%20Component%20Object%20References) (BurstCameraShakeInstance :type Camera%20Shake%20Base%20Object%20Reference) (BurstDecalInstance :type Decal%20Actor%20Object%20Reference))
  (|CallParentFunction MyTarget Parameters ParticleComponents AudioComponents BurstCameraShakeInstance BurstDecalInstance)
  (Utilities|FlowControl|Delay 0 1.0)
  (Components|Activation|Deactivate (Variables|Default|GetHandle_FX)))

(fn ConstructionScript)
