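; Looping gameplay cue that attaches a dust FX to a Fort pawn, named for a sand surface-change snare.
; On loop start it spawns the pooled effect on the pawn's capsule with manual release, keeping world position.
; Each tick it nudges the effect by the change in negative capsule half-height plus a ground offset, so it follows capsule height changes.
; On removal it deactivates the effect and releases it to the pool.
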(event EventOnLoopingStart(EitherVFX-preferthisversion)
  (params (MyTarget :type Actor%20Object%20Reference) (Parameters :type Gameplay%20Cue%20Parameters%20Structure%20%28by%20ref%29) (ParticleComponents :type Array%20of%20FXSystem%20Component%20Object%20References) (AudioComponents :type Array%20of%20Audio%20Component%20Object%20References))
  (bind (_asfort_pawn _bsuccess) (Utilities|Casting|CastToFortPawn MyTarget))
  ;[1 1 1] On Looping Start
  (|CallParentFunction MyTarget Parameters ParticleComponents AudioComponents)
  (if _bsuccess
    (Variables|Default|SetPawn _asfort_pawn)
    (bind _returnvalue (Effects|SpawnFXSystemAttached (Variables|Default|GetFX_Dust) (Variables|/Script/Engine.Character|GetCapsuleComponent (Variables|Default|GetPawn)) "None" (Transformation|GetWorldLocation (Variables|/Script/Engine.Character|GetCapsuleComponent (Variables|Default|GetPawn))) (Transformation|GetWorldRotation (Variables|/Script/Engine.Character|GetCapsuleComponent (Variables|Default|GetPawn))) (Variables|Default|GetFX_Scale) "KeepWorldPosition" true "ManualRelease"))
    (Variables|Default|SetHandle_FX_Dust _returnvalue)
    (Variables|Default|SetPreviousHeightOffset 0.0)))

(event EventOnRemoval(EitherVFX-preferthisversion)
  (params (MyTarget :type Actor%20Object%20Reference) (Parameters :type Gameplay%20Cue%20Parameters%20Structure%20%28by%20ref%29) (ParticleComponents :type Array%20of%20FXSystem%20Component%20Object%20References) (AudioComponents :type Array%20of%20Audio%20Component%20Object%20References) (BurstCameraShakeInstance :type Camera%20Shake%20Base%20Object%20Reference) (BurstDecalInstance :type Decal%20Actor%20Object%20Reference))
  ;[1 1 1] On removal
  (|CallParentFunction MyTarget Parameters ParticleComponents AudioComponents BurstCameraShakeInstance BurstDecalInstance)
  (Components|Activation|Deactivate (Variables|Default|GetHandle_FX_Dust))
  (Effects|Components|ParticleSystem|ReleasetoPool (Variables|Default|GetHandle_FX_Dust)))

(event EventTick
  (params (DeltaSeconds :type Float%20%28single-precision%29))
  ;[1 1 1] On Tick
  (|CallParentFunction DeltaSeconds)
  (Utilities|IsValid (Variables|Default|GetHandle_FX_Dust)
    (:Is%20Valid
      (Utilities|IsValid (Variables|Default|GetPawn)
        (:Is%20Valid
          (bind _deltaheightoffset (CallFunction|CalcZRelativeOffset (Variables|Default|GetPawn)))
          (Transformation|AddLocalOffset (Variables|Default|GetHandle_FX_Dust) (Math|Vector|MakeVector 0.0 0.0 _deltaheightoffset) false false))))))

(fn Calc_Z_RelativeOffset
  (params (Pawn :type Fort%20Pawn%20Object%20Reference))
  (Variables|Default|SetCurrentHeightOffset (+ (* (Components|Capsule|GetScaledCapsuleHalfHeight (Variables|/Script/Engine.Character|GetCapsuleComponent Pawn)) -1.0) (Variables|Default|GetHeightOffsetFromGround)))
  (Variables|Default|SetDeltaHeightOffset (- (Variables|Default|GetCurrentHeightOffset) (Variables|Default|GetPreviousHeightOffset)))
  (Variables|Default|SetPreviousHeightOffset (Variables|Default|GetCurrentHeightOffset))
  (return (Variables|Default|GetDeltaHeightOffset)))

(fn ConstructionScript)
