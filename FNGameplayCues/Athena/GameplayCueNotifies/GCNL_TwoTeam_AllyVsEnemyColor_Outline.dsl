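; Looping gameplay cue notify that picks a colored outline for the target pawn based on whether it is ally or enemy to the local player, apparently for two-team modes.
; GetColor finds the pawn's team affiliation relative to the local player and looks it up in the TeamAffiliationBaseColorMap.
; A missing map entry logs an error and returns the default color, as does an invalid pawn without logging.
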(fn GetColor
  (bind _teamaffiliationbasecolormap (Variables|Default|GetTeamAffiliationBaseColorMap))
  (bind _target_pawn (Variables|Default|GetTarget%20Pawn)
    (:then
      (if (.ReturnValue (Utilities|Map|Find _teamaffiliationbasecolormap (Team|GetTeamAffiliationtoLocalPlayer _target_pawn (Team|GetActorTeam _target_pawn))))
        (return (Utilities|Map|Find _teamaffiliationbasecolormap (Team|GetTeamAffiliationtoLocalPlayer _target_pawn (Team|GetActorTeam _target_pawn))))
        (else
          ;[1 0.4829 0.4583] Error - Failed to find the team affiliation, returning black (i.e. no glow)
          (Fortnite|FortShippingLog self (Utilities|String|Append "GCNL_TwoTeam_AllyVsEnemyColor_Outline::GetColor() Failed to find color data in TeamAffiliationbaseColorMap  for team affiliation: [" (Utilities|String|EnumtoString (Team|GetTeamAffiliationtoLocalPlayer _target_pawn (Team|GetActorTeam _target_pawn))) "].") true)
          (return (Variables|ColoredOutline|GetDefaultColor)))))
    (:else
      ;[1 0.4829 0.4583] Error - Pawn is Invalid
      (return (Variables|ColoredOutline|GetDefaultColor)))))

(fn ConstructionScript
  (|CallParentFunction))
