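; Looping gameplay cue notify for the VelvetMoth emote that shows the player's TotalRoyalRoyales cosmetic stat through a Niagara effect.
; On loop start it watches that stat on the target Fort player pawn.
; Each change writes the value to the first particle component's InputDigits parameter and plays a timeline driving RampIn.
; EndPlay starts a second timeline that drives RampOut on the NS_Looping component.
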(event EventBeginPlay)

(event EventOnLoopingStart(Niagara)
  (params (MyTarget :type Actor%20Object%20Reference) (Parameters :type Gameplay%20Cue%20Parameters%20Structure%20%28by%20ref%29) (ParticleComponents :type Array%20of%20Niagara%20Particle%20System%20Component%20Object%20References) (AudioComponents :type Array%20of%20Audio%20Component%20Object%20References))
  (bind _output (Utilities|Array|Get(acopy) ParticleComponents))
  (bind _asfort_player_pawn (Utilities|Casting|CastToFortPlayerPawn MyTarget))
  (bind (_statname _statvalue) (Cosmetics|WatchCosmeticStat _asfort_player_pawn "(TagName=\"StatManager.TotalRoyalRoyales\")")
    (:OnStatChanged
      (Effects|Components|ParticleSystem|SetIntParameter _output "InputDigits" _statvalue)
      (bind (_direction _rampin) (Timeline|Timeline_0 PlayFromStart)
        (:Update
          (Effects|Components|ParticleSystem|SetFloatParameter _output "RampIn" _rampin))))))

(event EventEndPlay
  (params (EndPlayReason :type EEndPlayReason%20Enum :default "Destroyed"))
  (bind (_direction _rampout) (Timeline|RampOut PlayFromStart)
    (:Update
      (Effects|Components|ParticleSystem|SetFloatParameter (Variables|Default|GetNS_Looping) "RampOut" _rampout))))

(event Collision|EventActorBeginOverlap
  (params (OtherActor :type Actor%20Object%20Reference)))

(event EventTick
  (params (DeltaSeconds :type Float%20%28single-precision%29)))

(fn ConstructionScript)
