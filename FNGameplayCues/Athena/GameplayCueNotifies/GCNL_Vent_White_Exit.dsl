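; Looping gameplay cue for exiting a white vent, adding screen effects and camera shake for the locally controlled player.
; On start it plays a configurable looping-cue shake at near-zero scale, plus a DudeBro vortex activation shake when a cue parameter equals 1.
; The FadePPVol timeline drives both post-process blend weight and the configurable shake's scale.
; On removal the timeline reverses to fade the effects out.
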(event EventOnLoopingStart(Cascade)
  (params (MyTarget :type Actor%20Object%20Reference) (Parameters :type Gameplay%20Cue%20Parameters%20Structure%20%28by%20ref%29) (ParticleComponents :type Array%20of%20Cascade%20Particle%20System%20Component%20Object%20References) (AudioComponents :type Array%20of%20Audio%20Component%20Object%20References))
  (Utilities|FlowControl|Sequence
    (:then_0
      (bind _asplayer_pawn_athena (Utilities|Casting|CastToPlayerPawn_Athena MyTarget))
      (if (Pawn|IsLocallyControlled _asplayer_pawn_athena)
        (Variables|Default|SetUsePP true)
        (bind _asplayer_controller (Utilities|Casting|CastToPlayerController (Pawn|GetController _asplayer_pawn_athena)))
        (Utilities|FlowControl|Sequence
          (:then_0
            (bind _returnvalue (CameraShakes|StartCameraShake :self (Variables|/Script/Engine.PlayerController|GetPlayerCameraManager _asplayer_controller) :ShakeClass (Variables|Default|GetCameraShakeClass) :Scale 0.0001 :PlaySpace "CameraLocal"))
            (Variables|Default|SetCameraShake (CameraShakes|ConvLegacyCameraShake _returnvalue)))
          (:then_1
            (if (== (Ability|GameplayCue|BreakGameplayCueParameters Parameters) 1.0)
              (CameraShakes|StartCameraShake :self (Variables|/Script/Engine.PlayerController|GetPlayerCameraManager _asplayer_controller) :ShakeClass "/FNGameplayCues/Athena/Environments/DudeBro/B_DudeBro_VortexActivate_Shake.B_DudeBro_VortexActivate_Shake_C" :Scale 1.0 :PlaySpace "CameraLocal"))))))
    (:then_1
      (if (Variables|Default|GetUsePP)
        (bind (_direction _lerp) (Timeline|FadePPVol Play)
          (:Update
            (Variables|/Script/Engine.PostProcessComponent|SetBlendWeight _lerp (Variables|Default|GetPostProcess))
            (bind _camerashake (Variables|Default|GetCameraShake)
              (:then
                (Variables|/Script/EngineCameras.LegacyCameraShake|SetShakeScale _lerp _camerashake))
              (:else))))))))

(event EventOnRemoval(Cascade)
  (params (MyTarget :type Actor%20Object%20Reference) (Parameters :type Gameplay%20Cue%20Parameters%20Structure%20%28by%20ref%29) (ParticleComponents :type Array%20of%20Cascade%20Particle%20System%20Component%20Object%20References) (AudioComponents :type Array%20of%20Audio%20Component%20Object%20References) (BurstCameraShakeInstance :type Camera%20Shake%20Base%20Object%20Reference) (BurstDecalInstance :type Decal%20Actor%20Object%20Reference))
  (if (Variables|Default|GetUsePP)
    (Timeline|FadePPVol Reverse)))

(fn ConstructionScript)
