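; Looping gameplay cue for the Wax mutator mode that colors a particle effect on a player by team standing.
; On start it caches the first Cascade component, Wax mutator and player state, then starts a repeating timer.
; Each update sets the "Color" parameter from a map, using team placement or the -1 entry depending on team-leader status.
; Removal clears the timer and deactivates the particle.
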(event EventOnLoopingStart(Cascade)
  (params (MyTarget :type Actor%20Object%20Reference) (Parameters :type Gameplay%20Cue%20Parameters%20Structure%20%28by%20ref%29) (ParticleComponents :type Array%20of%20Cascade%20Particle%20System%20Component%20Object%20References) (AudioComponents :type Array%20of%20Audio%20Component%20Object%20References))
  ;[0.09315 1 0] Event On Looping Start (Cascade)
  (Utilities|IsValid MyTarget
    (:Is%20Valid
      (Utilities|FlowControl|Sequence
        (:then_0
          ;;[0.2023 1 0.3454] Set the Particle Component
          (Variables|Default|SetParticleComponent (select (Utilities|Array|IsValidIndex ParticleComponents) (Utilities|Array|Get(acopy) ParticleComponents) 0)))
        (:then_1
          ;;[0.4233 0.8071 1] Grab the mutator
          (bind _asfort_game_state_athena (Utilities|Casting|CastToFortGameStateAthena (Game|GetGameState MyTarget))
            (:then
              (bind _asfort_athena_mutator_wax (Utilities|Casting|CastToFortAthenaMutator_Wax (Modifier|GetMutatorbyClass (Variables|/Script/FortniteGame.FortGameStateAthena|GetMutatorListComponent _asfort_game_state_athena) "/Script/FortniteGame.FortAthenaMutator_Wax"))
                (:then
                  (Variables|Default|SetWaxMutator _asfort_athena_mutator_wax))
                (:CastFailed
                  (Development|PrintString 0 "Couldn't Cast to Mutator" true true "(R=1.000000,G=0.000000,B=0.000000,A=1.000000)" 12.0))))
            (:CastFailed
              (Development|PrintString 0 "Couldn't Cast to Game State" true true "(R=1.000000,G=0.000000,B=0.000000,A=1.000000)" 12.0))))
        (:then_2
          ;;[0.05616 0.2499 1] Grab the player state
          (bind _asfort_player_pawn_athena (Utilities|Casting|CastToFortPlayerPawnAthena MyTarget)
            (:then
              (bind _asfort_player_state_athena (Utilities|Casting|CastToFortPlayerStateAthena (Variables|/Script/Engine.Pawn|GetPlayerState _asfort_player_pawn_athena))
                (:then
                  (Variables|Default|SetPlayerState _asfort_player_state_athena))
                (:CastFailed
                  (Development|PrintString 0 "Couldn't Cast to Player State" true true "(R=1.000000,G=0.000000,B=0.000000,A=1.000000)" 12.0))))
            (:CastFailed
              (Development|PrintString 0 "Couldn't Cast to Player Pawn" true true "(R=1.000000,G=0.000000,B=0.000000,A=1.000000)" 12.0))))
        (:then_3
          ;;[0.6132 0.5091 1] Don't set a timer if we don't have a mutator player state and particle component
          (if (Math|Boolean|ANDBoolean (Utilities|IsValid (Variables|Default|GetWaxMutator) :!class /Script/Engine.KismetSystemLibrary) (Utilities|IsValid (Variables|Default|GetPlayerState) :!class /Script/Engine.KismetSystemLibrary) (Utilities|IsValid (Variables|Default|GetParticleComponent) :!class /Script/Engine.KismetSystemLibrary))
            ;;[0 1 1] Set a timer for ourselves so we can listen for changes in placement
            (bind _returnvalue (Utilities|Time|SetTimerbyEvent (EventDispatchers|CreateEvent :!event Custom|UpdateVisuals) (Variables|Default|GetPeriodicUpdateFrequency) true))
            (Variables|Default|SetVisual%20Update%20TImer _returnvalue)
            (CallFunction|UpdateVisuals)))))))

(event EventOnRemoval(Cascade)
  (params (MyTarget :type Actor%20Object%20Reference) (Parameters :type Gameplay%20Cue%20Parameters%20Structure%20%28by%20ref%29) (ParticleComponents :type Array%20of%20Cascade%20Particle%20System%20Component%20Object%20References) (AudioComponents :type Array%20of%20Audio%20Component%20Object%20References) (BurstCameraShakeInstance :type Camera%20Shake%20Base%20Object%20Reference) (BurstDecalInstance :type Decal%20Actor%20Object%20Reference))
  (bind _output (Utilities|Array|Get(acopy) ParticleComponents))
  ;[1 0.5532 0] Event On Removal (Cascade)
  (Utilities|Time|ClearandInvalidateTimerbyHandle self (Variables|Default|GetVisual%20Update%20TImer))
  (if (Utilities|Array|IsValidIndex ParticleComponents)
    (Utilities|IsValid _output
      (:Is%20Valid
        (Components|Activation|Deactivate _output)))))

(event Custom|UpdateVisuals
  (bind _particlecomponent (Variables|Default|GetParticleComponent))
  ;[0.4825 0 1] Update Visuals
  (Utilities|IsValid _particlecomponent
    (:Is%20Valid
      (Effects|Components|ParticleSystem|SetVectorParameter _particlecomponent "Color" (Math|Conversions|ToVector(LinearColor) (select (Class|FortAthenaMutatorWax|IsPlayerTeamLeader (Variables|Default|GetWaxMutator) (Variables|Default|GetPlayerState)) (Utilities|Map|Find (Variables|Default|GetColors) (Class|FortAthenaMutatorWax|GetPlayerTeamPlacement (Variables|Default|GetWaxMutator) (Variables|Default|GetPlayerState))) (Utilities|Map|Find (Variables|Default|GetColors) -1)))))))

(fn ConstructionScript)
