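; Looping gameplay cue for the Zero Point low-gravity effect, adding landing FX and fall-reactive audio to the affected player.
; On start it caches the pawn and looping sound and binds OnPawnLanded to spawn a particle and landing sound; removal unbinds it.
; A gated update maps vertical speed times time from jump apex, while falling, into a smoothed value.
; That value and its inverse drive the loop's FallSpeed and FallSpeedInv parameters.
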(event Custom|TargetPlayerLanded
  (params (Hit :type Hit%20Result%20Structure%20%28by%20ref%29))
  (bind _targetplayer (Variables|Default|GetTargetPlayer)
    (:then
      (Effects|SpawnEmitterAttached "/FNGameplayCues/Athena/Items/ForagedItems/LowGravity/FX/P_Low_Gravity_Land_Blue.P_Low_Gravity_Land_Blue" (Variables|/Script/Engine.Character|GetMesh _targetplayer) "root")
      (Audio|PlaySoundatLocation 0 "/FNGameplayCues/Athena/Sounds/Events/Cattus/ZeroPoint/ZeroPoint_Land_Cue.ZeroPoint_Land_Cue" (Transformation|GetActorLocation _targetplayer)))
    (:else)))

(event EventOnLoopingStart(Cascade)
  (params (MyTarget :type Actor%20Object%20Reference) (Parameters :type Gameplay%20Cue%20Parameters%20Structure%20%28by%20ref%29) (ParticleComponents :type Array%20of%20Cascade%20Particle%20System%20Component%20Object%20References) (AudioComponents :type Array%20of%20Audio%20Component%20Object%20References))
  (bind _targetplayer (Variables|Default|GetTargetPlayer))
  (bind _charactermovement (Variables|/Script/Engine.Character|GetCharacterMovement _targetplayer))
  (bind _returnvalue (AI|Components|NavMovement|IsFalling _charactermovement))
  (bind (_velocity_x _velocity_y _velocity_z) (Math|Vector|BreakVector (Variables|/Script/Engine.MovementComponent|GetVelocity_X _charactermovement)))
  (bind _asfort_player_pawn (Utilities|Casting|CastToFortPlayerPawn MyTarget))
  (Variables|Default|SetTargetPlayer _asfort_player_pawn)
  (Variables|Default|SetLoopingAudio (Utilities|Array|Get(acopy) AudioComponents))
  (Utilities|FlowControl|Sequence
    (:then_0
      (EventDispatchers|BindEventtoOnPawnLanded (Variables|Default|GetTargetPlayer) (delegate Custom|TargetPlayerLanded)))
    (:then_1
      (Utilities|FlowControl|Gate true
        (:Exit
          (Variables|Default|GetTargetPlayer
            (:then
              (Variables|Default|GetLoopingAudio
                (:then
                  (Variables|Default|SetTargetDelta (Math|Float|SelectFloat (Math|Float|MapRangeClamped (Math|Float|Absolute(Float) (* (Math|Float|Absolute(Float) _velocity_z) (PawnStats|TimefromJumpApex _targetplayer))) 0.0 1e+03 0.0 1.0) 0.0 _returnvalue))
                  (bind _output_get (Variables|Default|SetInterpDelta (Math|Interpolation|FInterpTo (Variables|Default|GetInterpDelta) (Variables|Default|GetTargetDelta) (Utilities|Time|GetWorldDeltaSeconds (Variables|Default|GetTargetPlayer)) (Math|Float|SelectFloat 1e+01 2.0 (or (> (Variables|Default|GetTargetDelta) (Variables|Default|GetInterpDelta)) (not _returnvalue))))))
                  (Audio|Parameter|SetFloatParameter (Variables|Default|GetLoopingAudio) "FallSpeedInv" (- 1.0 _output_get) :!class /Script/AudioExtensions.AudioParameterControllerInterface)
                  (Audio|Parameter|SetFloatParameter (Variables|Default|GetLoopingAudio) "FallSpeed" _output_get :!class /Script/AudioExtensions.AudioParameterControllerInterface))
                (:else)))
            (:else)))))))

(event EventOnRemoval(Cascade)
  (params (MyTarget :type Actor%20Object%20Reference) (Parameters :type Gameplay%20Cue%20Parameters%20Structure%20%28by%20ref%29) (ParticleComponents :type Array%20of%20Cascade%20Particle%20System%20Component%20Object%20References) (AudioComponents :type Array%20of%20Audio%20Component%20Object%20References) (BurstCameraShakeInstance :type Camera%20Shake%20Base%20Object%20Reference) (BurstDecalInstance :type Decal%20Actor%20Object%20Reference))
  (bind _targetplayer (Variables|Default|GetTargetPlayer))
  (bind _charactermovement (Variables|/Script/Engine.Character|GetCharacterMovement _targetplayer))
  (bind _returnvalue (AI|Components|NavMovement|IsFalling _charactermovement))
  (bind (_velocity_x _velocity_y _velocity_z) (Math|Vector|BreakVector (Variables|/Script/Engine.MovementComponent|GetVelocity_X _charactermovement)))
  (bind _targetplayer_2 (Variables|Default|GetTargetPlayer))
  (Utilities|IsValid _targetplayer_2
    (:Is%20Valid
      (EventDispatchers|UnbindEventfromOnPawnLanded _targetplayer_2 (delegate Custom|TargetPlayerLanded)))
    (:Is%20Not%20Valid
      _))
  (Variables|Default|SetTargetPlayer 0)
  (Utilities|FlowControl|Gate true
    (:Exit
      (Variables|Default|GetTargetPlayer
        (:then
          (Variables|Default|GetLoopingAudio
            (:then
              (Variables|Default|SetTargetDelta (Math|Float|SelectFloat (Math|Float|MapRangeClamped (Math|Float|Absolute(Float) (* (Math|Float|Absolute(Float) _velocity_z) (PawnStats|TimefromJumpApex _targetplayer))) 0.0 1e+03 0.0 1.0) 0.0 _returnvalue))
              (bind _output_get (Variables|Default|SetInterpDelta (Math|Interpolation|FInterpTo (Variables|Default|GetInterpDelta) (Variables|Default|GetTargetDelta) (Utilities|Time|GetWorldDeltaSeconds (Variables|Default|GetTargetPlayer)) (Math|Float|SelectFloat 1e+01 2.0 (or (> (Variables|Default|GetTargetDelta) (Variables|Default|GetInterpDelta)) (not _returnvalue))))))
              (Audio|Parameter|SetFloatParameter (Variables|Default|GetLoopingAudio) "FallSpeedInv" (- 1.0 _output_get) :!class /Script/AudioExtensions.AudioParameterControllerInterface)
              (Audio|Parameter|SetFloatParameter (Variables|Default|GetLoopingAudio) "FallSpeed" _output_get :!class /Script/AudioExtensions.AudioParameterControllerInterface))
            (:else)))
        (:else)))))

(event EventTick
  (params (DeltaSeconds :type Float%20%28single-precision%29)))

(fn ConstructionScript)
