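; Looping gameplay cue for downhill zipline riding that adds a speed-driven camera shake for the player's local viewing controller.
; On loop start it caches the target pawn and starts a short repeating timer, which is cleared on removal.
; Each tick of that timer starts the downhill-speed shake, then overrides its rotational oscillation with Perlin noise.
; Amplitude and frequency interpolate between min and max values by the squared zipline speed factor.
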(event EventOnRemoval(EitherVFX-preferthisversion)
  (params (MyTarget :type Actor%20Object%20Reference) (Parameters :type Gameplay%20Cue%20Parameters%20Structure%20%28by%20ref%29) (ParticleComponents :type Array%20of%20FXSystem%20Component%20Object%20References) (AudioComponents :type Array%20of%20Audio%20Component%20Object%20References) (BurstCameraShakeInstance :type Camera%20Shake%20Base%20Object%20Reference) (BurstDecalInstance :type Decal%20Actor%20Object%20Reference))
  (Utilities|Time|ClearandInvalidateTimerbyHandle MyTarget (Variables|Default|GetCameraShakeTimerHandle)))

(event EventOnLoopingStart(EitherVFX-preferthisversion)
  (params (MyTarget :type Actor%20Object%20Reference) (Parameters :type Gameplay%20Cue%20Parameters%20Structure%20%28by%20ref%29) (ParticleComponents :type Array%20of%20FXSystem%20Component%20Object%20References) (AudioComponents :type Array%20of%20Audio%20Component%20Object%20References))
  (bind _asfort_player_pawn (Utilities|Casting|CastToFortPlayerPawn MyTarget))
  (Variables|Default|SetPlayerPawn _asfort_player_pawn)
  (bind _returnvalue (Utilities|Time|SetTimerbyEvent (EventDispatchers|CreateEvent :!event Custom|CameraShakeTimer) 0.3 true))
  (Variables|Default|SetCameraShakeTimerHandle _returnvalue))

(event Custom|CameraShakeTimer
  (bind _playerpawn (Variables|Default|GetPlayerPawn)
    (:then
      (Utilities|IsValid (Pawn|GetLocalViewingPlayerController _playerpawn)
        (:Is%20Valid
          (bind _playercameramanager (Variables|/Script/Engine.PlayerController|GetPlayerCameraManager (Pawn|GetLocalViewingPlayerController _playerpawn))
            (:then
              (bind _returnvalue (CameraShakes|StartCameraShake :self _playercameramanager :ShakeClass "/FNGameplayCues/Athena/Items/Consumables/ZipLine/CameraShake_ZipLine_DownhillSpeed.CameraShake_ZipLine_DownhillSpeed_C" :Scale 1.0 :PlaySpace "CameraLocal"))
              (bind _output_get (Variables|Default|SetmatineeCamShake (CameraShakes|ConvLegacyCameraShake _returnvalue)))
              (Variables|/Script/EngineCameras.LegacyCameraShake|SetRotOscillation_Pitch (Utilities|Struct|MakeROscillator "(Amplitude=0.000000,Frequency=0.000000,InitialOffset=EOO_OffsetRandom,Waveform=SineWave)" "(Amplitude=0.000000,Frequency=0.000000,InitialOffset=EOO_OffsetRandom,Waveform=SineWave)" (Utilities|Struct|MakeFOscillator (Math|Float|Lerp (Variables|Default|GetMinCamShakeAmp) (Variables|Default|GetMaxCamShakeAmp) (Math|Float|Square (Zipline|GetZiplineSpeedFactor _playerpawn))) (Math|Float|Lerp (Variables|Default|GetMinCamShakeFreq) (Variables|Default|GetMaxCamShakeFreq) (Math|Float|Square (Zipline|GetZiplineSpeedFactor _playerpawn))) "PerlinNoise")) _output_get))
            (:else)))))
    (:else)))

(fn ConstructionScript)
