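; Looping gameplay cue notify for zipline travel that drives the audio heard while a player rides a zipline.
; On application it caches the cue parameters and player pawn, snaps the inherited layered audio component to the pawn's root, and fades it in.
; It also stores the first spawned audio component as the attach sound.
; On removal it fades out both sounds, and both events print debug strings.
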(event EventOnApplication(EitherVFX-preferthisone)
  (params (MyTarget :type Actor%20Object%20Reference) (Parameters :type Gameplay%20Cue%20Parameters%20Structure%20%28by%20ref%29) (ParticleComponents :type Array%20of%20FXSystem%20Component%20Object%20References) (AudioComponents :type Array%20of%20Audio%20Component%20Object%20References) (BurstCameraShakeInstance :type Camera%20Shake%20Base%20Object%20Reference) (BurstDecalInstance :type Decal%20Actor%20Object%20Reference))
  (Development|PrintString 0 "Zipline GCNL Applied")
  (Variables|FortGameplayCueNotifyLoop_ZiplineTravel|SetCachedGCParameters Parameters)
  (bind _output_get (Variables|FortGameplayCueNotifyLoop_ZiplineTravel|SetPlayerPawn (Utilities|Casting|CastToFortPlayerPawn MyTarget)))
  (Transformation|AttachComponentToComponent (Variables|FortGameplayCueNotifyLoop_ZiplineTravel|GetLayeredAudioComp) (Variables|/Script/Engine.Actor|GetRootComponent _output_get) "None" "SnapToTarget" "SnapToTarget" "SnapToTarget" true)
  (Audio|Components|Audio|FadeIn (Variables|FortGameplayCueNotifyLoop_ZiplineTravel|GetLayeredAudioComp) 0.5 1.0 0.0 "Linear" :!class /Script/Engine.AudioComponent)
  (Variables|Default|SetAttachSoundAudioComp (Utilities|Array|Get(acopy) AudioComponents)))

(event EventOnRemoval(EitherVFX-preferthisversion)
  (params (MyTarget :type Actor%20Object%20Reference) (Parameters :type Gameplay%20Cue%20Parameters%20Structure%20%28by%20ref%29) (ParticleComponents :type Array%20of%20FXSystem%20Component%20Object%20References) (AudioComponents :type Array%20of%20Audio%20Component%20Object%20References) (BurstCameraShakeInstance :type Camera%20Shake%20Base%20Object%20Reference) (BurstDecalInstance :type Decal%20Actor%20Object%20Reference))
  (Development|PrintString 0 "Zipline GCNL Removed")
  (Audio|Components|Audio|FadeOut (Variables|FortGameplayCueNotifyLoop_ZiplineTravel|GetLayeredAudioComp) 0.5 0.0 "Linear" :!class /Script/Engine.AudioComponent)
  (bind _attachsoundaudiocomp (Variables|Default|GetAttachSoundAudioComp)
    (:then
      (Audio|Components|Audio|FadeOut _attachsoundaudiocomp 0.5 0.0 "Linear" :!class /Script/Engine.AudioComponent))
    (:else)))

(fn ConstructionScript)
