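; Gameplay cue notify for the Apollo Tinfoil Beacon pulse that sizes and times a sweep.
; On execute, it calls the parent, then sets the sweep radius from the cue's raw magnitude.
; Sweep duration is the scalable-float base duration multiplied by the clamped ratio of that radius to the default radius, so it grows with radius.
; The sweep itself is not built in this graph.
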(event EventBeginPlay)

(event Collision|EventActorBeginOverlap
  (params (OtherActor :type Actor%20Object%20Reference)))

(event EventTick
  (params (DeltaSeconds :type Float%20%28single-precision%29)))

(fn OnExecute
  (params (MyTarget :type Actor%20Object%20Reference) (Parameters :type Gameplay%20Cue%20Parameters%20Structure%20%28by%20ref%29))
  (bind (_normalizedmagnitude _rawmagnitude) (Ability|GameplayCue|BreakGameplayCueParameters Parameters))
  (bind _returnvalue (|CallParentFunction MyTarget Parameters))
  (Variables|Default|SetSweepRadius _rawmagnitude)
  (Variables|Default|SetSweepDuration (* (ScalableFloat|GetValueatLevel (Variables|Default|GetRow_SweepDuration)) (Math|Float|Clamp(Float) (Math|Float|SafeDivide _rawmagnitude (ScalableFloat|GetValueatLevel (Variables|Default|GetRow_DefaultRadius))) 0.01 1e+02)))
  (return _returnvalue))

(fn ConstructionScript
  (|CallParentFunction))
