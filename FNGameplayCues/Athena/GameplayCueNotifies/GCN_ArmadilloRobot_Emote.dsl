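; Looping gameplay cue for the Armadillo Robot emote that shows a spotlight and particle effect on the player.
; On start it attaches its Arrow to the player mesh's FX_Chest socket, creates dynamic materials for the spotlight and activates the Emote particles.
; The AnimLightFixture timeline drives spotlight scale and a flickering Alpha on materials and particles.
; On removal that timeline stops and TurnOff drives scale and Alpha, deactivating the particles when finished.
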(event EventOnLoopingStart(EitherVFX-preferthisversion)
  (params (MyTarget :type Actor%20Object%20Reference) (Parameters :type Gameplay%20Cue%20Parameters%20Structure%20%28by%20ref%29) (ParticleComponents :type Array%20of%20FXSystem%20Component%20Object%20References) (AudioComponents :type Array%20of%20Audio%20Component%20Object%20References))
  (Utilities|FlowControl|Sequence
    (:then_0
      (Utilities|IsValid MyTarget
        (:Is%20Valid
          (bind _asfort_player_pawn (Utilities|Casting|CastToFortPlayerPawn MyTarget))
          (bind _output_get (Variables|Default|SetPawn _asfort_player_pawn))
          (Variables|Default|SetPlayerMesh (Variables|/Script/Engine.Character|GetMesh _output_get)))))
    (:then_1
      (for (_array_element _array_index) (Rendering|Material|GetMaterials (Variables|Default|GetSpotlight))
        (bind _returnvalue (Rendering|Material|CreateDynamicMaterialInstance (Variables|Default|GetSpotlight) _array_index _array_element (Utilities|String|StringToName (Utilities|String|Append "MeshMID" "_" (Utilities|String|ToString(Integer) _array_index))) :!class /Script/Engine.PrimitiveComponent))
        (Utilities|Array|Add (Variables|Default|GetSpotlightMIDs) _returnvalue)))
    (:then_2
      (Transformation|SetActorLocation self (Transformation|GetSocketLocation (Variables|Default|GetPlayerMesh) "Root") false false)
      (Transformation|AttachComponentToComponent (Variables|Default|GetArrow) (Variables|Default|GetPlayerMesh) "FX_Chest" "SnapToTarget" "SnapToTarget" "KeepRelative" true)
      (Components|Activation|Activate (Variables|Default|GetEmote) false)
      (CallFunction|EnableFX))))

(event Custom|EnableFX
  (bind (_direction _scale _lightflicker _scaleup) (Timeline|AnimLightFixture PlayFromStart)
    (:Update
      (Transformation|SetWorldScale3D (Variables|Default|GetSpotlight) _scaleup)
      (for _array_element (Variables|Default|GetSpotlightMIDs)
        (Rendering|Material|SetScalarParameterValue _array_element "Alpha" _lightflicker :!class /Script/Engine.MaterialInstanceDynamic)
        (Effects|Components|ParticleSystem|SetFloatParameter (Variables|Default|GetEmote) "Alpha" _lightflicker)))))

(event EventOnRemoval(EitherVFX-preferthisversion)
  (params (MyTarget :type Actor%20Object%20Reference) (Parameters :type Gameplay%20Cue%20Parameters%20Structure%20%28by%20ref%29) (ParticleComponents :type Array%20of%20FXSystem%20Component%20Object%20References) (AudioComponents :type Array%20of%20Audio%20Component%20Object%20References) (BurstCameraShakeInstance :type Camera%20Shake%20Base%20Object%20Reference) (BurstDecalInstance :type Decal%20Actor%20Object%20Reference))
  (bind _spotlight (Variables|Default|GetSpotlight))
  (bind _emote (Variables|Default|GetEmote))
  (bind _spotlightmids (Variables|Default|GetSpotlightMIDs))
  (Utilities|FlowControl|Sequence
    (:then_0
      (Timeline|AnimLightFixture Stop))
    (:then_1
      (bind (_direction _scale_down _alpha _scaledown) (Timeline|TurnOff PlayFromStart)
        (:Update
          (Transformation|SetWorldScale3D _spotlight _scaledown)
          (for _array_element _spotlightmids
            (Rendering|Material|SetScalarParameterValue _array_element "Alpha" _alpha :!class /Script/Engine.MaterialInstanceDynamic)
            (Effects|Components|ParticleSystem|SetFloatParameter _emote "Alpha" _alpha)))
        (:Finished
          (Components|Activation|Deactivate (Variables|Default|GetEmote)))))))

(fn ConstructionScript)
