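; Burst gameplay cue notify that plays the launch sound when the Athena Directional Launcher launches a pawn or vehicle.
; OnBurst casts the target to FortPawn and, if that fails, to FortAthenaVehicle, using the vehicle's driver for the local-control check.
; A locally controlled pawn or driver gets the first-person cue as 2D sound; otherwise the standard cue plays at the target's location.
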(fn OnBurst(EitherVFX-preferthisversion)
  (params (MyTarget :type Actor%20Object%20Reference) (Parameters :type Gameplay%20Cue%20Parameters%20Structure%20%28by%20ref%29) (ParticleComponents :type Array%20of%20FXSystem%20Component%20Object%20References) (AudioComponents :type Array%20of%20Audio%20Component%20Object%20References) (BurstCameraShakeInstance :type Camera%20Shake%20Base%20Object%20Reference) (BurstDecalInstance :type Decal%20Actor%20Object%20Reference))
  (bind _asfort_pawn (Utilities|Casting|CastToFortPawn MyTarget)
    (:then
      (if (Pawn|IsLocallyControlled _asfort_pawn)
        (Audio|PlaySound2D MyTarget "/Game/Athena/Sounds/LaunchPad/DirectionalLauncher/DirectionalLauncher_Launch_1p_Cue.DirectionalLauncher_Launch_1p_Cue")
        (else
          (Audio|PlaySoundatLocation MyTarget "/Game/Athena/Sounds/LaunchPad/DirectionalLauncher/DirectionalLauncher_Launch_Cue.DirectionalLauncher_Launch_Cue" (Transformation|GetActorLocation MyTarget)))))
    (:CastFailed
      (bind _asfort_athena_vehicle (Utilities|Casting|CastToFortAthenaVehicle MyTarget))
      (if (Pawn|IsLocallyControlled (Vehicle|GetDriver _asfort_athena_vehicle))
        (Audio|PlaySound2D MyTarget "/Game/Athena/Sounds/LaunchPad/DirectionalLauncher/DirectionalLauncher_Launch_1p_Cue.DirectionalLauncher_Launch_1p_Cue")
        (else
          (Audio|PlaySoundatLocation MyTarget "/Game/Athena/Sounds/LaunchPad/DirectionalLauncher/DirectionalLauncher_Launch_Cue.DirectionalLauncher_Launch_Cue" (Transformation|GetActorLocation _asfort_athena_vehicle))))))
  (return))
