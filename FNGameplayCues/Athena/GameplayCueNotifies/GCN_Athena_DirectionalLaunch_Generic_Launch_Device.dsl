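; Gameplay cue notify for the generic directional launch device, overriding OnBurst to give launch feedback when an object is launched.
; Pawns and vehicles defer to the parent implementation, which plays the FX only for those target types.
; Other targets play a launch sound copied from the parent at their location, but only if they are physics pawn objects or report the PhysicsObject type.
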(fn OnBurst(EitherVFX-preferthisversion)
  (params (MyTarget :type Actor%20Object%20Reference) (Parameters :type Gameplay%20Cue%20Parameters%20Structure%20%28by%20ref%29) (ParticleComponents :type Array%20of%20FXSystem%20Component%20Object%20References) (AudioComponents :type Array%20of%20Audio%20Component%20Object%20References) (BurstCameraShakeInstance :type Camera%20Shake%20Base%20Object%20Reference) (BurstDecalInstance :type Decal%20Actor%20Object%20Reference))
  (bind (_asfort_athena_vehicle _bsuccess) (Utilities|Casting|CastToFortAthenaVehicle (Variables|GetMyTarget)))
  (bind (_asfort_pawn _bsuccess_2) (Utilities|Casting|CastToFortPawn (Variables|GetMyTarget)))
  (bind _mytarget (Variables|GetMyTarget))
  (if (or _bsuccess_2 _bsuccess)
    (|CallParentFunction MyTarget Parameters ParticleComponents AudioComponents BurstCameraShakeInstance BurstDecalInstance)
    (else
      ;[1 1 1] Parent BP only plays the FX for pawns and vehicles, so we need to check if this is a physics object
      (Utilities|Casting|CastToFortPhysicsPawnObject _mytarget
        (:then
          _)
        (:CastFailed
          (bind _returnvalue (FortPhysicsSimulationLibrary|GetPhysicsObjectType _mytarget))
          (if (not (Utilities|Enum|Equal(Enum) _returnvalue "PhysicsObject"))
            (return))))
      (Audio|PlaySoundatLocation (Variables|GetMyTarget) "/Game/Athena/Sounds/LaunchPad/DirectionalLauncher/DirectionalLauncher_Launch_Cue.DirectionalLauncher_Launch_Cue" (Transformation|GetActorLocation (Variables|GetMyTarget))) ; Copied from parent
    ))
  (return))
