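; Burst gameplay cue for the Athena scanner lock zap, drawing an electric strike between the lock and the instigator.
; Spawns a strike-source effect at the cue location plus a fixed, unrotated offset, using the target's forward rotation.
; Spawns the zap effect just above the instigator's location, aimed at that strike point.
; Sets the zap's Length parameter from the scaled distance between the two points so it spans the gap.
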(fn OnBurst(Cascade)
  (params (MyTarget :type Actor%20Object%20Reference) (Parameters :type Gameplay%20Cue%20Parameters%20Structure%20%28by%20ref%29) (ParticleComponents :type Array%20of%20Cascade%20Particle%20System%20Component%20Object%20References) (AudioComponents :type Array%20of%20Audio%20Component%20Object%20References) (BurstCameraShakeInstance :type Camera%20Shake%20Base%20Object%20Reference) (BurstDecalInstance :type Decal%20Actor%20Object%20Reference))
  (bind (_normalizedmagnitude _rawmagnitude _effectcontext _matchedtagname _originaltag _aggregatedsourcetags _aggregatedtargettags _location _normal _instigator) (Ability|GameplayCue|BreakGameplayCueParameters Parameters))
  (bind _returnvalue (Transformation|GetActorLocation _instigator))
  (bind _returnvalue_2 (+ _returnvalue "0,0.000000,50.000000"))
  (bind _returnvalue_3 (+ _location "18.000000,0.000000,8.000000"))
  (bind _rotation (Class|PlaysetCaptureLibrary|GetForwardRotation MyTarget))
  (Effects|Components|ParticleSystem|SpawnFXAtLocation 0 "/FNGameplayCues/Athena/Items/EnvironmentalItems/Locks/Scanner/FX/P_ScannerLock_Strike_Source.P_ScannerLock_Strike_Source" _returnvalue_3 _rotation)
  (bind _returnvalue_4 (Effects|Components|ParticleSystem|SpawnFXAtLocation 0 "/FNGameplayCues/Athena/Items/EnvironmentalItems/Locks/Scanner/FX/P_Faction_Lock_Zap.P_Faction_Lock_Zap" _returnvalue_2 (Math|Rotator|FindLookatRotation _returnvalue_2 _returnvalue_3)))
  (Effects|Components|ParticleSystem|SetFloatParameter _returnvalue_4 "Length" (/ (Math|Vector|Distance(Vector) _returnvalue_2 _returnvalue_3) 1e+03))
  (return))
