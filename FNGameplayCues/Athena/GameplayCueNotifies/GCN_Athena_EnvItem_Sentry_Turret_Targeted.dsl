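; Burst gameplay cue that, going by its name, fires when an Athena sentry turret targets an actor, adding a visual sound indicator to the parent cue's effects.
; OnBurst runs the parent burst, then passes the target to ShowSoundIndicator.
; If the target is valid and burst sounds exist, it shows a local Generic indicator at the target's location, using the first burst sound and that sound's max distance.
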(fn OnBurst(EitherVFX-preferthisversion)
  (params (MyTarget :type Actor%20Object%20Reference) (Parameters :type Gameplay%20Cue%20Parameters%20Structure%20%28by%20ref%29) (ParticleComponents :type Array%20of%20FXSystem%20Component%20Object%20References) (AudioComponents :type Array%20of%20Audio%20Component%20Object%20References) (BurstCameraShakeInstance :type Camera%20Shake%20Base%20Object%20Reference) (BurstDecalInstance :type Decal%20Actor%20Object%20Reference))
  (|CallParentFunction MyTarget Parameters ParticleComponents AudioComponents BurstCameraShakeInstance BurstDecalInstance)
  (CallFunction|ShowSoundIndicator MyTarget)
  (return))

(fn ShowSoundIndicator
  (params (OwningTarget :type Actor%20Object%20Reference))
  (bind _bursteffectdata (Variables|BurstEffects|GetBurstEffectData))
  (bind (_burstparticles _burstsounds) (Utilities|Struct|BreakFortBurstEffectData _bursteffectdata))
  (bind (_condition _attachment _soundcue) (Utilities|Struct|BreakFortGameplayCueAudioInfo (Utilities|Array|Get(aref) _burstsounds)))
  (Utilities|IsValid OwningTarget
    (:Is%20Valid
      (if (> (Utilities|Array|Length _burstsounds) 0)
        (Fortnite|ShowSoundIndicatorLocal :Location (Transformation|GetActorLocation OwningTarget) :MaxAudibleDistance (Variables|/Script/Engine.SoundBase|GetMaxDistance _soundcue) :IndicatorType "Generic" :Tint "(R=0.000000,G=0.000000,B=0.000000,A=0.000000)" :SoundBase _soundcue :AttenuationMaxRangeMultiplier 1.0 :AttenuationIntensityMultiplier 1.0)))))
