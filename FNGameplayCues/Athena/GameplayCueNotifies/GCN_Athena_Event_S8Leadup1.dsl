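; Burst gameplay cue for the Season 8 lead-up earthquake event, playing distance-scaled sound, camera shake and debris around an epicenter.
; The epicenter comes from the effect causer or a default, and effects require a walking pawn (the followed pawn in replays).
; XY distance drives curve-based volume and low-pass filtering.
; An upward trace detects cover, adding indoor audio and ceiling smoke, while uncovered targets get floor debris within range.
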(fn OnBurst(Cascade)
  (params (MyTarget :type Actor%20Object%20Reference) (Parameters :type Gameplay%20Cue%20Parameters%20Structure%20%28by%20ref%29) (ParticleComponents :type Array%20of%20Cascade%20Particle%20System%20Component%20Object%20References) (AudioComponents :type Array%20of%20Audio%20Component%20Object%20References) (BurstCameraShakeInstance :type Camera%20Shake%20Base%20Object%20Reference) (BurstDecalInstance :type Decal%20Actor%20Object%20Reference))
  (bind _localepicenter (LocalVariables|GetLocalEpicenter))
  (bind _returnvalue (Transformation|GetActorLocation MyTarget))
  (bind _returnvalue_2 (- _returnvalue _localepicenter))
  (bind _returnvalue_3 (Math|Vector|VectorLengthXY _returnvalue_2))
  (bind _returnvalue_4 (Math|Float|MapRangeClamped _returnvalue_3 5e+03 9e+04 0.0 1.0))
  (bind _returnvalue_5 (- 1.0 _returnvalue_4))
  (bind _audiointensitycurve_volume (Variables|Default|GetAudioIntensityCurve_Volume))
  (bind _returnvalue_6 (Math|Curves|GetFloatValue _audiointensitycurve_volume _returnvalue_5))
  (bind _returnvalue_7 (< _returnvalue_3 2.2e+05))
  (bind (_normalizedmagnitude _rawmagnitude _effectcontext _matchedtagname _originaltag _aggregatedsourcetags _aggregatedtargettags _location _normal _instigator _effectcauser) (Ability|GameplayCue|BreakGameplayCueParameters Parameters))
  (Utilities|IsValid _effectcauser
    (:Is%20Valid
      (LocalVariables|SetLocalEpicenter (Transformation|GetActorLocation _effectcauser)))
    (:Is%20Not%20Valid
      (LocalVariables|SetLocalEpicenter (Variables|Default|GetEpicenter))))
  (Development|PrintString MyTarget (Utilities|String|ToString(Vector) (LocalVariables|GetLocalEpicenter)) true true "(R=0.000000,G=0.660000,B=1.000000,A=1.000000)" 3e+01)
  (if (Replay|GetIsPlayingReplay MyTarget)
    (bind _asfort_replay_spectator (Utilities|Casting|CastToFortReplaySpectator (Fortnite|GetFortPlayerControllerfromActor MyTarget)))
    (LocalVariables|SetTargetActor (PlayerState|GetCurrentPawn (Class|FortPlayerControllerSpectating|GetFollowedPlayer _asfort_replay_spectator)))
    (else
      (LocalVariables|SetTargetActor MyTarget)))
  (bind _asplayer_pawn_athena (Utilities|Casting|CastToPlayerPawn_Athena (LocalVariables|GetTargetActor)))
  (switch Utilities|FlowControl|Switch|SwitchonEMovementMode (Variables|/Script/Engine.CharacterMovementComponent|GetMovementMode (Variables|/Script/Engine.Character|GetCharacterMovement _asplayer_pawn_athena))
    (:MOVE_Walking
      (bind (_outhit _returnvalue_8) (Collision|LineTraceByChannel MyTarget (+ _returnvalue "0, 0,100.000000") (+ _returnvalue "0, 0,600.000000")))
      (Utilities|FlowControl|Sequence
        (:then_0
          (if _returnvalue_7
            ;[1 1 1] outside audio
            (bind _returnvalue_9 (Audio|SpawnSound2D MyTarget "/FNGameplayCues/Athena/Sounds/Events/Cane/Cane_Cue.Cane_Cue"))
            (Audio|Components|Audio|SetLowPassFilterEnabled _returnvalue_9 true :!class /Script/Engine.AudioComponent)
            (Audio|Components|Audio|SetLowPassFilterFrequency _returnvalue_9 (* (Math|Curves|GetFloatValue (Variables|Default|GetAudioIntensityCurve_LPF) _returnvalue_5) 2e+04) :!class /Script/Engine.AudioComponent)
            (Audio|Parameter|SetFloatParameter _returnvalue_9 "Intensity" _returnvalue_6 :!class /Script/AudioExtensions.AudioParameterControllerInterface)))
        (:then_1
          ;[1 1 1] inside audio
          (if (and _returnvalue_8 _returnvalue_7)
            (bind _returnvalue_10 (Audio|SpawnSound2D MyTarget "/FNGameplayCues/Athena/Sounds/Events/Cane/Cane_Inside_Cue.Cane_Inside_Cue"))
            (Audio|Parameter|SetFloatParameter _returnvalue_10 "Intensity" _returnvalue_6 :!class /Script/AudioExtensions.AudioParameterControllerInterface)))
        (:then_2
          (Camera|PlayWorldCameraShake 0 "/FNGameplayCues/Effects/Fort_Effects/Camera_Effects/B_s8Leadup1_CameraShake_01.B_s8Leadup1_CameraShake_01_C" (LocalVariables|GetLocalEpicenter) 1e+04 2.2e+05 2.0)
          (if _returnvalue_8
            (if (not (> _returnvalue_3 9e+04))
              (Effects|SpawnEmitteratLocation 0 "/FNGameplayCues/Athena/Events/Earthquake/P_Events_Earthquake_Smokedrop.P_Events_Earthquake_Smokedrop" (.Location (Collision|BreakHitResult _outhit)))
              (return))
            (elif (> _returnvalue_3 7e+04)
              _
              (else
                (Effects|SpawnEmitteratLocation 0 "/FNGameplayCues/Athena/Events/Earthquake/P_Events_Earthquake_Floordrop.P_Events_Earthquake_Floordrop" (- _returnvalue "0, 0,100.000000"))))))))))
