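; Gameplay cue notify for the FringePlank reveal, playing a world-sweep post-process effect from the cue location.
; OnExecute takes sweep radius from raw magnitude and, if the source object is a player pawn, spawns particles and sound, adding a sound mix when locally controlled.
; Timers enable the post process after a delay, feeding origin, radius, durations and timestamp into a Material Parameter Collection, then disable it.
; EndPlay clears both timers and disables the effect.
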(event Custom|BeginSweepFX
  (bind _returnvalue (Utilities|Time|SetTimerbyEvent (EventDispatchers|CreateEvent :!event Custom|EndSweepFX) (+ (ScalableFloat|GetValueatLevel (Variables|Default|GetStencilDuration)) (Variables|Default|GetDelayBeforeSweep))))
  (Variables|Default|SetSweepDelayHandle _returnvalue)
  (bind _returnvalue_2 (Utilities|Time|SetTimerbyEvent (EventDispatchers|CreateEvent :!event Custom|PlayWorldSweep) (Variables|Default|GetDelayBeforeSweep)))
  (Variables|Default|SetDelayBeforeSweepHandle _returnvalue_2)
  (Effects|SpawnEmitteratLocation 0 "/FNGameplayCues/Creative/Items/Gameplay/Disguise/P_Creative_Disguise_DeviceDeactivate.P_Creative_Disguise_DeviceDeactivate" (Variables|Default|GetLocation)))

(event Custom|EndSweepFX
  (Variables|/Script/Engine.PostProcessComponent|SetBlendWeight 0.0 (Variables|Default|GetPostProcess))
  (Variables|/Script/Engine.PostProcessComponent|SetbEnabled false (Variables|Default|GetPostProcess)))

(event Custom|PlayWorldSweep
  (Variables|/Script/Engine.PostProcessComponent|SetBlendWeight 1.0 (Variables|Default|GetPostProcess))
  (Variables|/Script/Engine.PostProcessComponent|SetbEnabled true (Variables|Default|GetPostProcess))
  (Rendering|Material|SetVectorParameterValue 0 "/Game/Weapons/Prototype/FringePlank/PostProcess/MPC_FringePlank_Interrogate.MPC_FringePlank_Interrogate" "OriginPositionWS" (Math|Conversions|ToLinearColor(Vector) (Variables|Default|GetLocation)) :!class /Script/Engine.KismetMaterialLibrary)
  (Rendering|Material|SetScalarParameterValue 0 "/Game/Weapons/Prototype/FringePlank/PostProcess/MPC_FringePlank_Interrogate.MPC_FringePlank_Interrogate" "Duration" (Variables|Default|GetSweepDuration) :!class /Script/Engine.KismetMaterialLibrary)
  (Rendering|Material|SetScalarParameterValue 0 "/Game/Weapons/Prototype/FringePlank/PostProcess/MPC_FringePlank_Interrogate.MPC_FringePlank_Interrogate" "StencilDuration" (ScalableFloat|GetValueatLevel (Variables|Default|GetStencilDuration)) :!class /Script/Engine.KismetMaterialLibrary)
  (Rendering|Material|SetScalarParameterValue 0 "/Game/Weapons/Prototype/FringePlank/PostProcess/MPC_FringePlank_Interrogate.MPC_FringePlank_Interrogate" "Radius" (Variables|Default|GetSweepRadius) :!class /Script/Engine.KismetMaterialLibrary)
  (Rendering|Material|SetScalarParameterValue 0 "/Game/Weapons/Prototype/FringePlank/PostProcess/MPC_FringePlank_Interrogate.MPC_FringePlank_Interrogate" "ActivatedTimestamp" (Utilities|Time|GetGameTimeinSeconds 0) :!class /Script/Engine.KismetMaterialLibrary))

(event EventEndPlay
  (params (EndPlayReason :type EEndPlayReason%20Enum :default "Destroyed"))
  (Utilities|Time|ClearandInvalidateTimerbyHandle 0 (Variables|Default|GetSweepDelayHandle))
  (Utilities|Time|ClearandInvalidateTimerbyHandle 0 (Variables|Default|GetDelayBeforeSweepHandle))
  (CallFunction|EndSweepFX))

(fn OnExecute
  (params (MyTarget :type Actor%20Object%20Reference) (Parameters :type Gameplay%20Cue%20Parameters%20Structure%20%28by%20ref%29))
  (bind (_normalizedmagnitude _rawmagnitude _effectcontext _matchedtagname _originaltag _aggregatedsourcetags _aggregatedtargettags _location _normal _instigator _effectcauser _sourceobject) (Ability|GameplayCue|BreakGameplayCueParameters Parameters))
  (|CallParentFunction MyTarget Parameters)
  (Variables|Default|SetSweepRadius _rawmagnitude)
  (Variables|Default|SetLocation _location)
  (Utilities|IsValid _sourceobject
    (:Is%20Valid
      (bind _asfort_player_pawn (Utilities|Casting|CastToFortPlayerPawn _sourceobject)
        (:then
          (Variables|Default|SetPlayerPawnAthena _asfort_player_pawn)
          (CallFunction|BeginSweepFX)
          (Audio|SpawnSoundatLocation 0 (Variables|Default|GetSweep%20Sound) (Variables|Default|GetLocation))
          (if (Pawn|IsLocallyControlled (Variables|Default|GetPlayerPawnAthena))
            (Audio|PushSoundMixModifier 0 (Variables|Default|GetRevealSoundMix))
            (else
              _))
          (return true))
        (:CastFailed
          _)))
    (:Is%20Not%20Valid
      _))
  (return false))

(fn ConstructionScript)
