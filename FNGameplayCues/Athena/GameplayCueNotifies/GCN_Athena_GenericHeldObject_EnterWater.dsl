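; Burst gameplay cue notify that plays particle and sound feedback when a generic held object enters water.
; On burst, it sets WaterDepth and WaterBodyIndex on the first Cascade component from the cue's raw and normalized magnitudes, matching the effect to the water surface.
; It also spawns a configurable sound at the cue location and passes the ability level to it as an integer audio parameter.
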(fn OnBurst(Cascade)
  (params (MyTarget :type Actor%20Object%20Reference) (Parameters :type Gameplay%20Cue%20Parameters%20Structure%20%28by%20ref%29) (ParticleComponents :type Array%20of%20Cascade%20Particle%20System%20Component%20Object%20References) (AudioComponents :type Array%20of%20Audio%20Component%20Object%20References) (BurstCameraShakeInstance :type Camera%20Shake%20Base%20Object%20Reference) (BurstDecalInstance :type Decal%20Actor%20Object%20Reference))
  (bind (_normalizedmagnitude _rawmagnitude _effectcontext _matchedtagname _originaltag _aggregatedsourcetags _aggregatedtargettags _location _normal _instigator _effectcauser _sourceobject _physicalmaterial _gameplayeffectlevel _abilitylevel) (Ability|GameplayCue|BreakGameplayCueParameters Parameters))
  (bind _output (Utilities|Array|Get(acopy) ParticleComponents))
  ;[1 1 1] Set the water settings so that it matches the water surface
  (Effects|Components|ParticleSystem|SetFloatParameter _output "WaterDepth" _rawmagnitude)
  (Effects|Components|ParticleSystem|SetFloatParameter _output "WaterBodyIndex" _normalizedmagnitude)
  (bind _returnvalue (Audio|SpawnSoundatLocation 0 (Variables|Default|GetSound%20to%20Play) _location))
  (Audio|Parameter|SetIntegerParameter _returnvalue "AbilityLevel" _abilitylevel :!class /Script/AudioExtensions.AudioParameterControllerInterface)
  (return))
