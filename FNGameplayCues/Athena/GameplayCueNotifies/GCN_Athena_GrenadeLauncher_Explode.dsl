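; Gameplay cue notify for the Athena grenade launcher explosion, adjusting the burst Niagara effect by the surface the projectile stopped on.
; On burst, it casts the target through FortProjectileBase to the grenade launcher projectile parent and reads its stop surface type.
; For SurfaceType8 it swaps the first spawned Niagara component to the water explosion system; otherwise it reduces that component's Global Scale variable.
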(event EventOnBurst(Niagara)
  (params (MyTarget :type Actor%20Object%20Reference) (Parameters :type Gameplay%20Cue%20Parameters%20Structure%20%28by%20ref%29) (ParticleComponents :type Array%20of%20Niagara%20Particle%20System%20Component%20Object%20References) (AudioComponents :type Array%20of%20Audio%20Component%20Object%20References) (BurstCameraShakeInstance :type Camera%20Shake%20Base%20Object%20Reference) (BurstDecalInstance :type Decal%20Actor%20Object%20Reference))
  (bind _output (Utilities|Array|Get(acopy) ParticleComponents))
  (bind _asfort_projectile_base (Utilities|Casting|CastToFortProjectileBase MyTarget))
  (bind _asb_prj_ranged_grenade_launcher_athena_parent (Utilities|Casting|CastToB_Prj_Ranged_GrenadeLauncher_Athena_Parent _asfort_projectile_base))
  (if (Utilities|Enum|Equal(Enum) (Variables|/Game/Athena/Items/Weapons/B_Prj_Ranged_GrenadeLauncher_Athena_Parent.B_Prj_Ranged_GrenadeLauncher_Athena_Parent_C|GetStopSurfaceType _asb_prj_ranged_grenade_launcher_athena_parent) "SurfaceType8")
    (Niagara|SetNiagaraSystemAsset _output "/Game/Effects/Explosion_Update/NS_GrenadeLauncher_Explosion_Asteria_Water.NS_GrenadeLauncher_Explosion_Asteria_Water" true)
    (Niagara|SetNiagaraVariable(Float) _output "Global Scale" 0.6)))
