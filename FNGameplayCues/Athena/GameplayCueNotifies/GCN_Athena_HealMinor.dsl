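; Burst gameplay cue notify for minor heals in Athena, showing either a health or shield version of its Cascade particle effect.
; OnBurst reads the normalized magnitude, which carries the player's health percent so remote clients pick the right version.
; At full health it disables the HealthDecals emitter on the first particle component; otherwise it disables ShieldDecals and Squares.
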(fn OnBurst(Cascade)
  (params (MyTarget :type Actor%20Object%20Reference) (Parameters :type Gameplay%20Cue%20Parameters%20Structure%20%28by%20ref%29) (ParticleComponents :type Array%20of%20Cascade%20Particle%20System%20Component%20Object%20References) (AudioComponents :type Array%20of%20Audio%20Component%20Object%20References) (BurstCameraShakeInstance :type Camera%20Shake%20Base%20Object%20Reference) (BurstDecalInstance :type Decal%20Actor%20Object%20Reference))
  (bind _output (Utilities|Array|Get(acopy) ParticleComponents))
  ;[1 1 1] On burst check the passed in Normalized mag. The mag is the players health percent. If it is 1.0 then we have full health and should play the shield version. This must be passed in to show the correct version to remote clients as they will not know the other players health on their client.
  (if (>= (Ability|GameplayCue|BreakGameplayCueParameters Parameters) 1.0)
    (Effects|Components|ParticleSystem|SetEmitterEnable _output "HealthDecals" false)
    (else
      (Effects|Components|ParticleSystem|SetEmitterEnable _output "ShieldDecals" false)
      (Effects|Components|ParticleSystem|SetEmitterEnable _output "Squares" false)))
  (return))
