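; Burst gameplay cue notify for the rustle of a dumpster hiding prop, using Cascade particle effects.
; On burst, it moves each spawned particle component to the location passed in the cue parameters.
; It then shows a footstep-type sound indicator at the target actor's position, with a limited audible range, for players hostile or neutral to that actor's team.
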(fn OnBurst(Cascade)
  (params (MyTarget :type Actor%20Object%20Reference) (Parameters :type Gameplay%20Cue%20Parameters%20Structure%20%28by%20ref%29) (ParticleComponents :type Array%20of%20Cascade%20Particle%20System%20Component%20Object%20References) (AudioComponents :type Array%20of%20Audio%20Component%20Object%20References) (BurstCameraShakeInstance :type Camera%20Shake%20Base%20Object%20Reference) (BurstDecalInstance :type Decal%20Actor%20Object%20Reference))
  (bind (_normalizedmagnitude _rawmagnitude _effectcontext _matchedtagname _originaltag _aggregatedsourcetags _aggregatedtargettags _location) (Ability|GameplayCue|BreakGameplayCueParameters Parameters))
  (for _array_element ParticleComponents
    (Transformation|SetWorldLocation _array_element _location false false))
  (Fortnite|ShowSoundIndicatorLocalforTeam :WorldContextObject MyTarget :Location (Transformation|GetActorLocation MyTarget) :MaxAudibleDistance 7e+02 :IndicatorType "FootStep" :Team (Team|GetActorTeam MyTarget) :Affiliations (Utilities|Array|MakeArray "Hostile" "Neutral") :Instigator MyTarget :Tint "(R=0.000000,G=0.000000,B=0.000000,A=0.000000)" :AttenuationMaxRangeMultiplier 1.0 :AttenuationIntensityMultiplier 1.0)
  (return))
