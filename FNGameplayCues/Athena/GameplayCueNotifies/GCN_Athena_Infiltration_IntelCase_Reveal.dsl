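; Gameplay cue notify for the Infiltration intel case reveal, playing a world sweep effect centered on the cue target's location.
; On execute it caches that location, pushes a reveal sound mix and sets timers to start and end the sweep.
; The sweep enables post process, writes origin, radius, duration and timestamp to the shared Interrogate material parameter collection, and shakes the camera.
; Ending or EndPlay disables post process and zeroes the origin.
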(event Custom|BeginSweepFX
  (bind _returnvalue (Utilities|Time|SetTimerbyEvent (EventDispatchers|CreateEvent :!event Custom|EndSweepFX) (+ (Variables|Default|GetSweepDuration) (Variables|Default|GetDelayBeforeSweep))))
  (Variables|Default|SetSweepDelayHandle _returnvalue)
  (bind _returnvalue_2 (Utilities|Time|SetTimerbyEvent (EventDispatchers|CreateEvent :!event Custom|PlayWorldSweep) (Variables|Default|GetDelayBeforeSweep)))
  (Variables|Default|SetDelayBeforeSweepHandle _returnvalue_2))

(event Custom|EndSweepFX
  (Variables|/Script/Engine.PostProcessComponent|SetbEnabled false (Variables|Default|GetPostProcess))
  (Rendering|Material|SetVectorParameterValue 0 "/FNGameplayCues/Creative/Items/Gameplay/Interrogate/MPC_Interrogate.MPC_Interrogate" "OriginPositionWS" (Math|Color|MakeColor 0.0 0.0 0.0 0.0) :!class /Script/Engine.KismetMaterialLibrary))

(event Custom|PlayWorldSweep
  (Variables|/Script/Engine.PostProcessComponent|SetbEnabled true (Variables|Default|GetPostProcess))
  (Rendering|Material|SetVectorParameterValue 0 "/FNGameplayCues/Creative/Items/Gameplay/Interrogate/MPC_Interrogate.MPC_Interrogate" "OriginPositionWS" (Math|Conversions|ToLinearColor(Vector) (Variables|Default|GetLocalPlayerLocation)) :!class /Script/Engine.KismetMaterialLibrary)
  (Rendering|Material|SetScalarParameterValue 0 "/FNGameplayCues/Creative/Items/Gameplay/Interrogate/MPC_Interrogate.MPC_Interrogate" "Duration" (Variables|Default|GetSweepDuration) :!class /Script/Engine.KismetMaterialLibrary)
  (Rendering|Material|SetScalarParameterValue 0 "/FNGameplayCues/Creative/Items/Gameplay/Interrogate/MPC_Interrogate.MPC_Interrogate" "Radius" (Variables|Default|GetSweepRadius) :!class /Script/Engine.KismetMaterialLibrary)
  (Rendering|Material|SetScalarParameterValue 0 "/FNGameplayCues/Creative/Items/Gameplay/Interrogate/MPC_Interrogate.MPC_Interrogate" "ActivatedTimestamp" (Utilities|Time|GetGameTimeinSeconds 0) :!class /Script/Engine.KismetMaterialLibrary)
  (Camera|PlayWorldCameraShake 0 "/Game/Effects/Fort_Effects/BluePrints/CameraShakes/B_Small_Vertical_Jolt_CameraShake.B_Small_Vertical_Jolt_CameraShake_C" (Variables|Default|GetLocalPlayerLocation) 5e+02 8e+02))

(event EventEndPlay
  (params (EndPlayReason :type EEndPlayReason%20Enum :default "Destroyed"))
  (Utilities|Time|ClearandInvalidateTimerbyHandle 0 (Variables|Default|GetSweepDelayHandle))
  (Utilities|Time|ClearandInvalidateTimerbyHandle 0 (Variables|Default|GetDelayBeforeSweepHandle))
  (CallFunction|EndSweepFX))

(fn OnExecute
  (params (MyTarget :type Actor%20Object%20Reference) (Parameters :type Gameplay%20Cue%20Parameters%20Structure%20%28by%20ref%29))
  (|CallParentFunction MyTarget Parameters)
  (Utilities|IsValid MyTarget
    (:Is%20Valid
      (Variables|Default|SetLocalPlayerLocation (Transformation|GetActorLocation MyTarget))
      (CallFunction|BeginSweepFX)
      (Audio|PushSoundMixModifier 0 (Variables|Default|GetRevealSoundMix))
      (return true))
    (:Is%20Not%20Valid
      (return false))))

(fn ConstructionScript)
