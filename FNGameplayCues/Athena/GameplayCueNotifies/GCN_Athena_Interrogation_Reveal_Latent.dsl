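; Latent gameplay cue for the Athena interrogation reveal, playing a world sweep effect from the source object's location.
; OnExecute stores that location, starts the sweep timers, then caches the player pawn and pushes a sound mix if locally controlled.
; After a positive delay, the sweep enables post process, feeds MPC_Interrogate and shakes the camera; optional hand particles precede it.
; A second timer or EndPlay clears the effect; debug prints remain.
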(event Custom|BeginSweepFX
  (bind _playerpawnathena (Variables|Default|GetPlayerPawnAthena))
  (bind _delaybeforesweep (Variables|Default|GetDelayBeforeSweep))
  ;[0.02034 1 0.2723] Set Timer events for EndSweepFX and PlayWorldSweep and play particle effects on hands if bool is true
  (if (> _delaybeforesweep 0.0)
    (bind _returnvalue (Utilities|Time|SetTimerbyEvent (EventDispatchers|CreateEvent :!event Custom|EndSweepFX) (+ _delaybeforesweep (ScalableFloat|GetValueatLevel (Variables|Default|GetStencilDuration)))))
    (Variables|Default|SetSweepDelayHandle _returnvalue)
    (bind _returnvalue_2 (Utilities|Time|SetTimerbyEvent (EventDispatchers|CreateEvent :!event Custom|PlayWorldSweep) (Variables|Default|GetDelayBeforeSweep)))
    (Variables|Default|SetDelayBeforeSweepHandle _returnvalue_2)
    (if (Variables|Default|GetbShouldPlayParticleEffectOnHand)
      (Utilities|IsValid _playerpawnathena
        (:Is%20Valid
          (Effects|SpawnEmitterAttached :EmitterTemplate "/FNGameplayCues/Creative/Items/Gameplay/Disguise/P_Creative_Disguise_DeviceDeactivate.P_Creative_Disguise_DeviceDeactivate" :AttachToComponent (Variables|/Script/Engine.Character|GetMesh _playerpawnathena) :AttachPointName "hand_r" :Location (Variables|Default|GetParticleRelativeLocation) :Scale "1.000000,1.000000,1.000000" :LocationType "KeepRelativeOffset" :bAutoDestroy true :PoolingMethod "AutoRelease" :bAutoActivate true))))))

(event Custom|EndSweepFX
  (Variables|/Script/Engine.PostProcessComponent|SetBlendWeight 0.0 (Variables|Default|GetPostProcess))
  (Variables|/Script/Engine.PostProcessComponent|SetbEnabled false (Variables|Default|GetPostProcess))
  (Rendering|Material|SetVectorParameterValue 0 "/FNGameplayCues/Creative/Items/Gameplay/Interrogate/MPC_Interrogate.MPC_Interrogate" "OriginPositionWS" (Math|Color|MakeColor 0.0 0.0 0.0 0.0) :!class /Script/Engine.KismetMaterialLibrary))

(event Custom|PlayWorldSweep
  (Variables|/Script/Engine.PostProcessComponent|SetBlendWeight 1.0 (Variables|Default|GetPostProcess))
  (Variables|/Script/Engine.PostProcessComponent|SetbEnabled true (Variables|Default|GetPostProcess))
  (Development|PrintString 0 "PLAYER
" true true "(R=0.539531,G=1.000000,B=0.000000,A=1.000000)" 7.0)
  (Rendering|Material|SetVectorParameterValue 0 "/FNGameplayCues/Creative/Items/Gameplay/Interrogate/MPC_Interrogate.MPC_Interrogate" "OriginPositionWS" (Math|Conversions|ToLinearColor(Vector) (Variables|Default|GetSourceObjectLocation)) :!class /Script/Engine.KismetMaterialLibrary)
  (Rendering|Material|SetScalarParameterValue 0 "/FNGameplayCues/Creative/Items/Gameplay/Interrogate/MPC_Interrogate.MPC_Interrogate" "Duration" (Variables|Default|GetSweepDuration) :!class /Script/Engine.KismetMaterialLibrary)
  (Rendering|Material|SetScalarParameterValue 0 "/FNGameplayCues/Creative/Items/Gameplay/Interrogate/MPC_Interrogate.MPC_Interrogate" "StencilDuration" (ScalableFloat|GetValueatLevel (Variables|Default|GetStencilDuration)) :!class /Script/Engine.KismetMaterialLibrary)
  (Rendering|Material|SetScalarParameterValue 0 "/FNGameplayCues/Creative/Items/Gameplay/Interrogate/MPC_Interrogate.MPC_Interrogate" "Radius" (Variables|Default|GetSweepRadius) :!class /Script/Engine.KismetMaterialLibrary)
  (Rendering|Material|SetScalarParameterValue 0 "/FNGameplayCues/Creative/Items/Gameplay/Interrogate/MPC_Interrogate.MPC_Interrogate" "ActivatedTimestamp" (Utilities|Time|GetGameTimeinSeconds 0) :!class /Script/Engine.KismetMaterialLibrary)
  (Camera|PlayWorldCameraShake 0 "/Game/Effects/Fort_Effects/BluePrints/CameraShakes/B_Small_Vertical_Jolt_CameraShake.B_Small_Vertical_Jolt_CameraShake_C" (Variables|Default|GetSourceObjectLocation) 5e+02 8e+02))

(event EventEndPlay
  (params (EndPlayReason :type EEndPlayReason%20Enum :default "Destroyed"))
  (Utilities|Time|ClearandInvalidateTimerbyHandle 0 (Variables|Default|GetSweepDelayHandle))
  (Utilities|Time|ClearandInvalidateTimerbyHandle 0 (Variables|Default|GetDelayBeforeSweepHandle))
  (CallFunction|EndSweepFX))

(fn OnExecute
  (params (MyTarget :type Actor%20Object%20Reference) (Parameters :type Gameplay%20Cue%20Parameters%20Structure%20%28by%20ref%29))
  (bind (_normalizedmagnitude _rawmagnitude _effectcontext _matchedtagname _originaltag _aggregatedsourcetags _aggregatedtargettags _location _normal _instigator _effectcauser _sourceobject) (Ability|GameplayCue|BreakGameplayCueParameters Parameters))
  (|CallParentFunction MyTarget Parameters)
  (Development|PrintString MyTarget "Hello" true true "(R=1.000000,G=0.241172,B=0.000000,A=1.000000)" 5.0)
  (Utilities|IsValid _sourceobject
    (:Is%20Valid
      (bind _asactor (Utilities|Casting|CastToActor _sourceobject)
        (:then
          (Variables|Default|SetSourceObjectLocation (Transformation|GetActorLocation _asactor))
          (CallFunction|BeginSweepFX)
          (bind _asplayer_pawn_athena (Utilities|Casting|CastToPlayerPawn_Athena (select (Utilities|IsValid _instigator :!class /Script/Engine.KismetSystemLibrary) _instigator _sourceobject))
            (:then
              (Variables|Default|SetPlayerPawnAthena _asplayer_pawn_athena)
              (if (Pawn|IsLocallyControlled _asplayer_pawn_athena)
                (Audio|PushSoundMixModifier 0 (Variables|Default|GetRevealSoundMix))
                (else
                  _))
              (return true))
            (:CastFailed
              _)))
        (:CastFailed
          _)))
    (:Is%20Not%20Valid
      _))
  (return false))

(fn ConstructionScript)
