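; Latent gameplay cue for the interrogation reveal, playing a stencil-only world sweep from the source player's location.
; OnExecute caches that pawn's location, then sets timers for sweep start and end.
; The sweep enables a post-process and feeds origin, radius, durations and timestamp to MPC_Interrogate_StencilOnly.
; Timeout or EndPlay disables it but zeroes the origin on MPC_Interrogate, a different collection; debug PrintStrings remain.
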(event Custom|BeginSweepFX
  (bind _returnvalue (Utilities|Time|SetTimerbyEvent (EventDispatchers|CreateEvent :!event Custom|EndSweepFX) (+ (ScalableFloat|GetValueatLevel (Variables|Default|GetStencilDuration)) (Variables|Default|GetDelayBeforeSweep))))
  (Variables|Default|SetSweepDelayHandle _returnvalue)
  (bind _returnvalue_2 (Utilities|Time|SetTimerbyEvent (EventDispatchers|CreateEvent :!event Custom|PlayWorldSweep) (Variables|Default|GetDelayBeforeSweep)))
  (Variables|Default|SetDelayBeforeSweepHandle _returnvalue_2))

(event Custom|EndSweepFX
  (bind _postprocess (Variables|Default|GetPostProcess))
  (Variables|/Script/Engine.PostProcessComponent|SetBlendWeight 0.0 _postprocess)
  (Variables|/Script/Engine.PostProcessComponent|SetbEnabled false _postprocess)
  (Rendering|Material|SetVectorParameterValue 0 "/FNGameplayCues/Creative/Items/Gameplay/Interrogate/MPC_Interrogate.MPC_Interrogate" "OriginPositionWS" (Math|Color|MakeColor 0.0 0.0 0.0 0.0) :!class /Script/Engine.KismetMaterialLibrary))

(event Custom|PlayWorldSweep
  (Variables|/Script/Engine.PostProcessComponent|SetbEnabled true (Variables|Default|GetPostProcess))
  (Variables|/Script/Engine.PostProcessComponent|SetBlendWeight 1.0 (Variables|Default|GetPostProcess))
  (Development|PrintString 0 "SQUADMATE" true true "(R=0.000000,G=0.903307,B=1.000000,A=1.000000)" 7.0)
  (Rendering|Material|SetVectorParameterValue 0 "/FNGameplayCues/Creative/Items/Gameplay/Interrogate/MPC_Interrogate_StencilOnly.MPC_Interrogate_StencilOnly" "OriginPositionWS" (Math|Conversions|ToLinearColor(Vector) (Variables|Default|GetLocalPlayerLocation)) :!class /Script/Engine.KismetMaterialLibrary)
  (Rendering|Material|SetScalarParameterValue 0 "/FNGameplayCues/Creative/Items/Gameplay/Interrogate/MPC_Interrogate_StencilOnly.MPC_Interrogate_StencilOnly" "Duration" (Variables|Default|GetSweepDuration) :!class /Script/Engine.KismetMaterialLibrary)
  (Rendering|Material|SetScalarParameterValue 0 "/FNGameplayCues/Creative/Items/Gameplay/Interrogate/MPC_Interrogate_StencilOnly.MPC_Interrogate_StencilOnly" "StencilDuration" (ScalableFloat|GetValueatLevel (Variables|Default|GetStencilDuration)) :!class /Script/Engine.KismetMaterialLibrary)
  (Rendering|Material|SetScalarParameterValue 0 "/FNGameplayCues/Creative/Items/Gameplay/Interrogate/MPC_Interrogate_StencilOnly.MPC_Interrogate_StencilOnly" "Radius" (Variables|Default|GetSweepRadius) :!class /Script/Engine.KismetMaterialLibrary)
  (Rendering|Material|SetScalarParameterValue 0 "/FNGameplayCues/Creative/Items/Gameplay/Interrogate/MPC_Interrogate_StencilOnly.MPC_Interrogate_StencilOnly" "ActivatedTimestamp" (Utilities|Time|GetGameTimeinSeconds 0) :!class /Script/Engine.KismetMaterialLibrary))

(event EventEndPlay
  (params (EndPlayReason :type EEndPlayReason%20Enum :default "Destroyed"))
  (Utilities|Time|ClearandInvalidateTimerbyHandle 0 (Variables|Default|GetSweepDelayHandle))
  (Utilities|Time|ClearandInvalidateTimerbyHandle 0 (Variables|Default|GetDelayBeforeSweepHandle))
  (CallFunction|EndSweepFX))

(fn OnExecute
  (params (MyTarget :type Actor%20Object%20Reference) (Parameters :type Gameplay%20Cue%20Parameters%20Structure%20%28by%20ref%29))
  (bind (_normalizedmagnitude _rawmagnitude _effectcontext _matchedtagname _originaltag _aggregatedsourcetags _aggregatedtargettags _location _normal _instigator _effectcauser _sourceobject) (Ability|GameplayCue|BreakGameplayCueParameters Parameters))
  (|CallParentFunction MyTarget Parameters)
  (Development|PrintString MyTarget "Hello" true true "(R=1.000000,G=0.241172,B=0.000000,A=1.000000)" 5.0)
  (Utilities|IsValid _sourceobject
    (:Is%20Valid
      (bind _asplayer_pawn_athena (Utilities|Casting|CastToPlayerPawn_Athena _sourceobject)
        (:then
          (bind _output_get (Variables|Default|SetPlayerPawnAthena _asplayer_pawn_athena))
          (Variables|Default|SetLocalPlayerLocation (Transformation|GetActorLocation _output_get))
          (CallFunction|BeginSweepFX)
          (return true))
        (:CastFailed
          _)))
    (:Is%20Not%20Valid
      _))
  (return false))

(fn ConstructionScript)
