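; Burst gameplay cue notify for the end of the Athena DishFries item, playing a disguise-deactivation particle effect on the affected player.
; On burst, it casts the target to an Athena player pawn and spawns the Creative disguise-deactivated Cascade emitter attached to that pawn's mesh, auto-destroyed and pooled.
; The pawn is then set as the emitter's BoneSocketActor parameter so the particles can reference its skeleton.
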(fn OnBurst(Cascade)
  (params (MyTarget :type Actor%20Object%20Reference) (Parameters :type Gameplay%20Cue%20Parameters%20Structure%20%28by%20ref%29) (ParticleComponents :type Array%20of%20Cascade%20Particle%20System%20Component%20Object%20References) (AudioComponents :type Array%20of%20Audio%20Component%20Object%20References) (BurstCameraShakeInstance :type Camera%20Shake%20Base%20Object%20Reference) (BurstDecalInstance :type Decal%20Actor%20Object%20Reference))
  (bind _asplayer_pawn_athena (Utilities|Casting|CastToPlayerPawn_Athena MyTarget))
  (bind _returnvalue (Effects|SpawnEmitterAttached :EmitterTemplate "/Game/Creative/Items/Gameplay/Disguise/P_Creative_Disguise_Deactivated.P_Creative_Disguise_Deactivated" :AttachToComponent (Variables|/Script/Engine.Character|GetMesh _asplayer_pawn_athena) :AttachPointName "None" :Scale "1.000000,1.000000,1.000000" :LocationType "KeepRelativeOffset" :bAutoDestroy true :PoolingMethod "AutoRelease" :bAutoActivate true))
  (Effects|Components|ParticleSystem|SetActorParameter _returnvalue "BoneSocketActor" _asplayer_pawn_athena)
  (return))
