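; Gameplay cue notify that plays the burst effects when the SuperDingo item actor spawns.
; On burst, it casts the target to BGA_Athena_SuperDingo and calls the material VFX spawn function from its Dingo parent class.
; It then attaches the first burst Cascade particle component to that actor's mesh at the FX_Root socket, snapping location to the socket.
; A debug variable gates a print marking the burst.
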(fn OnBurst(Cascade)
  (params (MyTarget :type Actor%20Object%20Reference) (Parameters :type Gameplay%20Cue%20Parameters%20Structure%20%28by%20ref%29) (ParticleComponents :type Array%20of%20Cascade%20Particle%20System%20Component%20Object%20References) (AudioComponents :type Array%20of%20Audio%20Component%20Object%20References) (BurstCameraShakeInstance :type Camera%20Shake%20Base%20Object%20Reference) (BurstDecalInstance :type Decal%20Actor%20Object%20Reference))
  (bind _asbga_athena_super_dingo (Utilities|Casting|CastToBGA_Athena_SuperDingo MyTarget))
  (Class|BGAAthenaDingoParent|SpawnMaterialVFX _asbga_athena_super_dingo)
  (Development|PrintString 0 "*****Spawned_BURST_FIRED*****" (Variables|Default|GetDebug) (Variables|Default|GetDebug))
  (Transformation|AttachComponentToComponent (Utilities|Array|Get(acopy) ParticleComponents) (Variables|/Game/Athena/Items/Consumables/SuperDingo/BGA_Athena_SuperDingo.BGA_Athena_SuperDingo_C|GetMesh _asbga_athena_super_dingo) "FX_Root" "SnapToTarget" "KeepRelative" "KeepRelative" true)
  (return))
