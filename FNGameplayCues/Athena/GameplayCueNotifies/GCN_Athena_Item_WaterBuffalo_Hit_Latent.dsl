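; Gameplay cue notify for the WaterBuffalo item hit that makes the target player pawn look wet, then dries it.
; On the Executed event it stores the pawn and sets the "Wetness" material scalar to 1 on the skinned meshes of each custom part type.
; After a configurable delay, the DryingTime timeline lerps that scalar back toward 0 across those part meshes on each update.
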(event GameplayCueNotify|EventHandleGameplayCue
  (params (MyTarget :type Actor%20Object%20Reference) (EventType :type EGameplayCueEvent%20Enum :default "OnActive") (Parameters :type Gameplay%20Cue%20Parameters%20Structure%20%28by%20ref%29))
  (Utilities|IsValid MyTarget
    (:Is%20Valid
      (bind _asplayer_pawn_athena (Utilities|Casting|CastToPlayerPawn_Athena MyTarget))
      (Variables|Default|SetTarget%20Pawn _asplayer_pawn_athena)
      (switch Utilities|FlowControl|Switch|SwitchonEGameplayCueEvent EventType
        (:Executed
          (CallFunction|OnBurst))))))

(event Custom|On%20Burst
  (bind _target_pawn (Variables|Default|GetTarget%20Pawn))
  (bind _enumvalue (Utilities|Enum|ForEachEFortCustomPartType false)
    (:LoopBody
      (Utilities|IsValid _target_pawn
        (:Is%20Valid
          (for _array_element (Customization|GetSkinnedMeshComponentsforPartType _target_pawn _enumvalue)
            (Utilities|IsValid _array_element
              (:Is%20Valid
                (Rendering|Material|SetScalarParameterValueonMaterials _array_element "Wetness" 1.0)))))))
    (:then
      (Utilities|FlowControl|Delay 0 (Variables|Default|GetDelay))
      (CallFunction|Removal))))

(event Custom|Removal
  (bind _target_pawn (Variables|Default|GetTarget%20Pawn))
  (bind (_direction _lerp_controller) (Timeline|DryingTime PlayFromStart)
    (:Update
      (bind _enumvalue (Utilities|Enum|ForEachEFortCustomPartType true)
        (:LoopBody
          (Utilities|IsValid _target_pawn
            (:Is%20Valid
              (for _array_element (Customization|GetSkinnedMeshComponentsforPartType _target_pawn _enumvalue)
                (Utilities|IsValid _array_element
                  (:Is%20Valid
                    (Rendering|Material|SetScalarParameterValueonMaterials _array_element "Wetness" (Math|Float|Lerp 1.0 0.0 _lerp_controller))))))))))))

(fn ConstructionScript)
