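; Burst gameplay cue for the Ostrich mech vehicle's dash, driving its material and particle effects.
; On burst, it writes the current game time to the mech mesh's DashGameTime and BoostTime material parameters for time-based effects like engine afterglow.
; It then snaps each valid burst particle component onto thigh, foot and light sockets, chosen by array index, so particle order matters.
; A separate MeterFlare function spawns an eye boost emitter, with no caller shown here.
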(fn MeterFlare
  (Effects|SpawnEmitterAttached "/Game/Effects/Fort_Effects/Gameplay/Ostrich/P_Ostrich_Boost_Eyes.P_Ostrich_Boost_Eyes" (Variables|Default|GetMechMesh) "FX_Startup_Thigh_L"))

(fn OnBurst(Cascade)
  (params (MyTarget :type Actor%20Object%20Reference) (Parameters :type Gameplay%20Cue%20Parameters%20Structure%20%28by%20ref%29) (ParticleComponents :type Array%20of%20Cascade%20Particle%20System%20Component%20Object%20References) (AudioComponents :type Array%20of%20Audio%20Component%20Object%20References) (BurstCameraShakeInstance :type Camera%20Shake%20Base%20Object%20Reference) (BurstDecalInstance :type Decal%20Actor%20Object%20Reference))
  (bind _returnvalue (Utilities|Time|GetGameTimeinSeconds 0))
  (bind _asfort_character_vehicle_ostrich (Utilities|Casting|CastToFortCharacterVehicle_Ostrich MyTarget))
  (Utilities|FlowControl|Sequence
    (:then_0
      (Rendering|Material|SetScalarParameterValueonMaterials (Mesh|GetMechMesh _asfort_character_vehicle_ostrich) "DashGameTime" _returnvalue)
      ;[1 1 1] Engine afterglow
      (Rendering|Material|SetScalarParameterValueonMaterials (Mesh|GetMechMesh _asfort_character_vehicle_ostrich) "BoostTime" _returnvalue))
    (:then_1
      (for (_array_element _array_index) ParticleComponents
        (Utilities|IsValid _array_element
          (:Is%20Valid
            (Transformation|AttachComponentToComponent _array_element (Mesh|GetMechMesh _asfort_character_vehicle_ostrich) (Utilities|Select :Index _array_index :Option%200 "FX_Boost_Thigh_L" :Option%201 "FX_Boost_Thigh_R" :Option%202 "FX_Boost_Foot_L" :Option%203 "FX_Boost_Foot_R" :Option%204 "FX_StartUp_Light_L" :Option%205 "FX_Foot_Drag_L" :Option%206 "FX_Foot_Drag_R" :Option%207 "FX_Boost_Thigh_L" :Option%208 "FX_Boost_Thigh_R" :Option%209 "FX_Boost_Foot_L" :Option%2010 "FX_Boost_Foot_R" :Option%2011 "FX_Boost_Thigh_L" :Option%2012 "FX_Boost_Thigh_R" :Option%2013 "FX_Boost_Foot_L" :Option%2014 "FX_Boost_Foot_R") "SnapToTarget" "SnapToTarget" "KeepWorld" true))))
      (return))))
