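; Gameplay cue notify for the Ostrich vehicle explosion, which can swap its burst sounds based on how the vehicle detonates.
; ShouldUseBurstOverrides passes only when the target is an Ostrich in SelfDestruct or Instant detonation state and the override sound list is non-empty.
; When it passes, the Cascade burst handler replaces each audio component's sound with the override at the same index, skipping indices without one.
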(fn OnBurst(Cascade)
  (params (MyTarget :type Actor%20Object%20Reference) (Parameters :type Gameplay%20Cue%20Parameters%20Structure%20%28by%20ref%29) (ParticleComponents :type Array%20of%20Cascade%20Particle%20System%20Component%20Object%20References) (AudioComponents :type Array%20of%20Audio%20Component%20Object%20References) (BurstCameraShakeInstance :type Camera%20Shake%20Base%20Object%20Reference) (BurstDecalInstance :type Decal%20Actor%20Object%20Reference))
  (if (CallFunction|ShouldUseBurstOverrides MyTarget)
    (for (_array_element _array_index) AudioComponents
      (if (Utilities|Array|IsValidIndex (Variables|Default|GetOverrideSoundCues) _array_index)
        (Audio|Components|Audio|SetSound _array_element (Utilities|Array|Get(acopy) (Variables|Default|GetOverrideSoundCues) _array_index))))
    (return)
    (else
      (return))))

(fn ShouldUseBurstOverrides
  (params (Actor :type Actor%20Object%20Reference))
  (bind _asfort_character_vehicle_ostrich (Utilities|Casting|CastToFortCharacterVehicle_Ostrich Actor)
    (:then
      (return (and (or (Utilities|Enum|Equal(Enum) (Variables|/Script/FortniteGame.FortCharacterVehicle_Ostrich|GetDetonationState _asfort_character_vehicle_ostrich) "SelfDestruct") (Utilities|Enum|Equal(Enum) (Variables|/Script/FortniteGame.FortCharacterVehicle_Ostrich|GetDetonationState _asfort_character_vehicle_ostrich) "Instant")) (Math|Conversions|ToBoolean(Integer) (Utilities|Array|Length (Variables|Default|GetOverrideSoundCues))))))
    (:CastFailed
      (return false))))
