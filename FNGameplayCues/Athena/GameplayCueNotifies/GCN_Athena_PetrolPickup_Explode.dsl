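; Burst gameplay cue notify that provides the local feedback when a petrol pickup explodes.
; On burst it spawns a downward-facing scorch decal at the target that fades out, and shows a gunshot-type sound indicator for hostile, neutral and friendly affiliations.
; It also tunes one of the burst's particle components through parameters for scale, intensity, noise and a dark smoke color.
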(fn OnBurst(EitherVFX-preferthisversion)
  (params (MyTarget :type Actor%20Object%20Reference) (Parameters :type Gameplay%20Cue%20Parameters%20Structure%20%28by%20ref%29) (ParticleComponents :type Array%20of%20FXSystem%20Component%20Object%20References) (AudioComponents :type Array%20of%20Audio%20Component%20Object%20References) (BurstCameraShakeInstance :type Camera%20Shake%20Base%20Object%20Reference) (BurstDecalInstance :type Decal%20Actor%20Object%20Reference))
  (bind _output (Utilities|Array|Get(acopy) ParticleComponents))
  (bind _returnvalue (Rendering|Decal|SpawnDecalatLocation 0 "/FNGameplayCues/Athena/Prototype/Blueprints/Cube/FX/MD_Decal_Scorch_NoEmissive_CanFade.MD_Decal_Scorch_NoEmissive_CanFade" "100.000000,750.000000,750.000000" (Transformation|GetActorLocation MyTarget) (Math|Conversions|RotationFromXVector (* "0, 0,1.000000" -1.0)) 5.0))
  (Rendering|Components|Decal|SetFadeOut _returnvalue 2.0 2.0 true)
  (Fortnite|ShowSoundIndicatorLocalforTeam :WorldContextObject MyTarget :Location (Transformation|GetActorLocation MyTarget) :MaxAudibleDistance 2.5e+03 :IndicatorType "Gunshot" :Team (Team|GetActorTeam MyTarget) :Affiliations (Utilities|Array|MakeArray "Hostile" "Neutral" "Friendly") :Instigator MyTarget :Tint "(R=0.000000,G=0.000000,B=0.000000,A=0.000000)" :AttenuationMaxRangeMultiplier 1.0 :AttenuationIntensityMultiplier 1.0)
  (Effects|Components|ParticleSystem|SetFloatParameter _output "Global Scale" 0.75)
  (Effects|Components|ParticleSystem|SetFloatParameter _output "Initial High Intensity" 0.5)
  (Effects|Components|ParticleSystem|SetFloatParameter _output "Main Noise Scaling" 1.05)
  (Effects|Components|ParticleSystem|SetColorParameter _output "Explosion Smoke Color" "(R=0.078125,G=0.078125,B=0.078125,A=1.000000)")
  (return))
