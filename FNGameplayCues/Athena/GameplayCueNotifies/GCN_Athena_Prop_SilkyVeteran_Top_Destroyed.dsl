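; Burst gameplay cue notify for the destroyed top of the SilkyVeteran prop, which positions the cue's Cascade particle effects on the target actor.
; OnBurst offsets the target's location horizontally along its forward vector by XYOffset and vertically by ZOffset.
; It moves every spawned particle component to that point and sets its relative rotation from the Roll, Pitch and Yaw variables.
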(fn OnBurst(Cascade)
  (params (MyTarget :type Actor%20Object%20Reference) (Parameters :type Gameplay%20Cue%20Parameters%20Structure%20%28by%20ref%29) (ParticleComponents :type Array%20of%20Cascade%20Particle%20System%20Component%20Object%20References) (AudioComponents :type Array%20of%20Audio%20Component%20Object%20References) (BurstCameraShakeInstance :type Camera%20Shake%20Base%20Object%20Reference) (BurstDecalInstance :type Decal%20Actor%20Object%20Reference))
  (bind (_returnvalue_x _returnvalue_y) (Math|Vector|BreakVector (* (Transformation|GetActorForwardVector MyTarget) (Variables|Default|GetXYOffset))))
  (bind (_returnvalue_x_2 _returnvalue_y_2 _returnvalue_z) (Math|Vector|BreakVector (Transformation|GetActorLocation MyTarget)))
  (for _array_element ParticleComponents
    (Transformation|SetWorldLocation _array_element (Math|Vector|MakeVector (+ _returnvalue_x_2 _returnvalue_x) (+ _returnvalue_y_2 _returnvalue_y) (+ _returnvalue_z (Variables|Default|GetZOffset))) false false)
    (Transformation|SetRelativeRotation _array_element (Math|Rotator|MakeRotator (Variables|Default|GetRoll) (Variables|Default|GetPitch) (Variables|Default|GetYaw)) false false))
  (return))
