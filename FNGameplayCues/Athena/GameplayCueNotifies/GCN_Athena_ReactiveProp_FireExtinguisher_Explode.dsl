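; Burst gameplay cue for an exploding fire extinguisher reactive prop, adding a smoke mark and a sound indicator.
; OnBurst spawns a smoked-poof decal at the target's location, oriented from a constant zero vector rather than a hit normal, then fades it out after a short delay.
; It also shows a local gunshot-type sound indicator there, using the target's team with hostile, neutral and friendly affiliations.
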(fn OnBurst(Cascade)
  (params (MyTarget :type Actor%20Object%20Reference) (Parameters :type Gameplay%20Cue%20Parameters%20Structure%20%28by%20ref%29) (ParticleComponents :type Array%20of%20Cascade%20Particle%20System%20Component%20Object%20References) (AudioComponents :type Array%20of%20Audio%20Component%20Object%20References) (BurstCameraShakeInstance :type Camera%20Shake%20Base%20Object%20Reference) (BurstDecalInstance :type Decal%20Actor%20Object%20Reference))
  (bind _returnvalue (Rendering|Decal|SpawnDecalatLocation 0 "/FNGameplayCues/Creative/Test/Props/MD_Decal_Smoked_Poof.MD_Decal_Smoked_Poof" "1.000000,1.000000,1.000000" (Transformation|GetActorLocation MyTarget) (Math|Conversions|RotationFromXVector (* "0, 0, 0" -1.0))))
  (Rendering|Components|Decal|SetFadeOut _returnvalue 1.5 0.6 true)
  (Fortnite|ShowSoundIndicatorLocalforTeam :WorldContextObject MyTarget :Location (Transformation|GetActorLocation MyTarget) :MaxAudibleDistance 0.0 :IndicatorType "Gunshot" :Team (Team|GetActorTeam MyTarget) :Affiliations (Utilities|Array|MakeArray "Hostile" "Neutral" "Friendly") :Tint "(R=0.000000,G=0.000000,B=0.000000,A=0.000000)" :AttenuationMaxRangeMultiplier 1.0 :AttenuationIntensityMultiplier 1.0)
  (return))
