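; Burst gameplay cue notify that gives player feedback for a Rune Vent, apparently when the pawn is launched.
; On burst it casts the target to a Fort player pawn and plays the Rune Vent camera shake on that pawn's controller in camera-local space.
; It then attaches the DadBro vent launch sound to the pawn's root component.
; A debug "GC" print string runs first.
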(fn OnBurst(Cascade)
  (params (MyTarget :type Actor%20Object%20Reference) (Parameters :type Gameplay%20Cue%20Parameters%20Structure%20%28by%20ref%29) (ParticleComponents :type Array%20of%20Cascade%20Particle%20System%20Component%20Object%20References) (AudioComponents :type Array%20of%20Audio%20Component%20Object%20References) (BurstCameraShakeInstance :type Camera%20Shake%20Base%20Object%20Reference) (BurstDecalInstance :type Decal%20Actor%20Object%20Reference))
  (Development|PrintString 0 "GC")
  (bind _asfort_player_pawn (Utilities|Casting|CastToFortPlayerPawn MyTarget))
  (Game|Feedback|ClientStartCameraShake :self (Fortnite|GetFortPlayerControllerfromActor _asfort_player_pawn) :Shake "/Game/Effects/Fort_Effects/Camera_Effects/BP_Camera_Shake_RuneVent.BP_Camera_Shake_RuneVent_C" :Scale 1.0 :PlaySpace "CameraLocal")
  (Audio|SpawnSoundAttached "/FNGameplayCues/Sounds/Apollo/Events/Dadbro/DadBro_Vent_Launch_Cue.DadBro_Vent_Launch_Cue" (Variables|/Script/Engine.Actor|GetRootComponent _asfort_player_pawn))
  (return))
