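; Gameplay cue notify for the SCMachine (Figment) activation burst, adjusting its Niagara effect for enhanced respawns.
; OnBurst first runs the parent burst logic, which supplies the spawned particle components.
; If the aggregated source tags contain SCMachine.Respawn.EnhancedResurrection, it casts a spawned component to Niagara and sets the "Variant" int parameter to switch the effect's look.
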(fn OnBurst(EitherVFX-preferthisversion)
  (params (MyTarget :type Actor%20Object%20Reference) (Parameters :type Gameplay%20Cue%20Parameters%20Structure%20%28by%20ref%29) (ParticleComponents :type Array%20of%20FXSystem%20Component%20Object%20References) (AudioComponents :type Array%20of%20Audio%20Component%20Object%20References) (BurstCameraShakeInstance :type Camera%20Shake%20Base%20Object%20Reference) (BurstDecalInstance :type Decal%20Actor%20Object%20Reference))
  (bind (_normalizedmagnitude _rawmagnitude _effectcontext _matchedtagname _originaltag _aggregatedsourcetags) (Ability|GameplayCue|BreakGameplayCueParameters Parameters))
  ;[0.01432 0.15 0.07752 0.5] OnBurst: If an enhanced respawn, update the variant param on the spawned niagara effect.
  (|CallParentFunction MyTarget Parameters ParticleComponents AudioComponents BurstCameraShakeInstance BurstDecalInstance)
  (Utilities|FlowControl|Sequence
    (:then_0
      (if (and (Utilities|Array|IsValidIndex ParticleComponents) (GameplayTags|HasTag _aggregatedsourcetags "(TagName=\"SCMachine.Respawn.EnhancedResurrection\")" true))
        (bind _asniagara_particle_system_component (Utilities|Casting|CastToNiagaraComponent (Utilities|Array|Get(acopy) ParticleComponents)))
        (Niagara|SetNiagaraVariable(Int32) _asniagara_particle_system_component "Variant" 1)))
    (:then_1
      (return))))
