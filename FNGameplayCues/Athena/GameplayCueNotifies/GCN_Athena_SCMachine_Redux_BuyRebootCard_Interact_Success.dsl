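; Burst gameplay cue that plays visual and audio feedback when a reboot card is successfully bought at the Reboot Van.
; After the parent burst logic, it casts the target to the Redux SCMachine actor.
; It spawns a rift Niagara system and sound at the machine's location plus a configurable offset.
; If the machine's Van component is valid, it also attaches a light-flash particle emitter to it.
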(event EventOnBurst(EitherVFX-preferthisversion)
  (params (MyTarget :type Actor%20Object%20Reference) (Parameters :type Gameplay%20Cue%20Parameters%20Structure%20%28by%20ref%29) (ParticleComponents :type Array%20of%20FXSystem%20Component%20Object%20References) (AudioComponents :type Array%20of%20Audio%20Component%20Object%20References) (BurstCameraShakeInstance :type Camera%20Shake%20Base%20Object%20Reference) (BurstDecalInstance :type Decal%20Actor%20Object%20Reference))
  (bind _riftfxworldoffset (Variables|Setting|GetRiftFXWorldOffset))
  ;[0.15 0.06666 0.145 0.5] OnBurst: Execute visual and audio feedback for the successful buy reboot card interaction on the reboot van.
  ; Once this is solidified, it could be moved to the gameplay cue's defaults based on what works best for the FX team.
  (|CallParentFunction MyTarget Parameters ParticleComponents AudioComponents BurstCameraShakeInstance BurstDecalInstance)
  (bind _asbga_athena_scmachine_redux (Utilities|Casting|CastToBGA_Athena_SCMachine_Redux MyTarget))
  (Niagara|SpawnSystematLocation MyTarget (Variables|Setting|GetRiftNiagaraSystem) (+ (Transformation|GetActorLocation _asbga_athena_scmachine_redux) _riftfxworldoffset))
  (Audio|PlaySoundatLocation MyTarget (Variables|Setting|GetRiftSoundCue) (+ (Transformation|GetActorLocation _asbga_athena_scmachine_redux) _riftfxworldoffset))
  (bind _van (Variables|/Game/Athena/Items/EnvironmentalItems/SCMachine/BGA_Athena_SCMachine.BGA_Athena_SCMachine_C|GetVan _asbga_athena_scmachine_redux)
    (:then
      (Effects|SpawnEmitterAttached :EmitterTemplate (Variables|Setting|GetVanLightFlashParticleSystem) :AttachToComponent _van :AttachPointName "None" :Scale "1.000000,1.000000,1.000000" :LocationType "SnapToTargetIncludingScale" :bAutoDestroy true :PoolingMethod "None" :bAutoActivate true))
    (:else)))

(fn ConstructionScript)
