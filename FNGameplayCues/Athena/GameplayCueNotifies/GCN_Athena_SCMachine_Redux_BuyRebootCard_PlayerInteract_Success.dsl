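; Gameplay cue burst notify for a player successfully buying a reboot card at the SC Machine.
; On burst it runs the parent logic, then casts the target to a Fort player pawn.
; If the cast succeeds, it attaches an auto-released pooled card-collect particle effect to the root of the pawn's mesh.
; It then plays the machine's horn alert sound at the notify actor's own location, not explicitly at the pawn's.
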(event EventOnBurst(EitherVFX-preferthisversion)
  (params (MyTarget :type Actor%20Object%20Reference) (Parameters :type Gameplay%20Cue%20Parameters%20Structure%20%28by%20ref%29) (ParticleComponents :type Array%20of%20FXSystem%20Component%20Object%20References) (AudioComponents :type Array%20of%20Audio%20Component%20Object%20References) (BurstCameraShakeInstance :type Camera%20Shake%20Base%20Object%20Reference) (BurstDecalInstance :type Decal%20Actor%20Object%20Reference))
  (|CallParentFunction MyTarget Parameters ParticleComponents AudioComponents BurstCameraShakeInstance BurstDecalInstance)
  (bind _asfort_player_pawn (Utilities|Casting|CastToFortPlayerPawn MyTarget))
  (bind _mesh (Variables|/Script/Engine.Character|GetMesh _asfort_player_pawn)
    (:then
      (Effects|SpawnEmitterAttached :EmitterTemplate "/FNGameplayCues/Athena/Items/EnvironmentalItems/SCMachine/FX/P_SCMachine_Card_Collect_Shared.P_SCMachine_Card_Collect_Shared" :AttachToComponent _mesh :AttachPointName "root" :Scale "1.000000,1.000000,1.000000" :LocationType "SnapToTarget" :bAutoDestroy true :PoolingMethod "AutoRelease" :bAutoActivate true)
      (Audio|PlaySoundatLocation (Variables|Default|GetSC_Machine_Card_BuyBack) "/Game/Athena/Sounds/SC_Machine/SC_Machine_Horn_Alert_Cue.SC_Machine_Horn_Alert_Cue" (Transformation|GetActorLocation)))
    (:else)))

(fn ConstructionScript)
