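; Burst gameplay cue for a snow buildup event, playing quake-like audio, camera shake and overhead dust around an epicenter.
; The epicenter comes from the effect causer or a default, and the target's horizontal distance (followed pawn in replays) drives curve-based volume and low-pass filtering.
; An upward trace detects cover, adding an indoor sound layer and, nearer the epicenter, falling smoke.
; An ambient volcano snow sound takes its stage from raw magnitude.
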(fn OnBurst(Cascade)
  (params (MyTarget :type Actor%20Object%20Reference) (Parameters :type Gameplay%20Cue%20Parameters%20Structure%20%28by%20ref%29) (ParticleComponents :type Array%20of%20Cascade%20Particle%20System%20Component%20Object%20References) (AudioComponents :type Array%20of%20Audio%20Component%20Object%20References) (BurstCameraShakeInstance :type Camera%20Shake%20Base%20Object%20Reference) (BurstDecalInstance :type Decal%20Actor%20Object%20Reference))
  (bind _localepicenter (LocalVariables|GetLocalEpicenter))
  (bind _targetactor (LocalVariables|GetTargetActor))
  (bind _returnvalue (Transformation|GetActorLocation _targetactor))
  (bind _returnvalue_2 (- _returnvalue _localepicenter))
  (bind _returnvalue_3 (Math|Vector|VectorLengthXY _returnvalue_2))
  (bind _returnvalue_4 (Math|Float|MapRangeClamped _returnvalue_3 5e+03 9e+04 0.0 1.0))
  (bind _returnvalue_5 (- 1.0 _returnvalue_4))
  (bind _audiointensitycurve_volume (Variables|Default|GetAudioIntensityCurve_Volume))
  (bind _returnvalue_6 (Math|Curves|GetFloatValue _audiointensitycurve_volume _returnvalue_5))
  (bind _returnvalue_7 (< _returnvalue_3 4.2e+05))
  (bind (_normalizedmagnitude _rawmagnitude _effectcontext _matchedtagname _originaltag _aggregatedsourcetags _aggregatedtargettags _location _normal _instigator _effectcauser) (Ability|GameplayCue|BreakGameplayCueParameters Parameters))
  (Utilities|FlowControl|Sequence
    (:then_0
      (Utilities|IsValid _effectcauser
        (:Is%20Valid
          (LocalVariables|SetLocalEpicenter (Transformation|GetActorLocation _effectcauser)))
        (:Is%20Not%20Valid
          (LocalVariables|SetLocalEpicenter (Variables|Default|GetEpicenter))))
      (if (Replay|GetIsPlayingReplay MyTarget)
        (bind _asfort_replay_spectator (Utilities|Casting|CastToFortReplaySpectator (Fortnite|GetFortPlayerControllerfromActor MyTarget)))
        (LocalVariables|SetTargetActor (PlayerState|GetCurrentPawn (Class|FortPlayerControllerSpectating|GetFollowedPlayer _asfort_replay_spectator)))
        (else
          (LocalVariables|SetTargetActor MyTarget)))
      (bind _targetactor_2 (LocalVariables|GetTargetActor)
        (:then
          (bind (_outhit _returnvalue_8) (Collision|LineTraceByChannel _targetactor_2 (+ (Transformation|GetActorLocation _targetactor_2) "0, 0,100.000000") (+ (Transformation|GetActorLocation _targetactor_2) "0, 0,600.000000")))
          (Utilities|FlowControl|Sequence
            (:then_0
              (if _returnvalue_7
                ;[1 1 1] outside audio
                (bind _returnvalue_9 (Audio|SpawnSound2D (LocalVariables|GetTargetActor) "/FNGameplayCues/Athena/Sounds/Events/Cane/Cane_Cue.Cane_Cue"))
                (Audio|Components|Audio|SetLowPassFilterEnabled _returnvalue_9 true :!class /Script/Engine.AudioComponent)
                (Audio|Components|Audio|SetLowPassFilterFrequency _returnvalue_9 (* (Math|Curves|GetFloatValue (Variables|Default|GetAudioIntensityCurve_LPF) _returnvalue_5) 2e+04) :!class /Script/Engine.AudioComponent)
                (Audio|Parameter|SetFloatParameter _returnvalue_9 "Intensity" _returnvalue_6 :!class /Script/AudioExtensions.AudioParameterControllerInterface)))
            (:then_1
              (if (and _returnvalue_7 _returnvalue_8)
                ;[1 1 1] inside audio
                (bind _returnvalue_10 (Audio|SpawnSound2D (LocalVariables|GetTargetActor) "/FNGameplayCues/Athena/Sounds/Events/Cane/Cane_Inside_Cue.Cane_Inside_Cue"))
                (Audio|Parameter|SetFloatParameter _returnvalue_10 "Intensity" _returnvalue_6 :!class /Script/AudioExtensions.AudioParameterControllerInterface)))
            (:then_2
              (Camera|PlayWorldCameraShake 0 "/FNGameplayCues/Effects/Fort_Effects/Camera_Effects/B_Generic_Quake_CameraShake.B_Generic_Quake_CameraShake_C" (LocalVariables|GetLocalEpicenter) 1e+04 2.2e+05 2.0)
              (if _returnvalue_8
                (if (not (> _returnvalue_3 9e+04))
                  (Effects|SpawnEmitteratLocation 0 "/FNGameplayCues/Athena/Events/Earthquake/P_Events_Earthquake_Smokedrop.P_Events_Earthquake_Smokedrop" (.Location (Collision|BreakHitResult _outhit))))))))
        (:else)))
    (:then_1
      (bind _returnvalue_11 (Audio|SpawnSoundatLocation 0 "/FNGameplayCues/Athena/Sounds/Ambient/Volcano/Amb_Volcano_Snow_Cue.Amb_Volcano_Snow_Cue" (Variables|Default|GetEpicenter)))
      (Audio|Parameter|SetIntegerParameter _returnvalue_11 "Stage" (Math|Float|Truncate _rawmagnitude) :!class /Script/AudioExtensions.AudioParameterControllerInterface))
    (:then_2
      (return))))
