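; Gameplay cue burst notify for Athena's "Special Exit" event, spawning a water wall interaction effect on the player.
; On burst it casts the target to FortPlayerPawn and spawns the NS_WaterWall_Interaction Niagara system, unattached, at the pawn root's world location.
; Rotation is a look-at from that location to the same location scaled by a negative factor, so the effect points toward and past the world origin.
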(fn OnBurst(Cascade)
  (params (MyTarget :type Actor%20Object%20Reference) (Parameters :type Gameplay%20Cue%20Parameters%20Structure%20%28by%20ref%29) (ParticleComponents :type Array%20of%20Cascade%20Particle%20System%20Component%20Object%20References) (AudioComponents :type Array%20of%20Audio%20Component%20Object%20References) (BurstCameraShakeInstance :type Camera%20Shake%20Base%20Object%20Reference) (BurstDecalInstance :type Decal%20Actor%20Object%20Reference))
  (bind _asfort_player_pawn (Utilities|Casting|CastToFortPlayerPawn MyTarget))
  (Niagara|SpawnSystematLocation (Variables|/Script/Engine.Actor|GetRootComponent _asfort_player_pawn) "/FNGameplayCues/Effects/Niagara/Systems/Enviroment/Water/NS_WaterWall_Interaction.NS_WaterWall_Interaction" (Transformation|GetWorldLocation (Variables|/Script/Engine.Actor|GetRootComponent _asfort_player_pawn)) (Math|Rotator|FindLookatRotation (Transformation|GetWorldLocation (Variables|/Script/Engine.Actor|GetRootComponent _asfort_player_pawn)) (* (Transformation|GetWorldLocation (Variables|/Script/Engine.Actor|GetRootComponent _asfort_player_pawn)) -2.0)))
  (return))

(fn ConstructionScript)
