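; Burst gameplay cue for tether fall landings, playing audio, visual and camera feedback on the Fort player pawn and its controller.
; Passes the truncated normalized magnitude to the first audio component as "OnLand".
; With non-positive magnitude and a valid water body, adds a droplet lens effect and spawns a Niagara splash fed with water surface info.
; Otherwise treats it as a ground landing and plays a vertical jolt camera shake.
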(fn OnBurst(Cascade)
  (params (MyTarget :type Actor%20Object%20Reference) (Parameters :type Gameplay%20Cue%20Parameters%20Structure%20%28by%20ref%29) (ParticleComponents :type Array%20of%20Cascade%20Particle%20System%20Component%20Object%20References) (AudioComponents :type Array%20of%20Audio%20Component%20Object%20References) (BurstCameraShakeInstance :type Camera%20Shake%20Base%20Object%20Reference) (BurstDecalInstance :type Decal%20Actor%20Object%20Reference))
  (bind _normalizedmagnitude (Ability|GameplayCue|BreakGameplayCueParameters Parameters))
  (bind _asfort_player_pawn (Utilities|Casting|CastToFortPlayerPawn MyTarget)
    (:then
      (bind _asfort_player_controller (Utilities|Casting|CastToFortPlayerController (Pawn|GetController _asfort_player_pawn))
        (:then
          (Audio|Parameter|SetIntegerParameter (Utilities|Array|Get(acopy) AudioComponents) "OnLand" (Math|Float|Truncate _normalizedmagnitude) :!class /Script/AudioExtensions.AudioParameterControllerInterface)
          (if (not (> _normalizedmagnitude 0.0))
            (Utilities|IsValid (WaterBody|GetCurrentWaterBody _asfort_player_pawn)
              (:Is%20Valid
                ;[1 1 1] Landed in water
                (Camera|AddGenericCameraLensEffect (Variables|/Script/Engine.PlayerController|GetPlayerCameraManager _asfort_player_controller) "/Game/Effects/Fort_Effects/Camera_Effects/B_CameraLens_Splash_Water_Droplets.B_CameraLens_Splash_Water_Droplets_C")
                (bind _outwatersurfaceinfo (WaterBody|GetWaterSurfaceInfoStruct (WaterBody|GetCurrentWaterBody _asfort_player_pawn) (Transformation|GetActorLocation _asfort_player_pawn) false true false))
                (bind _returnvalue (Effects|Components|ParticleSystem|SpawnFXAtLocation 0 "/Game/Effects/Niagara/Systems/Enviroment/Water/NS_WaterSplash_Medium.NS_WaterSplash_Medium" (Utilities|Struct|BreakWaterSurfaceInfo _outwatersurfaceinfo)))
                (Effects|Components|ParticleSystem|SetVectorParameter _returnvalue "WaterPlaneLocation" (Utilities|Struct|BreakWaterSurfaceInfo _outwatersurfaceinfo))
                (Effects|Components|ParticleSystem|SetVectorParameter _returnvalue "WaterNormal" (.WaterPlaneNormal (Utilities|Struct|BreakWaterSurfaceInfo _outwatersurfaceinfo)))
                (Effects|Components|ParticleSystem|SetFloatParameter _returnvalue "WaterBodyIndex" (Math|Conversions|ToFloat(Integer) (.WaterBodyIdx (Utilities|Struct|BreakWaterSurfaceInfo _outwatersurfaceinfo))))
                (return))
              (:Is%20Not%20Valid
                _)))
          ;[1 1 1] Landed on ground
          (Game|Feedback|ClientStartCameraShake :self _asfort_player_controller :Shake "/Game/Effects/Fort_Effects/BluePrints/CameraShakes/B_Small_Vertical_Jolt_CameraShake.B_Small_Vertical_Jolt_CameraShake_C" :Scale 2.0 :PlaySpace "CameraLocal")
          (return))
        (:CastFailed
          _)))
    (:CastFailed
      _))
  (return))
