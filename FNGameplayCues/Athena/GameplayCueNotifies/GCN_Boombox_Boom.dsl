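; Burst gameplay cue notify for the Boombox's boom effect, adding placement tweaks and a camera shake to the parent burst.
; After the parent call, it yaws and offsets the particle components at indices 1 through 4 in mirrored pairs to fan them out.
; A nonzero ability level triggers a world camera shake at the target's location, with the cue's normalized and raw magnitudes repurposed as inner and outer radius.
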(fn OnBurst(Cascade)
  (params (MyTarget :type Actor%20Object%20Reference) (Parameters :type Gameplay%20Cue%20Parameters%20Structure%20%28by%20ref%29) (ParticleComponents :type Array%20of%20Cascade%20Particle%20System%20Component%20Object%20References) (AudioComponents :type Array%20of%20Audio%20Component%20Object%20References) (BurstCameraShakeInstance :type Camera%20Shake%20Base%20Object%20Reference) (BurstDecalInstance :type Decal%20Actor%20Object%20Reference))
  (bind (_normalizedmagnitude _rawmagnitude _effectcontext _matchedtagname _originaltag _aggregatedsourcetags _aggregatedtargettags _location _normal _instigator _effectcauser _sourceobject _physicalmaterial _gameplayeffectlevel _abilitylevel) (Ability|GameplayCue|BreakGameplayCueParameters Parameters))
  (|CallParentFunction MyTarget Parameters ParticleComponents AudioComponents BurstCameraShakeInstance BurstDecalInstance)
  (for (_array_element _array_index) ParticleComponents
    (switch int _array_index
      (:1
        (Transformation|AddLocalRotation _array_element "0,40.000000, 0" false false)
        (Transformation|AddLocalOffset _array_element "35.000000,10.000000,20.000000" false false))
      (:2
        (Transformation|AddLocalRotation _array_element "0,-40.000000, 0" false false)
        (Transformation|AddLocalOffset _array_element "35.000000,-10.000000,20.000000" false false))
      (:3
        (Transformation|AddLocalRotation _array_element "0,-140.000000, 0" false false)
        (Transformation|AddLocalOffset _array_element "35.000000,10.000000,20.000000" false false))
      (:4
        (Transformation|AddLocalRotation _array_element "0,140.000000, 0" false false)
        (Transformation|AddLocalOffset _array_element "35.000000,-10.000000,20.000000" false false))))
  (if (Math|Conversions|ToBoolean(Integer) _abilitylevel)
    (Camera|PlayWorldCameraShake MyTarget "/FNGameplayCues/Effects/Fort_Effects/Camera_Effects/BP_Camera_Shake_BoomBox.BP_Camera_Shake_BoomBox_C" (Transformation|GetActorLocation MyTarget) _normalizedmagnitude _rawmagnitude)
    (else
      _))
  (return))
