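; Gameplay cue notify for the jump pack boost that moves its particle effect onto the player's backpack.
; When the start particle system spawns, it casts the cue instigator to an Athena player pawn and gets the Backpack part's skeletal mesh.
; If that mesh is valid and has the configured attach socket, the spawned Cascade component attaches to that socket with relative transform rules.
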(fn OnStartParticleSystemSpawned
  (params (SpawnedParticleSysComponent :type Cascade%20Particle%20System%20Component%20Object%20Reference) (Parameters :type Gameplay%20Cue%20Parameters%20Structure%20%28by%20ref%29))
  (bind (_normalizedmagnitude _rawmagnitude _effectcontext _matchedtagname _originaltag _aggregatedsourcetags _aggregatedtargettags _location _normal _instigator) (Ability|GameplayCue|BreakGameplayCueParameters Parameters))
  (bind _asplayer_pawn_athena (Utilities|Casting|CastToPlayerPawn_Athena _instigator))
  (Utilities|IsValid (Customization|GetSkeletalMeshforPartType _asplayer_pawn_athena "Backpack" :!class /Script/FortniteGame.FortPlayerPawn)
    (:Is%20Valid
      (if (Transformation|DoesSocketExist (Customization|GetSkeletalMeshforPartType _asplayer_pawn_athena "Backpack" :!class /Script/FortniteGame.FortPlayerPawn) (Variables|Default|GetAttachSocketName))
        (Transformation|AttachComponentToComponent (Variables|GetSpawnedParticleSysComponent) (Customization|GetSkeletalMeshforPartType _asplayer_pawn_athena "Backpack" :!class /Script/FortniteGame.FortPlayerPawn) (Variables|Default|GetAttachSocketName) "KeepRelative" "KeepRelative" "KeepRelative" true)
        (return)))))
