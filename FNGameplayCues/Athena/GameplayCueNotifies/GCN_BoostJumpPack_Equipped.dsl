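; Gameplay cue notify for an equipped Boost Jump Pack that shows the pack's fuel level on the player's backpack.
; OnExecute and WhileActive cast the target to an Athena player pawn and write fuel percent to the FuelLevel material parameter on the Backpack mesh.
; Fuel and max fuel come from pawn attributes, falling back to simulated attribute values when the direct lookup fails.
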(fn GetFuel
  (params (PlayerPawn :type Player%20Pawn%20Athena%20Object%20Reference))
  (bind _attribute (LocalVariables|GetAttribute))
  (bind _playerpawn (Variables|GetPlayerPawn))
  (bind (_bsuccessfullyfoundattribute _returnvalue) (Ability|Attribute|GetFloatAttribute _playerpawn _attribute))
  (Utilities|IsValid (Variables|GetPlayerPawn)
    (:Is%20Valid
      (if _bsuccessfullyfoundattribute
        (return _returnvalue)
        (else
          (return (GadgetAttributeSets|GetSimulatedAttributeValue (Variables|GetPlayerPawn) (LocalVariables|GetAttribute) 0.0 :!class /Script/FortniteGame.FortPlayerPawn)))))
    (:Is%20Not%20Valid
      (return 0.0))))

(fn GetFuelPercent
  (params (PlayerPawn :type Player%20Pawn%20Athena%20Object%20Reference))
  (return (/ (CallFunction|GetFuel PlayerPawn) (CallFunction|GetMaxFuel PlayerPawn))))

(fn GetMaxFuel
  (params (PlayerPawn :type Player%20Pawn%20Athena%20Object%20Reference))
  (bind _attribute (LocalVariables|GetAttribute))
  (bind _playerpawn (Variables|GetPlayerPawn))
  (bind (_bsuccessfullyfoundattribute _returnvalue) (Ability|Attribute|GetFloatAttribute _playerpawn _attribute))
  (Utilities|IsValid (Variables|GetPlayerPawn)
    (:Is%20Valid
      (if _bsuccessfullyfoundattribute
        (return _returnvalue)
        (else
          (return (GadgetAttributeSets|GetSimulatedAttributeValue (Variables|GetPlayerPawn) (LocalVariables|GetAttribute) 1.0 :!class /Script/FortniteGame.FortPlayerPawn)))))
    (:Is%20Not%20Valid
      (return 1.0))))

(fn OnExecute
  (params (MyTarget :type Actor%20Object%20Reference) (Parameters :type Gameplay%20Cue%20Parameters%20Structure%20%28by%20ref%29))
  (bind _asplayer_pawn_athena (Utilities|Casting|CastToPlayerPawn_Athena MyTarget)
    (:then
      (bind _output_get (LocalVariables|SetPlayerPawn _asplayer_pawn_athena))
      (Utilities|IsValid _output_get
        (:Is%20Valid
          (Rendering|Material|SetScalarParameterValueonMaterials (Customization|GetSkeletalMeshforPartType (LocalVariables|GetPlayerPawn) "Backpack" :!class /Script/FortniteGame.FortPlayerPawn) "FuelLevel" (CallFunction|GetFuelPercent (LocalVariables|GetPlayerPawn)))
          (return true))))
    (:CastFailed
      (return false))))

(fn WhileActive
  (params (MyTarget :type Actor%20Object%20Reference) (Parameters :type Gameplay%20Cue%20Parameters%20Structure%20%28by%20ref%29))
  (bind _asplayer_pawn_athena (Utilities|Casting|CastToPlayerPawn_Athena MyTarget)
    (:then
      (bind _output_get (LocalVariables|SetPlayerPawn _asplayer_pawn_athena))
      (Utilities|IsValid _output_get
        (:Is%20Valid
          (Rendering|Material|SetScalarParameterValueonMaterials (Customization|GetSkeletalMeshforPartType (LocalVariables|GetPlayerPawn) "Backpack" :!class /Script/FortniteGame.FortPlayerPawn) "FuelLevel" (CallFunction|GetFuelPercent (LocalVariables|GetPlayerPawn)))
          (return true))))
    (:CastFailed
      (return false))))
