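; Gameplay cue notify that shows a butterfly body aura on an Athena player pawn while the cue loops.
; On looping start it manually spawns the aura particle at the mesh's chest socket with a local offset, because the notify's looping effects fail to spawn.
; It stores the component as BodyFX, makes its rotation and scale absolute, and deactivates it when the cue ceases to be relevant.
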(event EventOnLoopingStart(Cascade)
  (params (MyTarget :type Actor%20Object%20Reference) (Parameters :type Gameplay%20Cue%20Parameters%20Structure%20%28by%20ref%29) (ParticleComponents :type Array%20of%20Cascade%20Particle%20System%20Component%20Object%20References) (AudioComponents :type Array%20of%20Audio%20Component%20Object%20References))
  ;[1 1 1] For whatever reason looping effects do not spawn, so adding it the graph manually.
  (bind _asfort_player_pawn_athena (Utilities|Casting|CastToFortPlayerPawnAthena MyTarget))
  (bind _returnvalue (Effects|SpawnEmitterAttached :EmitterTemplate "/Game/Effects/Fort_Effects/Gameplay/Butterfly/P_Butterfly_BodyAura.P_Butterfly_BodyAura" :AttachToComponent (Variables|/Script/Engine.Character|GetMesh _asfort_player_pawn_athena) :AttachPointName "FX_Chest" :Location (Variables|Default|GetLocalOffset) :Scale "1.000000,1.000000,1.000000" :LocationType "KeepRelativeOffset" :bAutoDestroy true :PoolingMethod "AutoRelease" :bAutoActivate true))
  (bind _output_get (Variables|Default|SetBodyFX _returnvalue))
  (Transformation|SetAbsolute _output_get false true true))

(fn OnCeaseRelevant
  (params (MyTarget :type Actor%20Object%20Reference) (Parameters :type Gameplay%20Cue%20Parameters%20Structure%20%28by%20ref%29))
  (bind _bodyfx (Variables|Default|GetBodyFX)
    (:then
      (Components|Activation|Deactivate _bodyfx)
      (return true))
    (:else
      (return false))))

(fn ConstructionScript)
