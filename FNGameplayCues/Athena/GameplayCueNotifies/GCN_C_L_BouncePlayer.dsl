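; Burst gameplay cue notify that plays camera, particle and sound feedback when a player is bounced, using Cube smack and launch assets.
; On burst it casts the target to a Fort player pawn and starts a camera-local shake on that pawn's player controller.
; It moves a spawned Cascade particle component to the cue location's X and Y at a hardcoded world height, then plays the launch sound at the cue location.
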(fn OnBurst(Cascade)
  (params (MyTarget :type Actor%20Object%20Reference) (Parameters :type Gameplay%20Cue%20Parameters%20Structure%20%28by%20ref%29) (ParticleComponents :type Array%20of%20Cascade%20Particle%20System%20Component%20Object%20References) (AudioComponents :type Array%20of%20Audio%20Component%20Object%20References) (BurstCameraShakeInstance :type Camera%20Shake%20Base%20Object%20Reference) (BurstDecalInstance :type Decal%20Actor%20Object%20Reference))
  (bind (_normalizedmagnitude _rawmagnitude _effectcontext _matchedtagname _originaltag _aggregatedsourcetags _aggregatedtargettags _location) (Ability|GameplayCue|BreakGameplayCueParameters Parameters))
  (bind _asfort_player_pawn (Utilities|Casting|CastToFortPlayerPawn MyTarget))
  (Game|Feedback|ClientStartCameraShake :self (Fortnite|GetFortPlayerControllerfromActor _asfort_player_pawn) :Shake "/FNGameplayCues/Effects/Fort_Effects/Camera_Effects/BP_CameraShake_Cube_SmackL.BP_CameraShake_Cube_SmackL_C" :Scale 1.0 :PlaySpace "CameraLocal")
  (Transformation|SetWorldLocation (Utilities|Array|Get(acopy) ParticleComponents) (Math|Vector|MakeVector (.x _location) (.y _location) -816.0) false false)
  (Audio|PlaySoundatLocation (Variables|/Script/Engine.Actor|GetRootComponent _asfort_player_pawn) "/FNGameplayCues/Athena/Sounds/Events/Cube/Cube_LaunchPlayer_Surface_Cue.Cube_LaunchPlayer_Surface_Cue" _location)
  (return))
