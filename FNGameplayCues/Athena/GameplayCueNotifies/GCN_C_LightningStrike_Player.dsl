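; Gameplay cue burst for a lightning strike on a player, using Cube zap sound and smack shake assets.
; It acts only when normalized magnitude is not above zero, moving the first particle component to the target and aiming it at the cue location.
; Distance between those points drives the particle's "Length" parameter to stretch the bolt.
; It also shakes the struck player's camera and attaches the zap sound to the source object's root.
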(fn OnBurst(Cascade)
  (params (MyTarget :type Actor%20Object%20Reference) (Parameters :type Gameplay%20Cue%20Parameters%20Structure%20%28by%20ref%29) (ParticleComponents :type Array%20of%20Cascade%20Particle%20System%20Component%20Object%20References) (AudioComponents :type Array%20of%20Audio%20Component%20Object%20References) (BurstCameraShakeInstance :type Camera%20Shake%20Base%20Object%20Reference) (BurstDecalInstance :type Decal%20Actor%20Object%20Reference))
  (bind (_normalizedmagnitude _rawmagnitude _effectcontext _matchedtagname _originaltag _aggregatedsourcetags _aggregatedtargettags _location _normal _instigator _effectcauser _sourceobject) (Ability|GameplayCue|BreakGameplayCueParameters Parameters))
  (bind _returnvalue (Transformation|GetActorLocation MyTarget))
  (bind _output (Utilities|Array|Get(acopy) ParticleComponents))
  (if (not (> _normalizedmagnitude 0.0))
    (bind _asactor (Utilities|Casting|CastToActor _sourceobject))
    (Transformation|SetWorldLocationAndRotation _output _returnvalue (Math|Rotator|FindLookatRotation _returnvalue _location) false false)
    (Effects|Components|ParticleSystem|SetFloatParameter _output "Length" (/ (Math|Vector|Distance(Vector) _location _returnvalue) 1e+03))
    (bind _asfort_player_pawn_athena (Utilities|Casting|CastToFortPlayerPawnAthena MyTarget))
    (Game|Feedback|ClientStartCameraShake :self (Fortnite|GetFortPlayerControllerfromActor _asfort_player_pawn_athena) :Shake "/FNGameplayCues/Effects/Fort_Effects/Camera_Effects/BP_CameraShake_Cube_Smack.BP_CameraShake_Cube_Smack_C" :Scale 1.0 :PlaySpace "CameraLocal")
    (Audio|SpawnSoundAttached "/FNGameplayCues/Athena/Sounds/Events/Cube/Cube_AngryZap_Source_Cue.Cube_AngryZap_Source_Cue" (Variables|/Script/Engine.Actor|GetRootComponent _asactor))
    ;[1 1 1] Comment
    (return)))
