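; Burst gameplay cue for impacts on Daisy environment bouncers, adding audio and Niagara tuning to the parent burst.
; Audio picks the 1P or 3P sound, attaches it to the instigator or target respectively, then shows a sound indicator using the cap bouncer's settings.
; A positive cue magnitude, noted as heal cooldown, deactivates the particle effect.
; A cast to the side bouncer selects Niagara scale and offset parameters, falling back to unit scale.
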(fn HandleAudio
  (params (Parameters :type Gameplay%20Cue%20Parameters%20Structure%20%28by%20ref%29) (TargetActor :type Actor%20Object%20Reference))
  (bind (_normalizedmagnitude _rawmagnitude _effectcontext _matchedtagname _originaltag _aggregatedsourcetags _aggregatedtargettags _location _normal _instigator) (Ability|GameplayCue|BreakGameplayCueParameters Parameters))
  (bind _returnvalue (Audio|SpawnSoundAttached (AudioGameplay|GetAudioforPlayerEvent _instigator (Variables|Default|Get1P%20Sound) (Variables|Default|Get3P%20Sound)) (select (== (AudioGameplay|GetAudioforPlayerEvent _instigator (Variables|Default|Get1P%20Sound) (Variables|Default|Get3P%20Sound)) (Variables|Default|Get1P%20Sound)) (Variables|/Script/Engine.Actor|GetRootComponent _instigator) (Variables|/Script/Engine.Actor|GetRootComponent TargetActor))))
  (Utilities|IsValid _returnvalue
    (:Is%20Valid
      (bind _asartemis_daisy_cap_a_small_bounce (Utilities|Casting|CastToArtemis_Daisy_Cap_A_Small_Bounce TargetActor))
      (SoundIndicator|ShowSoundIndicator (Variables|/Game/Athena/Artemis/Environments/BuildingActors/Daisy/Props/Artemis_Daisy_Cap_A_Small_Bounce.Artemis_Daisy_Cap_A_Small_Bounce_C|GetFortSoundIndicator _asartemis_daisy_cap_a_small_bounce) (Transformation|GetActorLocation _asartemis_daisy_cap_a_small_bounce) (Variables|/Script/Engine.AudioComponent|GetSound _returnvalue) 1.0 0))))

(fn OnBurst(Niagara)
  (params (MyTarget :type Actor%20Object%20Reference) (Parameters :type Gameplay%20Cue%20Parameters%20Structure%20%28by%20ref%29) (ParticleComponents :type Array%20of%20Niagara%20Particle%20System%20Component%20Object%20References) (AudioComponents :type Array%20of%20Audio%20Component%20Object%20References) (BurstCameraShakeInstance :type Camera%20Shake%20Base%20Object%20Reference) (BurstDecalInstance :type Decal%20Actor%20Object%20Reference))
  (bind _output (Utilities|Array|Get(acopy) ParticleComponents))
  (|CallParentFunction MyTarget Parameters ParticleComponents AudioComponents BurstCameraShakeInstance BurstDecalInstance)
  (CallFunction|HandleAudio Parameters MyTarget)
  ;[1 0 0.7568] If the heal is on cooldown, deactivate the VFX
  (if (> (Ability|GameplayCue|BreakGameplayCueParameters Parameters) 0)
    (Utilities|IsValid _output
      (:Is%20Valid
        (Components|Activation|Deactivate _output))
      (:Is%20Not%20Valid
        _))
    (bind _asartemis_daisy_side_small_bounce (Utilities|Casting|CastToArtemis_Daisy_Side_Small_Bounce MyTarget)
      (:then
        ;[1 1 1] Artemis_Daisy_Side_Small_Bounce
        (Niagara|SetNiagaraVariable(Float) _output "User.FXGameplayScale" (Variables|/Game/Athena/Artemis/Environments/BuildingActors/Daisy/Props/Artemis_Daisy_Side_Small_Bounce.Artemis_Daisy_Side_Small_Bounce_C|GetFXParentScale _asartemis_daisy_side_small_bounce))
        (Niagara|SetNiagaraVariable(Float) _output "User.FXGameplayPosition" 0.0)
        (Niagara|SetNiagaraVariable(Vector3) _output "User.FXGameplayPThree" "0.000000,22.000000,-40.000000"))
      (:CastFailed
        ;[1 1 1] Artemis_Daisy_Cap_A_Small_Bounce
        (Niagara|SetNiagaraVariable(Float) _output "User.FXGameplayScale" 1.0)))
    (return)
    (else
      _)))
