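; Burst gameplay cue notify for a large Daisy environment item impact, adding perspective-aware audio and tuned Niagara effects.
; After calling the parent burst, HandleAudio picks a 1P or 3P sound through GetAudioforPlayerEvent for the instigator.
; The 1P sound attaches to the instigator's root component, while any other result attaches to the target actor's root.
; It then overrides scale, vector and position user parameters on the first spawned Niagara component with fixed values.
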(fn HandleAudio
  (params (Parameters :type Gameplay%20Cue%20Parameters%20Structure%20%28by%20ref%29) (TargetActor :type Actor%20Object%20Reference))
  (bind (_normalizedmagnitude _rawmagnitude _effectcontext _matchedtagname _originaltag _aggregatedsourcetags _aggregatedtargettags _location _normal _instigator) (Ability|GameplayCue|BreakGameplayCueParameters Parameters))
  (Audio|SpawnSoundAttached (AudioGameplay|GetAudioforPlayerEvent _instigator (Variables|Default|Get1P%20Sound) (Variables|Default|Get3P%20Sound)) (select (== (AudioGameplay|GetAudioforPlayerEvent _instigator (Variables|Default|Get1P%20Sound) (Variables|Default|Get3P%20Sound)) (Variables|Default|Get1P%20Sound)) (Variables|/Script/Engine.Actor|GetRootComponent _instigator) (Variables|/Script/Engine.Actor|GetRootComponent TargetActor))))

(fn OnBurst(Niagara)
  (params (MyTarget :type Actor%20Object%20Reference) (Parameters :type Gameplay%20Cue%20Parameters%20Structure%20%28by%20ref%29) (ParticleComponents :type Array%20of%20Niagara%20Particle%20System%20Component%20Object%20References) (AudioComponents :type Array%20of%20Audio%20Component%20Object%20References) (BurstCameraShakeInstance :type Camera%20Shake%20Base%20Object%20Reference) (BurstDecalInstance :type Decal%20Actor%20Object%20Reference))
  (bind _output (Utilities|Array|Get(acopy) ParticleComponents))
  (|CallParentFunction MyTarget Parameters ParticleComponents AudioComponents BurstCameraShakeInstance BurstDecalInstance)
  (CallFunction|HandleAudio Parameters MyTarget)
  ;[1 1 1] Artemis_Daisy_Side_Large_Bounce
  (Niagara|SetNiagaraVariable(Float) _output "User.FXGameplayScale" 1.7)
  (Niagara|SetNiagaraVariable(Vector3) _output "User.FXGameplayPThree" "0.000000,55.000000,-70.000000")
  (Niagara|SetNiagaraVariable(Float) _output "User.FXGameplayPosition" 0.3)
  (return))
