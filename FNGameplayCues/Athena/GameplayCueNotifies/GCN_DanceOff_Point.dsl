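; Burst gameplay cue notify for a point in the DanceOff feature, which tints its Cascade particle effect.
; On burst, it tests whether a value from the cue parameters is above zero and selects either a green or a gold color.
; It converts that color to a vector and writes it to the "Color" parameter of the first spawned particle component.
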(fn OnBurst(Cascade)
  (params (MyTarget :type Actor%20Object%20Reference) (Parameters :type Gameplay%20Cue%20Parameters%20Structure%20%28by%20ref%29) (ParticleComponents :type Array%20of%20Cascade%20Particle%20System%20Component%20Object%20References) (AudioComponents :type Array%20of%20Audio%20Component%20Object%20References) (BurstCameraShakeInstance :type Camera%20Shake%20Base%20Object%20Reference) (BurstDecalInstance :type Decal%20Actor%20Object%20Reference))
  (Effects|Components|ParticleSystem|SetVectorParameter (Utilities|Array|Get(acopy) ParticleComponents) "Color" (Math|Conversions|ToVector(LinearColor) (select (> (Ability|GameplayCue|BreakGameplayCueParameters Parameters) 0.0) "(R=0.029557,G=1.000000,B=0.242281,A=1.000000)" "(R=1.000000,G=0.838799,B=0.029557,A=1.000000)")))
  (return))
