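; Looping gameplay cue for the Elastic emote, projecting the player's body "Sticker" texture as the symbol on a spotlight effect.
; On loop start it feeds that texture into spotlight dynamic materials, attaches to the pawn and activates Niagara.
; Timelines flicker material alpha and point light intensity, then fade the materials on removal.
; Day phase sets Niagara smoke alpha; Switch destroys the point light, elsewhere effect quality gates its visibility and shadows.
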(event EventBeginPlay
  (TimeOfDay|RegisterforDayPhaseChangedEvent (TimeOfDay|GetContextualTimeOfDayManager self) (delegate Custom|TODF_Alpha%20Smoke%20Change) false)
  (if (TouchPlatform|IsNintendoSwitch)
    (Components|DestroyComponent (Variables|Default|GetPointLight) 0)
    (Rendering|Components|Light|SetCastShadows (Variables|Default|GetPointLight) false)
    (else
      (CallFunction|ValidateLightShadows))))

(event Custom|TODF_Alpha%20Smoke%20Change
  (params (CurrentDayPhase :type EFortDayPhase%20Enum :default "Morning") (PreviousDayPhase :type EFortDayPhase%20Enum :default "Morning") (bAtCreation :type Boolean))
  (Niagara|SetNiagaraVariable(Float) (Variables|Default|GetEmote) "AlphaFade" (Utilities|Select :Index CurrentDayPhase :Morning 0.5 :Day 0.75 :Evening 0.5 :Night 0.25)))

(event EventOnLoopingStart(Cascade)
  (params (MyTarget :type Actor%20Object%20Reference) (Parameters :type Gameplay%20Cue%20Parameters%20Structure%20%28by%20ref%29) (ParticleComponents :type Array%20of%20Cascade%20Particle%20System%20Component%20Object%20References) (AudioComponents :type Array%20of%20Audio%20Component%20Object%20References))
  (Utilities|FlowControl|Sequence
    (:then_0
      (Utilities|IsValid MyTarget
        (:Is%20Valid
          (bind _asfort_player_pawn (Utilities|Casting|CastToFortPlayerPawn MyTarget))
          (bind _output_get (Variables|Default|SetPawn _asfort_player_pawn))
          (Variables|Default|SetPlayerMesh (Variables|/Script/Engine.Character|GetMesh _output_get)))))
    (:then_1
      (bind _returnvalue (Fortnite|GetDynamicMaterialInstance (Customization|GetSkeletalMeshforPartType (Variables|Default|GetPawn) "Body" :!class /Script/FortniteGame.FortPlayerPawn)))
      (bind _returnvalue_2 (Rendering|Material|GetTextureParameterValue _returnvalue "Sticker" :!class /Script/Engine.MaterialInstanceDynamic))
      (Variables|Default|SetIcon _returnvalue_2))
    (:then_2
      (for (_array_element _array_index) (Rendering|Material|GetMaterials (Variables|Default|GetSpotlight))
        (bind _returnvalue_3 (Rendering|Material|CreateDynamicMaterialInstance (Variables|Default|GetSpotlight) _array_index _array_element (Utilities|String|StringToName (Utilities|String|Append "MeshMID" "_" (Utilities|String|ToString(Integer) _array_index))) :!class /Script/Engine.PrimitiveComponent))
        (Utilities|Array|Add (Variables|Default|GetSpotlightMIDs) _returnvalue_3)))
    (:then_3
      (for _array_element_2 (Variables|Default|GetSpotlightMIDs)
        (Rendering|Material|SetTextureParameterValue _array_element_2 "Symbol Texture" (Variables|Default|GetIcon))))
    (:then_4
      (Transformation|SetActorLocation self (Transformation|GetSocketLocation (Variables|Default|GetPlayerMesh) "Root") false false)
      (Transformation|AttachComponentToComponent (Variables|Default|GetArrow) (Variables|Default|GetPlayerMesh) "FX_Chest" "SnapToTarget" "SnapToTarget" "KeepRelative" true)
      (Components|Activation|Activate (Variables|Default|GetEmote) false)
      (CallFunction|EnableFX))))

(event Custom|EnableFX
  (bind (_direction _scale _lightflicker) (Timeline|AnimLightFixture PlayFromStart)
    (:Update
      (for _array_element (Variables|Default|GetSpotlightMIDs)
        (Rendering|Material|SetScalarParameterValue _array_element "Alpha" _lightflicker :!class /Script/Engine.MaterialInstanceDynamic)
        (bind _pointlight (Variables|Default|GetPointLight)
          (:then
            (Rendering|Components|Light|SetIntensity _pointlight (Math|Float|Lerp 0.0 3.0 _lightflicker)))
          (:else))))))

(event EventOnRemoval(Cascade)
  (params (MyTarget :type Actor%20Object%20Reference) (Parameters :type Gameplay%20Cue%20Parameters%20Structure%20%28by%20ref%29) (ParticleComponents :type Array%20of%20Cascade%20Particle%20System%20Component%20Object%20References) (AudioComponents :type Array%20of%20Audio%20Component%20Object%20References) (BurstCameraShakeInstance :type Camera%20Shake%20Base%20Object%20Reference) (BurstDecalInstance :type Decal%20Actor%20Object%20Reference))
  (bind _spotlightmids (Variables|Default|GetSpotlightMIDs))
  (Utilities|FlowControl|Sequence
    (:then_0
      (Timeline|AnimLightFixture Stop))
    (:then_1
      (bind (_direction _scale_down _alpha) (Timeline|TurnOff PlayFromStart)
        (:Update
          (for _array_element _spotlightmids
            (Rendering|Material|SetScalarParameterValue _array_element "Alpha" _alpha :!class /Script/Engine.MaterialInstanceDynamic)))
        (:Finished
          (Components|Activation|Deactivate (Variables|Default|GetEmote)))))))

(fn ConstructionScript)

(fn validateLightShadows
  (bind _pointlight (Variables|Default|GetPointLight))
  (bind _returnvalue (Settings|GetGameUserSettings))
  (Rendering|Components|Light|SetCastShadows _pointlight (>= (Settings|GetVisualEffectQuality _returnvalue) 3))
  (Rendering|SetVisibility _pointlight (>= (Settings|GetVisualEffectQuality _returnvalue) 3) false))
