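; Gameplay cue notify for the incendiary grenade fuel douse effect, which places its burst particles on an environment campfire.
; On the Cascade burst event, it casts the cue target to the Athena campfire actor.
; It then loops over the spawned particle components and moves each one to the world location of the campfire's FireLoc component.
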(fn OnBurst(Cascade)
  (params (MyTarget :type Actor%20Object%20Reference) (Parameters :type Gameplay%20Cue%20Parameters%20Structure%20%28by%20ref%29) (ParticleComponents :type Array%20of%20Cascade%20Particle%20System%20Component%20Object%20References) (AudioComponents :type Array%20of%20Audio%20Component%20Object%20References) (BurstCameraShakeInstance :type Camera%20Shake%20Base%20Object%20Reference) (BurstDecalInstance :type Decal%20Actor%20Object%20Reference))
  (bind _asb_bga_athena_env_camp_fire (Utilities|Casting|CastToB_BGA_Athena_EnvCampFire MyTarget))
  (for _array_element ParticleComponents
    (Transformation|SetWorldLocation _array_element (Transformation|GetWorldLocation (Variables|/Game/Athena/Items/ForagedItems/EnvCampFire/B_BGA_Athena_EnvCampFire.B_BGA_Athena_EnvCampFire_C|GetFireLoc _asb_bga_athena_env_camp_fire)) false false)
    (return)))
