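; Burst gameplay cue for a jump smash that places the impact particle effect and a scorched crack decal on the ground.
; When the burst fires, it takes the first Cascade particle component and runs a vertical line trace just ahead of it against world static objects.
; On a hit, it snaps the particle to the surface normal and spawns a short-lived fading decal there.
; Otherwise, it only pitches the particle upright.
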(fn OnBurst(Cascade)
  (params (MyTarget :type Actor%20Object%20Reference) (Parameters :type Gameplay%20Cue%20Parameters%20Structure%20%28by%20ref%29) (ParticleComponents :type Array%20of%20Cascade%20Particle%20System%20Component%20Object%20References) (AudioComponents :type Array%20of%20Audio%20Component%20Object%20References) (BurstCameraShakeInstance :type Camera%20Shake%20Base%20Object%20Reference) (BurstDecalInstance :type Decal%20Actor%20Object%20Reference))
  (bind _output (Utilities|Array|Get(acopy) ParticleComponents))
  (if (> (Utilities|Array|Length ParticleComponents) 0)
    (bind (_outhit _returnvalue) (Collision|LineTraceForObjects :Start (+ (+ (Transformation|GetWorldLocation _output) (* (Transformation|GetForwardVector _output) 7e+01)) "0, 0,50.000000") :End (- (+ (Transformation|GetWorldLocation _output) (* (Transformation|GetForwardVector _output) 7e+01)) "0, 0,250.000000") :ObjectTypes (Utilities|Array|MakeArray "ObjectTypeQuery1") :bTraceComplex false :DrawDebugType "None" :bIgnoreSelf true :TraceColor "(R=1.000000,G=0.000000,B=0.000000,A=1.000000)" :TraceHitColor "(R=0.000000,G=1.000000,B=0.000000,A=1.000000)" :DrawTime 3.0))
    (if _returnvalue
      (Transformation|SetWorldLocationAndRotation _output (.ImpactPoint (Collision|BreakHitResult _outhit)) (Math|Conversions|RotationFromXVector (.Normal (Collision|BreakHitResult _outhit))) false false)
      (bind _returnvalue_2 (Rendering|Decal|SpawnDecalatLocation 0 "/FNGameplayCues/Weapons/Materials/MD_Decal_ScorchCracks.MD_Decal_ScorchCracks" "128.000000,256.000000,256.000000" (.ImpactPoint (Collision|BreakHitResult _outhit)) (Math|Conversions|RotationFromXVector (* (.Normal (Collision|BreakHitResult _outhit)) -1.0)) 5.0))
      (Rendering|Components|Decal|SetFadeOut _returnvalue_2 0.2 4.5 true)
      (else
        (Transformation|SetWorldRotation _output "90.000000, 0, 0" false false)
        (Transformation|AddWorldOffset _output "0, 0, 0" false false)))
    (return)))
