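; Looping gameplay cue that applies a glow highlight to an Athena player pawn while it is targeted by an action trigger.
; On start it caches the pawn, registers a highlight with itself as source, then plays a timeline that lerps glow colors and fresnel values in through HighlightLerp.
; On removal it reverses the timeline to fade the glow out, removing its highlight source once the reverse finishes.
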(event EventOnLoopingStart(Cascade)
  (params (MyTarget :type Actor%20Object%20Reference) (Parameters :type Gameplay%20Cue%20Parameters%20Structure%20%28by%20ref%29) (ParticleComponents :type Array%20of%20Cascade%20Particle%20System%20Component%20Object%20References) (AudioComponents :type Array%20of%20Audio%20Component%20Object%20References))
  (bind _asplayer_pawn_athena (Utilities|Casting|CastToPlayerPawn_Athena MyTarget))
  (Variables|Default|SetTarget%20Player _asplayer_pawn_athena)
  (CallFunction|FadeMaterialsIn))

(event EventOnRemoval(Cascade)
  (params (MyTarget :type Actor%20Object%20Reference) (Parameters :type Gameplay%20Cue%20Parameters%20Structure%20%28by%20ref%29) (ParticleComponents :type Array%20of%20Cascade%20Particle%20System%20Component%20Object%20References) (AudioComponents :type Array%20of%20Audio%20Component%20Object%20References) (BurstCameraShakeInstance :type Camera%20Shake%20Base%20Object%20Reference) (BurstDecalInstance :type Decal%20Actor%20Object%20Reference))
  (CallFunction|FadeMaterialsOut))

(event Custom|FadeMaterialsIn
  (bind _target_player (Variables|Default|GetTarget%20Player)
    (:then
      (PawnHighlight|HighlightPawnFromSource _target_player self (CallFunction|HighlightLerp 1.0) false) ; Setting an empty highlight with self as Source to the current one.
      (bind (_direction _lerp) (Timeline|FadeMaterialsTL PlayFromStart)
        (:Update
          (bind _target_player_2 (Variables|Default|GetTarget%20Player)
            (:then
              (PawnHighlight|HighlightPawnFromSource _target_player_2 self (CallFunction|HighlightLerp _lerp) true) ; If current, lerp in the colors
            )
            (:else)))
        (:Finished
          (if (Utilities|Enum|Equal(Enum) _direction "Backward")
            (bind _target_player_3 (Variables|Default|GetTarget%20Player)
              (:then
                (PawnHighlight|RemovePawnHighlightFromSource _target_player_3 self))
              (:else))))))
    (:else)))

(event Custom|FadeMaterialsOut
  (Timeline|FadeMaterialsTL Reverse))

(fn HighlightLerp
  (params (Alpha :type Float%20%28double-precision%29))
  ;[1 1 1] Returns lerp between Off and Full from Values.
  (return (Variables|Default|GetGlow%20Priority) (Math|Color|Lerp(LinearColor) :B (Variables|Default|GetHitGlowColor) :Alpha Alpha) (Math|Color|Lerp(LinearColor) :B (Variables|Default|GetHotGlowOuterColor) :Alpha Alpha) (Math|Float|Lerp 0.0 (Variables|Default|GetHitGlowFresnelBrightness) Alpha) (Math|Float|Lerp 0.0 (Variables|Default|GetHitGlowFresnelExponent) Alpha) 1.0))

(fn ConstructionScript)
