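; Looping gameplay cue that attaches a snow particle effect to the locally controlled Athena player pawn.
; On loop start it clears any previous emitter, spawns one on the mesh root and binds the pawn's actor overlap events.
; Overlapping an actor whose display name is "SnowBelowLake" hides the emitter, and leaving shows it again.
; On removal it deactivates the emitter and unbinds both events.
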(event EventOnLoopingStart(Cascade)
  (params (MyTarget :type Actor%20Object%20Reference) (Parameters :type Gameplay%20Cue%20Parameters%20Structure%20%28by%20ref%29) (ParticleComponents :type Array%20of%20Cascade%20Particle%20System%20Component%20Object%20References) (AudioComponents :type Array%20of%20Audio%20Component%20Object%20References))
  (Utilities|FlowControl|Sequence
    (:then_0
      (Variables|Default|GetActiveLoopingSnow
        (:then
          (Components|DestroyComponent (Variables|Default|GetActiveLoopingSnow) 0))
        (:else)))
    (:then_1
      (bind _asplayer_pawn_athena (Utilities|Casting|CastToPlayerPawn_Athena MyTarget)
        (:then
          (if (Pawn|IsLocallyControlled _asplayer_pawn_athena)
            (bind _returnvalue (Effects|SpawnEmitterAttached :EmitterTemplate (Variables|Default|GetLoopingSnow) :AttachToComponent (Variables|/Script/Engine.Character|GetMesh _asplayer_pawn_athena) :AttachPointName "root" :Scale "1.000000,1.000000,1.000000" :LocationType "SnapToTarget" :bAutoDestroy true :PoolingMethod "None" :bAutoActivate true))
            (Variables|Default|SetActiveLoopingSnow _returnvalue)
            (Collision|BindEventtoOnActorBeginOverlap _asplayer_pawn_athena (EventDispatchers|CreateEvent :!event Custom|Begin%20Overlap))
            (Collision|BindEventtoOnActorEndOverlap _asplayer_pawn_athena (EventDispatchers|CreateEvent :!event Custom|End%20Overlap))))
        (:CastFailed
          (Development|PrintString 0 "Failed"))))))

(event Custom|Begin%20Overlap
  (params (OverlappedActor :type Actor%20Object%20Reference) (OtherActor :type Actor%20Object%20Reference))
  (if (Utilities|String|EqualExactly(String) (Utilities|GetDisplayName OtherActor) "SnowBelowLake")
    (Development|SetHiddeninGame (Variables|Default|GetActiveLoopingSnow) true false)))

(event Custom|End%20Overlap
  (params (OverlappedActor :type Actor%20Object%20Reference) (OtherActor :type Actor%20Object%20Reference))
  (if (Utilities|String|EqualExactly(String) (Utilities|GetDisplayName OtherActor) "SnowBelowLake")
    (Development|SetHiddeninGame (Variables|Default|GetActiveLoopingSnow) false false)))

(event EventOnRemoval(Cascade)
  (params (MyTarget :type Actor%20Object%20Reference) (Parameters :type Gameplay%20Cue%20Parameters%20Structure%20%28by%20ref%29) (ParticleComponents :type Array%20of%20Cascade%20Particle%20System%20Component%20Object%20References) (AudioComponents :type Array%20of%20Audio%20Component%20Object%20References) (BurstCameraShakeInstance :type Camera%20Shake%20Base%20Object%20Reference) (BurstDecalInstance :type Decal%20Actor%20Object%20Reference))
  (Components|Activation|Deactivate (Variables|Default|GetActiveLoopingSnow))
  (bind _asplayer_pawn_athena (Utilities|Casting|CastToPlayerPawn_Athena MyTarget))
  (Collision|UnbindEventfromOnActorBeginOverlap _asplayer_pawn_athena (EventDispatchers|CreateEvent :!event Custom|Begin%20Overlap))
  (Collision|UnbindEventfromOnActorEndOverlap _asplayer_pawn_athena (EventDispatchers|CreateEvent :!event Custom|End%20Overlap)))

(fn ConstructionScript)
