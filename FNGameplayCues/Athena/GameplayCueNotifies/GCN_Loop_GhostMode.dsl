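; Looping gameplay cue for Ghost Mode that applies a glow highlight to the instigating pawn and plays attached start and stop sounds.
; On start it stores the instigator as a PawnHighlight interface, adds the highlight, then a timeline updates it through HighlightLerp, which scales glow colors and fresnel by alpha.
; Removal reverses that timeline, clearing the highlight when the backward playback finishes.
; The start sound attaches to the instigator, the stop sound to the cue target.
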(event EventOnLoopingStart(Cascade)
  (params (MyTarget :type Actor%20Object%20Reference) (Parameters :type Gameplay%20Cue%20Parameters%20Structure%20%28by%20ref%29) (ParticleComponents :type Array%20of%20Cascade%20Particle%20System%20Component%20Object%20References) (AudioComponents :type Array%20of%20Audio%20Component%20Object%20References))
  (bind (_normalizedmagnitude _rawmagnitude _effectcontext _matchedtagname _originaltag _aggregatedsourcetags _aggregatedtargettags _location _normal _instigator) (Ability|GameplayCue|BreakGameplayCueParameters Parameters))
  (bind _startsoundcue (Variables|Default|GetStartSoundCue))
  (bind _aspawn_highlight_interface (Utilities|Casting|CastToPawnHighlight_Interface _instigator))
  (Variables|Default|SetTargetPawn _aspawn_highlight_interface)
  (CallFunction|FadeMaterialsIn)
  (Utilities|IsValid _instigator
    (:Is%20Valid
      (Utilities|IsValid _startsoundcue
        (:Is%20Valid
          (Audio|SpawnSoundAttached _startsoundcue (Variables|/Script/Engine.Actor|GetRootComponent _instigator)))))))

(event EventOnRemoval(Cascade)
  (params (MyTarget :type Actor%20Object%20Reference) (Parameters :type Gameplay%20Cue%20Parameters%20Structure%20%28by%20ref%29) (ParticleComponents :type Array%20of%20Cascade%20Particle%20System%20Component%20Object%20References) (AudioComponents :type Array%20of%20Audio%20Component%20Object%20References) (BurstCameraShakeInstance :type Camera%20Shake%20Base%20Object%20Reference) (BurstDecalInstance :type Decal%20Actor%20Object%20Reference))
  (bind _stopsoundcue (Variables|Default|GetStopSoundCue))
  (CallFunction|FadeMaterialsOut)
  (Utilities|IsValid MyTarget
    (:Is%20Valid
      (Utilities|IsValid _stopsoundcue
        (:Is%20Valid
          (Audio|SpawnSoundAttached _stopsoundcue (Variables|/Script/Engine.Actor|GetRootComponent MyTarget)))))))

(event Custom|FadeMaterialsIn
  (bind _targetpawn (Variables|Default|GetTargetPawn))
  (Utilities|IsValid (Utilities|ToObject(Interface) _targetpawn)
    (:Is%20Valid
      (PawnHighlight|AddHighlightfromSource _targetpawn self (CallFunction|HighlightLerp 1.0) false :!class /Game/Athena/PawnHighlight_Interface.PawnHighlight_Interface_C)
      (bind (_direction _lerp) (Timeline|FadeMaterialsTL PlayFromStart)
        (:Update
          (PawnHighlight|AddHighlightfromSource (Variables|Default|GetTargetPawn) self (CallFunction|HighlightLerp _lerp) true :!class /Game/Athena/PawnHighlight_Interface.PawnHighlight_Interface_C))
        (:Finished
          (if (Utilities|Enum|Equal(Enum) _direction "Backward")
            (PawnHighlight|RemoveHighlightfromSource (Variables|Default|GetTargetPawn) self :!class /Game/Athena/PawnHighlight_Interface.PawnHighlight_Interface_C)))))))

(event Custom|FadeMaterialsOut
  (Timeline|FadeMaterialsTL Reverse))

(fn HighlightLerp
  (params (Alpha :type Float%20%28double-precision%29))
  ;[1 1 1] Returns lerp between Off and Full from Values.
  (return (Variables|Default|GetGlow%20Priority) (Math|Color|Lerp(LinearColor) :B (Variables|Default|GetHitGlowColor) :Alpha Alpha) (Math|Color|Lerp(LinearColor) :B (Variables|Default|GetHotGlowOuterColor) :Alpha Alpha) (Math|Float|Lerp 0.0 (Variables|Default|GetHitGlowFresnelBrightness) Alpha) (Math|Float|Lerp 0.0 (Variables|Default|GetHitGlowFresnelExponent) Alpha) 0.0))

(fn ConstructionScript)
