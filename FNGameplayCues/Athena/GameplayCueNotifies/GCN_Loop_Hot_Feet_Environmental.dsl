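; Looping gameplay cue for environmental "hot feet", dressing the affected player pawn with fire-feet particles and movement-driven audio.
; On start it attaches the emitter to the mesh root and binds step, hit, landing and movement-mode events to play sounds, with local/remote footstep variants.
; Tick maps velocity to the loop's "Speed" parameter, zeroed while falling or flying.
; Removal spawns a lava burst, unbinds events and destroys the added collision audio component.
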(event EventOnLoopingStart(Cascade)
  (params (MyTarget :type Actor%20Object%20Reference) (Parameters :type Gameplay%20Cue%20Parameters%20Structure%20%28by%20ref%29) (ParticleComponents :type Array%20of%20Cascade%20Particle%20System%20Component%20Object%20References) (AudioComponents :type Array%20of%20Audio%20Component%20Object%20References))
  (Utilities|FlowControl|Sequence
    (:then_0
      (Variables|Default|SetSpawnedEmitters ParticleComponents)
      (Variables|Default|SetSpawnedAudioComps AudioComponents))
    (:then_1
      (Development|PrintString 0 (Utilities|GetDisplayName MyTarget))
      (bind _asfort_player_pawn (Utilities|Casting|CastToFortPlayerPawn MyTarget))
      (Variables|Default|SetTargetPlayer _asfort_player_pawn)
      (Transformation|AttachComponentToComponent (Utilities|Array|Get(acopy) (Variables|Default|GetSpawnedEmitters)) (Variables|/Script/Engine.Character|GetMesh _asfort_player_pawn) "root" "SnapToTarget" "SnapToTarget" "KeepWorld" true)
      (Effects|Components|ParticleSystem|SetFloatParameter (Variables|Default|GetSpawnedEmitters) "Time" (Utilities|Time|GetGameTimeinSeconds 0))
      ;[1 1 1] Bind footstep and collision events
      (bind _asfort_player_pawn_athena (Utilities|Casting|CastToFortPlayerPawnAthena (Variables|Default|GetTargetPlayer)))
      (EventDispatchers|BindEventtoOnFootstepEvent _asfort_player_pawn_athena (EventDispatchers|CreateEvent :!event Custom|On%20Player%20Step))
      (bind _returnvalue (AddComponent|NODE_AddFortCollisionAudioComponent-1))
      (Variables|Default|SetCollisionAudioComp _returnvalue)
      (Collision|BindEventtoOnActorHit _asfort_player_pawn_athena (EventDispatchers|CreateEvent :!event Custom|On%20Target%20Pawn%20Hit%20Event))
      (EventDispatchers|BindEventtoOnPawnLanded _asfort_player_pawn_athena (EventDispatchers|CreateEvent :!event Custom|On%20Target%20Pawn%20Landed))
      (Character|BindEventtoMovementModeChangedDelegate _asfort_player_pawn_athena (EventDispatchers|CreateEvent :!event Custom|On%20Target%20Pawn%20Movement%20Mode%20Changed)))))

(event EventOnRemoval(Cascade)
  (params (MyTarget :type Actor%20Object%20Reference) (Parameters :type Gameplay%20Cue%20Parameters%20Structure%20%28by%20ref%29) (ParticleComponents :type Array%20of%20Cascade%20Particle%20System%20Component%20Object%20References) (AudioComponents :type Array%20of%20Audio%20Component%20Object%20References) (BurstCameraShakeInstance :type Camera%20Shake%20Base%20Object%20Reference) (BurstDecalInstance :type Decal%20Actor%20Object%20Reference))
  (bind _asfort_player_pawn (Utilities|Casting|CastToFortPlayerPawn MyTarget))
  (Effects|SpawnEmitterAttached :EmitterTemplate "/FNGameplayCues/Athena/Effects/Particles/Gameplay/Lava/P_Lava_FireFeet_Remove.P_Lava_FireFeet_Remove" :AttachToComponent (Variables|/Script/Engine.Character|GetMesh _asfort_player_pawn) :AttachPointName "root" :Scale "1.000000,1.000000,1.000000" :LocationType "SnapToTarget" :bAutoDestroy true :PoolingMethod "None" :bAutoActivate true)
  ;[1 1 1] Unbind footstep and collision events
  (bind _asfort_player_pawn_athena (Utilities|Casting|CastToFortPlayerPawnAthena MyTarget))
  (EventDispatchers|UnbindEventfromOnFootstepEvent _asfort_player_pawn_athena (EventDispatchers|CreateEvent :!event Custom|On%20Player%20Step))
  (Collision|UnbindEventfromOnActorHit _asfort_player_pawn_athena (EventDispatchers|CreateEvent :!event Custom|On%20Target%20Pawn%20Hit%20Event))
  (EventDispatchers|UnbindEventfromOnPawnLanded _asfort_player_pawn_athena (EventDispatchers|CreateEvent :!event Custom|On%20Target%20Pawn%20Landed))
  (Character|UnbindEventfromMovementModeChangedDelegate _asfort_player_pawn_athena (EventDispatchers|CreateEvent :!event Custom|On%20Target%20Pawn%20Movement%20Mode%20Changed))
  (Components|DestroyComponent (Variables|Default|GetCollisionAudioComp) self))

(event Collision|EventHit
  (params (MyComp :type Primitive%20Component%20Object%20Reference) (Other :type Actor%20Object%20Reference) (OtherComp :type Primitive%20Component%20Object%20Reference) (bSelfMoved :type Boolean) (HitLocation :type Vector :default "0, 0, 0") (HitNormal :type Vector :default "0, 0, 0") (NormalImpulse :type Vector :default "0, 0, 0") (Hit :type Hit%20Result%20Structure%20%28by%20ref%29))
  (Development|PrintString MyComp))

(event Custom|On%20Player%20Step
  (params (FootIndex :type Integer))
  ;[1 1 1] On player step, play this audio
  (Audio|PlaySoundatLocation (Variables|Default|GetTargetPlayer) (select (Pawn|IsLocallyControlled (Variables|Default|GetTargetPlayer)) (Variables|Sound|GetSC_Footstep_Local) (Variables|Sound|GetSC_Footstep_Remote)) (Transformation|GetActorLocation (Variables|Default|GetTargetPlayer))))

(event Custom|On%20Target%20Pawn%20Hit%20Event
  (params (SelfActor :type Actor%20Object%20Reference) (OtherActor :type Actor%20Object%20Reference) (NormalImpulse :type Vector) (Hit :type Hit%20Result%20Structure%20%28by%20ref%29))
  (bind (_bblockinghit _binitialoverlap _time _distance _location _impactpoint _normal _impactnormal) (Collision|BreakHitResult Hit))
  ;[1 1 1] On player hit event, play this audio
  (Audio|OnCollision (Variables|Default|GetCollisionAudioComp) _location (* _impactnormal (Transformation|GetVelocity (Variables|Default|GetTargetPlayer)))))

(event EventTick
  (params (DeltaSeconds :type Float%20%28single-precision%29))
  (bind _spawnedaudiocomps (Variables|Default|GetSpawnedAudioComps))
  (bind _output (Utilities|Array|Get(acopy) _spawnedaudiocomps))
  ;[1 1 1] Modulate slide volume based on velocity
  (Utilities|IsValid _output
    (:Is%20Valid
      (if (or (AI|Components|NavMovement|IsFalling (Variables|/Script/Engine.Character|GetCharacterMovement (Variables|Default|GetTargetPlayer))) (Utilities|Enum|Equal(Enum) (Variables|/Script/Engine.CharacterMovementComponent|GetMovementMode (Variables|/Script/Engine.Character|GetCharacterMovement (Variables|Default|GetTargetPlayer))) "MOVE_Flying"))
        (Audio|Parameter|SetFloatParameter _output "Speed" 0.0 :!class /Script/AudioExtensions.AudioParameterControllerInterface)
        (else
          (Audio|Parameter|SetFloatParameter _output "Speed" (Math|Float|MapRangeClamped (Math|Vector|VectorLength (Transformation|GetVelocity (Variables|Default|GetTargetPlayer))) 0.0 1e+03 0.0 1.0) :!class /Script/AudioExtensions.AudioParameterControllerInterface))))))

(event Custom|On%20Target%20Pawn%20Landed
  (params (Hit :type Hit%20Result%20Structure%20%28by%20ref%29))
  ;[1 1 1] Landing sound
  (Audio|SpawnSoundAttached (Variables|Sound|GetSC_OnLand) (Variables|/Script/Engine.Actor|GetRootComponent (Variables|Default|GetTargetPlayer))))

(event Custom|On%20Target%20Pawn%20Movement%20Mode%20Changed
  (params (Character :type Character%20Object%20Reference) (PrevMovementMode :type EMovementMode%20Enum :default "MOVE_None") (PreviousCustomMode :type Byte))
  ;[1 1 1] Sound played when leaving ground
  (Utilities|IsValid Character
    (:Is%20Valid
      (switch Utilities|FlowControl|Switch|SwitchonEMovementMode (Variables|/Script/Engine.CharacterMovementComponent|GetMovementMode (Variables|/Script/Engine.Character|GetCharacterMovement Character))
        (:MOVE_Falling
          (Audio|SpawnSoundAttached (Variables|Sound|GetSC_OnJumpOrBeginFalling) (Variables|/Script/Engine.Actor|GetRootComponent (Variables|Default|GetTargetPlayer))))))))

(fn ConstructionScript)
