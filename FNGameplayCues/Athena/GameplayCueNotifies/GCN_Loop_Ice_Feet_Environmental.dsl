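; Looping gameplay cue that gives a player iced-feet effects and slide audio.
; On start it attaches two looping emitters to the foot bones, binds footstep, hit, landing and movement-mode events, and adds a collision audio component.
; Those events play footstep (local or remote), impact, landing and falling sounds, while Tick feeds velocity into the loop's Speed parameter.
; On removal it spawns ice-break particles on both feet and unbinds the events.
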(event EventOnLoopingStart(EitherVFX-preferthisversion)
  (params (MyTarget :type Actor%20Object%20Reference) (Parameters :type Gameplay%20Cue%20Parameters%20Structure%20%28by%20ref%29) (ParticleComponents :type Array%20of%20FXSystem%20Component%20Object%20References) (AudioComponents :type Array%20of%20Audio%20Component%20Object%20References))
  (bind _spawnedemitters (Variables|Default|GetSpawnedEmitters))
  ; On Looping Start
  (Variables|Default|SetSpawnedEmitters ParticleComponents)
  (Variables|Default|SetSpawnedAudioComps AudioComponents)
  (bind _output_get (Variables|Default|SetTargetActor MyTarget))
  (bind _asfort_player_pawn (Utilities|Casting|CastToFortPlayerPawn _output_get))
  (Transformation|AttachComponentToComponent (Utilities|Array|Get(acopy) _spawnedemitters) (Variables|/Script/Engine.Character|GetMesh _asfort_player_pawn) "foot_l" "SnapToTarget" "SnapToTarget" "KeepWorld" true)
  (Transformation|AttachComponentToComponent (Utilities|Array|Get(acopy) _spawnedemitters 1) (Variables|/Script/Engine.Character|GetMesh _asfort_player_pawn) "foot_r" "SnapToTarget" "SnapToTarget" "KeepWorld" true)
  (Effects|Components|ParticleSystem|SetFloatParameter (Variables|Default|GetSpawnedEmitters) "Time" (Utilities|Time|GetGameTimeinSeconds 0))
  (bind _asfort_player_pawn_athena (Utilities|Casting|CastToFortPlayerPawnAthena (Variables|Default|GetTargetActor)))
  (bind _output_get_2 (Variables|Default|SetTarget%20Player _asfort_player_pawn_athena))
  (Character|BindEventtoMovementModeChangedDelegate _output_get_2 (EventDispatchers|CreateEvent :!event Custom|OnPawnMovementModeChanged))
  (EventDispatchers|BindEventtoOnFootstepEvent _asfort_player_pawn_athena (EventDispatchers|CreateEvent :!event Custom|On%20Footstep%20Event))
  (bind _returnvalue (AddComponent|NODE_AddFortCollisionAudioComponent-5))
  (Variables|Default|SetCollisionAudioComp _returnvalue)
  (Collision|BindEventtoOnActorHit _asfort_player_pawn_athena (EventDispatchers|CreateEvent :!event Custom|OnTargetPawnHitEvent))
  (Audio|SetCollisionSoundatIndex (Variables|Default|GetCollisionAudioComp) 0 (Variables|Default|GetSC%20Collision))
  (EventDispatchers|BindEventtoOnPawnLanded _asfort_player_pawn_athena (EventDispatchers|CreateEvent :!event Custom|OnTargetPawnLanded)))

(event EventOnRemoval(EitherVFX-preferthisversion)
  (params (MyTarget :type Actor%20Object%20Reference) (Parameters :type Gameplay%20Cue%20Parameters%20Structure%20%28by%20ref%29) (ParticleComponents :type Array%20of%20FXSystem%20Component%20Object%20References) (AudioComponents :type Array%20of%20Audio%20Component%20Object%20References) (BurstCameraShakeInstance :type Camera%20Shake%20Base%20Object%20Reference) (BurstDecalInstance :type Decal%20Actor%20Object%20Reference))
  ; On Removal
  (bind _asfort_player_pawn_athena (Utilities|Casting|CastToFortPlayerPawnAthena MyTarget))
  (Effects|SpawnEmitterAttached :EmitterTemplate "/FNGameplayCues/Athena/Items/Traps/FreezeTrap/FX/P_Athena_Ice_Foot_Break.P_Athena_Ice_Foot_Break" :AttachToComponent (Variables|/Script/Engine.Character|GetMesh _asfort_player_pawn_athena) :AttachPointName "foot_l" :Scale "1.000000,1.000000,1.000000" :LocationType "SnapToTarget" :bAutoDestroy true :PoolingMethod "None" :bAutoActivate true)
  (Effects|SpawnEmitterAttached :EmitterTemplate "/FNGameplayCues/Athena/Items/Traps/FreezeTrap/FX/P_Athena_Ice_Foot_Break.P_Athena_Ice_Foot_Break" :AttachToComponent (Variables|/Script/Engine.Character|GetMesh _asfort_player_pawn_athena) :AttachPointName "foot_r" :Scale "1.000000,1.000000,1.000000" :LocationType "SnapToTarget" :bAutoDestroy true :PoolingMethod "None" :bAutoActivate true)
  (Character|UnbindEventfromMovementModeChangedDelegate _asfort_player_pawn_athena (EventDispatchers|CreateEvent :!event Custom|OnPawnMovementModeChanged))
  (EventDispatchers|UnbindEventfromOnFootstepEvent _asfort_player_pawn_athena (EventDispatchers|CreateEvent :!event Custom|On%20Footstep%20Event))
  (Collision|UnbindEventfromOnActorHit _asfort_player_pawn_athena (EventDispatchers|CreateEvent :!event Custom|OnTargetPawnHitEvent))
  (EventDispatchers|UnbindEventfromOnPawnLanded _asfort_player_pawn_athena (EventDispatchers|CreateEvent :!event Custom|OnTargetPawnLanded)))

(event EventTick
  (params (DeltaSeconds :type Float%20%28single-precision%29))
  ;[0 0 0] Modulate slide volume based on velocity
  (bind _targetactor (Variables|Default|GetTargetActor)
    (:then
      (bind _result (ArrayMacros|SafeArrayGet (Variables|Default|GetSpawnedAudioComps))
        (:%20%20
          (Audio|Parameter|SetFloatParameter _result "Speed" (Math|Float|MapRangeClamped (Math|Vector|VectorLength (Transformation|GetVelocity _targetactor)) 0.0 2e+03 0.0 1.0) :!class /Script/AudioExtensions.AudioParameterControllerInterface))))
    (:else)))

(event Custom|OnPawnMovementModeChanged
  (params (Character :type Character%20Object%20Reference) (PrevMovementMode :type EMovementMode%20Enum :default "MOVE_None") (PreviousCustomMode :type Byte))
  ; Sound played when leaving ground
  (Utilities|IsValid Character
    (:Is%20Valid
      (switch Utilities|FlowControl|Switch|SwitchonEMovementMode (Variables|/Script/Engine.CharacterMovementComponent|GetMovementMode (Variables|/Script/Engine.Character|GetCharacterMovement Character))
        (:MOVE_Falling
          (Audio|SpawnSoundAttached (Variables|Default|GetSC_OnJumpOrBeginFalling) (Variables|/Script/Engine.Actor|GetRootComponent (Variables|Default|GetTarget%20Player))))))))

(event Custom|On%20Footstep%20Event
  (params (FootIndex :type Integer))
  ;[0 0 0] On Pawn Footstep
  (bind _result (ArrayMacros|SafeArrayGet (Variables|Default|GetSpawnedAudioComps))
    (:%20%20
      (Audio|Parameter|ExecuteTriggerParameter _result "OnFootstep")
      (Audio|PlaySoundatLocation (Variables|Default|GetTarget%20Player) (select (Pawn|IsLocallyControlled (Variables|Default|GetTarget%20Player)) (Variables|Default|GetSC_Footstep_Local) (Variables|Default|GetSC_Footstep_Remote)) (Transformation|GetActorLocation (Variables|Default|GetTarget%20Player))))))

(event Custom|OnTargetPawnHitEvent
  (params (SelfActor :type Actor%20Object%20Reference) (OtherActor :type Actor%20Object%20Reference) (NormalImpulse :type Vector) (Hit :type Hit%20Result%20Structure%20%28by%20ref%29))
  (bind (_bblockinghit _binitialoverlap _time _distance _location _impactpoint _normal _impactnormal) (Collision|BreakHitResult Hit))
  ; On Player Hit Event, play this audio
  (Audio|OnCollision (Variables|Default|GetCollisionAudioComp) _location (* _impactnormal (Transformation|GetVelocity (Variables|Default|GetTarget%20Player)))))

(event Custom|OnTargetPawnLanded
  (params (Hit :type Hit%20Result%20Structure%20%28by%20ref%29))
  ; Landing Sound
  (Audio|SpawnSoundAttached (Variables|Default|GetSC_OnLand) (Variables|/Script/Engine.Actor|GetRootComponent (Variables|Default|GetTarget%20Player))))

(fn AttachFX
  (params (Mesh :type Scene%20Component%20Object%20Reference))
  (if (> (Utilities|Array|Length (Variables|Default|GetSpawnedEmitters)) 0)
    (Effects|Components|ParticleSystem|SetFloatParameter (Variables|Default|GetSpawnedEmitters) "Time" (Utilities|Time|GetGameTimeinSeconds (Variables|GetMesh)))
    (bind _result (ArrayMacros|SafeArrayGet (Variables|Default|GetSpawnedEmitters))
      (:%20%20
        (Transformation|AttachComponentToComponent _result (Variables|GetMesh) "foot_l" "SnapToTarget" "SnapToTarget" "KeepWorld" true)
        (bind _asniagara_particle_system_component (Utilities|Casting|CastToNiagaraComponent _result))
        (Rendering|SetVisibility _asniagara_particle_system_component true false)))))

(fn ConstructionScript)
