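; Looping gameplay cue that shows ice effects on a Fortnite vehicle's wheels while it is active.
; On start it attaches an ice emitter at each wheel FX socket, scaled up for tire mods, or ends if none exist.
; A repeating timer scales slide-loop volume by vehicle speed, muting it in the air.
; Removal deactivates emitters and clears that timer; the expire-warning crack and sound events are marked unused.
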(event Custom|BeginExpireTell
  ;[1 1 1] Handle ability expire warning - unused
  (bind _returnvalue (Utilities|Time|SetTimerbyEvent (delegate Custom|ExpireWarningTick) (Variables|Default|GetExpirationSoundPeriod) true))
  (Variables|Default|SetExpirationSoundTimer _returnvalue)
  (Variables|Default|SetCrack 0.0))

(event Custom|ExpireWarningTick
  ;[1 1 1] Handle ability expire warning - unused
  (bind _asactor (Utilities|Casting|CastToActor (Variables|Default|GetTargetVehicleInterface)))
  (Audio|SpawnSoundAttached "/Game/Sounds/Fort_Traps/Ice_Trap/IceTrap_PreExpireTell_Cue.IceTrap_PreExpireTell_Cue" (Variables|/Script/Engine.Actor|GetRootComponent _asactor))
  (bind _output_get (Variables|Default|SetCrack (+ (Variables|Default|GetCrack) (/ (- (Variables|Default|GetExpirationTellDuration) (Variables|Default|GetExpirationSoundPeriod)) (Variables|Default|GetExpirationTellDuration)))))
  (for _array_element (Variables|Default|GetLoopingEmitters)
    (Effects|Components|ParticleSystem|SetFloatParameter _array_element "Crack" _output_get)
    (Effects|Components|ParticleSystem|SetFloatParameter _array_element "Time" (Utilities|Time|GetGameTimeinSeconds 0))
    (Effects|SpawnEmitteratLocation 0 "/FNGameplayCues/Athena/Items/Traps/FreezeTrap/FX/P_Athena_Ice_Foot_Crack.P_Athena_Ice_Foot_Crack" (Transformation|GetWorldLocation _array_element))))

(event Custom|ClearExpireTell
  ;[1 1 1] Handle ability expire warning - unused
  (Utilities|Time|ClearandInvalidateTimerbyHandle 0 (Variables|Default|GetExpireTellDelayTimer))
  (Utilities|Time|ClearandInvalidateTimerbyHandle 0 (Variables|Default|GetExpirationSoundTimer)))

(event EventOnLoopingStart(Cascade)
  (params (MyTarget :type Actor%20Object%20Reference) (Parameters :type Gameplay%20Cue%20Parameters%20Structure%20%28by%20ref%29) (ParticleComponents :type Array%20of%20Cascade%20Particle%20System%20Component%20Object%20References) (AudioComponents :type Array%20of%20Audio%20Component%20Object%20References))
  (bind _loopingemitters (Variables|Default|GetLoopingEmitters))
  (Utilities|FlowControl|Sequence
    (:then_0
      (for _array_element _loopingemitters
        (Components|DestroyComponent _array_element 0))
      (Utilities|Array|Clear _loopingemitters))
    (:then_1
      (Variables|Default|SetSpawnedAudioComps AudioComponents)
      (Variables|Default|SetCrack 0.0))
    (:then_2
      (bind _asfort_vehicle_interface (Utilities|Casting|CastToFortVehicleInterface MyTarget))
      (bind _output_get (Variables|Default|SetTargetVehicleInterface _asfort_vehicle_interface))
      (if (> (Utilities|Array|Length (Vehicle|FX|GetWheelFXSockets _output_get)) 0)
        (bind _spawnedemitters1 (CallFunction|SpawnEmittersAttachedtoWheels (Vehicle|FX|GetWheelFXSockets _output_get) (Class|FortVehicleInterface|GetSkeletalMeshComponent _output_get) "/FNGameplayCues/Athena/Items/Traps/FreezeTrap/FX/P_Athena_Ice_Wheel.P_Athena_Ice_Wheel"))
        (Variables|Default|SetLoopingEmitters _spawnedemitters1)
        (else
          (GameplayCueNotify|End%28Recycle%29GameplayCue))))
    (:then_3
      (Utilities|Time|ClearandInvalidateTimerbyHandle self (Variables|Default|GetTimer_SoundModulation))
      (bind _returnvalue (Utilities|Time|SetTimerbyEvent (EventDispatchers|CreateEvent :!event Custom|Event%20ModulateSound) 0.2 true))
      (Variables|Default|SetTimer_SoundModulation _returnvalue))))

(event EventOnRemoval(Cascade)
  (params (MyTarget :type Actor%20Object%20Reference) (Parameters :type Gameplay%20Cue%20Parameters%20Structure%20%28by%20ref%29) (ParticleComponents :type Array%20of%20Cascade%20Particle%20System%20Component%20Object%20References) (AudioComponents :type Array%20of%20Audio%20Component%20Object%20References) (BurstCameraShakeInstance :type Camera%20Shake%20Base%20Object%20Reference) (BurstDecalInstance :type Decal%20Actor%20Object%20Reference))
  (Utilities|FlowControl|Sequence
    (:then_0
      (for _array_element (Variables|Default|GetLoopingEmitters)
        (Components|Activation|SetActive _array_element false false)))
    (:then_1
      (Utilities|Time|ClearandInvalidateTimerbyHandle self (Variables|Default|GetTimer_SoundModulation)))))

(event Custom|Event%20ModulateSound
  (bind _spawnedaudiocomps (Variables|Default|GetSpawnedAudioComps))
  (bind _output (Utilities|Array|Get(acopy) _spawnedaudiocomps))
  (bind _targetvehicleinterface (Variables|Default|GetTargetVehicleInterface))
  ;[1 1 1] Modulate slide volume based on velocity
  (Utilities|IsValid _output
    (:Is%20Valid
      (if (Vehicle|VehicleIsinAir _targetvehicleinterface)
        (Audio|Components|Audio|SetVolumeMultiplier _output 0.0 :!class /Script/Engine.AudioComponent)
        (else
          (Audio|Components|Audio|SetVolumeMultiplier _output (Math|Float|MapRangeClamped (Math|Vector|VectorLength (Vehicle|GetVehicleVelocity _targetvehicleinterface)) 0.0 1e+03 0.0 1.0) :!class /Script/Engine.AudioComponent))))))

(fn SpawnEmittersAttachedToWheels
  (params (Sockets :type Array%20of%20Vehicle%20Wheel%20FXSocket%20Info%20Structures) (AttachToComponent :type Scene%20Component%20Object%20Reference) (EmitterTemplate :type Particle%20System%20Object%20Reference))
  (Utilities|Array|Clear (LocalVariables|GetSpawnedEmitters))
  (for _array_element Sockets
    (bind _returnvalue (Effects|SpawnEmitterAttached :EmitterTemplate EmitterTemplate :AttachToComponent AttachToComponent :AttachPointName (Utilities|Struct|BreakVehicleWheelFXSocketInfo _array_element) :Location (.Offset (Utilities|Struct|BreakVehicleWheelFXSocketInfo _array_element)) :Scale (Math|Conversions|ToVector(Float) (* (.Scale (Utilities|Struct|BreakVehicleWheelFXSocketInfo _array_element)) (Math|Float|SelectFloat 1.5 1.0 (Vehicle|Mod|Tire|HasTireMod (Variables|Default|GetTargetVehicleInterface))))) :LocationType "SnapToTarget" :bAutoDestroy true :PoolingMethod "None" :bAutoActivate true))
    (Utilities|Array|Add (LocalVariables|GetSpawnedEmitters) _returnvalue))
  (return (LocalVariables|GetSpawnedEmitters)))

(fn ConstructionScript)
