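; Looping gameplay cue for environmental ice wheels on a vehicle, modulating the loop's sound with the vehicle's motion.
; On start it caches the vehicle interface and spawned audio, then runs a repeating timer until removal.
; Each tick mutes the first audio component while airborne, otherwise scales its volume by speed.
; An uncalled helper attaches emitters to wheel sockets, with per-wheel scaling and shopping-cart adjustments.
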(event EventOnLoopingStart(EitherVFX-preferthisversion)
  (params (MyTarget :type Actor%20Object%20Reference) (Parameters :type Gameplay%20Cue%20Parameters%20Structure%20%28by%20ref%29) (ParticleComponents :type Array%20of%20FXSystem%20Component%20Object%20References) (AudioComponents :type Array%20of%20Audio%20Component%20Object%20References))
  (bind _asfort_vehicle_interface (Utilities|Casting|CastToFortVehicleInterface MyTarget))
  (Variables|Default|SetTarget%20Vehicle%20Interface _asfort_vehicle_interface)
  (Utilities|Time|ClearandInvalidateTimerbyHandle self (Variables|Default|GetTimer%20Sound%20Modultation))
  (Variables|Default|SetSpawnedAudioComps AudioComponents)
  (bind _returnvalue (Utilities|Time|SetTimerbyEvent (EventDispatchers|CreateEvent :!event Custom|Event%20Modulate%20Sound) 0.2 true))
  (Variables|Default|SetTimer%20Sound%20Modultation _returnvalue))

(event EventOnRemoval(EitherVFX-preferthisversion)
  (params (MyTarget :type Actor%20Object%20Reference) (Parameters :type Gameplay%20Cue%20Parameters%20Structure%20%28by%20ref%29) (ParticleComponents :type Array%20of%20FXSystem%20Component%20Object%20References) (AudioComponents :type Array%20of%20Audio%20Component%20Object%20References) (BurstCameraShakeInstance :type Camera%20Shake%20Base%20Object%20Reference) (BurstDecalInstance :type Decal%20Actor%20Object%20Reference))
  (Utilities|Time|ClearandInvalidateTimerbyHandle self (Variables|Default|GetTimer%20Sound%20Modultation)))

(event Custom|Event%20Modulate%20Sound
  (bind _spawnedaudiocomps (Variables|Default|GetSpawnedAudioComps))
  (bind _output (Utilities|Array|Get(acopy) _spawnedaudiocomps))
  (bind _target_vehicle_interface (Variables|Default|GetTarget%20Vehicle%20Interface))
  (Utilities|IsValid _output
    (:Is%20Valid
      (if (Vehicle|VehicleIsinAir _target_vehicle_interface)
        (Audio|Components|Audio|SetVolumeMultiplier _output 0.0 :!class /Script/Engine.AudioComponent)
        (else
          (Audio|Components|Audio|SetVolumeMultiplier _output (Math|Float|MapRangeClamped (Math|Vector|VectorLength (Vehicle|GetVehicleVelocity _target_vehicle_interface)) 0.0 1e+03 0.0 1.0) :!class /Script/Engine.AudioComponent))))))

(fn SpawnEmittersAttachedToWheels
  (params (Sockets :type Array%20of%20Names) (AttachToComponent :type Scene%20Component%20Object%20Reference) (ShoppingCart :type Boolean) (EmitterTemplate :type Particle%20System%20Object%20Reference))
  (Utilities|Array|Clear (LocalVariables|GetSpawnedEmitters))
  (for (_array_element _array_index) Sockets
    (bind _returnvalue (Effects|SpawnEmitterAttached :EmitterTemplate EmitterTemplate :AttachToComponent AttachToComponent :AttachPointName _array_element :Location (Math|Vector|SelectVector "0, 0, 0" (Math|Vector|SelectVector "0,8.000000, 0" "0,-8.000000, 0" (or (== _array_index 1) (== _array_index 3))) ShoppingCart) :Scale (Math|Conversions|ToVector(Float) (* (Utilities|Select :Index _array_index :Option%200 1.07 :Option%201 1.07 :Option%202 0.95 :Option%203 0.95) (Math|Float|SelectFloat 0.3 1.0 ShoppingCart))) :LocationType "SnapToTarget" :bAutoDestroy true :PoolingMethod "None" :bAutoActivate true))
    (Utilities|Array|Add (LocalVariables|GetSpawnedEmitters) _returnvalue))
  (return (LocalVariables|GetSpawnedEmitters)))

(fn ConstructionScript)
