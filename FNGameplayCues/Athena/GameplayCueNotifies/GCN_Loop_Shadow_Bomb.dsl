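; Looping gameplay cue for the Shadow Bomb item that drives the misty player look, attached effects and audio while active.
; On start it fades opacity and sets SpookyMistMode for local, replay or broadcast views; otherwise it checks team and hides non-local players' weapons.
; Timers drive velocity-based audio and a flashing expiry warning, and a footstep audio mutator runs until removal.
; Setup repeats when character customization finishes, avoiding an Xbox One first-throw invisibility bug.
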(event Custom|BeginExpireTell
  ;[1 1 1] Handle ability expire warning - local player only
  (Actor|Tick|SetActorTickEnabled false)
  (bind _returnvalue (Utilities|Time|SetTimerbyEvent (delegate Custom|ExpireWarningTick) (Variables|Default|GetExpirationSoundPeriod) true))
  (Variables|Default|SetExpirationSoundTimer _returnvalue))

(event Custom|ExpireWarningTick
  (bind _expiretellopacity (Variables|Default|GetExpireTellOpacity))
  (bind _attachedfx (Variables|Default|GetAttachedFX))
  ;[1 1 1] Handle ability expire warning - local player only
  (Audio|SpawnSoundAttached "/FNGameplayCues/Athena/Sounds/Items/Shadow_Bomb/Sahdow_Bomb_Expire_Tell_Cue.Sahdow_Bomb_Expire_Tell_Cue" (Variables|Transformation|GetRootComponent (Variables|Default|GetTargetPlayer)))
  (for _array_element (Variables|Default|GetPlayerSkeletalMeshes)
    (Rendering|Material|SetScalarParameterValueonMaterials _array_element "FlashStartTime" (Utilities|Time|GetGameTimeinSeconds 0))
    (if (Math|Boolean|ORBoolean (Pawn|IsLocallyControlled (Variables|Default|GetTargetPlayer)) (Replay|GetIsPlayingReplay 0) (Broadcast|GetIsBroadcastClient 0))
      _
      (elif (> (Utilities|Array|Length _attachedfx) 0)
        (Effects|Components|ParticleSystem|SetFloatParameter (Utilities|Array|Get(acopy) _attachedfx) "MovementOpacity" _expiretellopacity)))
    (if (> (Utilities|Array|Length (Variables|Default|GetPlayerSkeletalMeshes)) 0)
      (Rendering|Material|SetScalarParameterValueonMaterials _array_element "MovementOpacity" _expiretellopacity))
    (Variables|Default|SetExpireTellOpacity (Math|Float|Clamp(Float) (+ (Variables|Default|GetExpireTellOpacity) 0.15)))))

(event Custom|StartedInteract
  ;[1 1 1] Fade in and out when interacting
  (bind (_direction _visbility) (Timeline|Timeline_0 Play)
    (:Update
      (Variables|Default|SetInteractVisibility _visbility))))

(event Custom|ClearExpireTell
  ;[1 1 1] Handle ability expire warning - local player only
  (Utilities|Time|ClearandInvalidateTimerbyHandle 0 (Variables|Default|GetExpireTellDelayTimer))
  (Utilities|Time|ClearandInvalidateTimerbyHandle 0 (Variables|Default|GetExpirationSoundTimer)))

(event Custom|EndedInteract
  ;[1 1 1] Fade in and out when interacting
  (Timeline|Timeline_0 Stop))

(event EventOnLoopingStart(Cascade)
  (params (MyTarget :type Actor%20Object%20Reference) (Parameters :type Gameplay%20Cue%20Parameters%20Structure%20%28by%20ref%29) (ParticleComponents :type Array%20of%20Cascade%20Particle%20System%20Component%20Object%20References) (AudioComponents :type Array%20of%20Audio%20Component%20Object%20References))
  (bind _playerskeletalmeshes (Variables|Default|GetPlayerSkeletalMeshes))
  (bind _targetplayer (Variables|Default|GetTargetPlayer))
  (bind _attachedfx (Variables|Default|GetAttachedFX))
  (bind _currentweapon (Variables|/Script/FortniteGame.FortPawn|GetCurrentWeapon _targetplayer))
  (bind _weaponmesh (Variables|/Script/FortniteGame.FortWeapon|GetWeaponMesh _currentweapon))
  (Utilities|FlowControl|Sequence
    (:then_0
      (Utilities|Time|SetTimerbyEvent (EventDispatchers|CreateEvent :!event Custom|Audio%20Tick) 0.1 true)
      (Variables|Default|SetVisbilityLevel 1.0)
      (Variables|Default|SetFriendlyPlayer false)
      (Variables|Default|SetTickENabled true)
      (Components|Timeline|ReversefromEnd (Variables|Default|GetFadeInOut))
      (Variables|Default|SetWalkVisMultiplier (ScalableFloat|GetValueatLevel (Variables|Default|GetDataWalkVisMult)))
      (Variables|Default|SetSprintVisMultiplier (ScalableFloat|GetValueatLevel (Variables|Default|GetDataSprintVisMult)))
      (Variables|Default|SetMinFriendlyOpacity (ScalableFloat|GetValueatLevel (Variables|Default|GetDataFriendlyVisMult)))
      (Variables|Default|SetExpireTellOpacity 0.0))
    (:then_1
      (bind _output_get (Variables|Default|SetAttachedFX ParticleComponents))
      (if (> (Utilities|Array|Length _output_get) 0)
        (Effects|Components|ParticleSystem|SetFloatParameter (Utilities|Array|Get(acopy) (Variables|Default|GetAttachedFX)) "MovementOpacity" 0.0))
      (if (> (Utilities|Array|Length _playerskeletalmeshes) 0)
        (for _array_element _playerskeletalmeshes
          (Rendering|Material|SetScalarParameterValueonMaterials _array_element "MovementOpacity" 0.0))))
    (:then_2
      (bind _asfort_player_pawn (Utilities|Casting|CastToFortPlayerPawn MyTarget))
      (bind _output_get_2 (Variables|Default|SetTargetPlayer _asfort_player_pawn))
      ;[1 0.01042 0.01042] DON'T USE DELAYS: This fixes the Xbox One issue where the first time you throw the smoke bomb, you are invisible instead of misty.
      ; We're binding to when the customization is done then do the things.
      ; We're also just passing all the way through in case the bind doesn't get hit on time.
      (Customization|BindEventtoOnFinishedCharacterCustomization _output_get_2 (delegate Custom|OnFinishedCharacterCustomization_Bind))
      (Utilities|FlowControl|Sequence
        (:then_0
          ;[1 1 1] Fade in and out when interacting
          (Components|Timeline|SetNewTime (Variables|Default|GetFadeInOut) 0.0))
        (:then_1
          ;[1 1 1] Handle ability expire warning - local player only
          (CallFunction|ClearExpireTell)
          (bind _returnvalue (Utilities|Time|SetTimerbyEvent (EventDispatchers|CreateEvent :!event Custom|BeginExpireTell) (- (select (Math|Float|NearlyEqual(Float) (Variables|Default|GetSmokeBombDurationUserSpecified) 0.0 0.01) (ScalableFloat|GetValueatLevel (Variables|Default|GetSmokeBombDuration)) (Variables|Default|GetSmokeBombDurationUserSpecified)) (Variables|Default|GetTellDuration))))
          (Variables|Default|SetExpireTellDelayTimer _returnvalue))
        (:then_2
          (if (Math|Boolean|ORBoolean (Pawn|IsLocallyControlled (Variables|Default|GetTargetPlayer)) (Replay|GetIsPlayingReplay 0) (Broadcast|GetIsBroadcastClient 0))
            (Rendering|Material|SetScalarParameterValue 0 "/Game/Packages/Fortress_SharedMaterials/GlobalMaterialParameters/FortniteMaterialParameters.FortniteMaterialParameters" "SpookyMistMode" (ScalableFloat|GetValueatLevel (Variables|Default|GetDataShadowPlaneOpacity)) :!class /Script/Engine.KismetMaterialLibrary) ; Reveal other players using ghost rocks
            (Variables|Default|SetFriendlyPlayer true)
            (if (> (Utilities|Array|Length _attachedfx) 0)
              (Effects|Components|ParticleSystem|SetFloatParameter (Utilities|Array|Get(acopy) _attachedfx 1) "Friendly" 1.0))
            (CallFunction|GetSkeletalMeshes)
            (for _array_element_2 (Variables|Default|GetPlayerSkeletalMeshes)
              (Rendering|Material|SetScalarParameterValueonMaterials _array_element_2 "MovementOpacity" 1.0))
            (else
              (CallFunction|GetSkeletalMeshes)
              (for _array_element_3 (Fortnite|GetLocalFortPlayerControllers 0)
                (bind _targetplayer_2 (Variables|Default|GetTargetPlayer)
                  (:then
                    (bind _output_get_3 (Variables|Default|SetFriendlyPlayer (Team|OnSameTeam (Variables|/Script/Engine.Pawn|GetPlayerState _targetplayer_2) (Variables|/Script/Engine.Controller|GetPlayerState _array_element_3))))
                    (if (not (Pawn|IsLocallyControlled (Variables|Default|GetTargetPlayer)))
                      (FortPawn|AssignOnWeaponEquippedDelegate _targetplayer (delegate Custom|OnWeaponEquippedDelegate_Event_0))
                      (Utilities|IsValid _currentweapon
                        (:Is%20Valid
                          (Utilities|IsValid _weaponmesh
                            (:Is%20Valid
                              (Rendering|SetVisibility _weaponmesh false true)))))))
                  (:else))))))))
    (:then_3
      (CallFunction|ActivateFootstepAudioMutator))))

(event Custom|OnFinishedCharacterCustomization_Bind
  (params (Pawn :type Fort%20Player%20Pawn%20Object%20Reference))
  (bind _targetplayer (Variables|Default|GetTargetPlayer))
  (bind _attachedfx (Variables|Default|GetAttachedFX))
  (bind _currentweapon (Variables|/Script/FortniteGame.FortPawn|GetCurrentWeapon _targetplayer))
  (bind _weaponmesh (Variables|/Script/FortniteGame.FortWeapon|GetWeaponMesh _currentweapon))
  ;[1 0.01042 0.01042] DON'T USE DELAYS: This fixes the Xbox One issue where the first time you throw the smoke bomb, you are invisible instead of misty.
  ; We're binding to when the customization is done then do the things.
  ; We're also just passing all the way through in case the bind doesn't get hit on time.
  (Customization|UnbindEventfromOnFinishedCharacterCustomization Pawn (delegate Custom|OnFinishedCharacterCustomization_Bind))
  (Utilities|FlowControl|Sequence
    (:then_0
      ;[1 1 1] Fade in and out when interacting
      (Components|Timeline|SetNewTime (Variables|Default|GetFadeInOut) 0.0))
    (:then_1
      ;[1 1 1] Handle ability expire warning - local player only
      (CallFunction|ClearExpireTell)
      (bind _returnvalue (Utilities|Time|SetTimerbyEvent (EventDispatchers|CreateEvent :!event Custom|BeginExpireTell) (- (select (Math|Float|NearlyEqual(Float) (Variables|Default|GetSmokeBombDurationUserSpecified) 0.0 0.01) (ScalableFloat|GetValueatLevel (Variables|Default|GetSmokeBombDuration)) (Variables|Default|GetSmokeBombDurationUserSpecified)) (Variables|Default|GetTellDuration))))
      (Variables|Default|SetExpireTellDelayTimer _returnvalue))
    (:then_2
      (if (Math|Boolean|ORBoolean (Pawn|IsLocallyControlled (Variables|Default|GetTargetPlayer)) (Replay|GetIsPlayingReplay 0) (Broadcast|GetIsBroadcastClient 0))
        (Rendering|Material|SetScalarParameterValue 0 "/Game/Packages/Fortress_SharedMaterials/GlobalMaterialParameters/FortniteMaterialParameters.FortniteMaterialParameters" "SpookyMistMode" (ScalableFloat|GetValueatLevel (Variables|Default|GetDataShadowPlaneOpacity)) :!class /Script/Engine.KismetMaterialLibrary) ; Reveal other players using ghost rocks
        (Variables|Default|SetFriendlyPlayer true)
        (if (> (Utilities|Array|Length _attachedfx) 0)
          (Effects|Components|ParticleSystem|SetFloatParameter (Utilities|Array|Get(acopy) _attachedfx 1) "Friendly" 1.0))
        (CallFunction|GetSkeletalMeshes)
        (for _array_element (Variables|Default|GetPlayerSkeletalMeshes)
          (Rendering|Material|SetScalarParameterValueonMaterials _array_element "MovementOpacity" 1.0))
        (else
          (CallFunction|GetSkeletalMeshes)
          (for _array_element_2 (Fortnite|GetLocalFortPlayerControllers 0)
            (bind _targetplayer_2 (Variables|Default|GetTargetPlayer)
              (:then
                (bind _output_get (Variables|Default|SetFriendlyPlayer (Team|OnSameTeam (Variables|/Script/Engine.Pawn|GetPlayerState _targetplayer_2) (Variables|/Script/Engine.Controller|GetPlayerState _array_element_2))))
                (if (not (Pawn|IsLocallyControlled (Variables|Default|GetTargetPlayer)))
                  (FortPawn|AssignOnWeaponEquippedDelegate _targetplayer (delegate Custom|OnWeaponEquippedDelegate_Event_0))
                  (Utilities|IsValid _currentweapon
                    (:Is%20Valid
                      (Utilities|IsValid _weaponmesh
                        (:Is%20Valid
                          (Rendering|SetVisibility _weaponmesh false true)))))))
              (:else))))))))

(event Custom|UpdateTeamVisbility
  (bind _targetplayer (Variables|Default|GetTargetPlayer))
  (bind _currentweapon (Variables|/Script/FortniteGame.FortPawn|GetCurrentWeapon _targetplayer))
  (bind _weaponmesh (Variables|/Script/FortniteGame.FortWeapon|GetWeaponMesh _currentweapon))
  (CallFunction|GetSkeletalMeshes)
  (for _array_element (Fortnite|GetLocalFortPlayerControllers 0)
    (bind _targetplayer_2 (Variables|Default|GetTargetPlayer)
      (:then
        (bind _output_get (Variables|Default|SetFriendlyPlayer (Team|OnSameTeam (Variables|/Script/Engine.Pawn|GetPlayerState _targetplayer_2) (Variables|/Script/Engine.Controller|GetPlayerState _array_element))))
        (if (not (Pawn|IsLocallyControlled (Variables|Default|GetTargetPlayer)))
          (FortPawn|AssignOnWeaponEquippedDelegate _targetplayer (delegate Custom|OnWeaponEquippedDelegate_Event_0))
          (Utilities|IsValid _currentweapon
            (:Is%20Valid
              (Utilities|IsValid _weaponmesh
                (:Is%20Valid
                  (Rendering|SetVisibility _weaponmesh false true)))))))
      (:else))))

(event Custom|Audio%20Tick
  (bind _targetplayer (Variables|Default|GetTargetPlayer)
    (:then
      (bind _output_get (Variables|Default|SetVelocityInterp (Math|Interpolation|FInterpTo (Variables|Default|GetVelocityInterp) (Math|Float|MapRangeClamped (Math|Float|Absolute(Float) (.z (Transformation|GetVelocity _targetplayer))) 6e+02 1e+03 0.0 1.0) (Utilities|Time|GetWorldDeltaSeconds 0) 5.0)))
      (Audio|Parameter|SetFloatParameter (Utilities|Array|Get(acopy) (stale EventOnLoopingStart(Cascade) AudioComponents)) "Velocity" _output_get :!class /Script/AudioExtensions.AudioParameterControllerInterface))
    (:else)))

(event Custom|OnWeaponEquippedDelegate_Event_0
  (params (NewWeapon :type Fort%20Weapon%20Object%20Reference) (PrevWeapon :type Fort%20Weapon%20Object%20Reference))
  (Utilities|IsValid PrevWeapon
    (:Is%20Valid
      (Rendering|SetVisibility (Variables|/Script/FortniteGame.FortWeapon|GetWeaponMesh PrevWeapon) true true))
    (:Is%20Not%20Valid
      _))
  (Utilities|IsValid NewWeapon
    (:Is%20Valid
      (Rendering|SetVisibility (Variables|/Script/FortniteGame.FortWeapon|GetWeaponMesh NewWeapon) false true))))

(event EventOnRemoval(Cascade)
  (params (MyTarget :type Actor%20Object%20Reference) (Parameters :type Gameplay%20Cue%20Parameters%20Structure%20%28by%20ref%29) (ParticleComponents :type Array%20of%20Cascade%20Particle%20System%20Component%20Object%20References) (AudioComponents :type Array%20of%20Audio%20Component%20Object%20References) (BurstCameraShakeInstance :type Camera%20Shake%20Base%20Object%20Reference) (BurstDecalInstance :type Decal%20Actor%20Object%20Reference))
  (bind _targetplayer (Variables|Default|GetTargetPlayer))
  (bind _targetplayer_2 (Variables|Default|GetTargetPlayer))
  (bind _currentweapon (Variables|/Script/FortniteGame.FortPawn|GetCurrentWeapon _targetplayer_2))
  (bind _asfort_player_pawn (Utilities|Casting|CastToFortPlayerPawn MyTarget)
    (:then
      (Variables|Default|SetTargetPlayer _asfort_player_pawn))
    (:CastFailed
      _))
  (Utilities|FlowControl|Sequence
    (:then_0
      (EventDispatchers|UnbindEventfromOnStartedInteractSearch _targetplayer (EventDispatchers|CreateEvent :!event Custom|StartedInteract))
      (EventDispatchers|UnbindEventfromOnStartedInteractSearch _targetplayer (EventDispatchers|CreateEvent :!event Custom|EndedInteract))
      (CallFunction|ClearExpireTell))
    (:then_1
      (if (Pawn|IsLocallyControlled (Variables|Default|GetTargetPlayer))
        (Rendering|Material|SetScalarParameterValue 0 "/Game/Packages/Fortress_SharedMaterials/GlobalMaterialParameters/FortniteMaterialParameters.FortniteMaterialParameters" "SpookyMistMode" (Math|Conversions|ToFloat(Boolean) (Replay|GetIsPlayingReplay 0)) :!class /Script/Engine.KismetMaterialLibrary)
        (else
          (Utilities|IsValid _targetplayer_2
            (:Is%20Valid
              (CallFunction|UnbindEquippedWeapon)
              (Utilities|IsValid _currentweapon
                (:Is%20Valid
                  (Rendering|SetVisibility (Variables|/Script/FortniteGame.FortWeapon|GetWeaponMesh _currentweapon) true true)))))))
      (Variables|Default|SetTargetPlayer 0))
    (:then_2
      (CallFunction|DeactivateFootstepAudioMutator))))

(event EventTick
  (params (DeltaSeconds :type Float%20%28single-precision%29)))

(event Custom|ActivateFootstepAudioMutator
  (Audio|Mutator|AddMutator (AudioEngineSubsystems|GetAudioMutatorsSubsystem 0) (Variables|Default|GetFootstepMutator)))

(event Custom|DeactivateFootstepAudioMutator
  (Audio|Mutator|RemoveMutator (AudioEngineSubsystems|GetAudioMutatorsSubsystem 0) (Variables|Default|GetFootstepMutator)))

(fn GetSkeletalMeshes
  (bind _targetplayer (Variables|Default|GetTargetPlayer))
  (Utilities|Array|Clear (Variables|Default|GetPlayerSkeletalMeshes))
  (Utilities|IsValid _targetplayer
    (:Is%20Valid
      (for _array_element (Components|GetChildrenComponents (Variables|/Script/Engine.Character|GetMesh _targetplayer) false)
        (bind _asskeletal_mesh_component (Utilities|Casting|CastToSkeletalMeshComponent _array_element))
        (Utilities|Array|AddUnique (Variables|Default|GetPlayerSkeletalMeshes) _asskeletal_mesh_component)))
    (:Is%20Not%20Valid
      (Development|PrintString 0 "not valid")))
  (return))

(fn UnbindEquippedWeapon
  (bind _targetplayer (Variables|Default|GetTargetPlayer))
  (Utilities|IsValid _targetplayer
    (:Is%20Valid
      (FortPawn|UnbindEventfromOnWeaponEquippedDelegate _targetplayer (EventDispatchers|CreateEvent :!event OnWeaponEquippedDelegate_Event_0)))))

(fn ConstructionScript)
