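; Looping gameplay cue for a player launched from a ship cannon, handling launch audio and character rotation.
; On start it turns off controller-desired rotation on the player pawn, plays a cannon fire sound and spawns an attached in-air loop.
; Both sounds switch between first-person and standard cues based on whether the pawn is locally controlled.
; On removal it turns controller-desired rotation back on and stops the stored loop.
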(event EventBeginPlay)

(event Collision|EventActorBeginOverlap
  (params (OtherActor :type Actor%20Object%20Reference)))

(event EventTick
  (params (DeltaSeconds :type Float%20%28single-precision%29)))

(event EventOnLoopingStart(EitherVFX-preferthisversion)
  (params (MyTarget :type Actor%20Object%20Reference) (Parameters :type Gameplay%20Cue%20Parameters%20Structure%20%28by%20ref%29) (ParticleComponents :type Array%20of%20FXSystem%20Component%20Object%20References) (AudioComponents :type Array%20of%20Audio%20Component%20Object%20References))
  (bind _asfort_player_pawn_athena (Utilities|Casting|CastToFortPlayerPawnAthena MyTarget))
  (Variables|/Script/Engine.CharacterMovementComponent|SetbUseControllerDesiredRotation false (Variables|/Script/Engine.Character|GetCharacterMovement _asfort_player_pawn_athena))
  (Audio|PlaySoundatLocation _asfort_player_pawn_athena (select (Pawn|IsLocallyControlled _asfort_player_pawn_athena) "/FNGameplayCues/Athena/Sounds/Vehicles/ShipCannon/Cannon_Fire_Player_1P_Cue.Cannon_Fire_Player_1P_Cue" "/Game/Athena/Sounds/Vehicles/ShipCannon/Cannon_Fire_Player_Cue.Cannon_Fire_Player_Cue") (Transformation|GetActorLocation _asfort_player_pawn_athena))
  (bind _returnvalue (Audio|SpawnSoundAttached (select (Pawn|IsLocallyControlled _asfort_player_pawn_athena) "/FNGameplayCues/Athena/Sounds/Vehicles/ShipCannon/Cannon_Fire_CannonBall_InAirLoop_1P_Cue.Cannon_Fire_CannonBall_InAirLoop_1P_Cue" "/FNGameplayCues/Athena/Sounds/Vehicles/ShipCannon/Cannon_Fire_CannonBall_InAirLoop_Cue.Cannon_Fire_CannonBall_InAirLoop_Cue") (Variables|/Script/Engine.Actor|GetRootComponent _asfort_player_pawn_athena)))
  (Variables|Default|SetInAirLoop _returnvalue))

(event EventOnRemoval(EitherVFX-preferthisversion)
  (params (MyTarget :type Actor%20Object%20Reference) (Parameters :type Gameplay%20Cue%20Parameters%20Structure%20%28by%20ref%29) (ParticleComponents :type Array%20of%20FXSystem%20Component%20Object%20References) (AudioComponents :type Array%20of%20Audio%20Component%20Object%20References) (BurstCameraShakeInstance :type Camera%20Shake%20Base%20Object%20Reference) (BurstDecalInstance :type Decal%20Actor%20Object%20Reference))
  (bind _asfort_player_pawn_athena (Utilities|Casting|CastToFortPlayerPawnAthena MyTarget))
  (Variables|/Script/Engine.CharacterMovementComponent|SetbUseControllerDesiredRotation true (Variables|/Script/Engine.Character|GetCharacterMovement _asfort_player_pawn_athena))
  (Audio|Components|Audio|Stop (Variables|Default|GetInAirLoop)))

(fn ConstructionScript)
