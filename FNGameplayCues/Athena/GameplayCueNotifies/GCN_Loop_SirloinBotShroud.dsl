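; Looping gameplay cue for the SirloinBot shroud, a ghost-style effect that varies the target player's visibility.
; Tick blends movement, vehicle, interaction and the global SpookyMistMode parameter into MovementOpacity on particles and child meshes, with a minimum for teammates, and feeds speed to audio.
; Start sets up interact-fade bindings, team checks and, for local, replay or broadcast views, mist mode plus a timed expiry warning with sound and flash.
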(event Custom|BeginExpireTell
  ;[1 1 1] Handle ability expire warning - local player only
  (bind _returnvalue (Utilities|Time|SetTimerbyEvent (delegate Custom|ExpireWarningTick) (Variables|Default|GetExpirationSoundPeriod) true))
  (Variables|Default|SetExpirationSoundTimer _returnvalue))

(event Custom|ExpireWarningTick
  ;[1 1 1] Handle ability expire warning - local player only
  (Audio|SpawnSoundAttached "/FNGameplayCues/Sounds/Fort_GamePlay_Sounds/Player/GhostRocks/GhostRocks_ExpireTell_Cue.GhostRocks_ExpireTell_Cue" (Variables|/Script/Engine.Actor|GetRootComponent (Variables|Default|GetTargetPlayer)))
  (for _array_element (Variables|Default|GetPlayerSkeletalMeshes)
    (Rendering|Material|SetScalarParameterValueonMaterials _array_element "FlashStartTime" (Utilities|Time|GetGameTimeinSeconds 0))))

(event Custom|StartedInteract
  ;[1 1 1] Fade in and out when interacting
  (bind (_direction _visbility) (Timeline|Timeline_0 Play)
    (:Update
      (Variables|Default|SetInteractVisibility _visbility))))

(event Custom|ClearExpireTell
  ;[1 1 1] Handle ability expire warning - local player only
  (Utilities|Time|ClearandInvalidateTimerbyHandle 0 (Variables|Default|GetExpireTellDelayTimer))
  (Utilities|Time|ClearandInvalidateTimerbyHandle 0 (Variables|Default|GetExpirationSoundTimer)))

(event Custom|EndedInteract
  ;[1 1 1] Fade in and out when interacting
  (Timeline|Timeline_0 Reverse))

(event EventOnLoopingStart(Cascade)
  (params (MyTarget :type Actor%20Object%20Reference) (Parameters :type Gameplay%20Cue%20Parameters%20Structure%20%28by%20ref%29) (ParticleComponents :type Array%20of%20Cascade%20Particle%20System%20Component%20Object%20References) (AudioComponents :type Array%20of%20Audio%20Component%20Object%20References))
  (bind _targetplayer (Variables|Default|GetTargetPlayer))
  (Utilities|FlowControl|Sequence
    (:then_0
      (Variables|Default|SetVisbilityLevel 1.0)
      (Variables|Default|SetFriendlyPlayer false)
      (Actor|Tick|SetActorTickEnabled true)
      (Components|Timeline|ReversefromEnd (Variables|Default|GetFadeInOut))
      (Variables|Default|SetWalkVisMultiplier (ScalableFloat|GetValueatLevel (Variables|Default|GetDataWalkVisMult)))
      (Variables|Default|SetSprintVisMultiplier (ScalableFloat|GetValueatLevel (Variables|Default|GetDataSprintVisMult)))
      (Variables|Default|SetMinFriendlyOpacity (ScalableFloat|GetValueatLevel (Variables|Default|GetDataFriendlyVisMult))))
    (:then_1
      (Variables|Default|SetAttachedFX ParticleComponents)
      (bind _asfort_player_pawn (Utilities|Casting|CastToFortPlayerPawn MyTarget))
      (Variables|Default|SetTargetPlayer _asfort_player_pawn)
      (Utilities|FlowControl|Sequence
        (:then_0
          ;[1 1 1] Fade in and out when interacting
          (Components|Timeline|SetNewTime (Variables|Default|GetFadeInOut) 0.0)
          (EventDispatchers|BindEventtoOnStartedInteractSearch _targetplayer (EventDispatchers|CreateEvent :!event Custom|StartedInteract))
          (EventDispatchers|BindEventtoOnEndedInteractSearch _targetplayer (EventDispatchers|CreateEvent :!event Custom|EndedInteract)))
        (:then_1
          (if (Math|Boolean|ORBoolean (Pawn|IsLocallyControlled (Variables|Default|GetTargetPlayer)) (Replay|GetIsPlayingReplay 0) (Broadcast|GetIsBroadcastClient 0))
            (Utilities|FlowControl|Sequence
              (:then_0
                ;[1 1 1] Handle ability expire warning - local player only
                (CallFunction|ClearExpireTell)
                (bind _returnvalue (Utilities|Time|SetTimerbyEvent (EventDispatchers|CreateEvent :!event Custom|BeginExpireTell) (- (ScalableFloat|GetValueatLevel (Variables|Default|GetDataDuration)) (ScalableFloat|GetValueatLevel (Variables|Default|GetDataTellDuration)))))
                (Variables|Default|SetExpireTellDelayTimer _returnvalue))
              (:then_1
                (Rendering|Material|SetScalarParameterValue 0 "/Game/Packages/Fortress_SharedMaterials/GlobalMaterialParameters/FortniteMaterialParameters.FortniteMaterialParameters" "SpookyMistMode" (ScalableFloat|GetValueatLevel (Variables|Default|GetDataShadowPlaneOpacity)) :!class /Script/Engine.KismetMaterialLibrary) ; Reveal other players using ghost rocks
                (Variables|Default|SetFriendlyPlayer true)
                (Effects|Components|ParticleSystem|SetFloatParameter (Utilities|Array|Get(acopy) (Variables|Default|GetAttachedFX) 1) "Friendly" 1.0)
                (Utilities|FlowControl|Delay 0 0.25) ; Give mesh swap time to finish
                (CallFunction|GetSkeletalMeshes)))
            (else
              (Utilities|FlowControl|Delay 0 0.25) ; Give mesh swap time to finish
              (CallFunction|GetSkeletalMeshes)
              (for _array_element (Fortnite|GetLocalFortPlayerControllers 0)
                (bind _targetplayer_2 (Variables|Default|GetTargetPlayer)
                  (:then
                    (bind _output_get (Variables|Default|SetFriendlyPlayer (Team|OnSameTeam (Variables|/Script/Engine.Pawn|GetPlayerState _targetplayer_2) (Variables|/Script/Engine.Controller|GetPlayerState _array_element))))
                    (Effects|Components|ParticleSystem|SetFloatParameter (Utilities|Array|Get(acopy) (Variables|Default|GetAttachedFX) 1) "Friendly" (Math|Conversions|ToFloat(Boolean) _output_get)))
                  (:else))))))))))

(event Custom|Audio%20Tick
  (params (Velocity :type Float%20%28double-precision%29))
  (bind _output_get (Variables|Default|SetVelocityInterp (Math|Interpolation|FInterpTo (Variables|Default|GetVelocityInterp) (Math|Float|MapRangeClamped Velocity 6e+02 1e+03 0.0 1.0) (Utilities|Time|GetWorldDeltaSeconds 0) 1.5)))
  (Audio|Parameter|SetFloatParameter (Utilities|Array|Get(acopy) (stale EventOnLoopingStart(Cascade) AudioComponents)) "Velocity" _output_get :!class /Script/AudioExtensions.AudioParameterControllerInterface))

(event Custom|UpdateTeamVisbility
  (CallFunction|GetSkeletalMeshes)
  (for _array_element (Fortnite|GetLocalFortPlayerControllers 0)
    (bind _targetplayer (Variables|Default|GetTargetPlayer)
      (:then
        (bind _output_get (Variables|Default|SetFriendlyPlayer (Team|OnSameTeam (Variables|/Script/Engine.Pawn|GetPlayerState _targetplayer) (Variables|/Script/Engine.Controller|GetPlayerState _array_element))))
        (Effects|Components|ParticleSystem|SetFloatParameter (Utilities|Array|Get(acopy) (Variables|Default|GetAttachedFX) 1) "Friendly" (Math|Conversions|ToFloat(Boolean) _output_get)))
      (:else))))

(event EventOnRemoval(Cascade)
  (params (MyTarget :type Actor%20Object%20Reference) (Parameters :type Gameplay%20Cue%20Parameters%20Structure%20%28by%20ref%29) (ParticleComponents :type Array%20of%20Cascade%20Particle%20System%20Component%20Object%20References) (AudioComponents :type Array%20of%20Audio%20Component%20Object%20References) (BurstCameraShakeInstance :type Camera%20Shake%20Base%20Object%20Reference) (BurstDecalInstance :type Decal%20Actor%20Object%20Reference))
  (bind _targetplayer (Variables|Default|GetTargetPlayer))
  (bind _asfort_player_pawn (Utilities|Casting|CastToFortPlayerPawn MyTarget)
    (:then
      (Variables|Default|SetTargetPlayer _asfort_player_pawn))
    (:CastFailed
      _))
  (Utilities|FlowControl|Sequence
    (:then_0
      (Actor|Tick|SetActorTickEnabled false)
      (EventDispatchers|UnbindEventfromOnStartedInteractSearch _targetplayer (EventDispatchers|CreateEvent :!event Custom|StartedInteract))
      (EventDispatchers|UnbindEventfromOnStartedInteractSearch _targetplayer (EventDispatchers|CreateEvent :!event Custom|EndedInteract))
      (CallFunction|ClearExpireTell))
    (:then_1
      (if (Pawn|IsLocallyControlled (Variables|Default|GetTargetPlayer))
        (Rendering|Material|SetScalarParameterValue 0 "/Game/Packages/Fortress_SharedMaterials/GlobalMaterialParameters/FortniteMaterialParameters.FortniteMaterialParameters" "SpookyMistMode" (Math|Conversions|ToFloat(Boolean) (Replay|GetIsPlayingReplay 0)) :!class /Script/Engine.KismetMaterialLibrary))
      (Variables|Default|SetTargetPlayer 0))))

(event EventTick
  (params (DeltaSeconds :type Float%20%28single-precision%29))
  (bind _friendlyplayer (Variables|Default|GetFriendlyPlayer))
  (bind _returnvalue (Math|Conversions|ToFloat(Boolean) _friendlyplayer))
  (bind _minfriendlyopacity (Variables|Default|GetMinFriendlyOpacity))
  (bind _returnvalue_2 (* _returnvalue _minfriendlyopacity))
  (bind _playerskeletalmeshes (Variables|Default|GetPlayerSkeletalMeshes))
  ;[1 1 1] Event tick
  (bind _targetplayer (Variables|Default|GetTargetPlayer)
    (:then
      (Utilities|FlowControl|Sequence
        (:then_0
          (bind _returnvalue_3 (Rendering|Material|GetScalarParameterValue 0 "/Game/Packages/Fortress_SharedMaterials/GlobalMaterialParameters/FortniteMaterialParameters.FortniteMaterialParameters" "SpookyMistMode" :!class /Script/Engine.KismetMaterialLibrary))
          (bind _output_get (Variables|Default|SetVisbilityLevel (Math|Interpolation|FInterpTo (Variables|Default|GetVisbilityLevel) (Math|Float|Clamp(Float) (Math|Float|float+float (* _returnvalue_3 (Math|Conversions|ToFloat(Boolean) (not (Pawn|IsLocallyControlled (Variables|Default|GetTargetPlayer))))) (* (Variables|Default|GetWalkVisMultiplier) (Math|Conversions|ToFloat(Boolean) (> (Math|Vector|VectorLength (Transformation|GetVelocity _targetplayer)) 7e+01))) (* (Variables|Default|GetSprintVisMultiplier) (Math|Conversions|ToFloat(Boolean) (Math|Boolean|ORBoolean (> (Math|Vector|VectorLength (Transformation|GetVelocity _targetplayer)) 6.7e+02) (Vehicle|IsinVehicle _targetplayer) (not (Utilities|Enum|Equal(Enum) (Variables|/Script/Engine.CharacterMovementComponent|GetMovementMode (Variables|/Script/Engine.Character|GetCharacterMovement _targetplayer)) "MOVE_Walking"))))) (Variables|Default|GetInteractVisibility))) (Utilities|Time|GetWorldDeltaSeconds 0) 6.4)))
          (Effects|Components|ParticleSystem|SetFloatParameter (Utilities|Array|Get(acopy) (Variables|Default|GetAttachedFX)) "MovementOpacity" (Math|Float|Clamp(Float) _output_get _returnvalue_2))
          (if (> (Utilities|Array|Length _playerskeletalmeshes) 0)
            (for _array_element _playerskeletalmeshes
              (Rendering|Material|SetScalarParameterValueonMaterials _array_element "MovementOpacity" (Math|Float|Clamp(Float) _output_get _returnvalue_2)))))
        (:then_2
          (CallFunction|AudioTick (Math|Vector|VectorLength (Transformation|GetVelocity _targetplayer)))
          (Utilities|FlowControl|Delay 0 3.0)
          (if (Variables|GameplayCueNotify|GetbTickEnabled)
            (CallFunction|UpdateTeamVisbility)))))
    (:else)))

(fn GetSkeletalMeshes
  (bind _targetplayer (Variables|Default|GetTargetPlayer))
  (Utilities|Array|Clear (Variables|Default|GetPlayerSkeletalMeshes))
  (Utilities|IsValid _targetplayer
    (:Is%20Valid
      (for _array_element (Components|GetChildrenComponents (Variables|/Script/Engine.Character|GetMesh _targetplayer) false)
        (bind _asskeletal_mesh_component (Utilities|Casting|CastToSkeletalMeshComponent _array_element))
        (Utilities|Array|AddUnique (Variables|Default|GetPlayerSkeletalMeshes) _asskeletal_mesh_component)))
    (:Is%20Not%20Valid
      (Development|PrintString 0 "not valid")))
  (return))

(fn ConstructionScript)
