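; Looping snow whiteout gameplay cue that fades the local player's camera to white with a post-process effect and attached fade sounds.
; On start, the FadeInOut timeline drives manual camera fade and post-process blend weight; normalized magnitude 1 makes it near-instant and sets the sound's FastFade parameter.
; Removal reverses the timeline at normal rate with a fade-out sound.
; Completed reverse fade or destruction clears the camera fade and disables post-processing.
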(event EventOnLoopingStart(Cascade)
  (params (MyTarget :type Actor%20Object%20Reference) (Parameters :type Gameplay%20Cue%20Parameters%20Structure%20%28by%20ref%29) (ParticleComponents :type Array%20of%20Cascade%20Particle%20System%20Component%20Object%20References) (AudioComponents :type Array%20of%20Audio%20Component%20Object%20References))
  (bind _normalizedmagnitude (Ability|GameplayCue|BreakGameplayCueParameters Parameters))
  (bind _returnvalue (== _normalizedmagnitude 1.0))
  (Variables|/Script/Engine.PostProcessComponent|SetbEnabled true (Variables|Default|GetPostProcess))
  (Components|Timeline|SetPlayRate (Variables|GCN_Loop_Snow_Whiteout|GetFadeinOut) (select _returnvalue 1e+02 1.0))
  (bind _returnvalue_2 (Audio|SpawnSoundAttached (Variables|Default|GetSoundOnFadeIn) (Variables|/Script/Engine.Actor|GetRootComponent MyTarget)))
  (Audio|Parameter|SetIntegerParameter _returnvalue_2 "FastFade" (Math|Conversions|ToInteger(Boolean) _returnvalue) :!class /Script/AudioExtensions.AudioParameterControllerInterface)
  (Utilities|FlowControl|Sequence
    (:then_0
      (bind (_direction _fade _post) (Timeline|FadeInOut PlayFromStart)
        (:Update
          (Variables|/Script/Engine.PostProcessComponent|SetBlendWeight _post (Variables|Default|GetPostProcess))
          (CameraFades|SetManualCameraFade (Variables|/Script/Engine.PlayerController|GetPlayerCameraManager (Utilities|Array|Get(acopy) (Fortnite|GetLocalFortPlayerControllers 0 true true))) _fade "(R=1.000000,G=1.000000,B=1.000000,A=1.000000)" false))
        (:Finished
          (switch Utilities|FlowControl|Switch|SwitchonETimelineDirection _direction
            (:Backward
              _)))))
    (:then_1
      (Utilities|FlowControl|Delay 0 6.5)))
  (CameraFades|StopCameraFade (Variables|/Script/Engine.PlayerController|GetPlayerCameraManager (Utilities|Array|Get(acopy) (Fortnite|GetLocalFortPlayerControllers 0 true true))))
  (Variables|/Script/Engine.PostProcessComponent|SetbEnabled false (Variables|Default|GetPostProcess)))

(event EventOnRemoval(Cascade)
  (params (MyTarget :type Actor%20Object%20Reference) (Parameters :type Gameplay%20Cue%20Parameters%20Structure%20%28by%20ref%29) (ParticleComponents :type Array%20of%20Cascade%20Particle%20System%20Component%20Object%20References) (AudioComponents :type Array%20of%20Audio%20Component%20Object%20References) (BurstCameraShakeInstance :type Camera%20Shake%20Base%20Object%20Reference) (BurstDecalInstance :type Decal%20Actor%20Object%20Reference))
  (Components|Timeline|SetPlayRate (Variables|GCN_Loop_Snow_Whiteout|GetFadeinOut) 1.0)
  (Audio|SpawnSoundAttached (Variables|Default|GetSoundOnFadeOut) (Variables|/Script/Engine.Actor|GetRootComponent MyTarget))
  (Timeline|FadeInOut ReverseFromEnd))

(event EventDestroyed
  (CameraFades|StopCameraFade (Variables|/Script/Engine.PlayerController|GetPlayerCameraManager (Utilities|Array|Get(acopy) (Fortnite|GetLocalFortPlayerControllers 0 true true))))
  (Variables|/Script/Engine.PostProcessComponent|SetbEnabled false (Variables|Default|GetPostProcess)))

(fn ConstructionScript)
