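; Looping gameplay cue for the Ghost Rocks (Specter) effect, controlling a player's visibility while active.
; On start it caches FX, audio and meshes, binds interact fades and, for local rendering, enables post-process, mist-reveal global parameter and expiry-warning timers.
; Tick interpolates opacity from movement, interaction and mist reveal, with a friendly minimum, driving materials, particles and velocity audio.
; Team FX assumes one local controller; split screen takes the last one's result.
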(event EventOnLoopingStart(Cascade)
  (params (MyTarget :type Actor%20Object%20Reference) (Parameters :type Gameplay%20Cue%20Parameters%20Structure%20%28by%20ref%29) (ParticleComponents :type Array%20of%20Cascade%20Particle%20System%20Component%20Object%20References) (AudioComponents :type Array%20of%20Audio%20Component%20Object%20References))
  (bind _returnvalue 1)
  (bind _returnvalue_2 0)
  (bind _targetplayer (Variables|Default|GetTargetPlayer))
  ;[0 1 0.05708] Looping Start
  (Utilities|FlowControl|Sequence
    (:then_0
      ;;[0 1 0.09341] Initialze variables
      (bind _asfort_player_pawn (Utilities|Casting|CastToFortPlayerPawn MyTarget)
        (:then
          (Variables|Default|SetTargetPlayer _asfort_player_pawn))
        (:CastFailed
          _))
      (Variables|Default|SetVisbilityLevel 1.0)
      (Actor|Tick|SetActorTickEnabled true)
      (Variables|Default|SetWalkVisMultiplier (ScalableFloat|GetValueatLevel (Variables|Default|GetDataWalkVisMult)))
      (Variables|Default|SetSprintVisMultiplier (ScalableFloat|GetValueatLevel (Variables|Default|GetDataSprintVisMult)))
      (Variables|Default|SetMinFriendlyOpacity (ScalableFloat|GetValueatLevel (Variables|Default|GetDataFriendlyVisMult))))
    (:then_1
      ;;;[1 0.9595 0.07245] ASSUMPTION: Particle Components[0] is for the head particle component
      ;;; Particle Components[1] is for the body particle component
      ;;; If that ever changes this code will need to be updated as well.
      (if (Utilities|Array|IsValidIndex ParticleComponents _returnvalue_2)
        (Variables|Default|SetAttached%20FX%20Head (Utilities|Array|Get(acopy) ParticleComponents _returnvalue_2))
        (else
          _))
      (if (Utilities|Array|IsValidIndex ParticleComponents _returnvalue)
        (Variables|Default|SetAttached%20FX%20Body (Utilities|Array|Get(acopy) ParticleComponents _returnvalue))))
    (:then_2
      (Variables|Default|SetLooping%20Audio%20Components AudioComponents))
    (:then_3
      ;;[0.241 0.4519 1] Hook up callbacks to fade in and out when interacting
      (Components|Timeline|SetNewTime (Variables|GCN_Loop_Specter|GetTimeline%20Interaction%20Fade%20In%20Out) 0.0)
      (EventDispatchers|BindEventtoOnStartedInteractSearch _targetplayer (EventDispatchers|CreateEvent :!event Custom|StartedInteract))
      (EventDispatchers|BindEventtoOnEndedInteractSearch _targetplayer (EventDispatchers|CreateEvent :!event Custom|EndedInteract)))
    (:then_4
      ;;[0 0.9642 1] Set-up for the local player
      (if (CallFunction|ShouldRenderasLocalPlayer)
        (Utilities|FlowControl|Sequence
          (:then_0
            ;;;[1 0.2138 0.6332] Handle ability expire warning - local player only
            (if (Variables|Default|GetbShowExpireTell%20Const)
              (CallFunction|ClearExpireTell)
              (bind _returnvalue_3 (Utilities|Time|SetTimerbyEvent (EventDispatchers|CreateEvent :!event Custom|BeginExpireTell) (- (ScalableFloat|GetValueatLevel (Variables|Default|GetDataDuration)) (ScalableFloat|GetValueatLevel (Variables|Default|GetDataTellDuration)))))
              (Variables|Default|SetExpireTellDelayTimer _returnvalue_3)))
          (:then_1
            ;;;[0.7478 0 1] Set-up VFX for the local player
            (bind _targetplayer_2 (Variables|Default|GetTargetPlayer)
              (:then
                (Variables|/Script/Engine.PostProcessComponent|SetbEnabled (Pawn|IsLocallyControlled _targetplayer_2) (Variables|Default|GetPostProcess))
                (Rendering|Material|SetScalarParameterValue 0 "/Game/Packages/Fortress_SharedMaterials/GlobalMaterialParameters/FortniteMaterialParameters.FortniteMaterialParameters" (Variables|ParameterName|GetMat%20Parameter%20Name%20Spooky%20Mist%20Mode%20Const) (ScalableFloat|GetValueatLevel (Variables|Default|GetDataShadowPlaneOpacity)) :!class /Script/Engine.KismetMaterialLibrary) ; Reveal other players using ghost rocks
              )
              (:else))))))
    (:then_5
      ;;[0.7159 1 0] Initialize the skeletal meshes
      (CallFunction|InitializePlayerSkeletalMeshes))))

(event Custom|Initialize%20Player%20Skeletal%20Meshes
  (bind _targetplayer (Variables|Default|GetTargetPlayer))
  ;[0.4358 1 0] Initialize Player Skeletal Meshes
  (Utilities|Array|Clear (Variables|Default|GetPlayerSkeletalMeshes))
  (Utilities|IsValid _targetplayer
    (:Is%20Valid
      (for _array_element (Components|GetChildrenComponents (Variables|/Script/Engine.Character|GetMesh _targetplayer) false)
        (bind _asskeletal_mesh_component (Utilities|Casting|CastToSkeletalMeshComponent _array_element))
        (Utilities|Array|AddUnique (Variables|Default|GetPlayerSkeletalMeshes) _asskeletal_mesh_component))
      (if (<= (Utilities|Array|Length (Variables|Default|GetPlayerSkeletalMeshes)) 0)
        ;;[1 0.5431 0] We have not yet found any skeletal meshes, schedule a retry.
        ;; The assumption here is our target has one or more skeletal meshes so we will eventually find something
        (Utilities|Time|SetTimerbyEvent (EventDispatchers|CreateEvent :!event Custom|Initialize%20Player%20Skeletal%20Meshes) 0.001)))))

(event EventOnRemoval(Cascade)
  (params (MyTarget :type Actor%20Object%20Reference) (Parameters :type Gameplay%20Cue%20Parameters%20Structure%20%28by%20ref%29) (ParticleComponents :type Array%20of%20Cascade%20Particle%20System%20Component%20Object%20References) (AudioComponents :type Array%20of%20Audio%20Component%20Object%20References) (BurstCameraShakeInstance :type Camera%20Shake%20Base%20Object%20Reference) (BurstDecalInstance :type Decal%20Actor%20Object%20Reference))
  (bind _targetplayer (Variables|Default|GetTargetPlayer))
  (bind _targetplayer_2 (Variables|Default|GetTargetPlayer))
  ;[1 0.4543 0] Handle Removal
  (bind _asfort_player_pawn (Utilities|Casting|CastToFortPlayerPawn MyTarget)
    (:then
      (Variables|Default|SetTargetPlayer _asfort_player_pawn))
    (:CastFailed
      _))
  (Utilities|FlowControl|Sequence
    (:then_0
      (Actor|Tick|SetActorTickEnabled false)
      (CallFunction|ClearExpireTell)
      (Variables|/Script/Engine.PostProcessComponent|SetbEnabled false (Variables|Default|GetPostProcess))
      (Utilities|IsValid _targetplayer
        (:Is%20Valid
          (EventDispatchers|UnbindEventfromOnStartedInteractSearch _targetplayer (EventDispatchers|CreateEvent :!event Custom|StartedInteract))
          (EventDispatchers|UnbindEventfromOnStartedInteractSearch _targetplayer (EventDispatchers|CreateEvent :!event Custom|EndedInteract)))))
    (:then_1
      (Utilities|IsValid _targetplayer_2
        (:Is%20Valid
          (if (Pawn|IsLocallyControlled _targetplayer_2)
            (Rendering|Material|SetScalarParameterValue 0 "/Game/Packages/Fortress_SharedMaterials/GlobalMaterialParameters/FortniteMaterialParameters.FortniteMaterialParameters" (Variables|ParameterName|GetMat%20Parameter%20Name%20Spooky%20Mist%20Mode%20Const) (Math|Conversions|ToFloat(Boolean) (Replay|GetIsPlayingReplay 0)) :!class /Script/Engine.KismetMaterialLibrary)
            (else
              _))
          (Variables|Default|SetTargetPlayer 0))))))

(event EventTick
  (params (DeltaSeconds :type Float%20%28single-precision%29))
  (bind _attached_fx_body (Variables|Default|GetAttached%20FX%20Body))
  (bind _mat_parameter_name_movement_opacity_const (Variables|ParameterName|GetMat%20Parameter%20Name%20Movement%20Opacity%20Const))
  (bind _attached_fx_body_2 (Variables|Default|GetAttached%20FX%20Body))
  (bind _attached_fx_body_3 (Variables|Default|GetAttached%20FX%20Body))
  ;[0.07843 0.4235 1] Event tick
  (Variables|Default|GetTargetPlayer
    (:then
      (Utilities|FlowControl|Sequence
        (:then_0
          ;;[0 1 1] Checks for what the player is doing. Determines the material opacity
          (bind _returnvalue (CallFunction|TickMovementOpacity))
          (Utilities|IsValid _attached_fx_body_3
            (:Is%20Valid
              (Effects|Components|ParticleSystem|SetFloatParameter _attached_fx_body_3 _mat_parameter_name_movement_opacity_const _returnvalue))
            (:Is%20Not%20Valid
              _))
          (Utilities|IsValid _attached_fx_body_2
            (:Is%20Valid
              (Effects|Components|ParticleSystem|SetFloatParameter _attached_fx_body_2 _mat_parameter_name_movement_opacity_const _returnvalue))
            (:Is%20Not%20Valid
              _))
          (for _array_element (Variables|Default|GetPlayerSkeletalMeshes)
            (Rendering|Material|SetScalarParameterValueonMaterials _array_element _mat_parameter_name_movement_opacity_const _returnvalue)))
        (:then_1
          ;;[0.1911 0.1703 1] Set if Ribbon effects are on or off
          (bind _should_show_speed_ribbons (CallFunction|ShouldSpeedRibbonsbeVisible))
          (Utilities|IsValid _attached_fx_body
            (:Is%20Valid
              (Effects|Components|ParticleSystem|SetFloatParameter _attached_fx_body (Variables|ParameterName|GetFX%20Parameter%20Name%20Ribbons%20Const) (Math|Conversions|ToFloat(Boolean) _should_show_speed_ribbons)))))
        (:then_2
          ;;[0.05415 1 0.7377] Update Team Visibility
          (CallFunction|UpdateTeamVisbility))
        (:then_3
          ;;[0.6823 0.3646 1] Perform the Audio Tick
          (CallFunction|AudioTick (Math|Vector|VectorLength (Transformation|GetVelocity (Variables|Default|GetTargetPlayer)))))))
    (:else)))

(event Custom|Audio%20Tick
  (params (Velocity :type Float%20%28double-precision%29))
  (bind _looping_audio_components (Variables|Default|GetLooping%20Audio%20Components))
  (bind _output (Utilities|Array|Get(acopy) _looping_audio_components))
  ;[0.2431 0.2588 1] Audio Tick
  (if (Utilities|Array|IsValidIndex _looping_audio_components)
    (Utilities|IsValid _output
      (:Is%20Valid
        (bind _output_get (Variables|Default|SetAudioVelocityInterp (Math|Interpolation|FInterpTo (Variables|Default|GetAudioVelocityInterp) (Math|Float|MapRangeClamped Velocity 6e+02 1e+03 0.0 1.0) (Utilities|Time|GetWorldDeltaSeconds 0) 1.5)))
        (Audio|Parameter|SetFloatParameter _output (Variables|ParameterName|GetAudio%20Parameter%20Name%20Velocity%20Const) _output_get :!class /Script/AudioExtensions.AudioParameterControllerInterface)))))

(event Custom|StartedInteract
  ;[0.2314 0.2118 0.4824] Interact (increase opacity while interacting)
  (bind (_direction _visbility) (Timeline|Timeline%20Interaction%20Fade%20In%20Out Play)
    (:Update
      (Variables|Default|SetInteractVisibility _visbility))))

(event Custom|EndedInteract
  ;[0.2314 0.2118 0.4824] Interact (increase opacity while interacting)
  (Timeline|Timeline%20Interaction%20Fade%20In%20Out Reverse))

(event Custom|UpdateTeamVisbility
  (bind _attached_fx_body (Variables|Default|GetAttached%20FX%20Body))
  ;;[1 0.1098 0.6784] Update what team player is on for visibility
  ;[0 0 0] TODO: THIS CODE IS NOT HANDLING SPLIT SCREEN PROPERLY
  ; This code will work fine if there is a single local Player Controller,
  ; but in split screen it will set the Friendly material param to the calculated value for the last local Player Controller
  ; If you has questions ask Christopher.Daniel@xa.epicgames.com
  (for _array_element (Fortnite|GetLocalFortPlayerControllers self)
    (Utilities|IsValid _attached_fx_body
      (:Is%20Valid
        (Effects|Components|ParticleSystem|SetFloatParameter _attached_fx_body (Variables|ParameterName|GetFX%20Parameter%20Name%20Friendly%20Const) (Math|Conversions|ToFloat(Boolean) (CallFunction|IsPlayerControllerFriendly _array_element)))))))

(event Custom|BeginExpireTell
  ;[0.543 0 1] Expiration Warning Sequence
  (bind _returnvalue (Utilities|Time|SetTimerbyEvent (delegate Custom|ExpireWarningTick) (Variables|Default|GetExpirationSoundPeriod) true))
  (Variables|Default|SetExpirationSoundTimer _returnvalue))

(event Custom|ExpireWarningTick
  ;[0.543 0 1] Expiration Warning Sequence
  (Audio|SpawnSoundAttached "/FNGameplayCues/Sounds/Fort_GamePlay_Sounds/Player/GhostRocks/GhostRocks_ExpireTell_Cue.GhostRocks_ExpireTell_Cue" (Variables|/Script/Engine.Actor|GetRootComponent (Variables|Default|GetTargetPlayer)))
  (for _array_element (Variables|Default|GetPlayerSkeletalMeshes)
    (Rendering|Material|SetScalarParameterValueonMaterials _array_element (Variables|ParameterName|GetMat%20Parameter%20Name%20Flash%20Start%20TimeConst) (Utilities|Time|GetGameTimeinSeconds 0))))

(event Custom|ClearExpireTell
  ;[0.543 0 1] Expiration Warning Sequence
  (Utilities|Time|ClearandInvalidateTimerbyHandle 0 (Variables|Default|GetExpireTellDelayTimer))
  (Utilities|Time|ClearandInvalidateTimerbyHandle 0 (Variables|Default|GetExpirationSoundTimer)))

(fn GetGoalVisibilityLevel
  (bind _targetplayer (Variables|Default|GetTargetPlayer)
    (:then
      (bind _should_show_speed_ribbons (CallFunction|ShouldSpeedRibbonsbeVisible))
      (bind _returnvalue (Rendering|Material|GetScalarParameterValue 0 "/Game/Packages/Fortress_SharedMaterials/GlobalMaterialParameters/FortniteMaterialParameters.FortniteMaterialParameters" (Variables|ParameterName|GetMat%20Parameter%20Name%20Spooky%20Mist%20Mode%20Const) :!class /Script/Engine.KismetMaterialLibrary))
      ;[1 0.5052 0.5052] Return goal visibility level
      (return (Math|Float|Clamp(Float) (Math|Float|float+float (* _returnvalue (Math|Conversions|ToFloat(Boolean) (not (Pawn|IsLocallyControlled _targetplayer)))) (* (Variables|Default|GetWalkVisMultiplier) (Math|Conversions|ToFloat(Boolean) (> (Math|Vector|VectorLength (Transformation|GetVelocity _targetplayer)) 7e+01))) (* (Variables|Default|GetSprintVisMultiplier) (Math|Conversions|ToFloat(Boolean) _should_show_speed_ribbons)) (Variables|Default|GetInteractVisibility)))))
    (:else
      ;[1 0.5052 0.5052] Failed to find Target Player,
      ; return current visibility level
      (return (Variables|Default|GetVisbilityLevel)))))

(fn GetSkeletalMeshes
  (bind _targetplayer (Variables|Default|GetTargetPlayer))
  (Utilities|Array|Clear (Variables|Default|GetPlayerSkeletalMeshes))
  (Utilities|IsValid _targetplayer
    (:Is%20Valid
      (for _array_element (Components|GetChildrenComponents (Variables|/Script/Engine.Character|GetMesh _targetplayer) false)
        (bind _asskeletal_mesh_component (Utilities|Casting|CastToSkeletalMeshComponent _array_element))
        (Utilities|Array|AddUnique (Variables|Default|GetPlayerSkeletalMeshes) _asskeletal_mesh_component)))
    (:Is%20Not%20Valid
      (Development|PrintString 0 "GCN_Loop_SpookyMist::GetSkeletalMeshes() Target Player is NOT valid"))))

(fn IsPlayerControllerFriendly
  (params (PlayerController :type Controller%20Object%20Reference))
  (Utilities|IsValid PlayerController
    (:Is%20Valid
      (bind _targetplayer (Variables|Default|GetTargetPlayer)
        (:then
          ;[1 0.4635 0.4635] Both players are valid, return if they are on the same team
          (return (Team|OnSameTeam (Variables|/Script/Engine.Pawn|GetPlayerState _targetplayer) (Variables|/Script/Engine.Controller|GetPlayerState PlayerController))))
        (:else
          _)))
    (:Is%20Not%20Valid
      _))
  ;[1 0.4635 0.4635] One or both players is nullptr, return FALSE
  (return false))

(fn ShouldRenderAsLocalPlayer
  (bind _targetplayer (Variables|Default|GetTargetPlayer)
    (:then
      ;[1 0.4896 0.4896] Return if this GC shourender as if this is the local player
      (return (Math|Boolean|ORBoolean (Pawn|IsLocallyControlled _targetplayer) (Replay|GetIsPlayingReplay 0) (Broadcast|GetIsBroadcastClient 0))))
    (:else
      ;[1 0.4896 0.4896] Could not find Target Player, Return FALSE
      (return false))))

(fn ShouldSpeedRibbonsBeVisible
  (bind _targetplayer (Variables|Default|GetTargetPlayer)
    (:then
      ;[1 0.5052 0.4896] Return TRUE if
      ; A) the player is moving fast enough
      ; OR B) the player is in a vehicle
      ; OR C) the player is NOT in walking movement mode
      (return (Math|Boolean|ORBoolean (> (Math|Vector|VectorLength (Transformation|GetVelocity _targetplayer)) (Variables|Default|GetSpeed%20Ribbon%20Velocity%20Threshold%20Const)) (Vehicle|IsinVehicle _targetplayer) (not (Utilities|Enum|Equal(Enum) (Variables|/Script/Engine.CharacterMovementComponent|GetMovementMode (Variables|/Script/Engine.Character|GetCharacterMovement _targetplayer)) "MOVE_Walking")))))
    (:else
      ;[1 0.4692 0.4635] No Target Player, return FALSE
      (return false))))

(fn TickMovementOpacity
  (Utilities|FlowControl|Sequence
    (:then_0
      ;[0.07549 0.4836 1] Interp Visibility Level to the desired level
      (bind _returnvalue (CallFunction|GetGoalVisibilityLevel))
      (Variables|Default|SetVisbilityLevel (Math|Float|Clamp(Float) (Math|Interpolation|FInterpTo (Variables|Default|GetVisbilityLevel) _returnvalue (Utilities|Time|GetWorldDeltaSeconds 0) (Variables|Default|GetVisibilityLevel%20Change%20Speed%20Const)) (* (Math|Conversions|ToFloat(Boolean) (CallFunction|IsPlayerControllerFriendly (Fortnite|GetFirstLocalFortPlayerController self))) (Variables|Default|GetMinFriendlyOpacity)))))
    (:then_1
      ;[1 0.474 0.474] Return Visibility Level
      (return (Variables|Default|GetVisbilityLevel)))))

(fn ConstructionScript)
