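; Looping gameplay cue for a spooky mist effect that reacts when the target player pawn starts or stops an interact search.
; On loop start it resets the fade timeline and binds the pawn's search events to play and reverse it.
; Comments describe this as raising opacity, but no timeline update logic appears.
; Removal unbinds both handlers from OnStartedInteractSearch, leaving the OnEndedInteractSearch binding in place.
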(event EventOnLoopingStart(Cascade)
  (params (MyTarget :type Actor%20Object%20Reference) (Parameters :type Gameplay%20Cue%20Parameters%20Structure%20%28by%20ref%29) (ParticleComponents :type Array%20of%20Cascade%20Particle%20System%20Component%20Object%20References) (AudioComponents :type Array%20of%20Audio%20Component%20Object%20References))
  ;[0 1 0.05708] Looping Start
  ;;[0.241 0.4519 1] Hook up callbacks to fade in and out when interacting
  (Components|Timeline|SetNewTime (Variables|GCN_Loop_SpookyMist|GetTimeline%20Interaction%20Fade%20In%20Out) 0.0)
  (bind _asfort_player_pawn (Utilities|Casting|CastToFortPlayerPawn MyTarget))
  (EventDispatchers|BindEventtoOnStartedInteractSearch _asfort_player_pawn (EventDispatchers|CreateEvent :!event Custom|StartedInteract))
  (EventDispatchers|BindEventtoOnEndedInteractSearch _asfort_player_pawn (EventDispatchers|CreateEvent :!event Custom|EndedInteract)))

(event EventOnRemoval(Cascade)
  (params (MyTarget :type Actor%20Object%20Reference) (Parameters :type Gameplay%20Cue%20Parameters%20Structure%20%28by%20ref%29) (ParticleComponents :type Array%20of%20Cascade%20Particle%20System%20Component%20Object%20References) (AudioComponents :type Array%20of%20Audio%20Component%20Object%20References) (BurstCameraShakeInstance :type Camera%20Shake%20Base%20Object%20Reference) (BurstDecalInstance :type Decal%20Actor%20Object%20Reference))
  ;[1 0.4543 0] Handle Removal
  (bind _asfort_player_pawn (Utilities|Casting|CastToFortPlayerPawn MyTarget))
  (Utilities|IsValid _asfort_player_pawn
    (:Is%20Valid
      (EventDispatchers|UnbindEventfromOnStartedInteractSearch _asfort_player_pawn (EventDispatchers|CreateEvent :!event Custom|StartedInteract))
      (EventDispatchers|UnbindEventfromOnStartedInteractSearch _asfort_player_pawn (EventDispatchers|CreateEvent :!event Custom|EndedInteract)))))

(event Custom|StartedInteract
  ;[0.2314 0.2118 0.4824] Interact (increase opacity while interacting)
  (Timeline|Timeline%20Interaction%20Fade%20In%20Out Play))

(event Custom|EndedInteract
  ;[0.2314 0.2118 0.4824] Interact (increase opacity while interacting)
  (Timeline|Timeline%20Interaction%20Fade%20In%20Out Reverse))

(fn ConstructionScript)
