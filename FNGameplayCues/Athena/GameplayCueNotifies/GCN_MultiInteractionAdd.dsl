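; Burst gameplay cue notify that gives visual and audio feedback when a player joins a multi-person interaction, including DBNO.
; The Cascade burst handler tints the first particle component green or yellow, depending on whether a cue parameter value is above zero.
; The preferred VFX burst handler soft-checks whether the effect causer is an Athena player pawn.
; If so, it plays the DBNO joined sound and pushes a foley-ducking mix; otherwise it plays the default joined sound.
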(fn OnBurst(Cascade)
  (params (MyTarget :type Actor%20Object%20Reference) (Parameters :type Gameplay%20Cue%20Parameters%20Structure%20%28by%20ref%29) (ParticleComponents :type Array%20of%20Cascade%20Particle%20System%20Component%20Object%20References) (AudioComponents :type Array%20of%20Audio%20Component%20Object%20References) (BurstCameraShakeInstance :type Camera%20Shake%20Base%20Object%20Reference) (BurstDecalInstance :type Decal%20Actor%20Object%20Reference))
  (Effects|Components|ParticleSystem|SetVectorParameter (Utilities|Array|Get(acopy) ParticleComponents) "Color" (Math|Conversions|ToVector(LinearColor) (select (> (Ability|GameplayCue|BreakGameplayCueParameters Parameters) 0.0) "(R=0.029557,G=1.000000,B=0.242281,A=1.000000)" "(R=1.000000,G=0.838799,B=0.029557,A=1.000000)")))
  (return))

(fn OnBurst(EitherVFX-preferthisversion)
  (params (MyTarget :type Actor%20Object%20Reference) (Parameters_NormalizedMagnitude :type Float%20%28single-precision%29) (Parameters_RawMagnitude :type Float%20%28single-precision%29) (Parameters_EffectContext :type Gameplay%20Effect%20Context%20Handle%20Structure) (Parameters_MatchedTagName :type Gameplay%20Tag%20Structure) (Parameters_OriginalTag :type Gameplay%20Tag%20Structure) (Parameters_AggregatedSourceTags :type Gameplay%20Tag%20Container%20Structure) (Parameters_AggregatedTargetTags :type Gameplay%20Tag%20Container%20Structure) (Parameters_Location :type Vector :default "0, 0, 0") (Parameters_Normal :type Vector :default "0, 0, 0") (Parameters_Instigator :type Actor%20Object%20Reference) (Parameters_EffectCauser :type Actor%20Object%20Reference) (Parameters_SourceObject :type Object%20Reference) (Parameters_PhysicalMaterial :type Physical%20Material%20Object%20Reference) (Parameters_GameplayEffectLevel :type Integer) (Parameters_AbilityLevel :type Integer) (Parameters_TargetAttachComponent :type Scene%20Component%20Object%20Reference) (Parameters_bReplicateLocationWhenUsingMinimalRepProxy :type Boolean) (ParticleComponents :type Array%20of%20FXSystem%20Component%20Object%20References) (AudioComponents :type Array%20of%20Audio%20Component%20Object%20References) (BurstCameraShakeInstance :type Camera%20Shake%20Base%20Object%20Reference) (BurstDecalInstance :type Decal%20Actor%20Object%20Reference))
  (Utilities|IsA(soft) Parameters_EffectCauser "/Game/Athena/PlayerPawn_Athena.PlayerPawn_Athena_C"
    (:True
      ;[0.04758 0.4115 0.127] Play Speed Up Sound  & duck foley
      (Audio|SpawnSoundatLocation Parameters_EffectCauser (Variables|Default|GetDBNO%20Player%20Joined%20Sound) Parameters_Location)
      (Audio|PushSoundMixModifier MyTarget "/FNGameplayCues/Sounds/FortSoundClassesAndModes/Fort_DBNO_Duck_Foley.Fort_DBNO_Duck_Foley"))
    (:False
      ;[0.159 0.2018 0.4115] Play Default Sound
      (Audio|SpawnSoundatLocation MyTarget (Variables|Default|GetDefault%20Player%20Joined%20Sound) Parameters_Location)
      (return))))
