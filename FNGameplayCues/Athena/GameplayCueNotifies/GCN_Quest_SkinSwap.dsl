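; Gameplay cue notify for a quest skin swap that tints its Cascade burst effect.
; On burst, it reads the gameplay effect level from the cue parameters and uses it as an index into the Primary Color array.
; The chosen color is converted to a vector and written to the Primary_Color parameter of the first spawned particle component.
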(fn OnBurst(Cascade)
  (params (MyTarget :type Actor%20Object%20Reference) (Parameters :type Gameplay%20Cue%20Parameters%20Structure%20%28by%20ref%29) (ParticleComponents :type Array%20of%20Cascade%20Particle%20System%20Component%20Object%20References) (AudioComponents :type Array%20of%20Audio%20Component%20Object%20References) (BurstCameraShakeInstance :type Camera%20Shake%20Base%20Object%20Reference) (BurstDecalInstance :type Decal%20Actor%20Object%20Reference))
  (bind (_normalizedmagnitude _rawmagnitude _effectcontext _matchedtagname _originaltag _aggregatedsourcetags _aggregatedtargettags _location _normal _instigator _effectcauser _sourceobject _physicalmaterial _gameplayeffectlevel) (Ability|GameplayCue|BreakGameplayCueParameters Parameters))
  (Effects|Components|ParticleSystem|SetVectorParameter (Utilities|Array|Get(acopy) ParticleComponents) "Primary_Color" (Math|Conversions|ToVector(LinearColor) (Utilities|Array|Get(acopy) (Variables|Default|GetPrimary%20Color) _gameplayeffectlevel)))
  (return))
