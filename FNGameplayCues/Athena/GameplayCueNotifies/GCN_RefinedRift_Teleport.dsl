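; Gameplay cue notify for Refined Rift teleports that adds a first-person portal enter/exit sound for a local player viewing someone else's pawn.
; On burst, it skips locally controlled pawns, then plays the 2D sound only if the first local Fort player controller's view target is the cue target, as when spectating.
; OnActive only defers to the parent, and the particle-spawn hook is empty.
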(fn OnActive
  (params (MyTarget :type Actor%20Object%20Reference) (Parameters :type Gameplay%20Cue%20Parameters%20Structure%20%28by%20ref%29))
  (return (|CallParentFunction MyTarget Parameters)))

(fn OnBurst(Cascade)
  (params (MyTarget :type Actor%20Object%20Reference) (Parameters :type Gameplay%20Cue%20Parameters%20Structure%20%28by%20ref%29) (ParticleComponents :type Array%20of%20Cascade%20Particle%20System%20Component%20Object%20References) (AudioComponents :type Array%20of%20Audio%20Component%20Object%20References) (BurstCameraShakeInstance :type Camera%20Shake%20Base%20Object%20Reference) (BurstDecalInstance :type Decal%20Actor%20Object%20Reference))
  (bind _returnvalue (Fortnite|GetLocalFortPlayerControllers MyTarget false true))
  (bind _output (Utilities|Array|Get(acopy) _returnvalue))
  (bind _asfort_pawn (Utilities|Casting|CastToFortPawn MyTarget))
  (if (Pawn|IsLocallyControlled _asfort_pawn)
    _
    (elif (> (Utilities|Array|Length _returnvalue) 0)
      (Utilities|IsValid _output
        (:Is%20Valid
          (if (Utilities|Equal(Object) MyTarget (Pawn|GetViewTarget _output))
            (Audio|PlaySound2D _asfort_pawn "/Game/Sounds/Fort_GamePlay_Sounds/Player/Portal/Portal_Teleport_EnterExit_1P_Cue.Portal_Teleport_EnterExit_1P_Cue")
            (else
              _))
          (return))
        (:Is%20Not%20Valid
          _))
      (else
        _))))

(fn OnStartParticleSystemSpawned
  (params (SpawnedParticleSysComponent :type Cascade%20Particle%20System%20Component%20Object%20Reference) (Parameters :type Gameplay%20Cue%20Parameters%20Structure%20%28by%20ref%29))
  (return))
