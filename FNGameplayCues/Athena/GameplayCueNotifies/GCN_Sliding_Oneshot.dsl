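; One-shot gameplay cue for crouch-slide turns that plays a surface-specific sound on the target Fort pawn.
; On burst it attaches a first- or third-person MetaSound to the pawn's mesh, based on whether the pawn is locally viewed.
; The current surface type selects an audio data table row, falling back to "Default", to supply the volume and wave array.
; An "AssetSwitch" trigger applies them; a missing table or default row prints a message.
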(fn OnBurst(EitherVFX-preferthisversion)
  (params (MyTarget :type Actor%20Object%20Reference) (Parameters :type Gameplay%20Cue%20Parameters%20Structure%20%28by%20ref%29) (ParticleComponents :type Array%20of%20FXSystem%20Component%20Object%20References) (AudioComponents :type Array%20of%20Audio%20Component%20Object%20References) (BurstCameraShakeInstance :type Camera%20Shake%20Base%20Object%20Reference) (BurstDecalInstance :type Decal%20Actor%20Object%20Reference))
  (bind _audiodatatable (Variables|Default|GetAudioDataTable))
  (bind _asfort_pawn (Utilities|Casting|CastToFortPawn MyTarget))
  (bind _returnvalue (Audio|SpawnSoundAttached (select (Pawn|IsLocallyViewed _asfort_pawn) "/FNGameplayCues/Athena/Sounds/Character/Sliding/MSS_CrouchSliding_Turn_1p.MSS_CrouchSliding_Turn_1p" "/FNGameplayCues/Athena/Sounds/Character/Sliding/MSS_CrouchSliding_Turn_3p.MSS_CrouchSliding_Turn_3p") (Variables|/Script/Engine.Character|GetMesh _asfort_pawn)))
  (Utilities|IsValid _audiodatatable
    (:Is%20Valid
      (bind _returnvalue_2 (Utilities|GetDataTableRow _audiodatatable (Utilities|String|StringToName (Utilities|String|EnumtoString (Class|FortSurfaceTrackingComponent|GetCurrentSurfaceType (Footsteps|GetSurfaceTrackingComponent _asfort_pawn)))))
        (:then
          (Audio|Parameter|SetFloatParameter _returnvalue "VolumeMultiplier" (.TurnVolume_37_1EAD27EA4DCE3E97D469B0864B4E2E6A (Utilities|Struct|BreakStrutAudioCrouchSlide _returnvalue_2)) :!class /Script/AudioExtensions.AudioParameterControllerInterface)
          (Audio|Parameter|SetObjectArrayParameter _returnvalue "Waves" (.TurnWaves_25_6742A24345EC2AFEF56E968736E1A8A2 (Utilities|Struct|BreakStrutAudioCrouchSlide _returnvalue_2))))
        (:RowNotFound
          (bind _returnvalue_3 (Utilities|GetDataTableRow (Variables|Default|GetAudioDataTable) "Default")
            (:then
              (Audio|Parameter|SetFloatParameter _returnvalue "VolumeMultiplier" (.TurnVolume_37_1EAD27EA4DCE3E97D469B0864B4E2E6A (Utilities|Struct|BreakStrutAudioCrouchSlide _returnvalue_3)) :!class /Script/AudioExtensions.AudioParameterControllerInterface)
              (Audio|Parameter|SetObjectArrayParameter _returnvalue "Waves" (.TurnWaves_25_6742A24345EC2AFEF56E968736E1A8A2 (Utilities|Struct|BreakStrutAudioCrouchSlide _returnvalue_3))))
            (:RowNotFound
              (Development|PrintString 0 "CrouchSliding Turn: No Default Sound Selected!")
              (return)))))
      (Audio|Parameter|ExecuteTriggerParameter _returnvalue "AssetSwitch")
      (return))
    (:Is%20Not%20Valid
      (Development|PrintString 0 "CrouchSliding Turn: No Table Selected!"))))
