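; Gameplay cue notify for the Spooky Mist dive effect in Athena, handling audio alongside its start particle system.
; When the start particle system spawns, it calls Play Sound at Location at that component's world position, though the graph shows no sound asset assigned to the call.
; The OnActive override has no logic.
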(fn OnActive
  (params (MyTarget :type Actor%20Object%20Reference) (Parameters :type Gameplay%20Cue%20Parameters%20Structure%20%28by%20ref%29)))

(fn OnStartParticleSystemSpawned
  (params (SpawnedParticleSysComponent :type Cascade%20Particle%20System%20Component%20Object%20Reference) (Parameters :type Gameplay%20Cue%20Parameters%20Structure%20%28by%20ref%29))
  (Audio|PlaySoundatLocation SpawnedParticleSysComponent 0 (Transformation|GetWorldLocation SpawnedParticleSysComponent))
  (return))
