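; Gameplay cue notify for the looping Zombie Elastic emote that animates pulsing spotlight effects.
; On EnableFX it runs the parent logic, then plays PulseTimeline to drive the spotlight dynamic material instances on every update.
; The timeline's random track sets intensity and phase on all spotlights, plus scale on index 5.
; Index 1 gets stronger day, morning and evening multipliers and its own vertical fade, while the others use dimmer multipliers.
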(event EventEnableFX
  (bind _spotlightmids (Variables|Default|GetSpotlightMIDs))
  (|CallParentFunction)
  (bind (_direction _random) (Timeline|PulseTimeline PlayFromStart)
    (:Update
      (Utilities|FlowControl|Sequence
        (:then_0
          (for (_array_element _array_index) _spotlightmids
            (Rendering|Material|SetScalarParameterValue _array_element "Final Multiplier" (* _random 2.0) :!class /Script/Engine.MaterialInstanceDynamic)
            (if (== _array_index 5)
              (Rendering|Material|SetScalarParameterValue _array_element "Scale 1" (Math|Float|Lerp 0.0 -4e+01 _random) :!class /Script/Engine.MaterialInstanceDynamic)
              (else
                (Rendering|Material|SetScalarParameterValue _array_element "Scale 1" 1.0 :!class /Script/Engine.MaterialInstanceDynamic)))))
        (:then_1
          (for (_array_element_2 _array_index_2) _spotlightmids
            (Rendering|Material|SetScalarParameterValue _array_element_2 "Phase" _random :!class /Script/Engine.MaterialInstanceDynamic)
            (if (== _array_index_2 1)
              (Rendering|Material|SetScalarParameterValue _array_element_2 "Day Multiplier" 1.0 :!class /Script/Engine.MaterialInstanceDynamic)
              (Rendering|Material|SetScalarParameterValue _array_element_2 "Morning Multiplier" 1.0 :!class /Script/Engine.MaterialInstanceDynamic)
              (Rendering|Material|SetScalarParameterValue _array_element_2 "Evening Multiplier" 1.0 :!class /Script/Engine.MaterialInstanceDynamic)
              (Rendering|Material|SetScalarParameterValue _array_element_2 "Vertical Fade Power" 1.2 :!class /Script/Engine.MaterialInstanceDynamic)
              (else
                (Rendering|Material|SetScalarParameterValue _array_element_2 "Day Multiplier" 0.5 :!class /Script/Engine.MaterialInstanceDynamic)
                (Rendering|Material|SetScalarParameterValue _array_element_2 "Morning Multiplier" 0.5 :!class /Script/Engine.MaterialInstanceDynamic)
                (Rendering|Material|SetScalarParameterValue _array_element_2 "Evening Multiplier" 0.5 :!class /Script/Engine.MaterialInstanceDynamic)))))))))

(fn ConstructionScript
  (|CallParentFunction))
