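; GC_Athena_AppleSauce is a gameplay cue notify for Athena that only adds debug prints in its graphs.
; OnExecute calls the parent implementation, prints "Hello" and returns the parent's result.
; The Cascade On Application event prints "APP" without using the supplied particle, audio, camera shake or decal references.
; BeginPlay, Tick, overlap and the construction script are empty.
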(event EventBeginPlay)

(event Collision|EventActorBeginOverlap
  (params (OtherActor :type Actor%20Object%20Reference)))

(event EventTick
  (params (DeltaSeconds :type Float%20%28single-precision%29)))

(event EventOnApplication(Cascade)
  (params (MyTarget :type Actor%20Object%20Reference) (Parameters :type Gameplay%20Cue%20Parameters%20Structure%20%28by%20ref%29) (ParticleComponents :type Array%20of%20Cascade%20Particle%20System%20Component%20Object%20References) (AudioComponents :type Array%20of%20Audio%20Component%20Object%20References) (BurstCameraShakeInstance :type Camera%20Shake%20Base%20Object%20Reference) (BurstDecalInstance :type Decal%20Actor%20Object%20Reference))
  (Development|PrintString 0 "APP" true true "(R=1.000000,G=0.060289,B=0.000000,A=1.000000)"))

(fn OnExecute
  (params (MyTarget :type Actor%20Object%20Reference) (Parameters :type Gameplay%20Cue%20Parameters%20Structure%20%28by%20ref%29))
  (bind _returnvalue (|CallParentFunction MyTarget Parameters))
  (Development|PrintString 0 "Hello" true true "(R=1.000000,G=0.060289,B=0.000000,A=1.000000)")
  (return _returnvalue))

(fn ConstructionScript)
