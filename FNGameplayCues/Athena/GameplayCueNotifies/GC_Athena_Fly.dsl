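; Looping gameplay cue for Athena flight that drives a "Speed" parameter on its spawned audio from the target pawn's movement.
; On loop start it caches the spawned audio components and the target as a FortPlayerPawn.
; Each tick it smooths velocity magnitude with FInterpTo and maps it, clamped, onto the first audio component's parameter.
; The parameter is forced to zero whenever the movement mode is not Flying.
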(event EventOnLoopingStart(Cascade)
  (params (MyTarget :type Actor%20Object%20Reference) (Parameters :type Gameplay%20Cue%20Parameters%20Structure%20%28by%20ref%29) (ParticleComponents :type Array%20of%20Cascade%20Particle%20System%20Component%20Object%20References) (AudioComponents :type Array%20of%20Audio%20Component%20Object%20References))
  (if (Utilities|Array|IsValidIndex AudioComponents)
    (Variables|Default|SetSpawnedAudioComps AudioComponents)
    (else
      _))
  (bind _asfort_player_pawn (Utilities|Casting|CastToFortPlayerPawn MyTarget))
  (Variables|Default|SetTargetPlayer _asfort_player_pawn))

(event EventTick
  (params (DeltaSeconds :type Float%20%28single-precision%29))
  (bind _spawnedaudiocomps (Variables|Default|GetSpawnedAudioComps))
  (bind _output (Utilities|Array|Get(acopy) _spawnedaudiocomps))
  (bind _targetplayer (Variables|Default|GetTargetPlayer))
  (if (Utilities|Array|IsValidIndex _spawnedaudiocomps)
    (Utilities|IsValid _output
      (:Is%20Valid
        (Utilities|IsValid _targetplayer
          (:Is%20Valid
            (Variables|Default|SetVelocitySmoothed (Math|Interpolation|FInterpTo (Variables|Default|GetVelocitySmoothed) (Math|Vector|VectorLength (Transformation|GetVelocity _targetplayer)) (Utilities|Time|GetWorldDeltaSeconds _targetplayer) 5.0))
            (if (not (Utilities|Enum|Equal(Enum) (Variables|/Script/Engine.CharacterMovementComponent|GetMovementMode (Variables|/Script/Engine.Character|GetCharacterMovement (Variables|Default|GetTargetPlayer))) "MOVE_Flying"))
              (Audio|Parameter|SetFloatParameter _output "Speed" 0.0 :!class /Script/AudioExtensions.AudioParameterControllerInterface)
              (else
                (Audio|Parameter|SetFloatParameter _output "Speed" (Math|Float|MapRangeClamped (Variables|Default|GetVelocitySmoothed) 0.0 1.5e+03 0.0 0.8) :!class /Script/AudioExtensions.AudioParameterControllerInterface)))))))))

(fn OnCeaseRelevant
  (params (MyTarget :type Actor%20Object%20Reference) (Parameters :type Gameplay%20Cue%20Parameters%20Structure%20%28by%20ref%29)))

(fn ConstructionScript)
