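; Looping gameplay cue for a player in the storm outside the safe zone, driving storm particles, ambient audio, lightning and screen effects.
; Storm stage selects particle intensity and audio tag, both audio entry and tag being cleared on removal.
; A self-rescheduling timer flashes a randomly placed light via timeline, sometimes with a strike sound.
; Optional post-process uses a camera modifier to avoid split-screen bleed, alongside a timeline-faded mesh material blend.
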(event EventOnLoopingStart(EitherVFX-preferthisversion)
  (params (MyTarget :type Actor%20Object%20Reference) (Parameters :type Gameplay%20Cue%20Parameters%20Structure%20%28by%20ref%29) (ParticleComponents :type Array%20of%20FXSystem%20Component%20Object%20References) (AudioComponents :type Array%20of%20Audio%20Component%20Object%20References))
  (bind _player_pawn (Variables|Default|GetPlayer%20Pawn))
  (bind _returnvalue (Pawn|GetLocalViewingPlayerController _player_pawn))
  (bind _output (Utilities|Array|Get(acopy) ParticleComponents 1))
  (bind _player_pawn_2 (Variables|Default|GetPlayer%20Pawn))
  (bind _returnvalue_2 (WorldSubsystems|GetAmbientAudioSubsystem _player_pawn_2))
  (bind _asfort_player_pawn_athena (Utilities|Casting|CastToFortPlayerPawnAthena MyTarget))
  (Variables|Default|SetPlayer%20Pawn _asfort_player_pawn_athena)
  (CallFunction|RandomLightingFlashes)
  (Utilities|FlowControl|Sequence
    (:then_0
      (switch int (- (Variables|Default|GetStorm%20Stage) 1)
        (:0
          (Effects|Components|ParticleSystem|SetFloatParameter _output "StormIntensity" 0.0)
          (AmbientAudio|AddGameplayTag _returnvalue_2 (Variables|Default|GetStormAudioTag_Low)))
        (:1
          (Effects|Components|ParticleSystem|SetFloatParameter _output "StormIntensity" 0.5)
          (AmbientAudio|AddGameplayTag _returnvalue_2 (Variables|Default|GetStormAudioTag_Medium)))
        (:2
          (Effects|Components|ParticleSystem|SetFloatParameter _output "StormIntensity" 1.0)
          (AmbientAudio|AddGameplayTag _returnvalue_2 (Variables|Default|GetStormAudioTag_High))))
      (AmbientAudio|AddAmbientEntry _returnvalue_2 "Storm" (Variables|Default|GetStormAudioBank) 10 0.5))
    (:then_1
      (bind _output_get (Variables|Default|SetbUsePostProcess (ScalableFloat|GetValueasBool (Variables|Default|GetHF_ApplyScreenFXInGameplayCue))))
      (if _output_get
        (Utilities|FlowControl|Sequence
          (:then_0
            ; Apply post process to camera instead of using a volume, otherwise teammates in split-screen will see it too
            (Utilities|IsValid _returnvalue
              (:Is%20Valid
                (bind _returnvalue_3 (Game|Player|AddNewCameraModifier (Variables|/Script/Engine.PlayerController|GetPlayerCameraManager _returnvalue) (Variables|Default|GetPostProcessCameraMod)))
                (Variables|Default|SetActiveCameraMod _returnvalue_3))))
          (:then_1
            (CallFunction|FadeScreenEffects true)))))))

(event EventOnRemoval(EitherVFX-preferthisversion)
  (params (MyTarget :type Actor%20Object%20Reference) (Parameters :type Gameplay%20Cue%20Parameters%20Structure%20%28by%20ref%29) (ParticleComponents :type Array%20of%20FXSystem%20Component%20Object%20References) (AudioComponents :type Array%20of%20Audio%20Component%20Object%20References) (BurstCameraShakeInstance :type Camera%20Shake%20Base%20Object%20Reference) (BurstDecalInstance :type Decal%20Actor%20Object%20Reference))
  (bind _player_pawn (Variables|Default|GetPlayer%20Pawn)
    (:then
      (AmbientAudio|RemoveAmbientEntry (WorldSubsystems|GetAmbientAudioSubsystem _player_pawn) "Storm" -1.0)
      (AmbientAudio|RemoveGameplayTag (WorldSubsystems|GetAmbientAudioSubsystem _player_pawn) (Utilities|Select :Index (- (Variables|Default|GetStorm%20Stage) 1) :Option%200 (Variables|Default|GetStormAudioTag_Low) :Option%201 (Variables|Default|GetStormAudioTag_Medium) :Option%202 (Variables|Default|GetStormAudioTag_High)))
      (CallFunction|StopLightningFlashTimer)
      (if (Variables|Default|GetbUsePostProcess)
        (Utilities|FlowControl|Sequence
          (:then_0
            (bind _activecameramod (Variables|Default|GetActiveCameraMod)
              (:then
                (CameraModifier|DisableModifier _activecameramod false)
                (Utilities|Time|SetTimerbyEvent (delegate Custom|RemoveCameraModifier) 1.0))
              (:else)))
          (:then_1
            (CallFunction|FadeScreenEffects false)))))
    (:else)))

(event Custom|RemoveCameraModifier
  (bind _player_pawn (Variables|Default|GetPlayer%20Pawn))
  (bind _returnvalue (Pawn|GetLocalViewingPlayerController _player_pawn))
  (bind _activecameramod (Variables|Default|GetActiveCameraMod)
    (:then
      (Utilities|IsValid _returnvalue
        (:Is%20Valid
          (Game|Player|RemoveCameraModifier (Variables|/Script/Engine.PlayerController|GetPlayerCameraManager _returnvalue) _activecameramod))))
    (:else)))

(event GCNLOutsideSafeZone|EventOnStormEffectsEnabled
  (params (bEnabled :type Boolean))
  (Variables|Default|SetStormEffectsEnabled bEnabled)
  (if (and (Variables|Default|GetStormEffectsEnabled) (Variables|/Script/FortniteGame.FortPlayerPawn|GetbIsInAnyStorm (Variables|Default|GetPlayer%20Pawn)))
    (bind _returnvalue (Utilities|Time|SetTimerbyEvent (delegate Custom|TryRandomLightningFlash) (Math|Random|RandomFloatinRange (Variables|Timeline|GetLightningFlashRepeatDelayMin) (Variables|Timeline|GetLightningFlashRepeatDelayMax))))
    (Variables|Default|SetLightningFlashTimerHandle _returnvalue)
    (Rendering|Material|CreateDynamicMaterialInstance (Variables|Default|GetStaticMesh) 0 0 "None" :!class /Script/Engine.PrimitiveComponent)))

(event Custom|FlashTimeline
  (bind _random_lighting_light (Variables|Default|GetRandom%20Lighting%20Light))
  (bind (_direction _lerp) (Timeline|LightningFlashTL PlayFromStart)
    (:Update
      (Rendering|Components|Light|SetIntensity _random_lighting_light (Math|Float|Lerp 0.1 (Variables|PointLIght|GetLightning%20Intensity) _lerp)))
    (:Finished
      (Rendering|SetVisibility _random_lighting_light false false)
      (if (and (Variables|Default|GetStormEffectsEnabled) (Variables|/Script/FortniteGame.FortPlayerPawn|GetbIsInAnyStorm (Variables|Default|GetPlayer%20Pawn)))
        (bind _returnvalue (Utilities|Time|SetTimerbyEvent (delegate Custom|TryRandomLightningFlash) (Math|Random|RandomFloatinRange (Variables|Timeline|GetLightningFlashRepeatDelayMin) (Variables|Timeline|GetLightningFlashRepeatDelayMax))))
        (Variables|Default|SetLightningFlashTimerHandle _returnvalue)
        (Rendering|Material|CreateDynamicMaterialInstance (Variables|Default|GetStaticMesh) 0 0 "None" :!class /Script/Engine.PrimitiveComponent)))))

(event Custom|TryRandomLightningFlash
  (if (Variables|/Script/FortniteGame.FortPlayerPawn|GetbIsInAnyStorm (Variables|Default|GetPlayer%20Pawn))
    (CallFunction|RandomLightingFlashes)))

(event Custom|StopLightningFlashTimer
  (Utilities|Time|ClearandInvalidateTimerbyHandle (Variables|Default|GetPlayer%20Pawn) (Variables|Default|GetLightningFlashTimerHandle)))

(event Custom|FadeScreenEffects
  (params (bFadeIn :type Boolean))
  ;[1 1 1] Timeline to fade post process effects
  (Components|Timeline|SetPlayRate (Variables|GC_Athena_OutsideSafeZone|GetTL_ScreenEffects) (Utilities|Reciprocal (Variables|Default|GetScreenEffectsFadeTime)))
  (if bFadeIn
    (bind (_direction _alpha) (Timeline|TL_ScreenEffects Play)
      (:Update
        (Rendering|Material|SetScalarParameterValueonMaterials (Variables|Default|GetStaticMesh) "BP_Blend" _alpha)))
    (else
      (Timeline|TL_ScreenEffects Reverse))))

(fn GetLocationinCircleXY
  (return (Math|Vector|Lerp(Vector) (Math|Vector|MakeVector 0.0 0.0 (Variables|Lightning|GetLightningFlashHeight)) (Math|Vector|MakeVector (* (Math|Trig|Sin(Radians) (Math|Float|Lerp 0.0 (* (Math|Trig|GetPI) 2) (Math|Random|RandomFloat))) (Variables|Lightning|GetLightningFlashDiameter)) (* (Math|Trig|Cos(Radians) (Math|Float|Lerp 0.0 (* (Math|Trig|GetPI) 2) (Math|Random|RandomFloat))) (Variables|Lightning|GetLightningFlashDiameter)) (Variables|Lightning|GetLightningFlashHeight)) (Math|Random|RandomFloatinRange 0.25 1))))

(fn RandomLightingFlashes
  (bind (_returnvalue_x _returnvalue_y _returnvalue_z) (Math|Vector|BreakVector (Transformation|GetActorLocation (Variables|Default|GetPlayer%20Pawn))))
  (Variables|Lightning|SetLightning%20Spawn%20Location (+ (Transformation|GetWorldLocation (Variables|/Script/Engine.Character|GetMesh (Variables|Default|GetPlayer%20Pawn))) (CallFunction|GetLocationInCircleXY)))
  (if (> (Math|Random|RandomFloat) (Variables|Lightning|Get%25ChanceOfLightningMesh))
    (Audio|PlaySoundatLocation 0 "/Game/Athena/Sounds/Shield/ApolloStorm/Apollo_Storm_Lightning_Cue.Apollo_Storm_Lightning_Cue" (Math|Vector|MakeVector (.x (Variables|Lightning|GetLightning%20Spawn%20Location)) (.y (Variables|Lightning|GetLightning%20Spawn%20Location)) _returnvalue_z)))
  (Variables|PointLIght|SetLightning%20Intensity (Math|Random|RandomFloatinRange (Variables|PointLIght|GetLightningIntensityMin) (Variables|PointLIght|GetLightningIntensityMax)))
  (Rendering|SetVisibility (Variables|Default|GetRandom%20Lighting%20Light) true false)
  (Components|Timeline|SetPlayRate (Variables|GC_Athena_OutsideSafeZone|GetLightningFlashTL) (Utilities|Reciprocal (Math|Random|RandomFloatinRange (Variables|Timeline|GetLightningTimelinePlaySpeedMin) (Variables|Timeline|GetLightningTimelinePlaySpeedMax))))
  (Transformation|SetWorldLocation (Variables|Default|GetRandom%20Lighting%20Light) (Variables|Lightning|GetLightning%20Spawn%20Location) false false)
  (CallFunction|FlashTimeline))

(fn ConstructionScript)
