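; Looping gameplay cue notify for Athena's stage 03 outside-safe-zone effect, adding sound to distant lightning.
; On looping start it runs the parent behavior, then binds the particle-spawn event of the second spawned Cascade component.
; When a "Light" particle spawn event fires, a random roll gives roughly half of them a crackling lightning sound at the particle's location.
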(event EventOnLoopingStart(Cascade)
  (params (MyTarget :type Actor%20Object%20Reference) (Parameters :type Gameplay%20Cue%20Parameters%20Structure%20%28by%20ref%29) (ParticleComponents :type Array%20of%20Cascade%20Particle%20System%20Component%20Object%20References) (AudioComponents :type Array%20of%20Audio%20Component%20Object%20References))
  (|CallParentFunction MyTarget Parameters ParticleComponents AudioComponents)
  (EventDispatchers|BindEventtoOnParticleSpawn (Utilities|Array|Get(acopy) ParticleComponents 1) (delegate Custom|On%20Distant%20Lightning%20Spawn)))

(event Custom|On%20Distant%20Lightning%20Spawn
  (params (EventName :type Name) (EmitterTime :type Float%20%28single-precision%29) (Location :type Vector) (Velocity :type Vector))
  (if (and (Utilities|Name|Equal(Name) EventName "Light") (> (Math|Random|RandomFloatinRange 0.0 1.0) 0.5))
    (Audio|SpawnSoundatLocation 0 "/FNGameplayCues/Athena/Sounds/Events/CrackLightning/Crack_Lightning_Pulse_Cue.Crack_Lightning_Pulse_Cue" Location)))

(fn ConstructionScript)
