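; Gameplay cue notify for the RustyFish applied effect, giving the affected player a phase-shift look built on Outlander teleport visuals.
; On looping start it overlays leader-posed copies of the character part meshes with phase-shift MIDs, animated by an Opacity/Sweep timeline.
; It optionally spawns a spline trail and drives a local-player-only post-process parameter.
; Removal stops the trail and reverses the timelines, destroying the duplicates when the reverse finishes.
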(event EventOnApplication(Cascade)
  (params (MyTarget :type Actor%20Object%20Reference) (Parameters :type Gameplay%20Cue%20Parameters%20Structure%20%28by%20ref%29) (ParticleComponents :type Array%20of%20Cascade%20Particle%20System%20Component%20Object%20References) (AudioComponents :type Array%20of%20Audio%20Component%20Object%20References) (BurstCameraShakeInstance :type Camera%20Shake%20Base%20Object%20Reference) (BurstDecalInstance :type Decal%20Actor%20Object%20Reference))
  (bind _asfort_player_pawn (Utilities|Casting|CastToFortPlayerPawn MyTarget))
  (if (Pawn|IsLocallyControlled _asfort_player_pawn)
    (Utilities|EffectsQualityCheck 0
      (:Medium
        _)
      (:Low
        _))
    (Variables|/Script/Engine.PostProcessComponent|SetbEnabled false (Variables|Default|GetTeleport_PostProcess))
    (else
      (Variables|/Script/Engine.PostProcessComponent|SetbEnabled false (Variables|Default|GetTeleport_PostProcess)))))

(event Custom|Activate%20PP
  (Utilities|EffectsQualityCheck 0
    (:Cinematic
      _)
    (:Epic
      _)
    (:High
      _))
  (Components|Timeline|SetPlayRate (Variables|GC_Athena_RustyFish_AppliedFX|GetTeleport%20PP%20TL) (/ 1.0 0.1))
  (bind (_direction _pp) (Timeline|Teleport%20PP%20TL PlayFromStart)
    (:Update
      (Rendering|Material|SetScalarParameterValue 0 "/Game/Packages/Fortress_SharedMaterials/GlobalMaterialParameters/FortniteMaterialParameters.FortniteMaterialParameters" "OutlanderTeleportPP" _pp :!class /Script/Engine.KismetMaterialLibrary))
    (:Finished
      (switch Utilities|FlowControl|Switch|SwitchonETimelineDirection _direction
        (:Backward
          (Variables|/Script/Engine.PostProcessComponent|SetbEnabled false (Variables|Default|GetTeleport_PostProcess)))))))

(event Custom|Remove%20PP
  (Utilities|EffectsQualityCheck 0
    (:Cinematic
      _)
    (:Epic
      _)
    (:High
      _))
  (Components|Timeline|SetPlayRate (Variables|GC_Athena_RustyFish_AppliedFX|GetTeleport%20PP%20TL) (/ 1.0 0.15))
  (Timeline|Teleport%20PP%20TL Reverse))

(event EventOnLoopingStart(Cascade)
  (params (MyTarget :type Actor%20Object%20Reference) (Parameters :type Gameplay%20Cue%20Parameters%20Structure%20%28by%20ref%29) (ParticleComponents :type Array%20of%20Cascade%20Particle%20System%20Component%20Object%20References) (AudioComponents :type Array%20of%20Audio%20Component%20Object%20References))
  (Utilities|FlowControl|Delay 0 0.1)
  (bind _asfort_player_pawn (Utilities|Casting|CastToFortPlayerPawn MyTarget))
  (if (not (AI|Components|NavMovement|IsMovingonGround (Variables|/Script/Engine.Character|GetCharacterMovement _asfort_player_pawn)))
    (for _array_element (Variables|GeneralVFX|GetDust%20Emitters)
      (Effects|Components|ParticleSystem|SetEmitterEnable (Utilities|Array|Get(acopy) ParticleComponents) _array_element false)))
  (CallFunction|StopSplinesIfValid)
  (if (Variables|Default|GetUseSpawnSplineActor)
    (CallFunction|SpawnSplineActor _asfort_player_pawn)
    (else
      _))
  (if (Pawn|IsLocallyControlled _asfort_player_pawn)
    (CallFunction|ActivatePP)
    (else
      _))
  (CallFunction|DuplicateCharacterMeshes _asfort_player_pawn)
  (Components|Timeline|SetPlayRate (Variables|GC_Athena_RustyFish_AppliedFX|GetPhase%20Shift%20Activation%20Timeline) (/ 1.0 0.4))
  (bind (_direction _opacity _sweep) (Timeline|Phase%20Shift%20Activation%20Timeline PlayFromStart)
    (:Update
      (for _array_element_2 (Variables|Default|GetDuplicate%20MIDs)
        (Utilities|IsValid _array_element_2
          (:Is%20Valid
            (Rendering|Material|SetScalarParameterValue _array_element_2 "Opacity" _opacity :!class /Script/Engine.MaterialInstanceDynamic)
            (Rendering|Material|SetScalarParameterValue _array_element_2 "Sweep" _sweep :!class /Script/Engine.MaterialInstanceDynamic)))))
    (:Finished
      (switch Utilities|FlowControl|Switch|SwitchonETimelineDirection _direction
        (:Backward
          (CallFunction|DestroyDuplicateCharacterMeshes))))))

(event Custom|Stop%20Splines%20if%20Valid
  (bind _teleportspline (Variables|Default|GetTeleportSpline))
  (Utilities|IsValid _teleportspline
    (:Is%20Valid
      (Class|BPSplineVolumeTrailV1b|ManuallyStopSplineSystem _teleportspline)
      (Variables|Default|SetTeleportSpline 0))))

(event Custom|Spawn%20Spline%20Actor
  (params (Player%20Pawn :type Fort%20Player%20Pawn%20Object%20Reference))
  (Utilities|EffectsQualityCheck 0
    (:Cinematic
      _)
    (:Epic
      _)
    (:High
      _))
  (bind _returnvalue (Game|SpawnActorfromClass :Class "/Game/Effects/SplineTrails/Blueprints/BP_SplineVolumeTrail_v1b.BP_SplineVolumeTrail_v1b_C" :SpawnTransform (Transformation|GetActorTransform Player%20Pawn) :CollisionHandlingOverride "Undefined" :TransformScaleMethod "OverrideRootScale" :Skeletal%20Mesh (Variables|/Script/Engine.Character|GetMesh Player%20Pawn) :Socket%20Attach%20Point "FX_Feet" :SplineMeshTemplate "/Game/Effects/SplineTrails/Meshes/Sm_Blink_Trail_01.Sm_Blink_Trail_01" :Lifetime 0.3 :Lifetime%20Buffer 0.1 :SplineSystemDuration 0.0 :LifetimeAfterCompletion 0.2 :KeepTickEnabledWhenCompleting true :NumberOfSplineMeshesToUpdate 2 :Initial%20Size 1.0 :SplineDistanceThreshold 125.0 :SplineDistanceThresholdEmpty 1e+01 :ColorAndAlphaOverLife "/Game/CharClasses/Outlander/Gadgets/Teleport/Cv_Outlander_Blink_ColorAndAlphaOverLife_01.Cv_Outlander_Blink_ColorAndAlphaOverLife_01" :Initial%20Rotation%20%28Roll%29 0.0 :DebugMode false :DebugPositionOffset "200.000000,0.000000,0.000000" :DebugRotationRate "0.000000,150.000000,0.000000" :Use%20Tick%20Interval%20LODs true :Use%20Recently%20Rendered%20LOD true :Recently%20Rendered%20Check%20Rate 0.2 :SplineDefaultUpVector "1.000000,0.000000,0.000000" :SplineDefaultUpVectorInWorldSpace false :SplineMeshForwardAxis "NewEnumerator0" :CeaseAndDetachSplineUponCompletion false :LinearColorOverLengthOfSpline true :RedistributePointsAlongSplineWhenCreated true :RenderTextureResolution "(X=128.000000,Y=1.000000)" :WriteBakedCurveDataToDisk false :File%20Path "D:\\Build\\FortniteGame\\Content\\Effects\\SplineTrails\\Textures" :File%20Name "test" :UseCheapSplines true :LinearColorTexture "/Game/Effects/SplineTrails/Textures/T_OutlanderBlink_LinearColorCurve_01.T_OutlanderBlink_LinearColorCurve_01"))
  (Variables|Default|SetTeleportSpline _returnvalue))

(event EventOnRemoval(Cascade)
  (params (MyTarget :type Actor%20Object%20Reference) (Parameters :type Gameplay%20Cue%20Parameters%20Structure%20%28by%20ref%29) (ParticleComponents :type Array%20of%20Cascade%20Particle%20System%20Component%20Object%20References) (AudioComponents :type Array%20of%20Audio%20Component%20Object%20References) (BurstCameraShakeInstance :type Camera%20Shake%20Base%20Object%20Reference) (BurstDecalInstance :type Decal%20Actor%20Object%20Reference))
  (bind _asfort_player_pawn (Utilities|Casting|CastToFortPlayerPawn MyTarget))
  (CallFunction|StopSplinesIfValid)
  (if (Pawn|IsLocallyControlled _asfort_player_pawn)
    (CallFunction|RemovePP)
    (else
      _))
  (Components|Timeline|SetPlayRate (Variables|GC_Athena_RustyFish_AppliedFX|GetPhase%20Shift%20Activation%20Timeline) (/ 1.0 0.4))
  (Timeline|Phase%20Shift%20Activation%20Timeline Reverse))

(fn DestroyDuplicateCharacterMeshes
  (bind _duplicate_face (Variables|Default|GetDuplicate%20Face))
  (bind _duplicate_backpack (Variables|Default|GetDuplicate%20Backpack))
  (bind _duplicate_charm (Variables|Default|GetDuplicate%20Charm))
  (bind _duplicate_head (Variables|Default|GetDuplicate%20Head))
  (bind _duplicate_body (Variables|Default|GetDuplicate%20Body))
  (bind _duplicate_hat (Variables|Default|GetDuplicate%20Hat))
  (Utilities|IsValid _duplicate_head
    (:Is%20Valid
      (Components|DestroyComponent _duplicate_head 0))
    (:Is%20Not%20Valid
      _))
  (Utilities|IsValid _duplicate_body
    (:Is%20Valid
      (Components|DestroyComponent _duplicate_body 0))
    (:Is%20Not%20Valid
      _))
  (Utilities|IsValid _duplicate_hat
    (:Is%20Valid
      (Components|DestroyComponent _duplicate_hat 0))
    (:Is%20Not%20Valid
      _))
  (Utilities|IsValid _duplicate_backpack
    (:Is%20Valid
      (Components|DestroyComponent _duplicate_backpack 0))
    (:Is%20Not%20Valid
      _))
  (Utilities|IsValid _duplicate_charm
    (:Is%20Valid
      (Components|DestroyComponent _duplicate_charm 0))
    (:Is%20Not%20Valid
      _))
  (Utilities|IsValid _duplicate_face
    (:Is%20Valid
      (Components|DestroyComponent _duplicate_face 0))))

(fn DuplicateCharacterMeshes
  (params (Pawn :type Fort%20Player%20Pawn%20Object%20Reference))
  (bind _returnvalue (Customization|GetSkeletalMeshforPartType Pawn "MiscOrTail" :!class /Script/FortniteGame.FortPlayerPawn))
  (bind _returnvalue_2 (Customization|GetSkeletalMeshforPartType Pawn "Head" :!class /Script/FortniteGame.FortPlayerPawn))
  (bind _returnvalue_3 (Customization|GetSkeletalMeshforPartType Pawn "Body" :!class /Script/FortniteGame.FortPlayerPawn))
  (bind _returnvalue_4 (Customization|GetSkeletalMeshforPartType Pawn "Face" :!class /Script/FortniteGame.FortPlayerPawn))
  (bind _returnvalue_5 (Customization|GetSkeletalMeshforPartType Pawn "Hat" :!class /Script/FortniteGame.FortPlayerPawn))
  (bind _returnvalue_6 (Customization|GetSkeletalMeshforPartType Pawn "Backpack" :!class /Script/FortniteGame.FortPlayerPawn))
  (bind _duplicate_head (Variables|Default|GetDuplicate%20Head))
  (bind _duplicate_body (Variables|Default|GetDuplicate%20Body))
  (bind _duplicate_hat (Variables|Default|GetDuplicate%20Hat))
  (bind _duplicate_backpack (Variables|Default|GetDuplicate%20Backpack))
  (bind _duplicate_face (Variables|Default|GetDuplicate%20Face))
  (bind _duplicate_charm (Variables|Default|GetDuplicate%20Charm))
  (Utilities|Array|Clear (Variables|Default|GetDuplicate%20MIDs))
  (Utilities|IsValid _returnvalue_2
    (:Is%20Valid
      (bind _returnvalue_7 (AddComponent|NODE_AddSkeletalMeshComponent-7))
      (bind _output_get (Variables|Default|SetDuplicate%20Head _returnvalue_7))
      (Components|SkinnedMesh|SetSkinnedAssetandUpdate _output_get (Variables|/Script/Engine.SkinnedMeshComponent|GetSkeletalMesh _returnvalue_2) true)
      (Components|SkinnedMesh|SetLeaderPoseComponent _output_get _returnvalue_2 false false)
      (Transformation|AttachComponentToComponent _output_get _returnvalue_2 "None" "SnapToTarget" "SnapToTarget" "SnapToTarget" true)
      (for (_ _array_index) (Rendering|Material|GetMaterials _duplicate_head)
        (bind _returnvalue_8 (Rendering|Material|CreateDynamicMaterialInstance _duplicate_head _array_index (Variables|MaterialChangeFX|GetPhase%20Shift%20Material) "None" :!class /Script/Engine.PrimitiveComponent))
        (Rendering|Material|CopyMaterialInstanceParameters _returnvalue_8 (Utilities|Array|Get(acopy) (Rendering|Material|GetMaterials _returnvalue_2) _array_index) false)
        (Rendering|Material|SetMaterial (Variables|Default|GetDuplicate%20Head) _array_index _returnvalue_8)
        (Utilities|Array|Add (Variables|Default|GetDuplicate%20MIDs) _returnvalue_8)))
    (:Is%20Not%20Valid
      _))
  (Utilities|IsValid _returnvalue_3
    (:Is%20Valid
      (bind _returnvalue_9 (AddComponent|NODE_AddSkeletalMeshComponent-8))
      (bind _output_get_2 (Variables|Default|SetDuplicate%20Body _returnvalue_9))
      (Components|SkinnedMesh|SetSkinnedAssetandUpdate _output_get_2 (Variables|/Script/Engine.SkinnedMeshComponent|GetSkeletalMesh _returnvalue_3) true)
      (Components|SkinnedMesh|SetLeaderPoseComponent _output_get_2 _returnvalue_3 false false)
      (Transformation|AttachComponentToComponent _output_get_2 _returnvalue_3 "None" "SnapToTarget" "SnapToTarget" "SnapToTarget" true)
      (for (_ _array_index_2) (Rendering|Material|GetMaterials _duplicate_body)
        (bind _returnvalue_10 (Rendering|Material|CreateDynamicMaterialInstance _duplicate_body _array_index_2 (Variables|MaterialChangeFX|GetPhase%20Shift%20Material) "None" :!class /Script/Engine.PrimitiveComponent))
        (Rendering|Material|CopyMaterialInstanceParameters _returnvalue_10 (Utilities|Array|Get(acopy) (Rendering|Material|GetMaterials _returnvalue_3) _array_index_2) false)
        (Rendering|Material|SetMaterial (Variables|Default|GetDuplicate%20Body) _array_index_2 _returnvalue_10)
        (Utilities|Array|Add (Variables|Default|GetDuplicate%20MIDs) _returnvalue_10)))
    (:Is%20Not%20Valid
      _))
  (Utilities|IsValid _returnvalue_5
    (:Is%20Valid
      (bind _returnvalue_11 (AddComponent|NODE_AddSkeletalMeshComponent-9))
      (bind _output_get_3 (Variables|Default|SetDuplicate%20Hat _returnvalue_11))
      (Components|SkinnedMesh|SetSkinnedAssetandUpdate _output_get_3 (Variables|/Script/Engine.SkinnedMeshComponent|GetSkeletalMesh _returnvalue_5) true)
      (Components|SkinnedMesh|SetLeaderPoseComponent _output_get_3 _returnvalue_5 false false)
      (Transformation|AttachComponentToComponent _output_get_3 _returnvalue_5 "None" "SnapToTarget" "SnapToTarget" "SnapToTarget" true)
      (for (_ _array_index_3) (Rendering|Material|GetMaterials _duplicate_hat)
        (bind _returnvalue_12 (Rendering|Material|CreateDynamicMaterialInstance _duplicate_hat _array_index_3 (Variables|MaterialChangeFX|GetPhase%20Shift%20Material) "None" :!class /Script/Engine.PrimitiveComponent))
        (Rendering|Material|CopyMaterialInstanceParameters _returnvalue_12 (Utilities|Array|Get(acopy) (Rendering|Material|GetMaterials _returnvalue_5) _array_index_3) false)
        (Rendering|Material|SetMaterial (Variables|Default|GetDuplicate%20Hat) _array_index_3 _returnvalue_12)
        (Utilities|Array|Add (Variables|Default|GetDuplicate%20MIDs) _returnvalue_12)))
    (:Is%20Not%20Valid
      _))
  (Utilities|IsValid _returnvalue_6
    (:Is%20Valid
      (bind _returnvalue_13 (AddComponent|NODE_AddSkeletalMeshComponent-10))
      (bind _output_get_4 (Variables|Default|SetDuplicate%20Backpack _returnvalue_13))
      (Components|SkinnedMesh|SetSkinnedAssetandUpdate _output_get_4 (Variables|/Script/Engine.SkinnedMeshComponent|GetSkeletalMesh _returnvalue_6) true)
      (Components|SkinnedMesh|SetLeaderPoseComponent _output_get_4 _returnvalue_6 false false)
      (Transformation|AttachComponentToComponent _output_get_4 _returnvalue_6 "None" "SnapToTarget" "SnapToTarget" "SnapToTarget" true)
      (for (_ _array_index_4) (Rendering|Material|GetMaterials _duplicate_backpack)
        (bind _returnvalue_14 (Rendering|Material|CreateDynamicMaterialInstance _duplicate_backpack _array_index_4 (Variables|MaterialChangeFX|GetPhase%20Shift%20Material) "None" :!class /Script/Engine.PrimitiveComponent))
        (Rendering|Material|CopyMaterialInstanceParameters _returnvalue_14 (Utilities|Array|Get(acopy) (Rendering|Material|GetMaterials _returnvalue_6) _array_index_4) false)
        (Rendering|Material|SetMaterial (Variables|Default|GetDuplicate%20Backpack) _array_index_4 _returnvalue_14)
        (Utilities|Array|Add (Variables|Default|GetDuplicate%20MIDs) _returnvalue_14)))
    (:Is%20Not%20Valid
      _))
  (Utilities|IsValid _returnvalue
    (:Is%20Valid
      (bind _returnvalue_15 (AddComponent|NODE_AddSkeletalMeshComponent-11))
      (bind _output_get_5 (Variables|Default|SetDuplicate%20Charm _returnvalue_15))
      (Components|SkinnedMesh|SetSkinnedAssetandUpdate _output_get_5 (Variables|/Script/Engine.SkinnedMeshComponent|GetSkeletalMesh _returnvalue) true)
      (Components|SkinnedMesh|SetLeaderPoseComponent _output_get_5 _returnvalue false false)
      (Transformation|AttachComponentToComponent _output_get_5 _returnvalue "None" "SnapToTarget" "SnapToTarget" "SnapToTarget" true)
      (for (_ _array_index_5) (Rendering|Material|GetMaterials _duplicate_charm)
        (bind _returnvalue_16 (Rendering|Material|CreateDynamicMaterialInstance _duplicate_charm _array_index_5 (Variables|MaterialChangeFX|GetPhase%20Shift%20Material) "None" :!class /Script/Engine.PrimitiveComponent))
        (Rendering|Material|CopyMaterialInstanceParameters _returnvalue_16 (Utilities|Array|Get(acopy) (Rendering|Material|GetMaterials _returnvalue) _array_index_5) false)
        (Rendering|Material|SetMaterial (Variables|Default|GetDuplicate%20Charm) _array_index_5 _returnvalue_16)
        (Utilities|Array|Add (Variables|Default|GetDuplicate%20MIDs) _returnvalue_16)))
    (:Is%20Not%20Valid
      _))
  (Utilities|IsValid _returnvalue_4
    (:Is%20Valid
      (bind _returnvalue_17 (AddComponent|NODE_AddSkeletalMeshComponent-12))
      (bind _output_get_6 (Variables|Default|SetDuplicate%20Face _returnvalue_17))
      (Components|SkinnedMesh|SetSkinnedAssetandUpdate _output_get_6 (Variables|/Script/Engine.SkinnedMeshComponent|GetSkeletalMesh _returnvalue_4) true)
      (Components|SkinnedMesh|SetLeaderPoseComponent _output_get_6 _returnvalue_4 false false)
      (Transformation|AttachComponentToComponent _output_get_6 _returnvalue_4 "None" "SnapToTarget" "SnapToTarget" "SnapToTarget" true)
      (for (_ _array_index_6) (Rendering|Material|GetMaterials _duplicate_face)
        (bind _returnvalue_18 (Rendering|Material|CreateDynamicMaterialInstance _duplicate_face _array_index_6 (Variables|MaterialChangeFX|GetPhase%20Shift%20Material) "None" :!class /Script/Engine.PrimitiveComponent))
        (Rendering|Material|CopyMaterialInstanceParameters _returnvalue_18 (Utilities|Array|Get(acopy) (Rendering|Material|GetMaterials _returnvalue_4) _array_index_6) false)
        (Rendering|Material|SetMaterial (Variables|Default|GetDuplicate%20Face) _array_index_6 _returnvalue_18)
        (Utilities|Array|Add (Variables|Default|GetDuplicate%20MIDs) _returnvalue_18)))))

(fn ConstructionScript)
